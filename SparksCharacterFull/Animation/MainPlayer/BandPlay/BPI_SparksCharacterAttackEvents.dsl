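; Blueprint interface for Sparks character attack events in the BandPlay animation folder of the main player.
; It declares four parameterless functions for attack stages: charging, windup before launch, launched and received.
; It holds no logic, so implementing Blueprints define the response to each event.
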(fn SparksCharAttackEventCharging)

(fn SparksCharAttackEventLaunched)

(fn SparksCharAttackEventReceived)

(fn SparksCharAttackEventWindupBeforeLaunched)
