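# Animation Blueprint for a Sparks band character playing bass in time with the music.
# A Body state machine runs beat-synced homebase, idle, play and solo variants plus intensity transitions, switching on transition beats, with a fidget when the beat stops.
# Its cached pose gets fret-hand and strum-arm layers via virtual-bone filters, with bool blends to fallback loops.
# Events drive variant changes, alternating pluck montages and random chord indices.
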AnimBlueprint /SparksCharacterFull/Animation/MainPlayer/BandPlay/Sparks_Bass_AnimBP
target_skeleton=&Fortnite_M_Avg_Player_Skeleton
graphs [
 AnimGraph {nodes [
   n1 = Sequence Player 'Bass_Play_Loop_125_Medium_A'!(PlayRate:float=0) -> (Pose) {playRate=0, sequence=/SparksCharacterFull/Animation/MainPlayer/BandPlay/Bass/Loops/BPM_125/Bass_Play_Loop_125_Medium_A, startPosition=1.3}
   Blend Poses by bool(BlendPose_0=@n1, bActiveValue=Get DEBUG_BodyNoMotion() -> (DEBUG_BodyNoMotion:bool)) {blendType=Linear}
   n2 = "Use cached pose 'Body (Cache Pose)'"()
   n3 = Get IsIdleBeatSynced() -> (IsIdleBeatSynced:bool)
   AnimGraphNode_StateMachine_1 = Right Arm() {is=title AnimGraphNode_StateMachine}
   AnimGraphNode_StateMachine_2 = Left Arm() {is=title AnimGraphNode_StateMachine, bAllowConduitEntryStates=true}
   n4 = Layered blend per bone!(
    BasePose="Use cached pose 'Body (Cache Pose)'"()
    BlendPoses_0=@AnimGraphNode_StateMachine_2
    BlendPoses_1:Pose Link Structure=@AnimGraphNode_StateMachine_1
    BlendWeights_0
    BlendWeights_1:float=1
   ) -> (Pose) {
    blendPoses [{linkId=-1, sourceLinkId=-1}, {linkId=-1, sourceLinkId=-1}]
    blendWeights [1,1]
    layerSetup [
     {branchFilters [{boneName=clavicle_l}, {boneName=VB VB spine_05_hand_attach_r_ik_hand_l}]}
     {branchFilters [{boneName=clavicle_r}, {boneName=VB VB spine_05_hand_attach_r_ik_hand_gun}]}
    ]
   }
   n5 = Apply Additive(Base=@n4, Additive=Slot 'AdditiveLayer'(Source=Additive Identity Pose()) {slotName=AdditiveLayer}) {alphaBoolBlend {blendOption=Linear}}
   n6 = Blend Poses by bool(
    BlendPose_0=Sequence Player 'Bass_Idle_Loop_A'() {sequence=/SparksCharacterFull/Animation/MainPlayer/BandPlay/Bass/Loops/Bass_Idle_Loop_A}
    BlendPose_1=Blend Poses by bool(
     BlendPose_0=Blend Poses by bool(BlendPose_0=@n2, BlendPose_1=@n5, BlendTime_0=0.5, BlendTime_1=0.5, bActiveValue=@n3) {blendType=Linear}
     BlendPose_1=Sequence Player 'Bass_Idle_Loop_A'() {sequence=&Bass_Idle_Loop_A}
     BlendTime_0=0.5
     BlendTime_1=0.5
     bActiveValue=Get bMusicBeatIsActive() -> (bMusicBeatIsActive:bool)
    ) {blendType=Linear}
    BlendTime_0=0.4
    BlendTime_1=0.4
    bActiveValue=Get IdleRealtime() -> (IdleRealtime:bool)
   ) {blendType=Linear}
   n7 = Sequence Player 'Bass_Play_Loop_125_Medium_A'!() -> (Pose) {sequence=&Bass_Play_Loop_125_Medium_A}
   n8 = Get bCanPlayMidiAnimations() -> (bCanPlayMidiAnimations:bool)
   Output Pose(Result=Apply Additive(
     Base=Blend Poses by bool(BlendPose_0=@n6, BlendPose_1=@n7, BlendTime_0=0.5, BlendTime_1=0.5, bActiveValue=@n8) {blendType=Linear}
     Additive="Sequence Player (additive) 'Bass_SpawnInstrument_Pose'"() {sequence=/SparksCharacterCommon/Animation/MainPlayer/BandPlay/Bass/Bass_SpawnInstrument_Pose}
    ) {alphaBoolBlend {blendOption=Linear}}) {layerGroup=None}
   AnimGraphNode_StateMachine_5 = Body() {is=title AnimGraphNode_StateMachine, bAllowConduitEntryStates=true}
   "Save cached pose 'Body (Cache Pose)'"(Pose=@AnimGraphNode_StateMachine_5) {cacheName="Body (Cache Pose)"}
   AnimGraphNode_StateMachine_0 = Head() {is=title AnimGraphNode_StateMachine, bAllowConduitEntryStates=true}
  ]}
 Head {entry=LipSync, path=AnimGraph.AnimGraphNode_StateMachine_0.Head, ids=id, nodes [AnimStateNode_0 = LipSync() {is=title AnimStateNode}]}
 LipSync {
  path=AnimGraph.AnimGraphNode_StateMachine_0.Head.AnimStateNode_0.LipSync
  nodes [Output Animation Pose(Result=Play Beat Synced Animation!(
     ExplicitTime:float=0
     bShouldLoop:bool=false
     InSequence:Anim Sequence Base Object Reference=Get LipSyncAnimSequence() -> (LipSyncAnimSequence:Anim Sequence Object Reference)
     MusicClock:Music Clock Object Reference
     TempoMultiplier:float=1
     bEnableTempoVariants:bool=False
     TempoVariantHysteresis:float=0.05
     bShowDebug:bool=False
     BeatAndTime:Beat And Time Structure=(Beat=0,Time=0,BeatsPerSecond=0,IsValid=False)
    ) -> (Pose:Pose Link Structure) {is=title AnimGraphNode_PlayBeatSyncedAnim})]
 }
 Right Arm {
  entry=Right Arm
  path=AnimGraph.AnimGraphNode_StateMachine_1.Right Arm
  ids=id
  nodes [AnimStateNode_0 = Right Arm() {is=title AnimStateNode}]
 }
 Right Arm {
  path=AnimGraph.AnimGraphNode_StateMachine_1.Right Arm.AnimStateNode_0.Right Arm
  nodes [
   Apply Additive(Additive=Sequence Player!(Sequence:Anim Sequence Base Object Reference, PlayRate:float=0, bLoopAnimation:bool=false) -> (Pose) {playRate=0}) {alphaBoolBlend {blendOption=Linear}}
   Sequence Player(PlayRate=0, bLoopAnimation=False) {bLoopAnimation=false, playRate=0}
   Output Animation Pose(Result=Apply Additive(
     Base="Use cached pose 'Body (Cache Pose)'"()
     Additive="Sequence Player (additive) 'Bass_Strum_Corrective_Pose'"() {sequence=/SparksCharacterCommon/Animation/MainPlayer/BandPlay/Bass/Strums/Bass_Strum_Corrective_Pose}
    ) {alphaBoolBlend {blendOption=Linear}})
  ]
 }
 Left Arm {
  entry=Conduit
  path=AnimGraph.AnimGraphNode_StateMachine_2.Left Arm
  ids=id
  nodes [
   AnimStateConduitNode_1 = Conduit()
   AnimStateNode_4 = "Frets && Chords"() {is=title AnimStateNode}
   AnimStateTransitionNode_12 = AnimStateTransitionNode()
   AnimStateTransitionNode_13 = AnimStateTransitionNode()
  ]
  links [
   {
    dst="Frets && Chords"
    rule=AnimGraph.AnimGraphNode_StateMachine_2.Left Arm.AnimStateTransitionNode_12.AnimationTransitionGraph_0
    src=Conduit
    via=AnimStateTransitionNode_12
   }
   {
    dst=Conduit
    rule=AnimGraph.AnimGraphNode_StateMachine_2.Left Arm.AnimStateTransitionNode_13.AnimationTransitionGraph_0
    src="Frets && Chords"
    via=AnimStateTransitionNode_13
   }
  ]
 }
 Conduit {path=AnimGraph.AnimGraphNode_StateMachine_2.Left Arm.AnimStateConduitNode_1.Conduit, nodes [Result(bCanEnterTransition=true)]}
 "Frets && Chords" {
  path="AnimGraph.AnimGraphNode_StateMachine_2.Left Arm.AnimStateNode_4.Frets && Chords"
  nodes [
   n1 = Blendspace Player 'Bass_FretHand_BS'(X=0, Y=0) {blendSpace=/SparksCharacterCommon/Animation/MainPlayer/BandPlay/Bass/FretHand/Bass_FretHand_BS, startPosition=1}
   n2 = Additive Identity Pose()
   Output Animation Pose(Result=Apply Additive(
     Base=Sequence Player 'Bass_FretHand_BasePose'() {sequence=/SparksCharacterCommon/Animation/MainPlayer/BandPlay/Bass/FretHand/Bass_FretHand_BasePose}
     Additive=Blend Poses by bool(BlendPose_0=@n1, BlendPose_1=@n2, bActiveValue=true) {blendType=Linear}
    ) {alphaBoolBlend {blendOption=Linear}})
  ]
 }
 AnimationTransitionGraph_0 {
  path=AnimGraph.AnimGraphNode_StateMachine_2.Left Arm.AnimStateTransitionNode_12.AnimationTransitionGraph_0
  nodes [Result(bCanEnterTransition=true)]
 }
 AnimationTransitionGraph_0 {
  path=AnimGraph.AnimGraphNode_StateMachine_2.Left Arm.AnimStateTransitionNode_13.AnimationTransitionGraph_0
  nodes [Result(bCanEnterTransition=NOT Boolean(A=false))]
 }
 Body {
  entry=Conduit
  path=AnimGraph.AnimGraphNode_StateMachine_5.Body
  ids=id
  nodes [
   AnimStateConduitNode_0 = Conduit()
   AnimStateNode_1 = "Play && Idle (Beat Sync)"() {
    is=title AnimStateNode
    stateEntered {guid=#1}
    stateFullyBlended {guid=#2}
    stateLeft {guid=#3}
   }
   AnimStateNode_6 = Fidget() {is=title AnimStateNode}
   AnimStateTransitionNode_0 = AnimStateTransitionNode() {crossfadeDuration=0.4, priorityOrder=100}
   AnimStateTransitionNode_1 = AnimStateTransitionNode()
   AnimStateTransitionNode_6 = AnimStateTransitionNode() {crossfadeDuration=0.4, priorityOrder=3}
   AnimStateTransitionNode_7 = AnimStateTransitionNode()
  ]
  links [
   {
    dst=Fidget
    rule=AnimGraph.AnimGraphNode_StateMachine_5.Body.AnimStateTransitionNode_6.Transition
    src=Conduit
    via=AnimStateTransitionNode_6
   }
   {
    dst="Play && Idle (Beat Sync)"
    rule=AnimGraph.AnimGraphNode_StateMachine_5.Body.AnimStateTransitionNode_0.Transition
    src=Conduit
    via=AnimStateTransitionNode_0
   }
   {
    dst=Conduit
    rule=AnimGraph.AnimGraphNode_StateMachine_5.Body.AnimStateTransitionNode_7.Transition
    src=Fidget
    via=AnimStateTransitionNode_7
   }
   {
    dst=Conduit
    rule=AnimGraph.AnimGraphNode_StateMachine_5.Body.AnimStateTransitionNode_1.Transition
    src="Play && Idle (Beat Sync)"
    via=AnimStateTransitionNode_1
   }
  ]
 }
 Conduit {path=AnimGraph.AnimGraphNode_StateMachine_5.Body.AnimStateConduitNode_0.Conduit, nodes [Result(bCanEnterTransition=true)]}
 "Play && Idle (Beat Sync)" {
  path="AnimGraph.AnimGraphNode_StateMachine_5.Body.AnimStateNode_1.Play && Idle (Beat Sync)"
  nodes [
   AnimGraphNode_StateMachine_0 = "Play && Idle (Beat Sync)"() {is=title AnimGraphNode_StateMachine, bAllowConduitEntryStates=true, bSkipFirstUpdateTransition=false}
   Output Animation Pose(Result=@AnimGraphNode_StateMachine_0)
  ]
 }
 "Play && Idle (Beat Sync)" {
  entry=Conduit
  path="AnimGraph.AnimGraphNode_StateMachine_5.Body.AnimStateNode_1.Play && Idle (Beat Sync).AnimGraphNode_StateMachine_0.Play && Idle (Beat Sync)"
  ids=id
  nodes [
   AnimStateConduitNode_0 = Conduit()
   AnimStateNode_10 = "Idle [Loop]"() {
    is=title AnimStateNode
    stateEntered {guid=#4}
    stateFullyBlended {guid=#5}
    stateLeft {guid=#6}
   }
   AnimStateNode_26 = "Play [Start]"() {
    is=title AnimStateNode
    stateEntered {guid=#4, notifyName=EnteredStateEvent_PlayVariant}
    stateFullyBlended {guid=#5}
    stateLeft {guid=#6}
   }
   AnimStateNode_27 = "Play [End]"() {
    is=title AnimStateNode
    stateEntered {guid=#4}
    stateFullyBlended {guid=#5}
    stateLeft {guid=#6, notifyName=LeftStateEvent_PlayVariant}
   }
   AnimStateNode_28 = "Play [Loop]"() {
    is=title AnimStateNode
    stateEntered {guid=#4}
    stateFullyBlended {guid=#5}
    stateLeft {guid=#6}
   }
   AnimStateNode_45 = IntensityTransition() {
    is=title AnimStateNode
    stateEntered {guid=#7}
    stateFullyBlended {guid=#8}
    stateLeft {guid=#9}
   }
   AnimStateNode_51 = "Solo [Start]"() {
    is=title AnimStateNode
    stateEntered {guid=#4, notifyName=EnteredStateEvent_SoloVariant}
    stateFullyBlended {guid=#5}
    stateLeft {guid=#6}
   }
   AnimStateNode_52 = "Solo [End]"() {
    is=title AnimStateNode
    stateEntered {guid=#4}
    stateFullyBlended {guid=#5}
    stateLeft {guid=#6, notifyName=LeftStateEvent_SoloVariant}
   }
   AnimStateNode_53 = "Solo [Loop]"() {
    is=title AnimStateNode
    stateEntered {guid=#4}
    stateFullyBlended {guid=#5}
    stateLeft {guid=#6}
   }
   AnimStateNode_6 = "Idle [Start]"() {
    is=title AnimStateNode
    stateEntered {guid=#4, notifyName=EnteredStateEvent_IdleVariant}
    stateFullyBlended {guid=#5}
    stateLeft {guid=#6}
   }
   AnimStateNode_7 = "Idle [End]"() {
    is=title AnimStateNode
    stateEntered {guid=#4}
    stateFullyBlended {guid=#5}
    stateLeft {guid=#6, notifyName=LeftStateEvent_IdleVariant}
   }
   AnimStateNode_8 = "Homebase [Loop]"() {
    is=title AnimStateNode
    stateEntered {guid=#10, notifyName=EnteredStateEvent_Homebase}
    stateFullyBlended {guid=#11}
    stateLeft {guid=#12, notifyName=LeftStateEvent_Homebase}
   }
   AnimStateTransitionNode_0 = AnimStateTransitionNode() {crossfadeDuration=0.5, priorityOrder=2}
   AnimStateTransitionNode_1 = AnimStateTransitionNode() {crossfadeDuration=0, priorityOrder=2}
   AnimStateTransitionNode_19 = AnimStateTransitionNode() {crossfadeDuration=0.4, priorityOrder=2}
   AnimStateTransitionNode_2 = AnimStateTransitionNode() {
    crossfadeDuration=0.5
    transitionEnd {guid=#13}
    transitionInterrupt {guid=#14}
    transitionStart {guid=#15}
   }
   AnimStateTransitionNode_20 = AnimStateTransitionNode() {crossfadeDuration=0.4}
   AnimStateTransitionNode_38 = AnimStateTransitionNode() {
    crossfadeDuration=0.4
    transitionEnd {guid=#16}
    transitionInterrupt {guid=#17}
    transitionStart {guid=#18}
   }
   AnimStateTransitionNode_40 = AnimStateTransitionNode() {
    crossfadeDuration=0.5
    transitionEnd {guid=#19}
    transitionInterrupt {guid=#20}
    transitionStart {guid=#21}
   }
   AnimStateTransitionNode_5 = AnimStateTransitionNode() {
    crossfadeDuration=0.4
    transitionEnd {guid=#22}
    transitionInterrupt {guid=#23}
    transitionStart {guid=#24}
   }
   AnimStateTransitionNode_50 = AnimStateTransitionNode() {crossfadeDuration=0.5, priorityOrder=2}
   AnimStateTransitionNode_52 = AnimStateTransitionNode() {crossfadeDuration=0.5}
   AnimStateTransitionNode_59 = AnimStateTransitionNode() {
    crossfadeDuration=0.5
    transitionEnd {guid=#25}
    transitionInterrupt {guid=#26}
    transitionStart {guid=#27}
   }
   AnimStateTransitionNode_6 = AnimStateTransitionNode() {
    crossfadeDuration=0.5
    transitionEnd {guid=#28}
    transitionInterrupt {guid=#29}
    transitionStart {guid=#30}
   }
   AnimStateTransitionNode_69 = AnimStateTransitionNode() {
    crossfadeDuration=0.5
    transitionEnd {guid=#28}
    transitionInterrupt {guid=#29}
    transitionStart {guid=#30}
   }
   AnimStateTransitionNode_70 = AnimStateTransitionNode() {crossfadeDuration=0.4}
   AnimStateTransitionNode_76 = AnimStateTransitionNode()
  ]
  links [
   {
    dst="Homebase [Loop]"
    rule="AnimGraph.AnimGraphNode_StateMachine_5.Body.AnimStateNode_1.Play && Idle (Beat Sync).AnimGraphNode_StateMachine_0.Play && Idle (Beat Sync).AnimStateTransitionNode_1.AnimationTransitionGraph_1"
    src=Conduit
    via=AnimStateTransitionNode_1
   }
   {
    dst="Idle [Start]"
    rule="AnimGraph.AnimGraphNode_StateMachine_5.Body.AnimStateNode_1.Play && Idle (Beat Sync).AnimGraphNode_StateMachine_0.Play && Idle (Beat Sync).AnimStateTransitionNode_19.Transition"
    src="Homebase [Loop]"
    via=AnimStateTransitionNode_19
   }
   {
    dst=IntensityTransition
    rule="AnimGraph.AnimGraphNode_StateMachine_5.Body.AnimStateNode_1.Play && Idle (Beat Sync).AnimGraphNode_StateMachine_0.Play && Idle (Beat Sync).AnimStateTransitionNode_59.Transition"
    src="Homebase [Loop]"
    via=AnimStateTransitionNode_59
   }
   {
    dst="Play [Start]"
    rule="AnimGraph.AnimGraphNode_StateMachine_5.Body.AnimStateNode_1.Play && Idle (Beat Sync).AnimGraphNode_StateMachine_0.Play && Idle (Beat Sync).AnimStateTransitionNode_0.Transition"
    src="Homebase [Loop]"
    via=AnimStateTransitionNode_0
   }
   {
    dst="Solo [Start]"
    rule="AnimGraph.AnimGraphNode_StateMachine_5.Body.AnimStateNode_1.Play && Idle (Beat Sync).AnimGraphNode_StateMachine_0.Play && Idle (Beat Sync).AnimStateTransitionNode_50.Transition"
    src="Homebase [Loop]"
    via=AnimStateTransitionNode_50
   }
   {
    dst="Homebase [Loop]"
    rule="AnimGraph.AnimGraphNode_StateMachine_5.Body.AnimStateNode_1.Play && Idle (Beat Sync).AnimGraphNode_StateMachine_0.Play && Idle (Beat Sync).AnimStateTransitionNode_20.Transition"
    src="Idle [End]"
    via=AnimStateTransitionNode_20
   }
   {
    dst="Idle [End]"
    rule="AnimGraph.AnimGraphNode_StateMachine_5.Body.AnimStateNode_1.Play && Idle (Beat Sync).AnimGraphNode_StateMachine_0.Play && Idle (Beat Sync).AnimStateTransitionNode_6.AnimationTransitionGraph_0"
    src="Idle [Loop]"
    via=AnimStateTransitionNode_6
   }
   {
    dst="Idle [Loop]"
    rule="AnimGraph.AnimGraphNode_StateMachine_5.Body.AnimStateNode_1.Play && Idle (Beat Sync).AnimGraphNode_StateMachine_0.Play && Idle (Beat Sync).AnimStateTransitionNode_5.AnimationTransitionGraph_1"
    src="Idle [Start]"
    via=AnimStateTransitionNode_5
   }
   {
    dst="Homebase [Loop]"
    rule="AnimGraph.AnimGraphNode_StateMachine_5.Body.AnimStateNode_1.Play && Idle (Beat Sync).AnimGraphNode_StateMachine_0.Play && Idle (Beat Sync).AnimStateTransitionNode_76.Transition"
    src=IntensityTransition
    via=AnimStateTransitionNode_76
   }
   {
    dst="Homebase [Loop]"
    rule="AnimGraph.AnimGraphNode_StateMachine_5.Body.AnimStateNode_1.Play && Idle (Beat Sync).AnimGraphNode_StateMachine_0.Play && Idle (Beat Sync).AnimStateTransitionNode_2.AnimationTransitionGraph_1"
    src="Play [End]"
    via=AnimStateTransitionNode_2
   }
   {
    dst="Play [End]"
    rule="AnimGraph.AnimGraphNode_StateMachine_5.Body.AnimStateNode_1.Play && Idle (Beat Sync).AnimGraphNode_StateMachine_0.Play && Idle (Beat Sync).AnimStateTransitionNode_40.AnimationTransitionGraph_0"
    src="Play [Loop]"
    via=AnimStateTransitionNode_40
   }
   {
    dst="Play [Loop]"
    rule="AnimGraph.AnimGraphNode_StateMachine_5.Body.AnimStateNode_1.Play && Idle (Beat Sync).AnimGraphNode_StateMachine_0.Play && Idle (Beat Sync).AnimStateTransitionNode_38.AnimationTransitionGraph_0"
    src="Play [Start]"
    via=AnimStateTransitionNode_38
   }
   {
    dst="Homebase [Loop]"
    rule="AnimGraph.AnimGraphNode_StateMachine_5.Body.AnimStateNode_1.Play && Idle (Beat Sync).AnimGraphNode_StateMachine_0.Play && Idle (Beat Sync).AnimStateTransitionNode_52.Transition"
    src="Solo [End]"
    via=AnimStateTransitionNode_52
   }
   {
    dst="Solo [End]"
    rule="AnimGraph.AnimGraphNode_StateMachine_5.Body.AnimStateNode_1.Play && Idle (Beat Sync).AnimGraphNode_StateMachine_0.Play && Idle (Beat Sync).AnimStateTransitionNode_69.AnimationTransitionGraph_0"
    src="Solo [Loop]"
    via=AnimStateTransitionNode_69
   }
   {
    dst="Solo [Loop]"
    rule="AnimGraph.AnimGraphNode_StateMachine_5.Body.AnimStateNode_1.Play && Idle (Beat Sync).AnimGraphNode_StateMachine_0.Play && Idle (Beat Sync).AnimStateTransitionNode_70.AnimationTransitionGraph_1"
    src="Solo [Start]"
    via=AnimStateTransitionNode_70
   }
  ]
 }
 Conduit {
  path="AnimGraph.AnimGraphNode_StateMachine_5.Body.AnimStateNode_1.Play && Idle (Beat Sync).AnimGraphNode_StateMachine_0.Play && Idle (Beat Sync).AnimStateConduitNode_0.Conduit"
  nodes [Result(bCanEnterTransition=true)]
 }
 "Idle [Loop]" {
  path="AnimGraph.AnimGraphNode_StateMachine_5.Body.AnimStateNode_1.Play && Idle (Beat Sync).AnimGraphNode_StateMachine_0.Play && Idle (Beat Sync).AnimStateNode_10.Idle [Loop]"
  nodes [Output Animation Pose(Result=Play Beat Synced Animation!(
     ExplicitTime=0
     bShouldLoop=True
     InSequence=Get!() -> (IdleAnim_Loop:Anim Sequence Object Reference, IdleAnim_Entry:Anim Sequence Object Reference, IdleAnim_Exit:Anim Sequence Object Reference)
     MusicClock
     TempoMultiplier=Get!() -> (IdleBPMs_BPMBuckets:Array of Sparks Anim Intensity Set Object References, IdleBPMs_CurrentTempoMultiplier:float=~).IdleBPMs_CurrentTempoMultiplier
     bEnableTempoVariants=False
     TempoVariantHysteresis=0.05
     bShowDebug=False
    ) -> (Pose))]
 }
 "Play [Start]" {
  path="AnimGraph.AnimGraphNode_StateMachine_5.Body.AnimStateNode_1.Play && Idle (Beat Sync).AnimGraphNode_StateMachine_0.Play && Idle (Beat Sync).AnimStateNode_26.Play [Start]"
  nodes [Output Animation Pose(Result=Play Beat Synced Animation(
     ExplicitTime=0
     bShouldLoop=false
     InSequence=Get!() -> (PlayAnim_Loop:Anim Sequence Object Reference, PlayAnim_Entry:Anim Sequence Object Reference, PlayAnim_Exit:Anim Sequence Object Reference).PlayAnim_Entry
     TempoMultiplier=Get!() -> (PlayBPMs_BPMBuckets:Array of Sparks Anim Intensity Set Object References, PlayBPMs_CurrentTempoMultiplier:float=~).PlayBPMs_CurrentTempoMultiplier
     bEnableTempoVariants=False
     TempoVariantHysteresis=0.05
     bShowDebug=False
    )) {
    node {stateEntryFunction {functionName=OnPlayEntry}}
    stateEntryFunction {bSelfContext=true, memberGuid=#31, memberName=OnPlayEntry}
   }]
 }
 "Play [End]" {
  path="AnimGraph.AnimGraphNode_StateMachine_5.Body.AnimStateNode_1.Play && Idle (Beat Sync).AnimGraphNode_StateMachine_0.Play && Idle (Beat Sync).AnimStateNode_27.Play [End]"
  nodes [Output Animation Pose(Result=Play Beat Synced Animation(
     ExplicitTime=0
     bShouldLoop=false
     InSequence=Get!() -> (PlayAnim_Loop:Anim Sequence Object Reference, PlayAnim_Entry:Anim Sequence Object Reference, PlayAnim_Exit:Anim Sequence Object Reference).PlayAnim_Exit
     TempoMultiplier=Get!() -> (PlayBPMs_BPMBuckets:Array of Sparks Anim Intensity Set Object References, PlayBPMs_CurrentTempoMultiplier:float=~).PlayBPMs_CurrentTempoMultiplier
     bEnableTempoVariants=False
     TempoVariantHysteresis=0.05
     bShowDebug=False
    ))]
 }
 "Play [Loop]" {
  path="AnimGraph.AnimGraphNode_StateMachine_5.Body.AnimStateNode_1.Play && Idle (Beat Sync).AnimGraphNode_StateMachine_0.Play && Idle (Beat Sync).AnimStateNode_28.Play [Loop]"
  nodes [Output Animation Pose(Result=Play Beat Synced Animation(
     ExplicitTime=0
     bShouldLoop=True
     InSequence=Get!() -> (PlayAnim_Loop:Anim Sequence Object Reference, PlayAnim_Entry:Anim Sequence Object Reference, PlayAnim_Exit:Anim Sequence Object Reference)
     TempoMultiplier=Get!() -> (PlayBPMs_BPMBuckets:Array of Sparks Anim Intensity Set Object References, PlayBPMs_CurrentTempoMultiplier:float=~).PlayBPMs_CurrentTempoMultiplier
     bEnableTempoVariants=False
     TempoVariantHysteresis=0.05
     bShowDebug=False
    ))]
 }
 IntensityTransition {
  path="AnimGraph.AnimGraphNode_StateMachine_5.Body.AnimStateNode_1.Play && Idle (Beat Sync).AnimGraphNode_StateMachine_0.Play && Idle (Beat Sync).AnimStateNode_45.IntensityTransition"
  nodes [Output Animation Pose(Result=Sequence Player(Sequence=Get TransitionAnim() -> (TransitionAnim:Anim Sequence Object Reference), PlayRate=1, bLoopAnimation=false)) {
    node {stateEntryFunction {functionName=OnIntensityTransitionEntry}}
    stateEntryFunction {bSelfContext=true, memberGuid=#32, memberName=OnIntensityTransitionEntry}
   }]
 }
 "Solo [Start]" {
  path="AnimGraph.AnimGraphNode_StateMachine_5.Body.AnimStateNode_1.Play && Idle (Beat Sync).AnimGraphNode_StateMachine_0.Play && Idle (Beat Sync).AnimStateNode_51.Solo [Start]"
  nodes [Output Animation Pose(Result=Play Beat Synced Animation(
     ExplicitTime=0
     bShouldLoop=false
     InSequence=Get!() -> (SoloAnim_Loop:Anim Sequence Object Reference, SoloAnim_Entry:Anim Sequence Object Reference, SoloAnim_Exit:Anim Sequence Object Reference).SoloAnim_Entry
     TempoMultiplier=Get!() -> (SoloBPMs_BPMBuckets:Array of Sparks Anim Intensity Set Object References, SoloBPMs_CurrentTempoMultiplier:float=~).SoloBPMs_CurrentTempoMultiplier
     bEnableTempoVariants=False
     TempoVariantHysteresis=0.05
     bShowDebug=False
    )) {
    node {stateEntryFunction {functionName=OnStateEntry}}
    stateEntryFunction {bSelfContext=true, memberGuid=#33, memberName=OnSoloEntry}
   }]
 }
 "Solo [End]" {
  path="AnimGraph.AnimGraphNode_StateMachine_5.Body.AnimStateNode_1.Play && Idle (Beat Sync).AnimGraphNode_StateMachine_0.Play && Idle (Beat Sync).AnimStateNode_52.Solo [End]"
  nodes [Output Animation Pose(Result=Play Beat Synced Animation(
     ExplicitTime=0
     bShouldLoop=false
     InSequence=Get!() -> (SoloAnim_Loop:Anim Sequence Object Reference, SoloAnim_Entry:Anim Sequence Object Reference, SoloAnim_Exit:Anim Sequence Object Reference).SoloAnim_Exit
     TempoMultiplier=Get!() -> (SoloBPMs_BPMBuckets:Array of Sparks Anim Intensity Set Object References, SoloBPMs_CurrentTempoMultiplier:float=~).SoloBPMs_CurrentTempoMultiplier
     bEnableTempoVariants=False
     TempoVariantHysteresis=0.05
     bShowDebug=False
    ))]
 }
 "Solo [Loop]" {
  path="AnimGraph.AnimGraphNode_StateMachine_5.Body.AnimStateNode_1.Play && Idle (Beat Sync).AnimGraphNode_StateMachine_0.Play && Idle (Beat Sync).AnimStateNode_53.Solo [Loop]"
  nodes [Output Animation Pose(Result=Play Beat Synced Animation(
     ExplicitTime=0
     bShouldLoop=True
     InSequence=Get!() -> (SoloAnim_Loop:Anim Sequence Object Reference, SoloAnim_Entry:Anim Sequence Object Reference, SoloAnim_Exit:Anim Sequence Object Reference)
     TempoMultiplier=Get!() -> (SoloBPMs_BPMBuckets:Array of Sparks Anim Intensity Set Object References, SoloBPMs_CurrentTempoMultiplier:float=~).SoloBPMs_CurrentTempoMultiplier
     bEnableTempoVariants=False
     TempoVariantHysteresis=0.05
     bShowDebug=False
    ))]
 }
 "Idle [Start]" {
  path="AnimGraph.AnimGraphNode_StateMachine_5.Body.AnimStateNode_1.Play && Idle (Beat Sync).AnimGraphNode_StateMachine_0.Play && Idle (Beat Sync).AnimStateNode_6.Idle [Start]"
  nodes [Output Animation Pose(Result=Play Beat Synced Animation(
     ExplicitTime=0
     bShouldLoop=false
     InSequence=Get!() -> (IdleAnim_Loop:Anim Sequence Object Reference, IdleAnim_Entry:Anim Sequence Object Reference, IdleAnim_Exit:Anim Sequence Object Reference).IdleAnim_Entry
     TempoMultiplier=Get!() -> (IdleBPMs_BPMBuckets:Array of Sparks Anim Intensity Set Object References, IdleBPMs_CurrentTempoMultiplier:float=~).IdleBPMs_CurrentTempoMultiplier
     bEnableTempoVariants=False
     TempoVariantHysteresis=0.05
     bShowDebug=False
    )) {
    node {stateEntryFunction {functionName=OnIdleEntry}}
    stateEntryFunction {bSelfContext=true, memberGuid=#34, memberName=OnIdleEntry}
   }]
 }
 "Idle [End]" {
  path="AnimGraph.AnimGraphNode_StateMachine_5.Body.AnimStateNode_1.Play && Idle (Beat Sync).AnimGraphNode_StateMachine_0.Play && Idle (Beat Sync).AnimStateNode_7.Idle [End]"
  nodes [Output Animation Pose(Result=Play Beat Synced Animation(
     ExplicitTime=0
     bShouldLoop=false
     InSequence=Get!() -> (IdleAnim_Loop:Anim Sequence Object Reference, IdleAnim_Entry:Anim Sequence Object Reference, IdleAnim_Exit:Anim Sequence Object Reference).IdleAnim_Exit
     TempoMultiplier=Get!() -> (IdleBPMs_BPMBuckets:Array of Sparks Anim Intensity Set Object References, IdleBPMs_CurrentTempoMultiplier:float=~).IdleBPMs_CurrentTempoMultiplier
     bEnableTempoVariants=False
     TempoVariantHysteresis=0.05
     bShowDebug=False
    ))]
 }
 "Homebase [Loop]" {
  path="AnimGraph.AnimGraphNode_StateMachine_5.Body.AnimStateNode_1.Play && Idle (Beat Sync).AnimGraphNode_StateMachine_0.Play && Idle (Beat Sync).AnimStateNode_8.Homebase [Loop]"
  nodes [Output Animation Pose(Result=Play Beat Synced Animation(
     ExplicitTime=0
     bShouldLoop=True
     InSequence=Get HomebaseAnim() -> (HomebaseAnim:Anim Sequence Object Reference)
     TempoMultiplier=Get!() -> (HomebaseBPMs_BPMBuckets:Array of Sparks Anim Intensity Set Object References, HomebaseBPMs_CurrentTempoMultiplier:float=~).HomebaseBPMs_CurrentTempoMultiplier
     bEnableTempoVariants=False
     TempoVariantHysteresis=0.05
     bShowDebug=False
    )) {
    node {stateEntryFunction {functionName=OnHomeBaseEntry}}
    stateEntryFunction {bSelfContext=true, memberGuid=#35, memberName=OnHomeBaseEntry}
   }]
 }
 Transition {
  path="AnimGraph.AnimGraphNode_StateMachine_5.Body.AnimStateNode_1.Play && Idle (Beat Sync).AnimGraphNode_StateMachine_0.Play && Idle (Beat Sync).AnimStateTransitionNode_0.Transition"
  nodes [Result(bCanEnterTransition=CanTransitionToSubgraphStart!(self, DesiredSubgraph:AnimSubgraph Enum=NewEnumerator2) -> (CanTransition:bool))]
 }
 AnimationTransitionGraph_1 {
  path="AnimGraph.AnimGraphNode_StateMachine_5.Body.AnimStateNode_1.Play && Idle (Beat Sync).AnimGraphNode_StateMachine_0.Play && Idle (Beat Sync).AnimStateTransitionNode_1.AnimationTransitionGraph_1"
  nodes [Result(bCanEnterTransition=true)]
 }
 Transition {
  path="AnimGraph.AnimGraphNode_StateMachine_5.Body.AnimStateNode_1.Play && Idle (Beat Sync).AnimGraphNode_StateMachine_0.Play && Idle (Beat Sync).AnimStateTransitionNode_19.Transition"
  nodes [Result(bCanEnterTransition=CanTransitionToSubgraphStart(DesiredSubgraph=NewEnumerator1))]
 }
 AnimationTransitionGraph_1 {
  path="AnimGraph.AnimGraphNode_StateMachine_5.Body.AnimStateNode_1.Play && Idle (Beat Sync).AnimGraphNode_StateMachine_0.Play && Idle (Beat Sync).AnimStateTransitionNode_2.AnimationTransitionGraph_1"
  nodes [Result(bCanEnterTransition=CanTransitionToLoop!(
     self
     RelevantAnimTime:double="Get Relevant Anim Time (Play [End])"!() -> (ReturnValue:float) {is=title K2Node_AnimGetter}
     RelevantAnimTimeRemaining:double="Get Relevant Anim Time Remaining (Play [End])"!() -> (ReturnValue:float) {is=title K2Node_AnimGetter}
    ) -> (CanTransition:bool))]
 }
 Transition {
  path="AnimGraph.AnimGraphNode_StateMachine_5.Body.AnimStateNode_1.Play && Idle (Beat Sync).AnimGraphNode_StateMachine_0.Play && Idle (Beat Sync).AnimStateTransitionNode_20.Transition"
  nodes [Result(bCanEnterTransition=CanTransitionToLoop(
     RelevantAnimTime="Get Relevant Anim Time (Idle [End])"!() -> (ReturnValue:float) {is=title K2Node_AnimGetter}
     RelevantAnimTimeRemaining="Get Relevant Anim Time Remaining (Idle [End])"!() -> (ReturnValue:float) {is=title K2Node_AnimGetter}
    ))]
 }
 AnimationTransitionGraph_0 {
  path="AnimGraph.AnimGraphNode_StateMachine_5.Body.AnimStateNode_1.Play && Idle (Beat Sync).AnimGraphNode_StateMachine_0.Play && Idle (Beat Sync).AnimStateTransitionNode_38.AnimationTransitionGraph_0"
  nodes [Result(bCanEnterTransition=CanTransitionToLoop(
     RelevantAnimTime="Get Relevant Anim Time (Play [Start])"!() -> (ReturnValue:float) {is=title K2Node_AnimGetter}
     RelevantAnimTimeRemaining="Get Relevant Anim Time Remaining (Play [Start])"!() -> (ReturnValue:float) {is=title K2Node_AnimGetter}
    ))]
 }
 AnimationTransitionGraph_0 {
  path="AnimGraph.AnimGraphNode_StateMachine_5.Body.AnimStateNode_1.Play && Idle (Beat Sync).AnimGraphNode_StateMachine_0.Play && Idle (Beat Sync).AnimStateTransitionNode_40.AnimationTransitionGraph_0"
  nodes [Result(bCanEnterTransition=CanTransitionToSubgraphEnd!(self, CurrentSubgraph:AnimSubgraph Enum=NewEnumerator2) -> (CanTransition:bool))]
 }
 AnimationTransitionGraph_1 {
  path="AnimGraph.AnimGraphNode_StateMachine_5.Body.AnimStateNode_1.Play && Idle (Beat Sync).AnimGraphNode_StateMachine_0.Play && Idle (Beat Sync).AnimStateTransitionNode_5.AnimationTransitionGraph_1"
  nodes [Result(bCanEnterTransition=CanTransitionToLoop(
     RelevantAnimTime="Get Relevant Anim Time (Idle [Start])"!() -> (ReturnValue:float) {is=title K2Node_AnimGetter}
     RelevantAnimTimeRemaining="Get Relevant Anim Time Remaining (Idle [Start])"!() -> (ReturnValue:float) {is=title K2Node_AnimGetter}
    ))]
 }
 Transition {
  path="AnimGraph.AnimGraphNode_StateMachine_5.Body.AnimStateNode_1.Play && Idle (Beat Sync).AnimGraphNode_StateMachine_0.Play && Idle (Beat Sync).AnimStateTransitionNode_50.Transition"
  nodes [Result(bCanEnterTransition=CanTransitionToSubgraphStart(DesiredSubgraph=NewEnumerator3))]
 }
 Transition {
  path="AnimGraph.AnimGraphNode_StateMachine_5.Body.AnimStateNode_1.Play && Idle (Beat Sync).AnimGraphNode_StateMachine_0.Play && Idle (Beat Sync).AnimStateTransitionNode_52.Transition"
  nodes [Result(bCanEnterTransition=CanTransitionToLoop(
     RelevantAnimTime="Get Relevant Anim Time (Solo [End])"!() -> (ReturnValue:float) {is=title K2Node_AnimGetter}
     RelevantAnimTimeRemaining="Get Relevant Anim Time Remaining (Solo [End])"!() -> (ReturnValue:float) {is=title K2Node_AnimGetter}
    ))]
 }
 Transition {
  path="AnimGraph.AnimGraphNode_StateMachine_5.Body.AnimStateNode_1.Play && Idle (Beat Sync).AnimGraphNode_StateMachine_0.Play && Idle (Beat Sync).AnimStateTransitionNode_59.Transition"
  nodes [
   n1 = Get SongIntensity() -> (SongIntensity:Gameplay Tag Structure)
   n2 = Get DesiredSongIntensity() -> (DesiredSongIntensity:Gameplay Tag Structure)
   Result(bCanEnterTransition=AND Boolean!(
     A=Get AllowAnimTransition() -> (AllowAnimTransition:bool)
     B=Get TransitionBeat() -> (TransitionBeat:bool)
     C:bool="Not Equal (GameplayTag)"(A:Gameplay Tag Structure=@n1, B:Gameplay Tag Structure=@n2) -> (ReturnValue:bool=~) {is=title K2Node_PromotableOperator}
    ) -> (ReturnValue))
  ]
 }
 AnimationTransitionGraph_0 {
  path="AnimGraph.AnimGraphNode_StateMachine_5.Body.AnimStateNode_1.Play && Idle (Beat Sync).AnimGraphNode_StateMachine_0.Play && Idle (Beat Sync).AnimStateTransitionNode_6.AnimationTransitionGraph_0"
  nodes [Result(bCanEnterTransition=CanTransitionToSubgraphEnd(CurrentSubgraph=NewEnumerator1))]
 }
 AnimationTransitionGraph_0 {
  path="AnimGraph.AnimGraphNode_StateMachine_5.Body.AnimStateNode_1.Play && Idle (Beat Sync).AnimGraphNode_StateMachine_0.Play && Idle (Beat Sync).AnimStateTransitionNode_69.AnimationTransitionGraph_0"
  nodes [Result(bCanEnterTransition=CanTransitionToSubgraphEnd(CurrentSubgraph=NewEnumerator3))]
 }
 AnimationTransitionGraph_1 {
  path="AnimGraph.AnimGraphNode_StateMachine_5.Body.AnimStateNode_1.Play && Idle (Beat Sync).AnimGraphNode_StateMachine_0.Play && Idle (Beat Sync).AnimStateTransitionNode_70.AnimationTransitionGraph_1"
  nodes [Result(bCanEnterTransition=CanTransitionToLoop(
     RelevantAnimTime="Get Relevant Anim Time (Solo [Start])"!() -> (ReturnValue:float) {is=title K2Node_AnimGetter}
     RelevantAnimTimeRemaining="Get Relevant Anim Time Remaining (Solo [Start])"!() -> (ReturnValue:float) {is=title K2Node_AnimGetter}
    ))]
 }
 Transition {
  path="AnimGraph.AnimGraphNode_StateMachine_5.Body.AnimStateNode_1.Play && Idle (Beat Sync).AnimGraphNode_StateMachine_0.Play && Idle (Beat Sync).AnimStateTransitionNode_76.Transition"
  nodes [Result(bCanEnterTransition=CanTransitionToLoop(
     RelevantAnimTime="Get Relevant Anim Time (IntensityTransition)"!() -> (ReturnValue:float) {is=title K2Node_AnimGetter}
     RelevantAnimTimeRemaining="Get Relevant Anim Time Remaining (IntensityTransition)"!() -> (ReturnValue:float) {is=title K2Node_AnimGetter}
    ))]
 }
 Fidget {
  path=AnimGraph.AnimGraphNode_StateMachine_5.Body.AnimStateNode_6.Fidget
  nodes [Output Animation Pose(Result=Sequence Player 'Bass_Fidget_3_A'!(
     Sequence:Anim Sequence Base Object Reference=/SparksCharacterFull/Animation/MainPlayer/BandPlay/Bass/Loops/Bass_Fidget_3_A
     bLoopAnimation:bool=false
    ) -> (Pose))]
 }
 Transition {path=AnimGraph.AnimGraphNode_StateMachine_5.Body.AnimStateTransitionNode_0.Transition, nodes [Result(bCanEnterTransition=true)]}
 Transition {
  path=AnimGraph.AnimGraphNode_StateMachine_5.Body.AnimStateTransitionNode_1.Transition
  nodes [Result(bCanEnterTransition=NOT Boolean(A=Get bMusicBeatIsActive()))]
 }
 Transition {
  path=AnimGraph.AnimGraphNode_StateMachine_5.Body.AnimStateTransitionNode_6.Transition
  nodes [Result(bCanEnterTransition=NOT Boolean(A=Get bMusicBeatIsActive()))]
 }
 Transition {
  path=AnimGraph.AnimGraphNode_StateMachine_5.Body.AnimStateTransitionNode_7.Transition
  nodes [Result(bCanEnterTransition=Get bMusicBeatIsActive())]
 }
 AnimNotify_PrintStringAssembler {nodes [
   n1 = AnimNotify_PrintStringAssembler!() -> (then, AnimNotifyEventName:String) {is=title K2Node_FunctionEntry}
   > Branch(Condition=Get DEBUG_PrintAnimLogic() -> (DEBUG_PrintAnimLogic:bool))
   n2 = Enum to String!(Enumerator:ESparksInstrumentType Enum=Get InstrumentType() -> (InstrumentType:ESparksInstrumentType Enum=Guitar)) -> (ReturnValue:String)
   > PrintString!(
    execute
    InString=JoinStringArray(SourceArray=Make Array!("[0]":String="[AnimNotify]", "[1]":String=@n2, "[2]":String=@n1) -> (Array:Array of Strings), Separator=" ")
    bPrintToScreen
    bPrintToLog
    TextColor=Get DEBUG_AnimNotify_TextColor() -> (DEBUG_AnimNotify_TextColor:Linear Color Structure)
    Duration
    Key
   ) -> (then)
   > Return Node()
  ]}
 CanTransitionToLoop {nodes [
   n1 = CanTransitionToLoop!() -> (then, RelevantAnimTime:double=~, RelevantAnimTimeRemaining:double=~) {is=title K2Node_FunctionEntry}
   n2 = "float > float"(A=@n1, B=0.2)
   n3 = Get TransitionBeat()
   > Return Node!(
    execute
    CanTransition:bool=OR Boolean(A=AND Boolean!(A=@n2, B=@n3) -> (ReturnValue), B="float <= float"(A=@n1.RelevantAnimTimeRemaining, B=0.4) -> (ReturnValue=~))
   )
  ]}
 CanTransitionToSubgraphEnd {nodes [
   n1 = CanTransitionToSubgraphEnd!() -> (then, CurrentSubgraph:AnimSubgraph Enum=NewEnumerator0) {is=title K2Node_FunctionEntry}
   n2 = Get CurrentAnimGraphLocation() -> (CurrentAnimGraphLocation:AnimSubgraph Enum=NewEnumerator0)
   > Return Node(CanTransition=AND Boolean!(
     A="Not Equal (Enum)"!(A:AnimSubgraph Enum=@n1, B:AnimSubgraph Enum=@n2) -> (ReturnValue:bool)
     B=Get AllowAnimTransition()
     C:bool=Get TransitionBeat()
    ) -> (ReturnValue))
  ]}
 CanTransitionToSubgraphStart {nodes [
   n1 = CanTransitionToSubgraphStart!() -> (then, DesiredSubgraph:AnimSubgraph Enum=NewEnumerator0) {is=title K2Node_FunctionEntry}
   n2 = Get CurrentAnimGraphLocation() -> (CurrentAnimGraphLocation=NewEnumerator0)
   > Return Node(CanTransition=AND Boolean(A="Equal (Enum)"(A:AnimSubgraph Enum=@n1, B:AnimSubgraph Enum=@n2), B=Get AllowAnimTransition(), C=Get TransitionBeat()))
  ]}
 ClearMidiListeners {nodes [
   ClearMidiListeners() {is=title K2Node_FunctionEntry}
   n1 = Get MidiListeners() -> (MidiListeners:Array of Async Action Listen For Midi Text Event Object References)
   > n2 = For Each Loop(Array:Array of Async Action Listen For Midi Text Event Object References=@n1) -> (Array Element:Async Action Listen For Midi Text Event Object Reference)
   Clear(execute=@n2.Completed, TargetArray:Array of Async Action Listen For Midi Text Event Object References=@n1)
   Is Valid(exec=@n2.LoopBody, InputObject=@n2)
   > Cancel(self:Cancellable Async Action Object Reference=@n2)
  ]}
 EventGraph {nodes [
   DEBUG_EVENT_TransitionToIdleVariant() {is=title K2Node_CustomEvent}
   > Branch(Condition=Get DEBUG_CALL_TransitionToIdleVariant() -> (DEBUG_CALL_TransitionToIdleVariant:bool))
   > TransitionToSubgraph!(execute, self, DesiredSubgraph:AnimSubgraph Enum=NewEnumerator1) -> (then)
   > Delay(Duration=3)
   > Set DEBUG_CALL_TransitionToIdleVariant(DEBUG_CALL_TransitionToIdleVariant:bool=false) -> (Output_Get:bool)
   DEBUG_EVENT_TransitionToHomebase() {is=title K2Node_CustomEvent}
   > Branch(Condition=Get DEBUG_CALL_TransitionToHomebase() -> (DEBUG_CALL_TransitionToHomebase:bool))
   > TransitionToSubgraph(DesiredSubgraph=NewEnumerator0)
   > Delay(Duration=3)
   > Set DEBUG_CALL_TransitionToHomebase(DEBUG_CALL_TransitionToHomebase:bool=false) -> (Output_Get:bool)
   Drive_PlayVariantJukebox() {is=title K2Node_CustomEvent}
   > n1 = Sequence()
   > n2 = QualifyForPlayVariant!(execute, self, WaitAfterLastTransition:double=13) -> (then, ReturnValue:bool)
   > Branch(Condition=@n2)
   > Do Once(Reset=@n3.then)
   > TransitionToSubgraph(DesiredSubgraph=NewEnumerator2)
   > n3 = Delay(Duration=3)
   n4 = QualifyForHomebase!(execute=@n1.then_1, self, WaitAfterLastTransition:double=18) -> (then, ReturnValue:bool)
   > Branch(Condition=@n4)
   > Do Once(Reset=@n5.then)
   > TransitionToSubgraph(DesiredSubgraph=NewEnumerator0)
   > n5 = Delay(Duration=3)
   Bass Strum() {is=title K2Node_CustomEvent}
   > Gate!(Enter:Exec, Open:Exec=@n9.then, Close:Exec=@n8.then_0, Toggle:Exec, bStartClosed:bool=false) -> (Exit:Exec) {is=title K2Node_MacroInstance}
   > n6 = Flip Flop!(None:Exec) -> (A:Exec, B:Exec, IsA:bool=~) {is=title K2Node_MacroInstance}
   > n7 = Montage_Play!(
    execute
    self:Anim Instance Object Reference
    MontageToPlay:Anim Montage Object Reference=/SparksCharacterCommon/Animation/MainPlayer/BandPlay/Bass/Strums/Bass_Strum_PluckThirdString_M
    InPlayRate:float=1
    ReturnValueType:EMontagePlayReturnType Enum=MontageLength
    InTimeToStartMontageAt:float=0
    bStopAllMontages:bool=true
   ) -> (then, ReturnValue:float)
   n8 = Sequence(execute=[@n10.then, @n7.then])
   n9 = Delay(execute=@n8.then_1, Duration=0.1)
   n10 = Montage_Play(
    execute=@n6.B
    MontageToPlay=/SparksCharacterCommon/Animation/MainPlayer/BandPlay/Bass/Strums/Bass_Strum_PluckFourthString_M
    InPlayRate=1
    ReturnValueType=MontageLength
    InTimeToStartMontageAt=0
    bStopAllMontages=true
   )
   Bass Change Chord() {is=title K2Node_CustomEvent}
   > Set ChordHandPoseIndex(ChordHandPoseIndex:int=RandomIntegerInRange!(Min:int=0, Max:int=4) -> (ReturnValue:int=0)) -> (Output_Get:int=0)
   DEBUG_EVENT_TransitionToPlayVariant() {is=title K2Node_CustomEvent}
   > Branch(Condition=Get DEBUG_CALL_TransitionToPlayVariant() -> (DEBUG_CALL_TransitionToPlayVariant:bool))
   > TransitionToSubgraph(DesiredSubgraph=NewEnumerator2)
   > Delay(Duration=3)
   > Set DEBUG_CALL_TransitionToPlayVariant(DEBUG_CALL_TransitionToPlayVariant:bool=false) -> (Output_Get:bool)
   DEBUG_EVENT_TransitionToSoloVariant() {is=title K2Node_CustomEvent}
   > Branch(Condition=Get DEBUG_CALL_TransitionToSoloVariant() -> (DEBUG_CALL_TransitionToSoloVariant:bool))
   > TransitionToSubgraph(DesiredSubgraph=NewEnumerator3)
   > Delay(Duration=3)
   > Set DEBUG_CALL_TransitionToSoloVariant(DEBUG_CALL_TransitionToSoloVariant:bool=false) -> (Output_Get:bool)
   DEBUG_EVENT_Strum() {is=title K2Node_CustomEvent}
   > Branch(Condition=Get DEBUG_CALL_Strum() -> (DEBUG_CALL_Strum:bool))
   > Bass Strum() {is=title K2Node_CallFunction}
   > Delay(Duration=1)
   > Set DEBUG_CALL_Strum(DEBUG_CALL_Strum:bool=false) -> (Output_Get:bool)
   Drive_IsIdleBeatSynced() {is=title K2Node_CustomEvent}
   > Set IsIdleBeatSynced(IsIdleBeatSynced:bool="Equal (Enum)"(A=Get CurrentAnimGraphLocation() -> (CurrentAnimGraphLocation=NewEnumerator0), B=NewEnumerator1)) -> (Output_Get:bool)
   Drive_TransitionTimers() {is=title K2Node_CustomEvent}
   > n11 = Sequence!(execute) -> (then_0, then_1, then_2:Exec, then_3:Exec)
   > ResetTimer_TimeSinceLastTransition()
   UpdateTimer_TimeSinceLastTransition(execute=@n11.then_1)
   ResetTimer_TimeSinceAllowAnimTransition!(execute=@n11.then_2, self, Seconds To Wait:double=6) -> (then)
   UpdateTimer_TimeSinceAllowAnimTransition(execute=@n11.then_3)
   Event On New Song() {is=title K2Node_Event}
   > SetSongDefaults!(
    execute
    AnimInstance:Sparks Instrument Anim Instance Object Reference=GetSelfAsSparksInstrumentAnimInstance!() -> (ReturnValue:Sparks Instrument Anim Instance Object Reference)
   ) -> (then)
   Event Sparks Debug Update() {is=title K2Node_Event}
   > n12 = Sequence!(execute) -> (then_0, then_1)
   > DEBUG_EVENT_Strum() {is=title K2Node_CallFunction}
   > DEBUG_EVENT_TransitionToHomebase() {is=title K2Node_CallFunction}
   > DEBUG_EVENT_TransitionToPlayVariant() {is=title K2Node_CallFunction}
   > DEBUG_EVENT_TransitionToSoloVariant() {is=title K2Node_CallFunction}
   > DEBUG_EVENT_TransitionToIdleVariant() {is=title K2Node_CallFunction}
   Branch(execute=@n12.then_1, Condition=Get DEBUG_PrintAnimLogic())
   n13 = Get TimeSinceLastTransition() -> (TimeSinceLastTransition:float)
   n14 = Get AllowAnimTransition()
   n15 = Get CurrentAnimGraphLocation() -> (CurrentAnimGraphLocation=NewEnumerator0)
   > PrintString(
    InString=Append!(
     A="[On Tick] Current AnimGraph Location: "
     B=Enum to String(Enumerator:AnimSubgraph Enum=@n15)
     C:String="\r\n[On Tick] Time Since Last Transition: "
     D:String="To String (Float)"(InDouble=@n13)
     E:String="\r\n[On Tick] Allow Anim Transition: "
     F:String="To String (Boolean)"(InBool=@n14)
     G:String="\r\n[On Tick] Current Song Section: "
     H:String=Get DEBUG_PrintCurrentSongSection() -> (DEBUG_PrintCurrentSongSection:String)
    ) -> (ReturnValue)
    TextColor=(R=0,G=1,B=0.502,A=1)
    Key=Print_TimeSinceLastTransition
   )
   n16 = Event On LipSync Asset Changed!() -> (OutputDelegate, then, NewLipSyncAsset:Anim Sequence Object Reference) {is=title K2Node_Event}
   > n17 = Is Valid(InputObject=@n16.NewLipSyncAsset)
   Set LipSyncAnimSequence(execute=@n17.Is Not Valid, LipSyncAnimSequence:Anim Sequence Object Reference) -> (Output_Get:Anim Sequence Object Reference)
   Set LipSyncAnimSequence(execute=@n17.Is Valid, LipSyncAnimSequence=@n16.NewLipSyncAsset)
   Event BlueprintPostEvaluateAnimation() {is=title K2Node_Event}
   > Set TransitionBeat(TransitionBeat:bool=false) -> (Output_Get:bool)
   n18 = Event On MIDI Event Data Parsed!() -> (OutputDelegate, then, EventDriver:"Midi Event Driver Interface Interface (by ref)") {is=title K2Node_Event}
   > Set MidiTrackName(MidiTrackName:Name=SelectName!(
     A:Name=GetMidiTrackNameFromSparksInstrument!(self:Midi Event Driver Interface Interface=@n18.EventDriver, InstrumentType:ESparksInstrumentType Enum=Bass) -> (ReturnValue:Name=None)
     B:Name=None
     bPickA:bool=IsValidInterface!(Interface:"Interface Interface (by ref)"=@n18.EventDriver) -> (ReturnValue:bool)
    ) -> (ReturnValue:Name=None)) -> (Output_Get:Name=None)
   > K2Node_Composite_3 = Midi Text Event Listeners!(In:Exec) -> (Out:Exec) {is=title K2Node_Composite}
   > K2Node_Composite_1 = Midi Note Event Fret Listener!(In:Exec) -> (Out:Exec) {is=title K2Node_Composite}
   > K2Node_Composite_2 = Midi Note Event Strum Listener!(In:Exec) -> (Out:Exec) {is=title K2Node_Composite}
   n19 = Event BlueprintUpdateAnimation()
   > Set DeltaTimeX(DeltaTimeX:float=@n19.DeltaTimeX) -> (Output_Get:float)
   > Drive_TransitionTimers() {is=title K2Node_CallFunction}
   > Drive_PlayVariantJukebox() {is=title K2Node_CallFunction}
   > Drive_IsIdleBeatSynced() {is=title K2Node_CallFunction}
   Event SparksCharAttackEventLaunched() {is=title K2Node_Event}
   > PrintString(InString=Play Attack Launched Anim Here.)
   n20 = TryGetPawnOwner()
   > n21 = Get Fort Pawn From Object!(execute, InObject:Object Reference=@n20) -> (Valid:Exec, NotValid:Exec, FortPawn:Fort Pawn Object Reference) {is=title K2Node_CallFunction}
   > n22 = FindGender!(execute, Object:Object Reference=@n21) -> (True:Exec, False:Exec, OutGender:EFortCustomGender Enum=Invalid)
   > n23 = Switch on EFortCustomGender!(execute, Selection:EFortCustomGender Enum=@n22) -> (Invalid:Exec, Male:Exec, Female:Exec, Both:Exec) {is=title K2Node_SwitchEnum}
   PlayAnimMontage!(
    execute=[@n23.Both, @n23.Male]
    Actor:Actor Object Reference=@n20
    NewAnimMontage:Anim Montage Object Reference=/SparksCharacterFull/Animation/MainPlayer/BandPlay/Guitar/Attacks/Guitar_Attack_Strum_A_M
    InPlayRate:float=1
    StartSectionName:Name=Default
   ) -> (then, ReturnValue:float)
   PlayAnimMontage(
    execute=@n23.Female
    Actor=@n20
    NewAnimMontage=&Guitar_Attack_Strum_A_M
    InPlayRate=1
    StartSectionName=Default
   )
   Event SparksCharAttackEventReceived() {is=title K2Node_Event}
   > PrintString(InString="Play Attack Landed (Reaction) Anim here.")
   n24 = Event On MIDI Event Driver Changed!() -> (OutputDelegate, then, NewMidiEventDriver:"Midi Event Driver Interface Interface (by ref)") {is=title K2Node_Event}
   > n25 = Branch(Condition=IsValidInterface(Interface=@n24.NewMidiEventDriver))
   ClearMidiListeners(execute=@n25.else) {is=title K2Node_CallFunction}
   Event BlueprintInitializeAnimation()
   > Parent: BlueprintInitializeAnimation() {is=title K2Node_CallParentFunction}
   > n26 = IntensityMidiTextEvents!(execute) -> (then, Events:Array of Strings)
   > Set BassIntensityMidiTextEvents(BassIntensityMidiTextEvents:Array of Strings=@n26) -> (Output_Get:Array of Strings)
   Event SparksCharAttackEventCharging() {is=title K2Node_Event}
   > PrintString(InString=Play Start Charging Attack Anim Here.)
   n27 = TryGetPawnOwner()
   > n28 = Get Fort Pawn From Object(InObject=@n27)
   > n29 = FindGender(Object=@n28) -> (OutGender=Invalid)
   > n30 = Switch on EFortCustomGender(Selection=@n29)
   n31 = PlayAnimMontage(
    execute=[@n30.Both, @n30.Male]
    Actor=@n27
    NewAnimMontage=/SparksCharacterFull/Animation/MainPlayer/BandPlay/Guitar/Attacks/Guitar_Charge_Strum_A_M
    InPlayRate=1
    StartSectionName=Intro
   )
   Set IsForceMediumIntensityOnEntry(execute=[@n31.then, @n32.then], IsForceMediumIntensityOnEntry:bool=true) -> (Output_Get:bool)
   n32 = PlayAnimMontage(
    execute=@n30.Female
    Actor=@n27
    NewAnimMontage=&Guitar_Charge_Strum_A_M
    InPlayRate=1
    StartSectionName=Intro
   )
   AnimNotify_BeatExit() {is=title K2Node_Event}
   > n33 = Set TransitionBeat(TransitionBeat=true)
   Reroute Node()
   n34 = Get DEBUG_PrintAnimLogic()
   > PrintString(
    InString=Append!(A="[AnimNotify] Bass_AnimBP Transition beat _", B="To String (Boolean)"(InBool=@n33)) -> (ReturnValue)
    bPrintToScreen=@n34
    bPrintToLog=@n34
    TextColor=(R=0.1379,G=0.6405,B=1,A=1)
   )
   Montage_Play(
    MontageToPlay=/SparksCharacterCommon/Animation/MainPlayer/BandPlay/Bass/Strums/Bass_Strum_PluckTwoFinger_M
    InPlayRate=1
    ReturnValueType=MontageLength
    InTimeToStartMontageAt=0
    bStopAllMontages=true
   )
  ]}
 Midi Note Event Fret Listener {
  path=EventGraph.K2Node_Composite_1.Midi Note Event Fret Listener
  nodes [
   Inputs!() -> (In:Exec)
   > n1 = Sequence!(execute) -> (then_0, then_1, then_2:Exec)
   n2 = Get MidiFretListener() -> (MidiFretListener:Async Action Listen For Midi Note Event Object Reference)
   > Is Valid(InputObject=@n2)
   > Cancel(self=@n2)
   > Set MidiFretListener(MidiFretListener:Async Action Listen For Midi Note Event Object Reference) -> (Output_Get:Async Action Listen For Midi Note Event Object Reference)
   Branch(execute=@n1.then_1, Condition=IsValidInterface(Interface=Get MidiEventDriver() -> (MidiEventDriver:Midi Event Driver Interface Interface)))
   > n3 = ListenForMidiNoteEvents!(
    execute:Exec
    MidiEventDriver:Midi Event Driver Interface Interface=Get MidiEventDriver()
    MidiTrackName:Name=Get MidiTrackName() -> (MidiTrackName:Name=None)
    EventNotes:Array of Bytes=GetBassMidiNoteIntegersFromEnums!(MidiNoteEvents:Array of EBassMidiNoteEvent Enums=Get MidiFretEvents() -> (MidiFretEvents:Array of EBassMidiNoteEvent Enums=None)) -> (ReturnValue:Array of Bytes)
    SecondsBeforeEventStart:float=0
    SecondsBeforeEventEnd:float=0
    InitBroadcast:EMidiEventListenerInitBroadcast Enum=MostRecentEvent
   ) -> (
    then:Exec
    AsyncTaskProxy:Async Action Listen For Midi Note Event Object Reference
    OnMidiEventStart:Exec
    OnMidiEventEnd:Exec
    OnFinished:Exec
    MidiEventNote:Byte
    MidiEventData:"Parsed Midi Note Event Structure (by ref)"
   ) {is=title K2Node_AsyncAction}
   > Set MidiFretListener(MidiFretListener=@n3)
   n4 = Set FretNumber(execute=@n3.OnMidiEventStart, FretNumber:int=GetFretNumberFromMidiNote!(MidiNote:Byte=@n3.MidiEventNote) -> (ReturnValue:int=0)) -> (Output_Get:int=0)
   > Branch(Condition="integer < integer"(A=@n4, B=7))
   > Set FretNumber(FretNumber=int + int(A=Get FretNumber() -> (FretNumber:int=0), B=6)) -> (Output_Get=0)
   Outputs!(Out:Exec=@n1.then_2)
  ]
 }
 Midi Note Event Strum Listener {
  path=EventGraph.K2Node_Composite_2.Midi Note Event Strum Listener
  nodes [
   Inputs()
   > n1 = Sequence()
   n2 = Get MidiStrumListener() -> (MidiStrumListener:Async Action Listen For Midi Note Event Object Reference)
   > Is Valid(InputObject=@n2)
   > Cancel(self=@n2)
   > Set MidiStrumListener(MidiStrumListener:Async Action Listen For Midi Note Event Object Reference) -> (Output_Get:Async Action Listen For Midi Note Event Object Reference)
   Branch(execute=@n1.then_1, Condition=IsValidInterface(Interface=Get MidiEventDriver()))
   > n3 = ListenForMidiNoteEvents(
    MidiEventDriver=Get MidiEventDriver()
    MidiTrackName=Get MidiTrackName() -> (MidiTrackName=None)
    EventNotes=Make Array!(
     "[0]":Byte=GetBassMidiNoteIntegerFromEnum!(MidiNoteEvent:EBassMidiNoteEvent Enum=StrumDown) -> (ReturnValue:Byte=0)
     "[1]":Byte=GetBassMidiNoteIntegerFromEnum(MidiNoteEvent=StrumUp) -> (ReturnValue=0)
    ) -> (Array:Array of Bytes)
    SecondsBeforeEventStart=0.06
    SecondsBeforeEventEnd=0
    InitBroadcast=NoPreviousEvents
   )
   > Set MidiStrumListener(MidiStrumListener=@n3)
   Bass Strum(execute=@n3.OnMidiEventStart)
   Outputs(Out=@n1.then_2)
  ]
 }
 Midi Text Event Listeners {
  path=EventGraph.K2Node_Composite_3.Midi Text Event Listeners
  nodes [
   Inputs()
   > n1 = Sequence()
   > ClearMidiListeners()
   Branch(execute=@n1.then_1, Condition=IsValidInterface(Interface=Get MidiEventDriver()))
   > n2 = Sequence!(execute) -> (then_0, then_1, then_2, then_3:Exec)
   > n3 = Sequence!(execute) -> (then_0, then_1)
   n4 = ListenForMidiTextEvents!(
    execute:Exec=@n3.then_1
    MidiEventDriver:Midi Event Driver Interface Interface=Get MidiEventDriver()
    MidiTrackName:Name=SECTION
    EventStrings:Array of Strings=Get SongSectionMidiTextEvents() -> (SongSectionMidiTextEvents:Array of Strings)
    SecondsBeforeEventStart:float=Get SecBefore_IdleIntensity_PreventEnterPlayVariant() -> (SecBefore_IdleIntensity_PreventEnterPlayVariant:float)
    InitBroadcast:EMidiEventListenerInitBroadcast Enum=MostRecentEvent
   ) -> (
    then:Exec
    AsyncTaskProxy:Async Action Listen For Midi Text Event Object Reference
    OnMidiEventStart:Exec
    OnFinished:Exec
    MidiEventText:String
    MidiEventData:"Parsed Midi Text Event Structure (by ref)"
   ) {is=title K2Node_AsyncAction}
   > Add(
    TargetArray:Array of Async Action Listen For Midi Text Event Object References=Get MidiListeners()
    NewItem:"Async Action Listen For Midi Text Event Object Reference (by ref)"=@n4
   ) {is=title K2Node_CallArrayFunction}
   n5 = ListenForMidiTextEvents(
    execute=@n3.then_0
    MidiEventDriver=Get MidiEventDriver()
    MidiTrackName=Get MidiTrackName() -> (MidiTrackName=None)
    EventStrings=Get BassIntensityMidiTextEvents() -> (BassIntensityMidiTextEvents:Array of Strings)
    SecondsBeforeEventStart=Get SecBefore_IdleIntensity_PreventEnterPlayVariant()
    InitBroadcast=MostRecentEvent
   )
   > Add(TargetArray=Get MidiListeners(), NewItem=@n5)
   MidiEvent_Intensity_PreventEnterPlayVariant!(execute=@n5.OnMidiEventStart, self, MidiEventText:String=@n5.MidiEventText) -> (then)
   Sequence(execute=@n2.then_1)
   > n6 = ListenForMidiTextEvents(
    MidiEventDriver=Get MidiEventDriver()
    MidiTrackName=Get MidiTrackName() -> (MidiTrackName=None)
    EventStrings=Get BassIntensityMidiTextEvents()
    SecondsBeforeEventStart=Get SecBefore_PlayIntensity_GoToHomebase() -> (SecBefore_PlayIntensity_GoToHomebase:float)
    InitBroadcast=MostRecentEvent
   )
   > Add(TargetArray=Get MidiListeners(), NewItem=@n6)
   MidiEvent_Intensity_GoToHomebase!(execute=@n6.OnMidiEventStart, self, MidiEventText:String=@n6.MidiEventText) -> (then)
   n7 = Sequence(execute=@n2.then_2)
   n8 = ListenForMidiTextEvents(
    execute=@n7.then_1
    MidiEventDriver=Get MidiEventDriver()
    MidiTrackName=SECTION
    EventStrings=Get SongSectionMidiTextEvents()
    SecondsBeforeEventStart=Get SecBefore_PlayIntensity_StartTransition() -> (SecBefore_PlayIntensity_StartTransition:float)
    InitBroadcast=MostRecentEvent
   )
   > Add(TargetArray=Get MidiListeners(), NewItem=@n8)
   n9 = ListenForMidiTextEvents(
    execute=@n7.then_0
    MidiEventDriver=Get MidiEventDriver()
    MidiTrackName=Get MidiTrackName() -> (MidiTrackName=None)
    EventStrings=Get BassIntensityMidiTextEvents()
    SecondsBeforeEventStart=Get SecBefore_PlayIntensity_StartTransition()
    InitBroadcast=MostRecentEvent
   )
   > Add(TargetArray=Get MidiListeners(), NewItem=@n9)
   MidiEvent_Intensity_StartTransition!(execute=@n9.OnMidiEventStart, self, MidiEventText:String=@n9.MidiEventText) -> (then)
   n10 = Sequence!(execute=@n2.then_3) -> (then_0, then_1, then_2:Exec)
   n11 = ListenForMidiTextEvents(
    execute=@n10.then_2
    MidiEventDriver=Get MidiEventDriver()
    MidiTrackName=SECTION
    EventStrings=Get SongSectionMidiTextEvents()
    SecondsBeforeEventStart=0
    InitBroadcast=MostRecentEvent
   )
   > Add(TargetArray=Get MidiListeners(), NewItem=@n11)
   Branch(execute=@n11.OnMidiEventStart, Condition=Get DEBUG_PrintAnimLogic())
   > PrintString(
    InString=Append!(A="\r\nMidi Track Name: ", B=SECTION, C:String="\r\nMidi Text Event: ", D:String=@n11.MidiEventText) -> (ReturnValue)
    TextColor=(R=1,G=0.6667,B=0.098,A=1)
    Duration=5
    Key=Print_SongSection_OnEvent
   )
   n12 = ListenForMidiTextEvents(
    execute=@n10.then_0
    MidiEventDriver=Get MidiEventDriver()
    MidiTrackName=Get MidiTrackName() -> (MidiTrackName=None)
    EventStrings=Get BassIntensityMidiTextEvents()
    SecondsBeforeEventStart=0
    InitBroadcast=MostRecentEvent
   )
   > Add(TargetArray=Get MidiListeners(), NewItem=@n12)
   MidiEvent_Intensity_OnEvent!(execute=@n12.OnMidiEventStart, self, MidiEventText:String=@n12.MidiEventText) -> (then)
   Outputs(Out=@n1.then_2)
   n13 = Get SongIntensity()
   PrintString(
    InString=Append(
     A="\r\nMidi Text Event: "
     B=@n12.MidiEventText
     C="\r\nSong Intensity: "
     D=GetDebugStringFromGameplayTag!(GameplayTag:Gameplay Tag Structure=@n13) -> (ReturnValue:String)
    )
    TextColor=(R=0,G=1,B=0,A=1)
    Duration=5
    Key=Print_OnEvent
   )
  ]
 }
 MidiEvent_Intensity_GoToHomebase {nodes [
   n1 = MidiEvent_Intensity_GoToHomebase!() -> (then, MidiEventText:String) {is=title K2Node_FunctionEntry}
   n2 = Sequence!(execute=[@n1.then, @n9.else]) -> (then_0, then_1)
   > n3 = TranslateMidiEventTextToSongIntensity!(execute, MidiEventText:Name=String To Name(InString=@n1)) -> (then, IdleRealtime:bool, MidiSongIntensityIsIdle:bool, SongIntensity:Gameplay Tag Structure)
   > n4 = Branch(Condition=@n3)
   n5 = Sequence(execute=@n4.else)
   > n6 = Branch(Condition=@n3.MidiSongIntensityIsIdle)
   n7 = "Not Equal (Enum)"(A=Get CurrentAnimGraphLocation() -> (CurrentAnimGraphLocation=NewEnumerator0), B=NewEnumerator0)
   n8 = "Not Equal (GameplayTag)"(A=@n3.SongIntensity, B=Get SongIntensity()) -> (ReturnValue=~)
   > Branch(Condition=OR Boolean(A=@n8, B=@n7))
   > Set SongIntensity(SongIntensity:Gameplay Tag Structure=@n3.SongIntensity) -> (Output_Get:Gameplay Tag Structure)
   > TransitionToSubgraph(DesiredSubgraph=NewEnumerator0)
   Branch(execute=@n6.else, Condition=AND Boolean!(A=@n8, B=@n7) -> (ReturnValue))
   > TransitionToSubgraph(DesiredSubgraph=NewEnumerator0)
   Set IntensityTransitionInProgress(execute=@n5.then_1, IntensityTransitionInProgress:bool=true) -> (Output_Get:bool)
   Branch(execute=@n2.then_1, Condition=Get DEBUG_PrintAnimLogic())
   > MidiEvent_PrintStringAssembler!(
    execute
    self
    InstrumentType:ESparksInstrumentType Enum=Get InstrumentType() -> (InstrumentType=Guitar)
    MidiEventText:String=@n1
    SecondsBeforeEvent:double=SelectFloat(
     A=Get SecBefore_IdleIntensity_GoToHomebase() -> (SecBefore_IdleIntensity_GoToHomebase:float)
     B=Get SecBefore_PlayIntensity_GoToHomebase()
     bPickA=@n3.MidiSongIntensityIsIdle
    )
   ) -> (then)
   n9 = Branch(Condition=Get IntensityTransitionInProgress() -> (IntensityTransitionInProgress:bool))
   n10 = Get DEBUG_PrintAnimLogic()
   > PrintString(
    InString=Append!(A="Transition In Progress... Blocking Midi Event: ", B=@n1) -> (ReturnValue)
    bPrintToScreen=@n10
    bPrintToLog=@n10
    TextColor=Get DEBUG_AnimNotify_TextColor()
    Duration=5
   )
  ]}
 MidiEvent_Intensity_OnEvent {nodes [
   n1 = MidiEvent_Intensity_OnEvent!() -> (then, MidiEventText:String) {is=title K2Node_FunctionEntry}
   > n2 = Sequence()
   > n3 = TranslateMidiEventTextToSongIntensity(MidiEventText=String To Name(InString=@n1))
   > n4 = Set IdleRealtime(IdleRealtime:bool=@n3) -> (Output_Get:bool)
   > n5 = Branch(Condition=@n4)
   Set CurrentAnimGraphLocation(
    execute=@n5.else
    CurrentAnimGraphLocation:AnimSubgraph Enum=Select(Option 0:AnimSubgraph Enum=NewEnumerator0, Option 1:AnimSubgraph Enum=NewEnumerator1, Index=@n3.MidiSongIntensityIsIdle) -> (ReturnValue:AnimSubgraph Enum)
   ) -> (Output_Get:AnimSubgraph Enum=NewEnumerator0)
   > Set IntensityTransitionInProgress(IntensityTransitionInProgress=false)
   > Set PreventEnterPlayVariant(PreventEnterPlayVariant:bool=false) -> (Output_Get:bool)
   MidiEvent_PrintStringAssembler(execute=@n2.then_1, InstrumentType=Get InstrumentType() -> (InstrumentType=Guitar), MidiEventText=@n1, SecondsBeforeEvent=0)
  ]}
 MidiEvent_Intensity_PreventEnterPlayVariant {nodes [
   n1 = MidiEvent_Intensity_PreventEnterPlayVariant!() -> (then, MidiEventText:String) {is=title K2Node_FunctionEntry}
   > n2 = Sequence!(execute) -> (then_0, then_1, then_2:Exec)
   > n3 = TranslateMidiEventTextToSongIntensity(MidiEventText=String To Name(InString=@n1))
   > n4 = Branch(Condition=@n3)
   Set PreventEnterPlayVariant(execute=@n4.else, PreventEnterPlayVariant=true)
   Branch(execute=@n2.then_1, Condition=Get DEBUG_PrintAnimLogic())
   > PrintString(InString="...enabling `PreventEnterPlayVariant`.", TextColor=Get DEBUG_AnimNotify_TextColor(), Duration=4)
   MidiEvent_PrintStringAssembler(execute=@n2.then_2, InstrumentType=Get InstrumentType() -> (InstrumentType=Guitar), MidiEventText=@n1, SecondsBeforeEvent=0)
  ]}
 MidiEvent_Intensity_StartTransition {nodes [
   n1 = MidiEvent_Intensity_StartTransition!() -> (then, MidiEventText:String) {is=title K2Node_FunctionEntry}
   > n2 = Sequence()
   > Branch(Condition=Get DEBUG_PrintAnimLogic())
   > Branch(Condition=Get IntensityTransitionInProgress())
   n3 = TranslateMidiEventTextToSongIntensity(execute=@n2.then_1, MidiEventText=String To Name(InString=@n1))
   > n4 = Branch(Condition=@n3)
   Set SongIntensity(execute=@n4.else, SongIntensity=@n3.SongIntensity)
   > TransitionToSubgraph(DesiredSubgraph=Select(Option 0=NewEnumerator0, Option 1=NewEnumerator1, Index=@n3.MidiSongIntensityIsIdle))
   > Set IntensityTransitionInProgress(IntensityTransitionInProgress=true)
   Branch(execute=@n2.then_2, Condition=Get DEBUG_PrintAnimLogic())
   > MidiEvent_PrintStringAssembler(
    InstrumentType=Get InstrumentType() -> (InstrumentType=Guitar)
    MidiEventText=@n1
    SecondsBeforeEvent=SelectFloat(
     A=Get SecBefore_IdleIntensity_StartTransition() -> (SecBefore_IdleIntensity_StartTransition:float)
     B=Get SecBefore_PlayIntensity_StartTransition()
     bPickA=@n3.MidiSongIntensityIsIdle
    )
   )
   PrintString(InString=Append(A="Transition In Progress... Blocking Midi Event: ", B=@n1), TextColor=Get DEBUG_AnimNotify_TextColor(), Duration=5)
  ]}
 MidiEvent_PrintStringAssembler {nodes [
   n1 = MidiEvent_PrintStringAssembler!() -> (then, InstrumentType:ESparksInstrumentType Enum=Guitar, MidiEventText:String, SecondsBeforeEvent:double=~) {is=title K2Node_FunctionEntry}
   > Branch(Condition=Get DEBUG_PrintAnimLogic())
   > n2 = Branch(Condition="Equal (Float)"(A=@n1.SecondsBeforeEvent))
   > n3 = Set Duration(Duration:float=2) -> (Output_Get:float)
   n4 = "To String (Float)"(InDouble=@n1.SecondsBeforeEvent)
   n5 = Enum to String(Enumerator:ESparksInstrumentType Enum=@n1)
   PrintString(
    execute=[@n3.then, @n6.then]
    InString=JoinStringArray(
     SourceArray=Make Array!("[0]":String="[Anim Midi]", "[1]":String=@n5, "[2]":String=@n1.MidiEventText, "[3]":String=in, "[4]":String=@n4, "[5]":String=seconds...) -> (Array:Array of Strings)
     Separator=" "
    )
    TextColor=Get DEBUG_AnimNotify_TextColor()
    Duration=Get Duration() -> (Duration:float)
   )
   > Return Node!(execute)
   n6 = Set Duration(execute=@n2.else, Duration=@n1.SecondsBeforeEvent)
  ]}
 OnHomeBaseEntry {nodes [
   OnHomeBaseEntry!() -> (then, Context:"Anim Update Context Structure (by ref)", Node:"Anim Node Reference Structure (by ref)") {is=title K2Node_FunctionEntry}
   > n1 = Sequence!(execute) -> (then_0, then_1)
   > Branch(Condition=Get IsForceMediumIntensityOnEntry() -> (IsForceMediumIntensityOnEntry:bool))
   > Set SongIntensity(SongIntensity=(TagName="Sparks.Anim.Intensity.Medium"))
   > Set IsForceMediumIntensityOnEntry(IsForceMediumIntensityOnEntry=false)
   n2 = PickIntensityAnimLoopCycle!(
    execute=@n1.then_1
    self:Sparks Instrument Anim Instance Object Reference
    BPMSet:"Sparks Anim BPMSet Structure (by ref)"=Get HomebaseBPMs() -> (HomebaseBPMs:Sparks Anim BPMSet Structure)
    TransitionTag:Gameplay Tag Structure
    ForceVariant:bool=false
    VariantIndex:int=0
   ) -> (then, ReturnValue:"Sparks Anim Loop Cycle Structure (by ref)")
   > Set HomebaseAnim(HomebaseAnim:Anim Sequence Object Reference=Break Sparks Anim Loop Cycle!(SparksAnimLoopCycle:"Sparks Anim Loop Cycle Structure (by ref)"=@n2) -> (Loop:Anim Sequence Object Reference, Entry:Anim Sequence Object Reference, Exit:Anim Sequence Object Reference) {is=title K2Node_BreakStruct}) -> (Output_Get:Anim Sequence Object Reference)
   > Return Node()
  ]}
 OnIdleEntry {nodes [
   OnIdleEntry!() -> (then, Context:"Anim Update Context Structure (by ref)", Node:"Anim Node Reference Structure (by ref)") {is=title K2Node_FunctionEntry}
   > n1 = PickIntensityAnimLoopCycle(BPMSet=Get IdleBPMs() -> (IdleBPMs:Sparks Anim BPMSet Structure), ForceVariant=false, VariantIndex=0)
   > Set IdleAnim(IdleAnim:Sparks Anim Loop Cycle Structure=@n1) -> (Output_Get:Sparks Anim Loop Cycle Structure)
   > Return Node()
  ]}
 OnIntensityTransitionEntry {nodes [
   OnIntensityTransitionEntry!() -> (then, Context:"Anim Update Context Structure (by ref)", Node:"Anim Node Reference Structure (by ref)") {is=title K2Node_FunctionEntry}
   n1 = Get DesiredSongIntensity()
   > n2 = PickIntensityAnimLoopCycle(BPMSet=Get TransitionBPMs() -> (TransitionBPMs:Sparks Anim BPMSet Structure), TransitionTag=@n1, ForceVariant=false, VariantIndex=0)
   > Set TransitionAnim(TransitionAnim:Anim Sequence Object Reference=Break Sparks Anim Loop Cycle(SparksAnimLoopCycle=@n2)) -> (Output_Get:Anim Sequence Object Reference)
   > Set SongIntensity(SongIntensity=@n1)
   > Return Node()
  ]}
 OnPlayEntry {nodes [
   OnPlayEntry!() -> (then, Context:"Anim Update Context Structure (by ref)", Node:"Anim Node Reference Structure (by ref)") {is=title K2Node_FunctionEntry}
   > n1 = PickIntensityAnimLoopCycle(
    BPMSet=Get PlayBPMs() -> (PlayBPMs:Sparks Anim BPMSet Structure)
    ForceVariant=Get DEBUG_DisablePlayVariantLoopJukebox() -> (DEBUG_DisablePlayVariantLoopJukebox:bool)
    VariantIndex=Get DEBUG_PlayVariantLoopOverride() -> (DEBUG_PlayVariantLoopOverride:int=0)
   )
   > Set PlayAnim(PlayAnim:Sparks Anim Loop Cycle Structure=@n1) -> (Output_Get:Sparks Anim Loop Cycle Structure)
   > Return Node()
  ]}
 OnSoloEntry {nodes [
   OnSoloEntry!() -> (then, Context:"Anim Update Context Structure (by ref)", Node:"Anim Node Reference Structure (by ref)") {is=title K2Node_FunctionEntry}
   > n1 = PickIntensityAnimLoopCycle(BPMSet=Get SoloBPMs() -> (SoloBPMs:Sparks Anim BPMSet Structure), ForceVariant=false, VariantIndex=0)
   > Set SoloAnim(SoloAnim:Sparks Anim Loop Cycle Structure=@n1) -> (Output_Get:Sparks Anim Loop Cycle Structure)
   > Return Node()
  ]}
 QualifyForHomebase {nodes [
   n1 = QualifyForHomebase!() -> (then, WaitAfterLastTransition:double=~) {is=title K2Node_FunctionEntry}
   n2 = Get IntensityTransitionInProgress()
   n3 = Get CurrentAnimGraphLocation() -> (CurrentAnimGraphLocation=NewEnumerator0)
   n4 = Get TimeSinceLastTransition()
   > Return Node!(
    execute
    ReturnValue:bool=AND Boolean!(A="float >= float"(A=@n4, B=@n1), B=NOT Boolean(A=@n2), C:bool="Equal (Enum)"(A=@n3, B=NewEnumerator2)) -> (ReturnValue)
   )
  ]}
 QualifyForPlayVariant {nodes [
   n1 = QualifyForPlayVariant!() -> (then, WaitAfterLastTransition:double=~) {is=title K2Node_FunctionEntry}
   n2 = Get PreventEnterPlayVariant() -> (PreventEnterPlayVariant:bool)
   n3 = Get IntensityTransitionInProgress()
   n4 = Get CurrentAnimGraphLocation() -> (CurrentAnimGraphLocation=NewEnumerator0)
   n5 = Get TimeSinceLastTransition()
   > Return Node(ReturnValue=AND Boolean!(A="float >= float"(A=@n5, B=@n1), B=NOT Boolean(A=@n3), C=NOT Boolean(A=@n2), D:bool="Equal (Enum)"(A=@n4, B=NewEnumerator0)) -> (ReturnValue))
  ]}
 ResetTimer_TimeSinceAllowAnimTransition {nodes [
   n1 = ResetTimer_TimeSinceAllowAnimTransition!() -> (then, Seconds To Wait:double=~) {is=title K2Node_FunctionEntry}
   > Branch(Condition="float > float"(A=Get TimeSinceAllowAnimTransitionFalse() -> (TimeSinceAllowAnimTransitionFalse:float), B=@n1))
   > n2 = Branch(Condition=Get DEBUG_PrintAnimLogic())
   n3 = Get TimeSinceAllowAnimTransitionFalse()
   > n4 = PrintString(InString=Append!(A="AllowAnimTransition set to False\r\nafter ", B="To String (Float)"(InDouble=@n3), C:String=" seconds.") -> (ReturnValue))
   Set TimeSinceLastTransition(execute=[@n4.then, @n2.else], TimeSinceLastTransition:float=0) -> (Output_Get:float)
   > Set AllowAnimTransition(AllowAnimTransition:bool=false) -> (Output_Get:bool)
  ]}
 ResetTimer_TimeSinceLastTransition {nodes [
   ResetTimer_TimeSinceLastTransition() {is=title K2Node_FunctionEntry}
   > n1 = Branch(Condition=AND Boolean!(A=Get AllowAnimTransition(), B=Get TransitionBeat()) -> (ReturnValue))
   > Do Once(Reset=@n1.else, Start Closed=true)
   > Set TimeSinceLastTransition(TimeSinceLastTransition=0)
  ]}
 TransitionToSubgraph {nodes [
   n1 = TransitionToSubgraph!() -> (then, DesiredSubgraph:AnimSubgraph Enum=NewEnumerator0) {is=title K2Node_FunctionEntry}
   > Set CurrentAnimGraphLocation(CurrentAnimGraphLocation=@n1) -> (Output_Get=NewEnumerator0)
   > Set TimeSinceLastTransition(TimeSinceLastTransition=0)
   > Set TimeSinceAllowAnimTransitionFalse(TimeSinceAllowAnimTransitionFalse:float=0) -> (Output_Get:float)
   > Set AllowAnimTransition(AllowAnimTransition=true)
  ]}
 UpdateTimer_TimeSinceAllowAnimTransition {nodes [
   UpdateTimer_TimeSinceAllowAnimTransition() {is=title K2Node_FunctionEntry}
   > n1 = Branch(Condition=Get AllowAnimTransition())
   > Set TimeSinceAllowAnimTransitionFalse(TimeSinceAllowAnimTransitionFalse=float + float(A=Get DeltaTimeX() -> (DeltaTimeX:float), B=Get TimeSinceAllowAnimTransitionFalse()))
   Set TimeSinceAllowAnimTransitionFalse(execute=@n1.else, TimeSinceAllowAnimTransitionFalse=0)
  ]}
 UpdateTimer_TimeSinceLastTransition {nodes [
   UpdateTimer_TimeSinceLastTransition() {is=title K2Node_FunctionEntry}
   > Set TimeSinceLastTransition(TimeSinceLastTransition=float + float(A=Get DeltaTimeX(), B=Get TimeSinceLastTransition()))
  ]}
]
variables [
 AllowAnimTransition {type=boolean}
 BassIntensityMidiTextEvents {type="array<string>", default []}
 BPM Bucket {type=integer}
 BPM Preview {type=number, default=125}
 ChordHandPoseIndex {type=integer}
 CurrentAnimGraphLocation {type=AnimSubgraph, default=Homebase}
 DEBUG_AllowFeature_3200 {type=boolean}
 DEBUG_AllowFeature_Future {type=boolean}
 DEBUG_AlwaysFalse {type=boolean}
 DEBUG_AnimNotify_TextColor {type=LinearColor, default=<rgba 1 0 1 1>}
 DEBUG_BodyNoMotion {type=boolean}
 DEBUG_CALL_Strum {type=boolean}
 DEBUG_CALL_TransitionToHomebase {type=boolean}
 DEBUG_CALL_TransitionToIdleVariant {type=boolean}
 DEBUG_CALL_TransitionToPlayVariant {type=boolean}
 DEBUG_CALL_TransitionToSoloVariant {type=boolean}
 DEBUG_DisablePlayVariantLoopJukebox {type=boolean}
 DEBUG_PlayVariantLoopOverride {type=integer}
 DEBUG_PrintAnimLogic {type=boolean}
 DEBUG_PrintCurrentSongSection {type=string, default=""}
 DeltaTimeX {type=number}
 HomebaseAnim {type=AnimSequence, default=null}
 IdleAnim {type=SparksAnimLoopCycle, default {entry=null, exit=null, loop=null}}
 IdleRealtime {type=boolean, default=true}
 InstrumentType {type=ESparksInstrumentType, default=Bass}
 IntensityTransitionInProgress {type=boolean}
 IsForceMediumIntensityOnEntry {type=boolean}
 IsIdleBeatSynced {type=boolean}
 LipSyncAnimSequence {type=AnimSequence, default=null}
 MidiFretEvents {
  type="array<EBassMidiNoteEvent>"
  default [
   FretPosition1
   FretPosition2
   FretPosition3
   FretPosition4
   FretPosition5
   FretPosition6
   FretPosition7
   FretPosition8
   FretPosition9
   FretPosition10
   FretPosition11
   FretPosition12
   FretPosition13
   FretPosition14
   FretPosition15
   FretPosition16
   FretPosition17
   FretPosition18
   FretPosition19
   FretPosition20
  ]
 }
 MidiFretListener {type=AsyncAction_ListenForMidiNoteEvent, default=null}
 MidiListeners {type="array<AsyncAction_ListenForMidiTextEvent>", default []}
 MidiStrumListener {type=AsyncAction_ListenForMidiNoteEvent, default=null}
 MidiTrackName {type=string, default=None}
 PlayAnim {type=SparksAnimLoopCycle, default {entry=null, exit=null, loop=null}}
 PreventEnterPlayVariant {type=boolean, default=true}
 SecBefore_IdleIntensity_GoToHomebase {type=number, default=8}
 SecBefore_IdleIntensity_PreventEnterPlayVariant {type=number, default=12}
 SecBefore_IdleIntensity_StartTransition {type=number, default=3.5}
 SecBefore_PlayIntensity_GoToHomebase {type=number, default=8}
 SecBefore_PlayIntensity_StartTransition {type=number, default=3.5}
 SoloAnim {type=SparksAnimLoopCycle, default {entry=null, exit=null, loop=null}}
 SongSectionMidiTextEvents {
  type="array<string>"
  default [music_start,music_end,intro,verse,prechorus,chorus,bridge,breakdown,build,drop,outro,solo_guitar,solo_bass,solo_vocals,solo_drums,solo_keys]
 }
 StrumCorrectivePose {type=AnimSequenceBase, default=null}
 StrumStartPose {type=AnimSequenceBase, default=null}
 TimeSinceAllowAnimTransitionFalse {type=number}
 TimeSinceLastTransition {type=number}
 TransitionAnim {type=AnimSequence, default=null}
 TransitionBeat {type=boolean}
]
