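# Animation Blueprint for a Sparks band-play drummer, playing MIDI-driven hits or falling back to play, idle and fidget loops.
# Layered bone blends combine a spine-twisted body, additive arm-hit state machines and separate leg state machines.
# Arm machines alternate A/B hit states so strikes can crossfade, with nested choosers per drum piece.
# Visible left-arm choosers pick short, medium or long swings by delay flags, offsetting start position by the shortened delay.
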AnimBlueprint /SparksCharacterFull/Animation/MainPlayer/BandPlay/Sparks_Drum_AnimBP
target_skeleton=&Fortnite_M_Avg_Player_Skeleton
graphs [
 AnimGraph {nodes [
   Apply Additive(
    Base=Sequence Player 'Drum_SpineTwist_Base_Pose'() {sequence=/SparksCharacterCommon/Animation/MainPlayer/BandPlay/Drum/SpineTwist/Drum_SpineTwist_Base_Pose}
    Additive=Use cached pose 'FullBodyLeftLegHitADD'()
   ) {alphaBoolBlend {blendOption=Linear}}
   Apply Additive(
    Base=Sequence Player 'Drum_SpineTwist_Base_Pose'() {sequence=&Drum_SpineTwist_Base_Pose}
    Additive=Use cached pose 'FullBodyRightLegHitADD'()
   ) {alphaBoolBlend {blendOption=Linear}}
   AnimGraphNode_StateMachine_6 = RightArmAdditiveHits() {is=title AnimGraphNode_StateMachine, bReinitializeOnBecomingRelevant=false}
   AnimGraphNode_StateMachine_1 = LeftArmAdditiveHits() {is=title AnimGraphNode_StateMachine, bReinitializeOnBecomingRelevant=false}
   n1 = Layered blend per bone!(
    BasePose="Use cached pose 'Body (Cache Pose)'"()
    BlendPoses_0=Apply Additive(
     Base=Sequence Player 'Drum_SpineTwist_Base_Pose'() {sequence=&Drum_SpineTwist_Base_Pose}
     Additive=@AnimGraphNode_StateMachine_1
    ) {alphaBoolBlend {blendOption=Linear}}
    BlendPoses_1:Pose Link Structure=Apply Additive(
     Base=Sequence Player 'Drum_SpineTwist_Base_Pose'() {sequence=&Drum_SpineTwist_Base_Pose}
     Additive=@AnimGraphNode_StateMachine_6
    ) {alphaBoolBlend {blendOption=Linear}}
    BlendWeights_0=Get LeftArmIsPlayingHitBlend() -> (LeftArmIsPlayingHitBlend:double)
    BlendWeights_1:float=Get RightArmIsPlayingHitBlend() -> (RightArmIsPlayingHitBlend:double)
   ) -> (Pose) {
    blendPoses [{linkId=-1, sourceLinkId=-1}, {linkId=-1, sourceLinkId=-1}]
    blendWeights [1,1]
    layerSetup [{branchFilters [{boneName=clavicle_l}, {boneName=VB Root_hand_l}]}, {branchFilters [{boneName=clavicle_r}, {boneName=VB Root_hand_r}]}]
   }
   AnimGraphNode_StateMachine_4 = Left Leg() {is=title AnimGraphNode_StateMachine}
   AnimGraphNode_StateMachine_5 = Right Leg() {is=title AnimGraphNode_StateMachine}
   n2 = Layered blend per bone!(
    BasePose=Sequence Player 'Drum_SpineTwist_Base_Pose'() {sequence=&Drum_SpineTwist_Base_Pose}
    BlendPoses_0=@n1
    BlendPoses_1=@AnimGraphNode_StateMachine_4
    BlendPoses_2:Pose Link Structure=@AnimGraphNode_StateMachine_5
    BlendWeights_0
    BlendWeights_1=1
    BlendWeights_2:float=1
   ) -> (Pose) {
    blendPoses [{linkId=-1, sourceLinkId=-1}, {linkId=-1, sourceLinkId=-1}, {linkId=-1, sourceLinkId=-1}]
    blendWeights [1,1,1]
    layerSetup [
     {branchFilters [{boneName=pelvis}, {boneName=VB Root_hand_l}, {boneName=VB Root_hand_r}]}
     {branchFilters [{boneName=thigh_l}, {boneName=ik_foot_l}]}
     {branchFilters [{boneName=thigh_r}, {boneName=ik_foot_r}]}
    ]
   }
   n3 = Blend Poses by bool(
    BlendPose_0=Blend Poses by bool(
     BlendPose_0=@n2
     BlendPose_1=Sequence Player 'Drum_Fidget_3_A'!(
      Sequence:Anim Sequence Base Object Reference=/SparksCharacterFull/Animation/MainPlayer/BandPlay/Drum/Loops/Drum_Fidget_3_A
      bLoopAnimation:bool=True
     ) -> (Pose)
     BlendTime_0=0.5
     BlendTime_1=0.5
     bActiveValue=Get bMusicBeatIsActive() -> (bMusicBeatIsActive:bool)
    ) {blendType=Linear}
    BlendPose_1=Sequence Player 'Drum_Idle_Loop_A'() {sequence=/SparksCharacterFull/Animation/MainPlayer/BandPlay/Drum/Loops/Drum_Idle_Loop_A}
    BlendTime_0=0.5
    BlendTime_1=0.5
    bActiveValue=OR Boolean(A=Get bMusicBeatIsActive(), B=Get IdleRealtime() -> (IdleRealtime:bool))
   ) {blendType=Linear}
   n4 = Sequence Player 'Drum_Play_Loop_125_Medium_A'() {sequence=/SparksCharacterFull/Animation/MainPlayer/BandPlay/Drum/Loops/BPM_125/Drum_Play_Loop_125_Medium_A}
   n5 = Get bCanPlayMidiAnimations() -> (bCanPlayMidiAnimations:bool)
   Output Pose(Result=Apply Additive(
     Base=Blend Poses by bool(BlendPose_0=@n3, BlendPose_1=@n4, BlendTime_0=0.5, BlendTime_1=0.5, bActiveValue=@n5) {blendType=Linear}
     Additive="Sequence Player (additive) 'Drum_SpawnInstrument_Pose'"() {sequence=/SparksCharacterCommon/Animation/MainPlayer/BandPlay/Drum/Drum_SpawnInstrument_Pose}
    ) {alphaBoolBlend {blendOption=Linear}}) {layerGroup=None}
   n6 = Get SpineTwist() -> (SpineTwist:double)
   AnimGraphNode_StateMachine_2 = Body() {is=title AnimGraphNode_StateMachine, bAllowConduitEntryStates=true}
   n7 = Layered blend per bone!(
    BasePose=Apply Mesh Space Additive!(
     Base:Pose Link Structure=@AnimGraphNode_StateMachine_2
     Additive:Pose Link Structure=Blendspace Player 'Drum_SpineTwist_BS_Additive'!(X=@n6) -> (Pose) {blendSpace=/SparksCharacterCommon/Animation/MainPlayer/BandPlay/Drum/SpineTwist/Drum_SpineTwist_BS_Additive}
     Alpha:float=1
    ) -> (Pose:Pose Link Structure) {is=title AnimGraphNode_ApplyMeshSpaceAdditive, alphaBoolBlend {blendOption=Linear}}
    BlendPoses_0=Apply Additive(
     Base=Sequence Player 'Drum_SpineTwist_Base_Pose'() {sequence=&Drum_SpineTwist_Base_Pose}
     Additive="Sequence Evaluator (by time) 'Drum_Hit_Snare_Normal_Medium'"!(ExplicitTime:float=0) -> (Pose:Pose Link Structure) {
      is=title AnimGraphNode_SequenceEvaluator
      sequence=/SparksCharacterCommon/Animation/MainPlayer/BandPlay/Drum/Hits/Normal_Medium/Drum_Hit_Snare_Normal_Medium
     }
    ) {alphaBoolBlend {blendOption=Linear}}
    BlendWeights_0
   ) -> (Pose) {layerSetup [{branchFilters [
       {boneName=hand_attach_l}
       {boneName=index_metacarpal_l}
       {boneName=middle_metacarpal_l}
       {boneName=pinky_metacarpal_l}
       {boneName=ring_metacarpal_l}
       {boneName=thumb_01_l}
       {boneName=hand_attach_r}
       {boneName=index_metacarpal_r}
       {boneName=middle_metacarpal_r}
       {boneName=pinky_metacarpal_r}
       {boneName=ring_metacarpal_r}
       {boneName=thumb_01_r}
       {boneName=weapon_l}
       {boneName=weapon_r}
      ]}]}
   "Save cached pose 'Body (Cache Pose)'"(Pose=@n7) {cacheName="Body (Cache Pose)"}
   Save cached pose 'FullBodyLeftLegHitADD'(Pose=Slot 'AdditiveCorrectiveUpperBody'(Source=Additive Identity Pose()) {slotName=AdditiveCorrectiveUpperBody}) {cacheName=FullBodyLeftLegHitADD}
   Save cached pose 'FullBodyRightLegHitADD'(Pose=Slot 'AdditiveCorrectiveDBNO'(Source=Additive Identity Pose()) {slotName=AdditiveCorrectiveDBNO}) {cacheName=FullBodyRightLegHitADD}
  ]}
 LeftArmAdditiveHits {
  entry=Neutral
  path=AnimGraph.AnimGraphNode_StateMachine_1.LeftArmAdditiveHits
  ids=id
  nodes [
   AnimStateNode_0 = LeftArmHit_B() {is=title AnimStateNode, bAlwaysResetOnEntry=true, stateEntered {guid=#1, notifyName=LeftArmHit_B_Entered}}
   AnimStateNode_1 = LeftArmHit_A() {is=title AnimStateNode, bAlwaysResetOnEntry=true, stateEntered {guid=#2, notifyName=LeftArmHit_A_Entered}}
   AnimStateNode_2 = Neutral() {is=title AnimStateNode, stateEntered {guid=#3, notifyName=LeftArm_Neutral_Entered}}
   AnimStateTransitionNode_0 = AnimStateTransitionNode() {crossfadeDuration=0.05}
   AnimStateTransitionNode_1 = AnimStateTransitionNode() {crossfadeDuration=0.05}
   AnimStateTransitionNode_10 = AnimStateTransitionNode() {crossfadeDuration=0.4}
   AnimStateTransitionNode_2 = AnimStateTransitionNode() {crossfadeDuration=0.15}
   AnimStateTransitionNode_3 = AnimStateTransitionNode() {crossfadeDuration=0.15}
   AnimStateTransitionNode_4 = AnimStateTransitionNode() {bAutomaticRuleBasedOnSequencePlayerInState=true, crossfadeDuration=1.2}
   AnimStateTransitionNode_5 = AnimStateTransitionNode() {bAutomaticRuleBasedOnSequencePlayerInState=true, crossfadeDuration=1.2}
   AnimStateTransitionNode_6 = AnimStateTransitionNode() {crossfadeDuration=0.25}
   AnimStateTransitionNode_7 = AnimStateTransitionNode() {crossfadeDuration=0.4}
   AnimStateTransitionNode_8 = AnimStateTransitionNode() {crossfadeDuration=0.25}
  ]
  links [
   {
    dst=LeftArmHit_B
    rule=AnimGraph.AnimGraphNode_StateMachine_1.LeftArmAdditiveHits.AnimStateTransitionNode_0.Transition
    src=LeftArmHit_A
    via=AnimStateTransitionNode_0
   }
   {
    dst=LeftArmHit_B
    rule=AnimGraph.AnimGraphNode_StateMachine_1.LeftArmAdditiveHits.AnimStateTransitionNode_2.Transition
    src=LeftArmHit_A
    via=AnimStateTransitionNode_2
   }
   {
    dst=LeftArmHit_B
    rule=AnimGraph.AnimGraphNode_StateMachine_1.LeftArmAdditiveHits.AnimStateTransitionNode_8.Transition
    src=LeftArmHit_A
    via=AnimStateTransitionNode_8
   }
   {
    dst=Neutral
    rule=AnimGraph.AnimGraphNode_StateMachine_1.LeftArmAdditiveHits.AnimStateTransitionNode_4.Transition
    src=LeftArmHit_A
    via=AnimStateTransitionNode_4
   }
   {
    dst=LeftArmHit_A
    rule=AnimGraph.AnimGraphNode_StateMachine_1.LeftArmAdditiveHits.AnimStateTransitionNode_1.Transition
    src=LeftArmHit_B
    via=AnimStateTransitionNode_1
   }
   {
    dst=LeftArmHit_A
    rule=AnimGraph.AnimGraphNode_StateMachine_1.LeftArmAdditiveHits.AnimStateTransitionNode_3.Transition
    src=LeftArmHit_B
    via=AnimStateTransitionNode_3
   }
   {
    dst=LeftArmHit_A
    rule=AnimGraph.AnimGraphNode_StateMachine_1.LeftArmAdditiveHits.AnimStateTransitionNode_6.Transition
    src=LeftArmHit_B
    via=AnimStateTransitionNode_6
   }
   {
    dst=Neutral
    rule=AnimGraph.AnimGraphNode_StateMachine_1.LeftArmAdditiveHits.AnimStateTransitionNode_5.Transition
    src=LeftArmHit_B
    via=AnimStateTransitionNode_5
   }
   {
    dst=LeftArmHit_A
    rule=AnimGraph.AnimGraphNode_StateMachine_1.LeftArmAdditiveHits.AnimStateTransitionNode_10.Transition
    src=Neutral
    via=AnimStateTransitionNode_10
   }
   {
    dst=LeftArmHit_B
    rule=AnimGraph.AnimGraphNode_StateMachine_1.LeftArmAdditiveHits.AnimStateTransitionNode_7.Transition
    src=Neutral
    via=AnimStateTransitionNode_7
   }
  ]
 }
 LeftArmHit_B {
  path=AnimGraph.AnimGraphNode_StateMachine_1.LeftArmAdditiveHits.AnimStateNode_0.LeftArmHit_B
  nodes [
   AnimGraphNode_StateMachine_2 = LeftArmHitChooser_StartTime_DelayFloat() {is=title AnimGraphNode_StateMachine, bSkipFirstUpdateTransition=false}
   Output Animation Pose(Result=@AnimGraphNode_StateMachine_2)
  ]
 }
 LeftArmHitChooser_StartTime_DelayFloat {
  entry=Identity
  path=AnimGraph.AnimGraphNode_StateMachine_1.LeftArmAdditiveHits.AnimStateNode_0.LeftArmHit_B.AnimGraphNode_StateMachine_2.LeftArmHitChooser_StartTime_DelayFloat
  ids=id
  nodes [
   AnimStateNode_0 = Hit_Snare_L() {
    is=title AnimStateNode
    stateEntered {guid=#4}
    stateFullyBlended {guid=#5}
    stateLeft {guid=#6}
   }
   AnimStateNode_1 = Hit_HiHat_L() {
    is=title AnimStateNode
    stateEntered {guid=#7}
    stateFullyBlended {guid=#8}
    stateLeft {guid=#9}
   }
   AnimStateNode_10 = Identity() {is=title AnimStateNode}
   AnimStateNode_11 = Hit_Crash2_Hard_L() {
    is=title AnimStateNode
    stateEntered {guid=#10}
    stateFullyBlended {guid=#11}
    stateLeft {guid=#12}
   }
   AnimStateNode_2 = Hit_Crash1_Hard_L() {
    is=title AnimStateNode
    stateEntered {guid=#13}
    stateFullyBlended {guid=#14}
    stateLeft {guid=#15}
   }
   AnimStateNode_3 = Hit_Crash1_Soft_L() {
    is=title AnimStateNode
    stateEntered {guid=#16}
    stateFullyBlended {guid=#17}
    stateLeft {guid=#18}
   }
   AnimStateNode_4 = Hit_Soft() {is=title AnimStateNode}
   AnimStateNode_5 = Hit_Crash2_Soft_L() {
    is=title AnimStateNode
    stateEntered {guid=#19}
    stateFullyBlended {guid=#20}
    stateLeft {guid=#21}
   }
   AnimStateNode_6 = Hit_Ride_L() {
    is=title AnimStateNode
    stateEntered {guid=#22}
    stateFullyBlended {guid=#23}
    stateLeft {guid=#24}
   }
   AnimStateNode_7 = Hit_Tom1_L() {
    is=title AnimStateNode
    stateEntered {guid=#25}
    stateFullyBlended {guid=#26}
    stateLeft {guid=#27}
   }
   AnimStateNode_8 = Hit_Tom2_L() {
    is=title AnimStateNode
    stateEntered {guid=#28}
    stateFullyBlended {guid=#29}
    stateLeft {guid=#30}
   }
   AnimStateNode_9 = Hit_FloorTom_L() {
    is=title AnimStateNode
    stateEntered {guid=#31}
    stateFullyBlended {guid=#32}
    stateLeft {guid=#33}
   }
   AnimStateTransitionNode_10 = AnimStateTransitionNode() {crossfadeDuration=0}
   AnimStateTransitionNode_12 = AnimStateTransitionNode() {crossfadeDuration=0}
   AnimStateTransitionNode_14 = AnimStateTransitionNode() {crossfadeDuration=0}
   AnimStateTransitionNode_16 = AnimStateTransitionNode() {crossfadeDuration=0}
   AnimStateTransitionNode_18 = AnimStateTransitionNode() {crossfadeDuration=0}
   AnimStateTransitionNode_20 = AnimStateTransitionNode() {crossfadeDuration=0}
   AnimStateTransitionNode_22 = AnimStateTransitionNode() {crossfadeDuration=0}
   AnimStateTransitionNode_24 = AnimStateTransitionNode() {crossfadeDuration=0}
   AnimStateTransitionNode_26 = AnimStateTransitionNode() {crossfadeDuration=0}
   AnimStateTransitionNode_28 = AnimStateTransitionNode() {crossfadeDuration=0}
   AnimStateTransitionNode_5 = AnimStateTransitionNode() {crossfadeDuration=0}
  ]
  links [
   {
    dst=Hit_Crash1_Hard_L
    rule=AnimGraph.AnimGraphNode_StateMachine_1.LeftArmAdditiveHits.AnimStateNode_0.LeftArmHit_B.AnimGraphNode_StateMachine_2.LeftArmHitChooser_StartTime_DelayFloat.AnimStateTransitionNode_14.Transition
    src=Identity
    via=AnimStateTransitionNode_14
   }
   {
    dst=Hit_Crash1_Soft_L
    rule=AnimGraph.AnimGraphNode_StateMachine_1.LeftArmAdditiveHits.AnimStateNode_0.LeftArmHit_B.AnimGraphNode_StateMachine_2.LeftArmHitChooser_StartTime_DelayFloat.AnimStateTransitionNode_16.Transition
    src=Identity
    via=AnimStateTransitionNode_16
   }
   {
    dst=Hit_Crash2_Hard_L
    rule=AnimGraph.AnimGraphNode_StateMachine_1.LeftArmAdditiveHits.AnimStateNode_0.LeftArmHit_B.AnimGraphNode_StateMachine_2.LeftArmHitChooser_StartTime_DelayFloat.AnimStateTransitionNode_18.Transition
    src=Identity
    via=AnimStateTransitionNode_18
   }
   {
    dst=Hit_Crash2_Soft_L
    rule=AnimGraph.AnimGraphNode_StateMachine_1.LeftArmAdditiveHits.AnimStateNode_0.LeftArmHit_B.AnimGraphNode_StateMachine_2.LeftArmHitChooser_StartTime_DelayFloat.AnimStateTransitionNode_20.Transition
    src=Identity
    via=AnimStateTransitionNode_20
   }
   {
    dst=Hit_FloorTom_L
    rule=AnimGraph.AnimGraphNode_StateMachine_1.LeftArmAdditiveHits.AnimStateNode_0.LeftArmHit_B.AnimGraphNode_StateMachine_2.LeftArmHitChooser_StartTime_DelayFloat.AnimStateTransitionNode_28.Transition
    src=Identity
    via=AnimStateTransitionNode_28
   }
   {
    dst=Hit_HiHat_L
    rule=AnimGraph.AnimGraphNode_StateMachine_1.LeftArmAdditiveHits.AnimStateNode_0.LeftArmHit_B.AnimGraphNode_StateMachine_2.LeftArmHitChooser_StartTime_DelayFloat.AnimStateTransitionNode_12.Transition
    src=Identity
    via=AnimStateTransitionNode_12
   }
   {
    dst=Hit_Ride_L
    rule=AnimGraph.AnimGraphNode_StateMachine_1.LeftArmAdditiveHits.AnimStateNode_0.LeftArmHit_B.AnimGraphNode_StateMachine_2.LeftArmHitChooser_StartTime_DelayFloat.AnimStateTransitionNode_22.Transition
    src=Identity
    via=AnimStateTransitionNode_22
   }
   {
    dst=Hit_Snare_L
    rule=AnimGraph.AnimGraphNode_StateMachine_1.LeftArmAdditiveHits.AnimStateNode_0.LeftArmHit_B.AnimGraphNode_StateMachine_2.LeftArmHitChooser_StartTime_DelayFloat.AnimStateTransitionNode_10.Transition
    src=Identity
    via=AnimStateTransitionNode_10
   }
   {
    dst=Hit_Soft
    rule=AnimGraph.AnimGraphNode_StateMachine_1.LeftArmAdditiveHits.AnimStateNode_0.LeftArmHit_B.AnimGraphNode_StateMachine_2.LeftArmHitChooser_StartTime_DelayFloat.AnimStateTransitionNode_5.Transition
    src=Identity
    via=AnimStateTransitionNode_5
   }
   {
    dst=Hit_Tom1_L
    rule=AnimGraph.AnimGraphNode_StateMachine_1.LeftArmAdditiveHits.AnimStateNode_0.LeftArmHit_B.AnimGraphNode_StateMachine_2.LeftArmHitChooser_StartTime_DelayFloat.AnimStateTransitionNode_24.Transition
    src=Identity
    via=AnimStateTransitionNode_24
   }
   {
    dst=Hit_Tom2_L
    rule=AnimGraph.AnimGraphNode_StateMachine_1.LeftArmAdditiveHits.AnimStateNode_0.LeftArmHit_B.AnimGraphNode_StateMachine_2.LeftArmHitChooser_StartTime_DelayFloat.AnimStateTransitionNode_26.Transition
    src=Identity
    via=AnimStateTransitionNode_26
   }
  ]
 }
 Hit_Snare_L {
  path=AnimGraph.AnimGraphNode_StateMachine_1.LeftArmAdditiveHits.AnimStateNode_0.LeftArmHit_B.AnimGraphNode_StateMachine_2.LeftArmHitChooser_StartTime_DelayFloat.AnimStateNode_0.Hit_Snare_L
  nodes [
   AnimGraphNode_StateMachine_0 = Hit_Snare_L_SM() {is=title AnimGraphNode_StateMachine, bAllowConduitEntryStates=true}
   Output Animation Pose(Result=@AnimGraphNode_StateMachine_0)
  ]
 }
 Hit_Snare_L_SM {
  entry=Conduit
  path=AnimGraph.AnimGraphNode_StateMachine_1.LeftArmAdditiveHits.AnimStateNode_0.LeftArmHit_B.AnimGraphNode_StateMachine_2.LeftArmHitChooser_StartTime_DelayFloat.AnimStateNode_0.Hit_Snare_L.AnimGraphNode_StateMachine_0.Hit_Snare_L_SM
  ids=id
  nodes [
   AnimStateConduitNode_0 = Conduit()
   AnimStateNode_2 = Hit_Long() {is=title AnimStateNode}
   AnimStateNode_3 = Hit_Medium() {is=title AnimStateNode}
   AnimStateNode_4 = Hit_Short() {is=title AnimStateNode}
   AnimStateTransitionNode_4 = AnimStateTransitionNode() {crossfadeDuration=0, priorityOrder=2}
   AnimStateTransitionNode_5 = AnimStateTransitionNode() {crossfadeDuration=0, priorityOrder=2}
   AnimStateTransitionNode_6 = AnimStateTransitionNode() {crossfadeDuration=0, priorityOrder=2}
  ]
  links [
   {
    dst=Hit_Long
    rule=AnimGraph.AnimGraphNode_StateMachine_1.LeftArmAdditiveHits.AnimStateNode_0.LeftArmHit_B.AnimGraphNode_StateMachine_2.LeftArmHitChooser_StartTime_DelayFloat.AnimStateNode_0.Hit_Snare_L.AnimGraphNode_StateMachine_0.Hit_Snare_L_SM.AnimStateTransitionNode_6.Transition
    src=Conduit
    via=AnimStateTransitionNode_6
   }
   {
    dst=Hit_Medium
    rule=AnimGraph.AnimGraphNode_StateMachine_1.LeftArmAdditiveHits.AnimStateNode_0.LeftArmHit_B.AnimGraphNode_StateMachine_2.LeftArmHitChooser_StartTime_DelayFloat.AnimStateNode_0.Hit_Snare_L.AnimGraphNode_StateMachine_0.Hit_Snare_L_SM.AnimStateTransitionNode_5.Transition
    src=Conduit
    via=AnimStateTransitionNode_5
   }
   {
    dst=Hit_Short
    rule=AnimGraph.AnimGraphNode_StateMachine_1.LeftArmAdditiveHits.AnimStateNode_0.LeftArmHit_B.AnimGraphNode_StateMachine_2.LeftArmHitChooser_StartTime_DelayFloat.AnimStateNode_0.Hit_Snare_L.AnimGraphNode_StateMachine_0.Hit_Snare_L_SM.AnimStateTransitionNode_4.Transition
    src=Conduit
    via=AnimStateTransitionNode_4
   }
  ]
 }
 Conduit {
  path=AnimGraph.AnimGraphNode_StateMachine_1.LeftArmAdditiveHits.AnimStateNode_0.LeftArmHit_B.AnimGraphNode_StateMachine_2.LeftArmHitChooser_StartTime_DelayFloat.AnimStateNode_0.Hit_Snare_L.AnimGraphNode_StateMachine_0.Hit_Snare_L_SM.AnimStateConduitNode_0.Conduit
  nodes [Result(bCanEnterTransition=true)]
 }
 Hit_Long {
  path=AnimGraph.AnimGraphNode_StateMachine_1.LeftArmAdditiveHits.AnimStateNode_0.LeftArmHit_B.AnimGraphNode_StateMachine_2.LeftArmHitChooser_StartTime_DelayFloat.AnimStateNode_0.Hit_Snare_L.AnimGraphNode_StateMachine_0.Hit_Snare_L_SM.AnimStateNode_2.Hit_Long
  nodes [Output Animation Pose(Result="Sequence Player (additive) 'Drum_Hit_Snare_Slow_Big'"!(StartPosition:float=Get LeftArmMaxDelayShortenedBy() -> (LeftArmMaxDelayShortenedBy:float), bLoopAnimation:bool=false) -> (Pose) {sequence=/SparksCharacterFull/Animation/MainPlayer/BandPlay/Drum/Hits/Slow_Big/Drum_Hit_Snare_Slow_Big, startPosition=0.05})]
 }
 Hit_Medium {
  path=AnimGraph.AnimGraphNode_StateMachine_1.LeftArmAdditiveHits.AnimStateNode_0.LeftArmHit_B.AnimGraphNode_StateMachine_2.LeftArmHitChooser_StartTime_DelayFloat.AnimStateNode_0.Hit_Snare_L.AnimGraphNode_StateMachine_0.Hit_Snare_L_SM.AnimStateNode_3.Hit_Medium
  nodes [
   n1 = Get LeftArmMaxDelayShortenedBy()
   n2 = "Sequence Player (additive) 'Drum_Hit_CrossHiHatSnare_Medium'"!(StartPosition:float=@n1, bLoopAnimation:bool=False) -> (Pose) {bLoopAnimation=false, sequence=/SparksCharacterCommon/Animation/MainPlayer/BandPlay/Drum/Hits/Normal_Medium/Drum_Hit_CrossHiHatSnare_Medium}
   n3 = "Sequence Player (additive) 'Drum_Hit_Snare_Normal_Medium'"!(StartPosition:float=@n1, bLoopAnimation:bool=false) -> (Pose) {sequence=&Drum_Hit_Snare_Normal_Medium, startPosition=0.05}
   Output Animation Pose(Result=Blend Poses by bool(BlendPose_0=@n2, BlendPose_1=@n3, bActiveValue=Get Hit_HiHat_R() -> (Hit_HiHat_R:bool)) {blendType=Linear})
  ]
 }
 Hit_Short {
  path=AnimGraph.AnimGraphNode_StateMachine_1.LeftArmAdditiveHits.AnimStateNode_0.LeftArmHit_B.AnimGraphNode_StateMachine_2.LeftArmHitChooser_StartTime_DelayFloat.AnimStateNode_0.Hit_Snare_L.AnimGraphNode_StateMachine_0.Hit_Snare_L_SM.AnimStateNode_4.Hit_Short
  nodes [
   n1 = "Sequence Player (additive) 'Drum_Hit_CrossHiHatSnare_Tap_Fast'"!(StartPosition:float=Get LeftArmMaxDelayShortenedBy(), bLoopAnimation:bool=False) -> (Pose) {bLoopAnimation=false, sequence=/SparksCharacterCommon/Animation/MainPlayer/BandPlay/Drum/Hits/Tap_Fast/Drum_Hit_CrossHiHatSnare_Tap_Fast}
   n2 = Get LeftArmMaxDelayShortenedBy()
   Output Animation Pose(Result=Blend Poses by bool(
     BlendPose_0=@n1
     BlendPose_1="Sequence Player (additive) 'Drum_Hit_Snare_Tap_Fast'"!(StartPosition:float=@n2, bLoopAnimation:bool=false) -> (Pose) {sequence=/SparksCharacterCommon/Animation/MainPlayer/BandPlay/Drum/Hits/Tap_Fast/Drum_Hit_Snare_Tap_Fast, startPosition=0.05}
     bActiveValue=Get Hit_HiHat_R()
    ) {blendType=Linear})
  ]
 }
 Transition {
  path=AnimGraph.AnimGraphNode_StateMachine_1.LeftArmAdditiveHits.AnimStateNode_0.LeftArmHit_B.AnimGraphNode_StateMachine_2.LeftArmHitChooser_StartTime_DelayFloat.AnimStateNode_0.Hit_Snare_L.AnimGraphNode_StateMachine_0.Hit_Snare_L_SM.AnimStateTransitionNode_4.Transition
  nodes [Result(bCanEnterTransition=Get LeftArmIsShortDelay() -> (LeftArmIsShortDelay:bool))]
 }
 Transition {
  path=AnimGraph.AnimGraphNode_StateMachine_1.LeftArmAdditiveHits.AnimStateNode_0.LeftArmHit_B.AnimGraphNode_StateMachine_2.LeftArmHitChooser_StartTime_DelayFloat.AnimStateNode_0.Hit_Snare_L.AnimGraphNode_StateMachine_0.Hit_Snare_L_SM.AnimStateTransitionNode_5.Transition
  nodes [
   n1 = Get LeftArmIsLongDelay() -> (LeftArmIsLongDelay:bool)
   n2 = Get LeftArmIsShortDelay()
   Result(bCanEnterTransition=AND Boolean(A=NOT Boolean(A=@n2), B=NOT Boolean(A=@n1)))
  ]
 }
 Transition {
  path=AnimGraph.AnimGraphNode_StateMachine_1.LeftArmAdditiveHits.AnimStateNode_0.LeftArmHit_B.AnimGraphNode_StateMachine_2.LeftArmHitChooser_StartTime_DelayFloat.AnimStateNode_0.Hit_Snare_L.AnimGraphNode_StateMachine_0.Hit_Snare_L_SM.AnimStateTransitionNode_6.Transition
  nodes [Result(bCanEnterTransition=Get LeftArmIsLongDelay())]
 }
 Hit_HiHat_L {
  path=AnimGraph.AnimGraphNode_StateMachine_1.LeftArmAdditiveHits.AnimStateNode_0.LeftArmHit_B.AnimGraphNode_StateMachine_2.LeftArmHitChooser_StartTime_DelayFloat.AnimStateNode_1.Hit_HiHat_L
  nodes [
   n1 = Get LeftArmMaxDelayShortenedBy()
   n2 = "Sequence Player (additive) 'Drum_Hit_HiHat_Normal_Medium'"!(StartPosition:float=Get LeftArmMaxDelayShortenedBy(), bLoopAnimation:bool=false) -> (Pose) {sequence=/SparksCharacterCommon/Animation/MainPlayer/BandPlay/Drum/Hits/Normal_Medium/Drum_Hit_HiHat_Normal_Medium, startPosition=0.05}
   Blend Poses by bool(
    BlendPose_0="Sequence Player (additive) 'Drum_Hit_HiHat_Slow_Big'"!(StartPosition:float=Get LeftArmMaxDelayShortenedBy(), bLoopAnimation:bool=false) -> (Pose) {sequence=/SparksCharacterFull/Animation/MainPlayer/BandPlay/Drum/Hits/Slow_Big/Drum_Hit_HiHat_Slow_Big, startPosition=0.05}
    BlendPose_1=Blend Poses by bool(
     BlendPose_0="Sequence Player (additive) 'Drum_Hit_HiHat_Tap_Fast'"!(StartPosition:float=@n1, bLoopAnimation:bool=false) -> (Pose) {sequence=/SparksCharacterCommon/Animation/MainPlayer/BandPlay/Drum/Hits/Tap_Fast/Drum_Hit_HiHat_Tap_Fast, startPosition=0.05}
     BlendPose_1=@n2
     BlendTime_0=0
     BlendTime_1=0
     bActiveValue=Get LeftArmIsShortDelay()
    ) {blendType=Linear}
    BlendTime_0=0
    BlendTime_1=0
    bActiveValue=Get LeftArmIsLongDelay()
   ) {blendType=Linear}
   AnimGraphNode_StateMachine_0 = Hit_HiHat_L_SM() {is=title AnimGraphNode_StateMachine, bAllowConduitEntryStates=true}
   Output Animation Pose(Result=@AnimGraphNode_StateMachine_0)
  ]
 }
 Hit_HiHat_L_SM {
  entry=Conduit
  path=AnimGraph.AnimGraphNode_StateMachine_1.LeftArmAdditiveHits.AnimStateNode_0.LeftArmHit_B.AnimGraphNode_StateMachine_2.LeftArmHitChooser_StartTime_DelayFloat.AnimStateNode_1.Hit_HiHat_L.AnimGraphNode_StateMachine_0.Hit_HiHat_L_SM
  ids=id
  nodes [
   AnimStateConduitNode_0 = Conduit()
   AnimStateNode_0 = Hit_Short()
   AnimStateNode_1 = Hit_Medium()
   AnimStateNode_2 = Hit_Long()
   AnimStateTransitionNode_4 = AnimStateTransitionNode() {crossfadeDuration=0}
   AnimStateTransitionNode_5 = AnimStateTransitionNode() {crossfadeDuration=0}
   AnimStateTransitionNode_6 = AnimStateTransitionNode() {crossfadeDuration=0}
  ]
  links [
   {
    dst=Hit_Long
    rule=AnimGraph.AnimGraphNode_StateMachine_1.LeftArmAdditiveHits.AnimStateNode_0.LeftArmHit_B.AnimGraphNode_StateMachine_2.LeftArmHitChooser_StartTime_DelayFloat.AnimStateNode_1.Hit_HiHat_L.AnimGraphNode_StateMachine_0.Hit_HiHat_L_SM.AnimStateTransitionNode_6.Transition
    src=Conduit
    via=AnimStateTransitionNode_6
   }
   {
    dst=Hit_Medium
    rule=AnimGraph.AnimGraphNode_StateMachine_1.LeftArmAdditiveHits.AnimStateNode_0.LeftArmHit_B.AnimGraphNode_StateMachine_2.LeftArmHitChooser_StartTime_DelayFloat.AnimStateNode_1.Hit_HiHat_L.AnimGraphNode_StateMachine_0.Hit_HiHat_L_SM.AnimStateTransitionNode_5.Transition
    src=Conduit
    via=AnimStateTransitionNode_5
   }
   {
    dst=Hit_Short
    rule=AnimGraph.AnimGraphNode_StateMachine_1.LeftArmAdditiveHits.AnimStateNode_0.LeftArmHit_B.AnimGraphNode_StateMachine_2.LeftArmHitChooser_StartTime_DelayFloat.AnimStateNode_1.Hit_HiHat_L.AnimGraphNode_StateMachine_0.Hit_HiHat_L_SM.AnimStateTransitionNode_4.Transition
    src=Conduit
    via=AnimStateTransitionNode_4
   }
  ]
 }
 Conduit {
  path=AnimGraph.AnimGraphNode_StateMachine_1.LeftArmAdditiveHits.AnimStateNode_0.LeftArmHit_B.AnimGraphNode_StateMachine_2.LeftArmHitChooser_StartTime_DelayFloat.AnimStateNode_1.Hit_HiHat_L.AnimGraphNode_StateMachine_0.Hit_HiHat_L_SM.AnimStateConduitNode_0.Conduit
  nodes [Result(bCanEnterTransition=true)]
 }
 Hit_Short {
  path=AnimGraph.AnimGraphNode_StateMachine_1.LeftArmAdditiveHits.AnimStateNode_0.LeftArmHit_B.AnimGraphNode_StateMachine_2.LeftArmHitChooser_StartTime_DelayFloat.AnimStateNode_1.Hit_HiHat_L.AnimGraphNode_StateMachine_0.Hit_HiHat_L_SM.AnimStateNode_0.Hit_Short
  nodes [Output Animation Pose(Result="Sequence Player (additive) 'Drum_Hit_HiHat_Tap_Fast'"(StartPosition=Get LeftArmMaxDelayShortenedBy(), bLoopAnimation=false) {sequence=&Drum_Hit_HiHat_Tap_Fast, startPosition=0.05})]
 }
 Hit_Medium {
  path=AnimGraph.AnimGraphNode_StateMachine_1.LeftArmAdditiveHits.AnimStateNode_0.LeftArmHit_B.AnimGraphNode_StateMachine_2.LeftArmHitChooser_StartTime_DelayFloat.AnimStateNode_1.Hit_HiHat_L.AnimGraphNode_StateMachine_0.Hit_HiHat_L_SM.AnimStateNode_1.Hit_Medium
  nodes [
   n1 = "Sequence Player (additive) 'Drum_Hit_HiHat_Normal_Medium'"(StartPosition=Get LeftArmMaxDelayShortenedBy(), bLoopAnimation=false) {sequence=&Drum_Hit_HiHat_Normal_Medium, startPosition=0.05}
   Output Animation Pose(Result=@n1)
  ]
 }
 Hit_Long {
  path=AnimGraph.AnimGraphNode_StateMachine_1.LeftArmAdditiveHits.AnimStateNode_0.LeftArmHit_B.AnimGraphNode_StateMachine_2.LeftArmHitChooser_StartTime_DelayFloat.AnimStateNode_1.Hit_HiHat_L.AnimGraphNode_StateMachine_0.Hit_HiHat_L_SM.AnimStateNode_2.Hit_Long
  nodes [Output Animation Pose(Result="Sequence Player (additive) 'Drum_Hit_HiHat_Slow_Big'"(StartPosition=Get LeftArmMaxDelayShortenedBy(), bLoopAnimation=false) {sequence=&Drum_Hit_HiHat_Slow_Big, startPosition=0.05})]
 }
 Transition {
  path=AnimGraph.AnimGraphNode_StateMachine_1.LeftArmAdditiveHits.AnimStateNode_0.LeftArmHit_B.AnimGraphNode_StateMachine_2.LeftArmHitChooser_StartTime_DelayFloat.AnimStateNode_1.Hit_HiHat_L.AnimGraphNode_StateMachine_0.Hit_HiHat_L_SM.AnimStateTransitionNode_4.Transition
  nodes [Result(bCanEnterTransition=Get LeftArmIsShortDelay())]
 }
 Transition {
  path=AnimGraph.AnimGraphNode_StateMachine_1.LeftArmAdditiveHits.AnimStateNode_0.LeftArmHit_B.AnimGraphNode_StateMachine_2.LeftArmHitChooser_StartTime_DelayFloat.AnimStateNode_1.Hit_HiHat_L.AnimGraphNode_StateMachine_0.Hit_HiHat_L_SM.AnimStateTransitionNode_5.Transition
  nodes [
   n1 = Get LeftArmIsLongDelay()
   n2 = Get LeftArmIsShortDelay()
   Result(bCanEnterTransition=AND Boolean(A=NOT Boolean(A=@n2), B=NOT Boolean(A=@n1)))
  ]
 }
 Transition {
  path=AnimGraph.AnimGraphNode_StateMachine_1.LeftArmAdditiveHits.AnimStateNode_0.LeftArmHit_B.AnimGraphNode_StateMachine_2.LeftArmHitChooser_StartTime_DelayFloat.AnimStateNode_1.Hit_HiHat_L.AnimGraphNode_StateMachine_0.Hit_HiHat_L_SM.AnimStateTransitionNode_6.Transition
  nodes [Result(bCanEnterTransition=Get LeftArmIsLongDelay())]
 }
 Identity {
  path=AnimGraph.AnimGraphNode_StateMachine_1.LeftArmAdditiveHits.AnimStateNode_0.LeftArmHit_B.AnimGraphNode_StateMachine_2.LeftArmHitChooser_StartTime_DelayFloat.AnimStateNode_10.Identity
  nodes [
   "Sequence Player (additive) 'Drum_SpineTwist_Center_Pose'"() {bLoopAnimation=false, sequence=/SparksCharacterCommon/Animation/MainPlayer/BandPlay/Drum/SpineTwist/Drum_SpineTwist_Center_Pose}
   Output Animation Pose(Result=Additive Identity Pose())
  ]
 }
 Hit_Crash2_Hard_L {
  path=AnimGraph.AnimGraphNode_StateMachine_1.LeftArmAdditiveHits.AnimStateNode_0.LeftArmHit_B.AnimGraphNode_StateMachine_2.LeftArmHitChooser_StartTime_DelayFloat.AnimStateNode_11.Hit_Crash2_Hard_L
  nodes [
   n1 = Get LeftArmMaxDelayShortenedBy()
   n2 = "Sequence Player (additive) 'Drum_Hit_Crash2_Normal_Medium'"!(StartPosition:float=Get LeftArmMaxDelayShortenedBy(), bLoopAnimation:bool=false) -> (Pose) {sequence=/SparksCharacterCommon/Animation/MainPlayer/BandPlay/Drum/Hits/Normal_Medium/Drum_Hit_Crash2_Normal_Medium, startPosition=0.05}
   Blend Poses by bool(
    BlendPose_0="Sequence Player (additive) 'Drum_Hit_Crash2_Slow_Big'"!(StartPosition:float=Get LeftArmMaxDelayShortenedBy(), bLoopAnimation:bool=false) -> (Pose) {sequence=/SparksCharacterFull/Animation/MainPlayer/BandPlay/Drum/Hits/Slow_Big/Drum_Hit_Crash2_Slow_Big, startPosition=0.05}
    BlendPose_1=Blend Poses by bool(
     BlendPose_0="Sequence Player (additive) 'Drum_Hit_Crash2_Tap_Fast'"!(StartPosition:float=@n1, bLoopAnimation:bool=false) -> (Pose) {sequence=/SparksCharacterCommon/Animation/MainPlayer/BandPlay/Drum/Hits/Tap_Fast/Drum_Hit_Crash2_Tap_Fast, startPosition=0.05}
     BlendPose_1=@n2
     BlendTime_0=0
     BlendTime_1=0
     bActiveValue=Get LeftArmIsShortDelay()
    ) {blendType=Linear}
    BlendTime_0=0
    BlendTime_1=0
    bActiveValue=Get LeftArmIsLongDelay()
   ) {blendType=Linear}
   AnimGraphNode_StateMachine_0 = Hit_Crash2_Hard_L_SM() {is=title AnimGraphNode_StateMachine, bAllowConduitEntryStates=true}
   Output Animation Pose(Result=@AnimGraphNode_StateMachine_0)
  ]
 }
 Hit_Crash2_Hard_L_SM {
  entry=Conduit
  path=AnimGraph.AnimGraphNode_StateMachine_1.LeftArmAdditiveHits.AnimStateNode_0.LeftArmHit_B.AnimGraphNode_StateMachine_2.LeftArmHitChooser_StartTime_DelayFloat.AnimStateNode_11.Hit_Crash2_Hard_L.AnimGraphNode_StateMachine_0.Hit_Crash2_Hard_L_SM
  ids=id
  nodes [
   AnimStateConduitNode_0 = Conduit()
   AnimStateNode_0 = Hit_Short()
   AnimStateNode_1 = Hit_Medium()
   AnimStateNode_2 = Hit_Long()
   AnimStateTransitionNode_4 = AnimStateTransitionNode() {crossfadeDuration=0}
   AnimStateTransitionNode_5 = AnimStateTransitionNode() {crossfadeDuration=0}
   AnimStateTransitionNode_6 = AnimStateTransitionNode() {crossfadeDuration=0}
  ]
  links [
   {
    dst=Hit_Long
    rule=AnimGraph.AnimGraphNode_StateMachine_1.LeftArmAdditiveHits.AnimStateNode_0.LeftArmHit_B.AnimGraphNode_StateMachine_2.LeftArmHitChooser_StartTime_DelayFloat.AnimStateNode_11.Hit_Crash2_Hard_L.AnimGraphNode_StateMachine_0.Hit_Crash2_Hard_L_SM.AnimStateTransitionNode_6.Transition
    src=Conduit
    via=AnimStateTransitionNode_6
   }
   {
    dst=Hit_Medium
    rule=AnimGraph.AnimGraphNode_StateMachine_1.LeftArmAdditiveHits.AnimStateNode_0.LeftArmHit_B.AnimGraphNode_StateMachine_2.LeftArmHitChooser_StartTime_DelayFloat.AnimStateNode_11.Hit_Crash2_Hard_L.AnimGraphNode_StateMachine_0.Hit_Crash2_Hard_L_SM.AnimStateTransitionNode_5.Transition
    src=Conduit
    via=AnimStateTransitionNode_5
   }
   {
    dst=Hit_Short
    rule=AnimGraph.AnimGraphNode_StateMachine_1.LeftArmAdditiveHits.AnimStateNode_0.LeftArmHit_B.AnimGraphNode_StateMachine_2.LeftArmHitChooser_StartTime_DelayFloat.AnimStateNode_11.Hit_Crash2_Hard_L.AnimGraphNode_StateMachine_0.Hit_Crash2_Hard_L_SM.AnimStateTransitionNode_4.Transition
    src=Conduit
    via=AnimStateTransitionNode_4
   }
  ]
 }
 Conduit {
  path=AnimGraph.AnimGraphNode_StateMachine_1.LeftArmAdditiveHits.AnimStateNode_0.LeftArmHit_B.AnimGraphNode_StateMachine_2.LeftArmHitChooser_StartTime_DelayFloat.AnimStateNode_11.Hit_Crash2_Hard_L.AnimGraphNode_StateMachine_0.Hit_Crash2_Hard_L_SM.AnimStateConduitNode_0.Conduit
  nodes [Result(bCanEnterTransition=true)]
 }
 Hit_Short {
  path=AnimGraph.AnimGraphNode_StateMachine_1.LeftArmAdditiveHits.AnimStateNode_0.LeftArmHit_B.AnimGraphNode_StateMachine_2.LeftArmHitChooser_StartTime_DelayFloat.AnimStateNode_11.Hit_Crash2_Hard_L.AnimGraphNode_StateMachine_0.Hit_Crash2_Hard_L_SM.AnimStateNode_0.Hit_Short
  nodes [Output Animation Pose(Result="Sequence Player (additive) 'Drum_Hit_Crash2_Tap_Fast'"(StartPosition=Get LeftArmMaxDelayShortenedBy(), bLoopAnimation=false) {sequence=&Drum_Hit_Crash2_Tap_Fast, startPosition=0.05})]
 }
 Hit_Medium {
  path=AnimGraph.AnimGraphNode_StateMachine_1.LeftArmAdditiveHits.AnimStateNode_0.LeftArmHit_B.AnimGraphNode_StateMachine_2.LeftArmHitChooser_StartTime_DelayFloat.AnimStateNode_11.Hit_Crash2_Hard_L.AnimGraphNode_StateMachine_0.Hit_Crash2_Hard_L_SM.AnimStateNode_1.Hit_Medium
  nodes [
   n1 = "Sequence Player (additive) 'Drum_Hit_Crash2_Normal_Medium'"(StartPosition=Get LeftArmMaxDelayShortenedBy(), bLoopAnimation=false) {sequence=&Drum_Hit_Crash2_Normal_Medium, startPosition=0.05}
   Output Animation Pose(Result=@n1)
  ]
 }
 Hit_Long {
  path=AnimGraph.AnimGraphNode_StateMachine_1.LeftArmAdditiveHits.AnimStateNode_0.LeftArmHit_B.AnimGraphNode_StateMachine_2.LeftArmHitChooser_StartTime_DelayFloat.AnimStateNode_11.Hit_Crash2_Hard_L.AnimGraphNode_StateMachine_0.Hit_Crash2_Hard_L_SM.AnimStateNode_2.Hit_Long
  nodes [Output Animation Pose(Result="Sequence Player (additive) 'Drum_Hit_Crash2_Slow_Big'"(StartPosition=Get LeftArmMaxDelayShortenedBy(), bLoopAnimation=false) {sequence=&Drum_Hit_Crash2_Slow_Big, startPosition=0.05})]
 }
 Transition {
  path=AnimGraph.AnimGraphNode_StateMachine_1.LeftArmAdditiveHits.AnimStateNode_0.LeftArmHit_B.AnimGraphNode_StateMachine_2.LeftArmHitChooser_StartTime_DelayFloat.AnimStateNode_11.Hit_Crash2_Hard_L.AnimGraphNode_StateMachine_0.Hit_Crash2_Hard_L_SM.AnimStateTransitionNode_4.Transition
  nodes [Result(bCanEnterTransition=Get LeftArmIsShortDelay())]
 }
 Transition {
  path=AnimGraph.AnimGraphNode_StateMachine_1.LeftArmAdditiveHits.AnimStateNode_0.LeftArmHit_B.AnimGraphNode_StateMachine_2.LeftArmHitChooser_StartTime_DelayFloat.AnimStateNode_11.Hit_Crash2_Hard_L.AnimGraphNode_StateMachine_0.Hit_Crash2_Hard_L_SM.AnimStateTransitionNode_5.Transition
  nodes [
   n1 = Get LeftArmIsLongDelay()
   n2 = Get LeftArmIsShortDelay()
   Result(bCanEnterTransition=AND Boolean(A=NOT Boolean(A=@n2), B=NOT Boolean(A=@n1)))
  ]
 }
 Transition {
  path=AnimGraph.AnimGraphNode_StateMachine_1.LeftArmAdditiveHits.AnimStateNode_0.LeftArmHit_B.AnimGraphNode_StateMachine_2.LeftArmHitChooser_StartTime_DelayFloat.AnimStateNode_11.Hit_Crash2_Hard_L.AnimGraphNode_StateMachine_0.Hit_Crash2_Hard_L_SM.AnimStateTransitionNode_6.Transition
  nodes [Result(bCanEnterTransition=Get LeftArmIsLongDelay())]
 }
 Hit_Crash1_Hard_L {
  path=AnimGraph.AnimGraphNode_StateMachine_1.LeftArmAdditiveHits.AnimStateNode_0.LeftArmHit_B.AnimGraphNode_StateMachine_2.LeftArmHitChooser_StartTime_DelayFloat.AnimStateNode_2.Hit_Crash1_Hard_L
  nodes [
   n1 = Get LeftArmMaxDelayShortenedBy()
   n2 = "Sequence Player (additive) 'Drum_Hit_Crash1_Normal_Medium'"!(StartPosition:float=Get LeftArmMaxDelayShortenedBy(), bLoopAnimation:bool=false) -> (Pose) {sequence=/SparksCharacterCommon/Animation/MainPlayer/BandPlay/Drum/Hits/Normal_Medium/Drum_Hit_Crash1_Normal_Medium, startPosition=0.05}
   Blend Poses by bool(
    BlendPose_0="Sequence Player (additive) 'Drum_Hit_Crash1_Slow_Big'"!(StartPosition:float=Get LeftArmMaxDelayShortenedBy(), bLoopAnimation:bool=false) -> (Pose) {sequence=/SparksCharacterFull/Animation/MainPlayer/BandPlay/Drum/Hits/Slow_Big/Drum_Hit_Crash1_Slow_Big, startPosition=0.05}
    BlendPose_1=Blend Poses by bool(
     BlendPose_0="Sequence Player (additive) 'Drum_Hit_Crash1_Tap_Fast'"!(StartPosition:float=@n1, bLoopAnimation:bool=false) -> (Pose) {sequence=/SparksCharacterCommon/Animation/MainPlayer/BandPlay/Drum/Hits/Tap_Fast/Drum_Hit_Crash1_Tap_Fast, startPosition=0.05}
     BlendPose_1=@n2
     BlendTime_0=0
     BlendTime_1=0
     bActiveValue=Get LeftArmIsShortDelay()
    ) {blendType=Linear}
    BlendTime_0=0
    BlendTime_1=0
    bActiveValue=Get LeftArmIsLongDelay()
   ) {blendType=Linear}
   AnimGraphNode_StateMachine_0 = Hit_Crash1_Hard_L_SM() {is=title AnimGraphNode_StateMachine, bAllowConduitEntryStates=true}
   Output Animation Pose(Result=@AnimGraphNode_StateMachine_0)
  ]
 }
 Hit_Crash1_Hard_L_SM {
  entry=Conduit
  path=AnimGraph.AnimGraphNode_StateMachine_1.LeftArmAdditiveHits.AnimStateNode_0.LeftArmHit_B.AnimGraphNode_StateMachine_2.LeftArmHitChooser_StartTime_DelayFloat.AnimStateNode_2.Hit_Crash1_Hard_L.AnimGraphNode_StateMachine_0.Hit_Crash1_Hard_L_SM
  ids=id
  nodes [
   AnimStateConduitNode_0 = Conduit()
   AnimStateNode_0 = Hit_Short()
   AnimStateNode_1 = Hit_Medium()
   AnimStateNode_2 = Hit_Long()
   AnimStateTransitionNode_4 = AnimStateTransitionNode() {crossfadeDuration=0}
   AnimStateTransitionNode_5 = AnimStateTransitionNode() {crossfadeDuration=0}
   AnimStateTransitionNode_6 = AnimStateTransitionNode() {crossfadeDuration=0}
  ]
  links [
   {
    dst=Hit_Long
    rule=AnimGraph.AnimGraphNode_StateMachine_1.LeftArmAdditiveHits.AnimStateNode_0.LeftArmHit_B.AnimGraphNode_StateMachine_2.LeftArmHitChooser_StartTime_DelayFloat.AnimStateNode_2.Hit_Crash1_Hard_L.AnimGraphNode_StateMachine_0.Hit_Crash1_Hard_L_SM.AnimStateTransitionNode_6.Transition
    src=Conduit
    via=AnimStateTransitionNode_6
   }
   {
    dst=Hit_Medium
    rule=AnimGraph.AnimGraphNode_StateMachine_1.LeftArmAdditiveHits.AnimStateNode_0.LeftArmHit_B.AnimGraphNode_StateMachine_2.LeftArmHitChooser_StartTime_DelayFloat.AnimStateNode_2.Hit_Crash1_Hard_L.AnimGraphNode_StateMachine_0.Hit_Crash1_Hard_L_SM.AnimStateTransitionNode_5.Transition
    src=Conduit
    via=AnimStateTransitionNode_5
   }
   {
    dst=Hit_Short
    rule=AnimGraph.AnimGraphNode_StateMachine_1.LeftArmAdditiveHits.AnimStateNode_0.LeftArmHit_B.AnimGraphNode_StateMachine_2.LeftArmHitChooser_StartTime_DelayFloat.AnimStateNode_2.Hit_Crash1_Hard_L.AnimGraphNode_StateMachine_0.Hit_Crash1_Hard_L_SM.AnimStateTransitionNode_4.Transition
    src=Conduit
    via=AnimStateTransitionNode_4
   }
  ]
 }
 Conduit {
  path=AnimGraph.AnimGraphNode_StateMachine_1.LeftArmAdditiveHits.AnimStateNode_0.LeftArmHit_B.AnimGraphNode_StateMachine_2.LeftArmHitChooser_StartTime_DelayFloat.AnimStateNode_2.Hit_Crash1_Hard_L.AnimGraphNode_StateMachine_0.Hit_Crash1_Hard_L_SM.AnimStateConduitNode_0.Conduit
  nodes [Result(bCanEnterTransition=true)]
 }
 Hit_Short {
  path=AnimGraph.AnimGraphNode_StateMachine_1.LeftArmAdditiveHits.AnimStateNode_0.LeftArmHit_B.AnimGraphNode_StateMachine_2.LeftArmHitChooser_StartTime_DelayFloat.AnimStateNode_2.Hit_Crash1_Hard_L.AnimGraphNode_StateMachine_0.Hit_Crash1_Hard_L_SM.AnimStateNode_0.Hit_Short
  nodes [Output Animation Pose(Result="Sequence Player (additive) 'Drum_Hit_Crash1_Tap_Fast'"(StartPosition=Get LeftArmMaxDelayShortenedBy(), bLoopAnimation=false) {sequence=&Drum_Hit_Crash1_Tap_Fast, startPosition=0.05})]
 }
 Hit_Medium {
  path=AnimGraph.AnimGraphNode_StateMachine_1.LeftArmAdditiveHits.AnimStateNode_0.LeftArmHit_B.AnimGraphNode_StateMachine_2.LeftArmHitChooser_StartTime_DelayFloat.AnimStateNode_2.Hit_Crash1_Hard_L.AnimGraphNode_StateMachine_0.Hit_Crash1_Hard_L_SM.AnimStateNode_1.Hit_Medium
  nodes [
   n1 = "Sequence Player (additive) 'Drum_Hit_Crash1_Normal_Medium'"(StartPosition=Get LeftArmMaxDelayShortenedBy(), bLoopAnimation=false) {sequence=&Drum_Hit_Crash1_Normal_Medium, startPosition=0.05}
   Output Animation Pose(Result=@n1)
  ]
 }
 Hit_Long {
  path=AnimGraph.AnimGraphNode_StateMachine_1.LeftArmAdditiveHits.AnimStateNode_0.LeftArmHit_B.AnimGraphNode_StateMachine_2.LeftArmHitChooser_StartTime_DelayFloat.AnimStateNode_2.Hit_Crash1_Hard_L.AnimGraphNode_StateMachine_0.Hit_Crash1_Hard_L_SM.AnimStateNode_2.Hit_Long
  nodes [Output Animation Pose(Result="Sequence Player (additive) 'Drum_Hit_Crash1_Slow_Big'"(StartPosition=Get LeftArmMaxDelayShortenedBy(), bLoopAnimation=false) {sequence=&Drum_Hit_Crash1_Slow_Big, startPosition=0.05})]
 }
 Transition {
  path=AnimGraph.AnimGraphNode_StateMachine_1.LeftArmAdditiveHits.AnimStateNode_0.LeftArmHit_B.AnimGraphNode_StateMachine_2.LeftArmHitChooser_StartTime_DelayFloat.AnimStateNode_2.Hit_Crash1_Hard_L.AnimGraphNode_StateMachine_0.Hit_Crash1_Hard_L_SM.AnimStateTransitionNode_4.Transition
  nodes [Result(bCanEnterTransition=Get LeftArmIsShortDelay())]
 }
 Transition {
  path=AnimGraph.AnimGraphNode_StateMachine_1.LeftArmAdditiveHits.AnimStateNode_0.LeftArmHit_B.AnimGraphNode_StateMachine_2.LeftArmHitChooser_StartTime_DelayFloat.AnimStateNode_2.Hit_Crash1_Hard_L.AnimGraphNode_StateMachine_0.Hit_Crash1_Hard_L_SM.AnimStateTransitionNode_5.Transition
  nodes [
   n1 = Get LeftArmIsLongDelay()
   n2 = Get LeftArmIsShortDelay()
   Result(bCanEnterTransition=AND Boolean(A=NOT Boolean(A=@n2), B=NOT Boolean(A=@n1)))
  ]
 }
 Transition {
  path=AnimGraph.AnimGraphNode_StateMachine_1.LeftArmAdditiveHits.AnimStateNode_0.LeftArmHit_B.AnimGraphNode_StateMachine_2.LeftArmHitChooser_StartTime_DelayFloat.AnimStateNode_2.Hit_Crash1_Hard_L.AnimGraphNode_StateMachine_0.Hit_Crash1_Hard_L_SM.AnimStateTransitionNode_6.Transition
  nodes [Result(bCanEnterTransition=Get LeftArmIsLongDelay())]
 }
 Hit_Crash1_Soft_L {
  path=AnimGraph.AnimGraphNode_StateMachine_1.LeftArmAdditiveHits.AnimStateNode_0.LeftArmHit_B.AnimGraphNode_StateMachine_2.LeftArmHitChooser_StartTime_DelayFloat.AnimStateNode_3.Hit_Crash1_Soft_L
  nodes [
   n1 = Get LeftArmMaxDelayShortenedBy()
   n2 = "Sequence Player (additive) 'Drum_Hit_Crash1_Normal_Medium'"(StartPosition=Get LeftArmMaxDelayShortenedBy(), bLoopAnimation=false) {sequence=&Drum_Hit_Crash1_Normal_Medium, startPosition=0.05}
   Blend Poses by bool(
    BlendPose_0="Sequence Player (additive) 'Drum_Hit_Crash1_Slow_Big'"(StartPosition=Get LeftArmMaxDelayShortenedBy(), bLoopAnimation=false) {sequence=&Drum_Hit_Crash1_Slow_Big, startPosition=0.05}
    BlendPose_1=Blend Poses by bool(
     BlendPose_0="Sequence Player (additive) 'Drum_Hit_Crash1_Tap_Fast'"(StartPosition=@n1, bLoopAnimation=false) {sequence=&Drum_Hit_Crash1_Tap_Fast, startPosition=0.05}
     BlendPose_1=@n2
     BlendTime_0=0
     BlendTime_1=0
     bActiveValue=Get LeftArmIsShortDelay()
    ) {blendType=Linear}
    BlendTime_0=0
    BlendTime_1=0
    bActiveValue=Get LeftArmIsLongDelay()
   ) {blendType=Linear}
   AnimGraphNode_StateMachine_0 = Hit_Crash1_Soft_L_SM() {is=title AnimGraphNode_StateMachine, bAllowConduitEntryStates=true}
   Output Animation Pose(Result=@AnimGraphNode_StateMachine_0)
  ]
 }
 Hit_Crash1_Soft_L_SM {
  entry=Conduit
  path=AnimGraph.AnimGraphNode_StateMachine_1.LeftArmAdditiveHits.AnimStateNode_0.LeftArmHit_B.AnimGraphNode_StateMachine_2.LeftArmHitChooser_StartTime_DelayFloat.AnimStateNode_3.Hit_Crash1_Soft_L.AnimGraphNode_StateMachine_0.Hit_Crash1_Soft_L_SM
  ids=id
  nodes [
   AnimStateConduitNode_0 = Conduit()
   AnimStateNode_0 = Hit_Short()
   AnimStateNode_1 = Hit_Medium()
   AnimStateNode_2 = Hit_Long()
   AnimStateTransitionNode_4 = AnimStateTransitionNode() {crossfadeDuration=0}
   AnimStateTransitionNode_5 = AnimStateTransitionNode() {crossfadeDuration=0}
   AnimStateTransitionNode_6 = AnimStateTransitionNode() {crossfadeDuration=0}
  ]
  links [
   {
    dst=Hit_Long
    rule=AnimGraph.AnimGraphNode_StateMachine_1.LeftArmAdditiveHits.AnimStateNode_0.LeftArmHit_B.AnimGraphNode_StateMachine_2.LeftArmHitChooser_StartTime_DelayFloat.AnimStateNode_3.Hit_Crash1_Soft_L.AnimGraphNode_StateMachine_0.Hit_Crash1_Soft_L_SM.AnimStateTransitionNode_6.Transition
    src=Conduit
    via=AnimStateTransitionNode_6
   }
   {
    dst=Hit_Medium
    rule=AnimGraph.AnimGraphNode_StateMachine_1.LeftArmAdditiveHits.AnimStateNode_0.LeftArmHit_B.AnimGraphNode_StateMachine_2.LeftArmHitChooser_StartTime_DelayFloat.AnimStateNode_3.Hit_Crash1_Soft_L.AnimGraphNode_StateMachine_0.Hit_Crash1_Soft_L_SM.AnimStateTransitionNode_5.Transition
    src=Conduit
    via=AnimStateTransitionNode_5
   }
   {
    dst=Hit_Short
    rule=AnimGraph.AnimGraphNode_StateMachine_1.LeftArmAdditiveHits.AnimStateNode_0.LeftArmHit_B.AnimGraphNode_StateMachine_2.LeftArmHitChooser_StartTime_DelayFloat.AnimStateNode_3.Hit_Crash1_Soft_L.AnimGraphNode_StateMachine_0.Hit_Crash1_Soft_L_SM.AnimStateTransitionNode_4.Transition
    src=Conduit
    via=AnimStateTransitionNode_4
   }
  ]
 }
 Conduit {
  path=AnimGraph.AnimGraphNode_StateMachine_1.LeftArmAdditiveHits.AnimStateNode_0.LeftArmHit_B.AnimGraphNode_StateMachine_2.LeftArmHitChooser_StartTime_DelayFloat.AnimStateNode_3.Hit_Crash1_Soft_L.AnimGraphNode_StateMachine_0.Hit_Crash1_Soft_L_SM.AnimStateConduitNode_0.Conduit
  nodes [Result(bCanEnterTransition=true)]
 }
 Hit_Short {
  path=AnimGraph.AnimGraphNode_StateMachine_1.LeftArmAdditiveHits.AnimStateNode_0.LeftArmHit_B.AnimGraphNode_StateMachine_2.LeftArmHitChooser_StartTime_DelayFloat.AnimStateNode_3.Hit_Crash1_Soft_L.AnimGraphNode_StateMachine_0.Hit_Crash1_Soft_L_SM.AnimStateNode_0.Hit_Short
  nodes [Output Animation Pose(Result="Sequence Player (additive) 'Drum_Hit_Crash1_Tap_Fast'"(StartPosition=Get LeftArmMaxDelayShortenedBy(), bLoopAnimation=false) {sequence=&Drum_Hit_Crash1_Tap_Fast, startPosition=0.05})]
 }
 Hit_Medium {
  path=AnimGraph.AnimGraphNode_StateMachine_1.LeftArmAdditiveHits.AnimStateNode_0.LeftArmHit_B.AnimGraphNode_StateMachine_2.LeftArmHitChooser_StartTime_DelayFloat.AnimStateNode_3.Hit_Crash1_Soft_L.AnimGraphNode_StateMachine_0.Hit_Crash1_Soft_L_SM.AnimStateNode_1.Hit_Medium
  nodes [
   n1 = "Sequence Player (additive) 'Drum_Hit_Crash1_Normal_Medium'"(StartPosition=Get LeftArmMaxDelayShortenedBy(), bLoopAnimation=false) {sequence=&Drum_Hit_Crash1_Normal_Medium, startPosition=0.05}
   Output Animation Pose(Result=@n1)
  ]
 }
 Hit_Long {
  path=AnimGraph.AnimGraphNode_StateMachine_1.LeftArmAdditiveHits.AnimStateNode_0.LeftArmHit_B.AnimGraphNode_StateMachine_2.LeftArmHitChooser_StartTime_DelayFloat.AnimStateNode_3.Hit_Crash1_Soft_L.AnimGraphNode_StateMachine_0.Hit_Crash1_Soft_L_SM.AnimStateNode_2.Hit_Long
  nodes [Output Animation Pose(Result="Sequence Player (additive) 'Drum_Hit_Crash1_Slow_Big'"(StartPosition=Get LeftArmMaxDelayShortenedBy(), bLoopAnimation=false) {sequence=&Drum_Hit_Crash1_Slow_Big, startPosition=0.05})]
 }
 Transition {
  path=AnimGraph.AnimGraphNode_StateMachine_1.LeftArmAdditiveHits.AnimStateNode_0.LeftArmHit_B.AnimGraphNode_StateMachine_2.LeftArmHitChooser_StartTime_DelayFloat.AnimStateNode_3.Hit_Crash1_Soft_L.AnimGraphNode_StateMachine_0.Hit_Crash1_Soft_L_SM.AnimStateTransitionNode_4.Transition
  nodes [Result(bCanEnterTransition=Get LeftArmIsShortDelay())]
 }
 Transition {
  path=AnimGraph.AnimGraphNode_StateMachine_1.LeftArmAdditiveHits.AnimStateNode_0.LeftArmHit_B.AnimGraphNode_StateMachine_2.LeftArmHitChooser_StartTime_DelayFloat.AnimStateNode_3.Hit_Crash1_Soft_L.AnimGraphNode_StateMachine_0.Hit_Crash1_Soft_L_SM.AnimStateTransitionNode_5.Transition
  nodes [
   n1 = Get LeftArmIsLongDelay()
   n2 = Get LeftArmIsShortDelay()
   Result(bCanEnterTransition=AND Boolean(A=NOT Boolean(A=@n2), B=NOT Boolean(A=@n1)))
  ]
 }
 Transition {
  path=AnimGraph.AnimGraphNode_StateMachine_1.LeftArmAdditiveHits.AnimStateNode_0.LeftArmHit_B.AnimGraphNode_StateMachine_2.LeftArmHitChooser_StartTime_DelayFloat.AnimStateNode_3.Hit_Crash1_Soft_L.AnimGraphNode_StateMachine_0.Hit_Crash1_Soft_L_SM.AnimStateTransitionNode_6.Transition
  nodes [Result(bCanEnterTransition=Get LeftArmIsLongDelay())]
 }
 Hit_Soft {
  path=AnimGraph.AnimGraphNode_StateMachine_1.LeftArmAdditiveHits.AnimStateNode_0.LeftArmHit_B.AnimGraphNode_StateMachine_2.LeftArmHitChooser_StartTime_DelayFloat.AnimStateNode_4.Hit_Soft
  nodes [Output Animation Pose(Result="Sequence Player (additive) 'Drum_Hit_SnareSoft_Tap_Fast'"!(StartPosition:float=Get LeftArmMaxDelayShortenedBy(), bLoopAnimation:bool=False) -> (Pose) {bLoopAnimation=false, sequence=/SparksCharacterCommon/Animation/MainPlayer/BandPlay/Drum/Hits/Tap_Fast/Drum_Hit_SnareSoft_Tap_Fast})]
 }
 Hit_Crash2_Soft_L {
  path=AnimGraph.AnimGraphNode_StateMachine_1.LeftArmAdditiveHits.AnimStateNode_0.LeftArmHit_B.AnimGraphNode_StateMachine_2.LeftArmHitChooser_StartTime_DelayFloat.AnimStateNode_5.Hit_Crash2_Soft_L
  nodes [
   n1 = Get LeftArmMaxDelayShortenedBy()
   n2 = "Sequence Player (additive) 'Drum_Hit_Crash2_Normal_Medium'"(StartPosition=Get LeftArmMaxDelayShortenedBy(), bLoopAnimation=false) {sequence=&Drum_Hit_Crash2_Normal_Medium, startPosition=0.05}
   Blend Poses by bool(
    BlendPose_0="Sequence Player (additive) 'Drum_Hit_Crash2_Slow_Big'"(StartPosition=Get LeftArmMaxDelayShortenedBy(), bLoopAnimation=false) {sequence=&Drum_Hit_Crash2_Slow_Big, startPosition=0.05}
    BlendPose_1=Blend Poses by bool(
     BlendPose_0="Sequence Player (additive) 'Drum_Hit_Crash2_Tap_Fast'"(StartPosition=@n1, bLoopAnimation=false) {sequence=&Drum_Hit_Crash2_Tap_Fast, startPosition=0.05}
     BlendPose_1=@n2
     BlendTime_0=0
     BlendTime_1=0
     bActiveValue=Get LeftArmIsShortDelay()
    ) {blendType=Linear}
    BlendTime_0=0
    BlendTime_1=0
    bActiveValue=Get LeftArmIsLongDelay()
   ) {blendType=Linear}
   AnimGraphNode_StateMachine_0 = Hit_Crash2_Soft_L_SM() {is=title AnimGraphNode_StateMachine, bAllowConduitEntryStates=true}
   Output Animation Pose(Result=@AnimGraphNode_StateMachine_0)
  ]
 }
 Hit_Crash2_Soft_L_SM {
  entry=Conduit
  path=AnimGraph.AnimGraphNode_StateMachine_1.LeftArmAdditiveHits.AnimStateNode_0.LeftArmHit_B.AnimGraphNode_StateMachine_2.LeftArmHitChooser_StartTime_DelayFloat.AnimStateNode_5.Hit_Crash2_Soft_L.AnimGraphNode_StateMachine_0.Hit_Crash2_Soft_L_SM
  ids=id
  nodes [
   AnimStateConduitNode_0 = Conduit()
   AnimStateNode_0 = Hit_Short()
   AnimStateNode_1 = Hit_Medium()
   AnimStateNode_2 = Hit_Long()
   AnimStateTransitionNode_4 = AnimStateTransitionNode() {crossfadeDuration=0}
   AnimStateTransitionNode_5 = AnimStateTransitionNode() {crossfadeDuration=0}
   AnimStateTransitionNode_6 = AnimStateTransitionNode() {crossfadeDuration=0}
  ]
  links [
   {
    dst=Hit_Long
    rule=AnimGraph.AnimGraphNode_StateMachine_1.LeftArmAdditiveHits.AnimStateNode_0.LeftArmHit_B.AnimGraphNode_StateMachine_2.LeftArmHitChooser_StartTime_DelayFloat.AnimStateNode_5.Hit_Crash2_Soft_L.AnimGraphNode_StateMachine_0.Hit_Crash2_Soft_L_SM.AnimStateTransitionNode_6.Transition
    src=Conduit
    via=AnimStateTransitionNode_6
   }
   {
    dst=Hit_Medium
    rule=AnimGraph.AnimGraphNode_StateMachine_1.LeftArmAdditiveHits.AnimStateNode_0.LeftArmHit_B.AnimGraphNode_StateMachine_2.LeftArmHitChooser_StartTime_DelayFloat.AnimStateNode_5.Hit_Crash2_Soft_L.AnimGraphNode_StateMachine_0.Hit_Crash2_Soft_L_SM.AnimStateTransitionNode_5.Transition
    src=Conduit
    via=AnimStateTransitionNode_5
   }
   {
    dst=Hit_Short
    rule=AnimGraph.AnimGraphNode_StateMachine_1.LeftArmAdditiveHits.AnimStateNode_0.LeftArmHit_B.AnimGraphNode_StateMachine_2.LeftArmHitChooser_StartTime_DelayFloat.AnimStateNode_5.Hit_Crash2_Soft_L.AnimGraphNode_StateMachine_0.Hit_Crash2_Soft_L_SM.AnimStateTransitionNode_4.Transition
    src=Conduit
    via=AnimStateTransitionNode_4
   }
  ]
 }
 Conduit {
  path=AnimGraph.AnimGraphNode_StateMachine_1.LeftArmAdditiveHits.AnimStateNode_0.LeftArmHit_B.AnimGraphNode_StateMachine_2.LeftArmHitChooser_StartTime_DelayFloat.AnimStateNode_5.Hit_Crash2_Soft_L.AnimGraphNode_StateMachine_0.Hit_Crash2_Soft_L_SM.AnimStateConduitNode_0.Conduit
  nodes [Result(bCanEnterTransition=true)]
 }
 Hit_Short {
  path=AnimGraph.AnimGraphNode_StateMachine_1.LeftArmAdditiveHits.AnimStateNode_0.LeftArmHit_B.AnimGraphNode_StateMachine_2.LeftArmHitChooser_StartTime_DelayFloat.AnimStateNode_5.Hit_Crash2_Soft_L.AnimGraphNode_StateMachine_0.Hit_Crash2_Soft_L_SM.AnimStateNode_0.Hit_Short
  nodes [Output Animation Pose(Result="Sequence Player (additive) 'Drum_Hit_Crash2_Tap_Fast'"(StartPosition=Get LeftArmMaxDelayShortenedBy(), bLoopAnimation=false) {sequence=&Drum_Hit_Crash2_Tap_Fast, startPosition=0.05})]
 }
 Hit_Medium {
  path=AnimGraph.AnimGraphNode_StateMachine_1.LeftArmAdditiveHits.AnimStateNode_0.LeftArmHit_B.AnimGraphNode_StateMachine_2.LeftArmHitChooser_StartTime_DelayFloat.AnimStateNode_5.Hit_Crash2_Soft_L.AnimGraphNode_StateMachine_0.Hit_Crash2_Soft_L_SM.AnimStateNode_1.Hit_Medium
  nodes [
   n1 = "Sequence Player (additive) 'Drum_Hit_Crash2_Normal_Medium'"(StartPosition=Get LeftArmMaxDelayShortenedBy(), bLoopAnimation=false) {sequence=&Drum_Hit_Crash2_Normal_Medium, startPosition=0.05}
   Output Animation Pose(Result=@n1)
  ]
 }
 Hit_Long {
  path=AnimGraph.AnimGraphNode_StateMachine_1.LeftArmAdditiveHits.AnimStateNode_0.LeftArmHit_B.AnimGraphNode_StateMachine_2.LeftArmHitChooser_StartTime_DelayFloat.AnimStateNode_5.Hit_Crash2_Soft_L.AnimGraphNode_StateMachine_0.Hit_Crash2_Soft_L_SM.AnimStateNode_2.Hit_Long
  nodes [Output Animation Pose(Result="Sequence Player (additive) 'Drum_Hit_Crash2_Slow_Big'"(StartPosition=Get LeftArmMaxDelayShortenedBy(), bLoopAnimation=false) {sequence=&Drum_Hit_Crash2_Slow_Big, startPosition=0.05})]
 }
 Transition {
  path=AnimGraph.AnimGraphNode_StateMachine_1.LeftArmAdditiveHits.AnimStateNode_0.LeftArmHit_B.AnimGraphNode_StateMachine_2.LeftArmHitChooser_StartTime_DelayFloat.AnimStateNode_5.Hit_Crash2_Soft_L.AnimGraphNode_StateMachine_0.Hit_Crash2_Soft_L_SM.AnimStateTransitionNode_4.Transition
  nodes [Result(bCanEnterTransition=Get LeftArmIsShortDelay())]
 }
 Transition {
  path=AnimGraph.AnimGraphNode_StateMachine_1.LeftArmAdditiveHits.AnimStateNode_0.LeftArmHit_B.AnimGraphNode_StateMachine_2.LeftArmHitChooser_StartTime_DelayFloat.AnimStateNode_5.Hit_Crash2_Soft_L.AnimGraphNode_StateMachine_0.Hit_Crash2_Soft_L_SM.AnimStateTransitionNode_5.Transition
  nodes [
   n1 = Get LeftArmIsLongDelay()
   n2 = Get LeftArmIsShortDelay()
   Result(bCanEnterTransition=AND Boolean(A=NOT Boolean(A=@n2), B=NOT Boolean(A=@n1)))
  ]
 }
 Transition {
  path=AnimGraph.AnimGraphNode_StateMachine_1.LeftArmAdditiveHits.AnimStateNode_0.LeftArmHit_B.AnimGraphNode_StateMachine_2.LeftArmHitChooser_StartTime_DelayFloat.AnimStateNode_5.Hit_Crash2_Soft_L.AnimGraphNode_StateMachine_0.Hit_Crash2_Soft_L_SM.AnimStateTransitionNode_6.Transition
  nodes [Result(bCanEnterTransition=Get LeftArmIsLongDelay())]
 }
 Hit_Ride_L {
  path=AnimGraph.AnimGraphNode_StateMachine_1.LeftArmAdditiveHits.AnimStateNode_0.LeftArmHit_B.AnimGraphNode_StateMachine_2.LeftArmHitChooser_StartTime_DelayFloat.AnimStateNode_6.Hit_Ride_L
  nodes [
   n1 = Get LeftArmMaxDelayShortenedBy()
   n2 = "Sequence Player (additive) 'Drum_Hit_Ride_Normal_Medium'"!(StartPosition:float=Get LeftArmMaxDelayShortenedBy(), bLoopAnimation:bool=false) -> (Pose) {sequence=/SparksCharacterCommon/Animation/MainPlayer/BandPlay/Drum/Hits/Normal_Medium/Drum_Hit_Ride_Normal_Medium, startPosition=0.05}
   Blend Poses by bool(
    BlendPose_0="Sequence Player (additive) 'Drum_Hit_Ride_Slow_Big'"!(StartPosition:float=Get LeftArmMaxDelayShortenedBy(), bLoopAnimation:bool=false) -> (Pose) {sequence=/SparksCharacterFull/Animation/MainPlayer/BandPlay/Drum/Hits/Slow_Big/Drum_Hit_Ride_Slow_Big, startPosition=0.05}
    BlendPose_1=Blend Poses by bool(
     BlendPose_0="Sequence Player (additive) 'Drum_Hit_Ride_Tap_Fast'"!(StartPosition:float=@n1, bLoopAnimation:bool=false) -> (Pose) {sequence=/SparksCharacterCommon/Animation/MainPlayer/BandPlay/Drum/Hits/Tap_Fast/Drum_Hit_Ride_Tap_Fast, startPosition=0.05}
     BlendPose_1=@n2
     BlendTime_0=0
     BlendTime_1=0
     bActiveValue=Get LeftArmIsShortDelay()
    ) {blendType=Linear}
    BlendTime_0=0
    BlendTime_1=0
    bActiveValue=Get LeftArmIsLongDelay()
   ) {blendType=Linear}
   AnimGraphNode_StateMachine_0 = Hit_Ride_L_SM() {is=title AnimGraphNode_StateMachine, bAllowConduitEntryStates=true}
   Output Animation Pose(Result=@AnimGraphNode_StateMachine_0)
  ]
 }
 Hit_Ride_L_SM {
  entry=Conduit
  path=AnimGraph.AnimGraphNode_StateMachine_1.LeftArmAdditiveHits.AnimStateNode_0.LeftArmHit_B.AnimGraphNode_StateMachine_2.LeftArmHitChooser_StartTime_DelayFloat.AnimStateNode_6.Hit_Ride_L.AnimGraphNode_StateMachine_0.Hit_Ride_L_SM
  ids=id
  nodes [
   AnimStateConduitNode_0 = Conduit()
   AnimStateNode_0 = Hit_Short()
   AnimStateNode_1 = Hit_Medium()
   AnimStateNode_2 = Hit_Long()
   AnimStateTransitionNode_4 = AnimStateTransitionNode() {crossfadeDuration=0}
   AnimStateTransitionNode_5 = AnimStateTransitionNode() {crossfadeDuration=0}
   AnimStateTransitionNode_6 = AnimStateTransitionNode() {crossfadeDuration=0}
  ]
  links [
   {
    dst=Hit_Long
    rule=AnimGraph.AnimGraphNode_StateMachine_1.LeftArmAdditiveHits.AnimStateNode_0.LeftArmHit_B.AnimGraphNode_StateMachine_2.LeftArmHitChooser_StartTime_DelayFloat.AnimStateNode_6.Hit_Ride_L.AnimGraphNode_StateMachine_0.Hit_Ride_L_SM.AnimStateTransitionNode_6.Transition
    src=Conduit
    via=AnimStateTransitionNode_6
   }
   {
    dst=Hit_Medium
    rule=AnimGraph.AnimGraphNode_StateMachine_1.LeftArmAdditiveHits.AnimStateNode_0.LeftArmHit_B.AnimGraphNode_StateMachine_2.LeftArmHitChooser_StartTime_DelayFloat.AnimStateNode_6.Hit_Ride_L.AnimGraphNode_StateMachine_0.Hit_Ride_L_SM.AnimStateTransitionNode_5.Transition
    src=Conduit
    via=AnimStateTransitionNode_5
   }
   {
    dst=Hit_Short
    rule=AnimGraph.AnimGraphNode_StateMachine_1.LeftArmAdditiveHits.AnimStateNode_0.LeftArmHit_B.AnimGraphNode_StateMachine_2.LeftArmHitChooser_StartTime_DelayFloat.AnimStateNode_6.Hit_Ride_L.AnimGraphNode_StateMachine_0.Hit_Ride_L_SM.AnimStateTransitionNode_4.Transition
    src=Conduit
    via=AnimStateTransitionNode_4
   }
  ]
 }
 Conduit {
  path=AnimGraph.AnimGraphNode_StateMachine_1.LeftArmAdditiveHits.AnimStateNode_0.LeftArmHit_B.AnimGraphNode_StateMachine_2.LeftArmHitChooser_StartTime_DelayFloat.AnimStateNode_6.Hit_Ride_L.AnimGraphNode_StateMachine_0.Hit_Ride_L_SM.AnimStateConduitNode_0.Conduit
  nodes [Result(bCanEnterTransition=true)]
 }
 Hit_Short {
  path=AnimGraph.AnimGraphNode_StateMachine_1.LeftArmAdditiveHits.AnimStateNode_0.LeftArmHit_B.AnimGraphNode_StateMachine_2.LeftArmHitChooser_StartTime_DelayFloat.AnimStateNode_6.Hit_Ride_L.AnimGraphNode_StateMachine_0.Hit_Ride_L_SM.AnimStateNode_0.Hit_Short
  nodes [Output Animation Pose(Result="Sequence Player (additive) 'Drum_Hit_Ride_Tap_Fast'"(StartPosition=Get LeftArmMaxDelayShortenedBy(), bLoopAnimation=false) {sequence=&Drum_Hit_Ride_Tap_Fast, startPosition=0.05})]
 }
 Hit_Medium {
  path=AnimGraph.AnimGraphNode_StateMachine_1.LeftArmAdditiveHits.AnimStateNode_0.LeftArmHit_B.AnimGraphNode_StateMachine_2.LeftArmHitChooser_StartTime_DelayFloat.AnimStateNode_6.Hit_Ride_L.AnimGraphNode_StateMachine_0.Hit_Ride_L_SM.AnimStateNode_1.Hit_Medium
  nodes [
   n1 = "Sequence Player (additive) 'Drum_Hit_Ride_Normal_Medium'"(StartPosition=Get LeftArmMaxDelayShortenedBy(), bLoopAnimation=false) {sequence=&Drum_Hit_Ride_Normal_Medium, startPosition=0.05}
   Output Animation Pose(Result=@n1)
  ]
 }
 Hit_Long {
  path=AnimGraph.AnimGraphNode_StateMachine_1.LeftArmAdditiveHits.AnimStateNode_0.LeftArmHit_B.AnimGraphNode_StateMachine_2.LeftArmHitChooser_StartTime_DelayFloat.AnimStateNode_6.Hit_Ride_L.AnimGraphNode_StateMachine_0.Hit_Ride_L_SM.AnimStateNode_2.Hit_Long
  nodes [Output Animation Pose(Result="Sequence Player (additive) 'Drum_Hit_Ride_Slow_Big'"(StartPosition=Get LeftArmMaxDelayShortenedBy(), bLoopAnimation=false) {sequence=&Drum_Hit_Ride_Slow_Big, startPosition=0.05})]
 }
 Transition {
  path=AnimGraph.AnimGraphNode_StateMachine_1.LeftArmAdditiveHits.AnimStateNode_0.LeftArmHit_B.AnimGraphNode_StateMachine_2.LeftArmHitChooser_StartTime_DelayFloat.AnimStateNode_6.Hit_Ride_L.AnimGraphNode_StateMachine_0.Hit_Ride_L_SM.AnimStateTransitionNode_4.Transition
  nodes [Result(bCanEnterTransition=Get LeftArmIsShortDelay())]
 }
 Transition {
  path=AnimGraph.AnimGraphNode_StateMachine_1.LeftArmAdditiveHits.AnimStateNode_0.LeftArmHit_B.AnimGraphNode_StateMachine_2.LeftArmHitChooser_StartTime_DelayFloat.AnimStateNode_6.Hit_Ride_L.AnimGraphNode_StateMachine_0.Hit_Ride_L_SM.AnimStateTransitionNode_5.Transition
  nodes [
   n1 = Get LeftArmIsLongDelay()
   n2 = Get LeftArmIsShortDelay()
   Result(bCanEnterTransition=AND Boolean(A=NOT Boolean(A=@n2), B=NOT Boolean(A=@n1)))
  ]
 }
 Transition {
  path=AnimGraph.AnimGraphNode_StateMachine_1.LeftArmAdditiveHits.AnimStateNode_0.LeftArmHit_B.AnimGraphNode_StateMachine_2.LeftArmHitChooser_StartTime_DelayFloat.AnimStateNode_6.Hit_Ride_L.AnimGraphNode_StateMachine_0.Hit_Ride_L_SM.AnimStateTransitionNode_6.Transition
  nodes [Result(bCanEnterTransition=Get LeftArmIsLongDelay())]
 }
 Hit_Tom1_L {
  path=AnimGraph.AnimGraphNode_StateMachine_1.LeftArmAdditiveHits.AnimStateNode_0.LeftArmHit_B.AnimGraphNode_StateMachine_2.LeftArmHitChooser_StartTime_DelayFloat.AnimStateNode_7.Hit_Tom1_L
  nodes [
   n1 = Get LeftArmMaxDelayShortenedBy()
   n2 = "Sequence Player (additive) 'Drum_Hit_Tom1_Normal_Medium'"!(StartPosition:float=Get LeftArmMaxDelayShortenedBy(), bLoopAnimation:bool=false) -> (Pose) {sequence=/SparksCharacterCommon/Animation/MainPlayer/BandPlay/Drum/Hits/Normal_Medium/Drum_Hit_Tom1_Normal_Medium, startPosition=0.05}
   Blend Poses by bool(
    BlendPose_0="Sequence Player (additive) 'Drum_Hit_Tom1_Slow_Big'"!(StartPosition:float=Get LeftArmMaxDelayShortenedBy(), bLoopAnimation:bool=false) -> (Pose) {sequence=/SparksCharacterFull/Animation/MainPlayer/BandPlay/Drum/Hits/Slow_Big/Drum_Hit_Tom1_Slow_Big, startPosition=0.05}
    BlendPose_1=Blend Poses by bool(
     BlendPose_0="Sequence Player (additive) 'Drum_Hit_Tom1_Tap_Fast'"!(StartPosition:float=@n1, bLoopAnimation:bool=false) -> (Pose) {sequence=/SparksCharacterCommon/Animation/MainPlayer/BandPlay/Drum/Hits/Tap_Fast/Drum_Hit_Tom1_Tap_Fast, startPosition=0.05}
     BlendPose_1=@n2
     BlendTime_0=0
     BlendTime_1=0
     bActiveValue=Get LeftArmIsShortDelay()
    ) {blendType=Linear}
    BlendTime_0=0
    BlendTime_1=0
    bActiveValue=Get LeftArmIsLongDelay()
   ) {blendType=Linear}
   AnimGraphNode_StateMachine_0 = Hit_Tom1_L_SM() {is=title AnimGraphNode_StateMachine, bAllowConduitEntryStates=true}
   Output Animation Pose(Result=@AnimGraphNode_StateMachine_0)
  ]
 }
 Hit_Tom1_L_SM {
  entry=Conduit
  path=AnimGraph.AnimGraphNode_StateMachine_1.LeftArmAdditiveHits.AnimStateNode_0.LeftArmHit_B.AnimGraphNode_StateMachine_2.LeftArmHitChooser_StartTime_DelayFloat.AnimStateNode_7.Hit_Tom1_L.AnimGraphNode_StateMachine_0.Hit_Tom1_L_SM
  ids=id
  nodes [
   AnimStateConduitNode_0 = Conduit()
   AnimStateNode_0 = Hit_Short()
   AnimStateNode_1 = Hit_Medium()
   AnimStateNode_2 = Hit_Long()
   AnimStateTransitionNode_4 = AnimStateTransitionNode() {crossfadeDuration=0}
   AnimStateTransitionNode_5 = AnimStateTransitionNode() {crossfadeDuration=0}
   AnimStateTransitionNode_6 = AnimStateTransitionNode() {crossfadeDuration=0}
  ]
  links [
   {
    dst=Hit_Long
    rule=AnimGraph.AnimGraphNode_StateMachine_1.LeftArmAdditiveHits.AnimStateNode_0.LeftArmHit_B.AnimGraphNode_StateMachine_2.LeftArmHitChooser_StartTime_DelayFloat.AnimStateNode_7.Hit_Tom1_L.AnimGraphNode_StateMachine_0.Hit_Tom1_L_SM.AnimStateTransitionNode_6.Transition
    src=Conduit
    via=AnimStateTransitionNode_6
   }
   {
    dst=Hit_Medium
    rule=AnimGraph.AnimGraphNode_StateMachine_1.LeftArmAdditiveHits.AnimStateNode_0.LeftArmHit_B.AnimGraphNode_StateMachine_2.LeftArmHitChooser_StartTime_DelayFloat.AnimStateNode_7.Hit_Tom1_L.AnimGraphNode_StateMachine_0.Hit_Tom1_L_SM.AnimStateTransitionNode_5.Transition
    src=Conduit
    via=AnimStateTransitionNode_5
   }
   {
    dst=Hit_Short
    rule=AnimGraph.AnimGraphNode_StateMachine_1.LeftArmAdditiveHits.AnimStateNode_0.LeftArmHit_B.AnimGraphNode_StateMachine_2.LeftArmHitChooser_StartTime_DelayFloat.AnimStateNode_7.Hit_Tom1_L.AnimGraphNode_StateMachine_0.Hit_Tom1_L_SM.AnimStateTransitionNode_4.Transition
    src=Conduit
    via=AnimStateTransitionNode_4
   }
  ]
 }
 Conduit {
  path=AnimGraph.AnimGraphNode_StateMachine_1.LeftArmAdditiveHits.AnimStateNode_0.LeftArmHit_B.AnimGraphNode_StateMachine_2.LeftArmHitChooser_StartTime_DelayFloat.AnimStateNode_7.Hit_Tom1_L.AnimGraphNode_StateMachine_0.Hit_Tom1_L_SM.AnimStateConduitNode_0.Conduit
  nodes [Result(bCanEnterTransition=true)]
 }
 Hit_Short {
  path=AnimGraph.AnimGraphNode_StateMachine_1.LeftArmAdditiveHits.AnimStateNode_0.LeftArmHit_B.AnimGraphNode_StateMachine_2.LeftArmHitChooser_StartTime_DelayFloat.AnimStateNode_7.Hit_Tom1_L.AnimGraphNode_StateMachine_0.Hit_Tom1_L_SM.AnimStateNode_0.Hit_Short
  nodes [Output Animation Pose(Result="Sequence Player (additive) 'Drum_Hit_Tom1_Tap_Fast'"(StartPosition=Get LeftArmMaxDelayShortenedBy(), bLoopAnimation=false) {sequence=&Drum_Hit_Tom1_Tap_Fast, startPosition=0.05})]
 }
 Hit_Medium {
  path=AnimGraph.AnimGraphNode_StateMachine_1.LeftArmAdditiveHits.AnimStateNode_0.LeftArmHit_B.AnimGraphNode_StateMachine_2.LeftArmHitChooser_StartTime_DelayFloat.AnimStateNode_7.Hit_Tom1_L.AnimGraphNode_StateMachine_0.Hit_Tom1_L_SM.AnimStateNode_1.Hit_Medium
  nodes [
   n1 = "Sequence Player (additive) 'Drum_Hit_Tom1_Normal_Medium'"(StartPosition=Get LeftArmMaxDelayShortenedBy(), bLoopAnimation=false) {sequence=&Drum_Hit_Tom1_Normal_Medium, startPosition=0.05}
   Output Animation Pose(Result=@n1)
  ]
 }
 Hit_Long {
  path=AnimGraph.AnimGraphNode_StateMachine_1.LeftArmAdditiveHits.AnimStateNode_0.LeftArmHit_B.AnimGraphNode_StateMachine_2.LeftArmHitChooser_StartTime_DelayFloat.AnimStateNode_7.Hit_Tom1_L.AnimGraphNode_StateMachine_0.Hit_Tom1_L_SM.AnimStateNode_2.Hit_Long
  nodes [Output Animation Pose(Result="Sequence Player (additive) 'Drum_Hit_Tom1_Slow_Big'"(StartPosition=Get LeftArmMaxDelayShortenedBy(), bLoopAnimation=false) {sequence=&Drum_Hit_Tom1_Slow_Big, startPosition=0.05})]
 }
 Transition {
  path=AnimGraph.AnimGraphNode_StateMachine_1.LeftArmAdditiveHits.AnimStateNode_0.LeftArmHit_B.AnimGraphNode_StateMachine_2.LeftArmHitChooser_StartTime_DelayFloat.AnimStateNode_7.Hit_Tom1_L.AnimGraphNode_StateMachine_0.Hit_Tom1_L_SM.AnimStateTransitionNode_4.Transition
  nodes [Result(bCanEnterTransition=Get LeftArmIsShortDelay())]
 }
 Transition {
  path=AnimGraph.AnimGraphNode_StateMachine_1.LeftArmAdditiveHits.AnimStateNode_0.LeftArmHit_B.AnimGraphNode_StateMachine_2.LeftArmHitChooser_StartTime_DelayFloat.AnimStateNode_7.Hit_Tom1_L.AnimGraphNode_StateMachine_0.Hit_Tom1_L_SM.AnimStateTransitionNode_5.Transition
  nodes [
   n1 = Get LeftArmIsLongDelay()
   n2 = Get LeftArmIsShortDelay()
   Result(bCanEnterTransition=AND Boolean(A=NOT Boolean(A=@n2), B=NOT Boolean(A=@n1)))
  ]
 }
 Transition {
  path=AnimGraph.AnimGraphNode_StateMachine_1.LeftArmAdditiveHits.AnimStateNode_0.LeftArmHit_B.AnimGraphNode_StateMachine_2.LeftArmHitChooser_StartTime_DelayFloat.AnimStateNode_7.Hit_Tom1_L.AnimGraphNode_StateMachine_0.Hit_Tom1_L_SM.AnimStateTransitionNode_6.Transition
  nodes [Result(bCanEnterTransition=Get LeftArmIsLongDelay())]
 }
 Hit_Tom2_L {
  path=AnimGraph.AnimGraphNode_StateMachine_1.LeftArmAdditiveHits.AnimStateNode_0.LeftArmHit_B.AnimGraphNode_StateMachine_2.LeftArmHitChooser_StartTime_DelayFloat.AnimStateNode_8.Hit_Tom2_L
  nodes [
   n1 = Get LeftArmMaxDelayShortenedBy()
   n2 = "Sequence Player (additive) 'Drum_Hit_Tom2_Normal_Medium'"!(StartPosition:float=Get LeftArmMaxDelayShortenedBy(), bLoopAnimation:bool=false) -> (Pose) {sequence=/SparksCharacterCommon/Animation/MainPlayer/BandPlay/Drum/Hits/Normal_Medium/Drum_Hit_Tom2_Normal_Medium, startPosition=0.05}
   Blend Poses by bool(
    BlendPose_0="Sequence Player (additive) 'Drum_Hit_Tom2_Slow_Big'"!(StartPosition:float=Get LeftArmMaxDelayShortenedBy(), bLoopAnimation:bool=false) -> (Pose) {sequence=/SparksCharacterFull/Animation/MainPlayer/BandPlay/Drum/Hits/Slow_Big/Drum_Hit_Tom2_Slow_Big, startPosition=0.05}
    BlendPose_1=Blend Poses by bool(
     BlendPose_0="Sequence Player (additive) 'Drum_Hit_Tom2_Tap_Fast'"!(StartPosition:float=@n1, bLoopAnimation:bool=false) -> (Pose) {sequence=/SparksCharacterCommon/Animation/MainPlayer/BandPlay/Drum/Hits/Tap_Fast/Drum_Hit_Tom2_Tap_Fast, startPosition=0.05}
     BlendPose_1=@n2
     BlendTime_0=0
     BlendTime_1=0
     bActiveValue=Get LeftArmIsShortDelay()
    ) {blendType=Linear}
    BlendTime_0=0
    BlendTime_1=0
    bActiveValue=Get LeftArmIsLongDelay()
   ) {blendType=Linear}
   AnimGraphNode_StateMachine_0 = Hit_Tom2_L_SM() {is=title AnimGraphNode_StateMachine, bAllowConduitEntryStates=true}
   Output Animation Pose(Result=@AnimGraphNode_StateMachine_0)
  ]
 }
 Hit_Tom2_L_SM {
  entry=Conduit
  path=AnimGraph.AnimGraphNode_StateMachine_1.LeftArmAdditiveHits.AnimStateNode_0.LeftArmHit_B.AnimGraphNode_StateMachine_2.LeftArmHitChooser_StartTime_DelayFloat.AnimStateNode_8.Hit_Tom2_L.AnimGraphNode_StateMachine_0.Hit_Tom2_L_SM
  ids=id
  nodes [
   AnimStateConduitNode_0 = Conduit()
   AnimStateNode_0 = Hit_Short()
   AnimStateNode_1 = Hit_Medium()
   AnimStateNode_2 = Hit_Long()
   AnimStateTransitionNode_4 = AnimStateTransitionNode() {crossfadeDuration=0}
   AnimStateTransitionNode_5 = AnimStateTransitionNode() {crossfadeDuration=0}
   AnimStateTransitionNode_6 = AnimStateTransitionNode() {crossfadeDuration=0}
  ]
  links [
   {
    dst=Hit_Long
    rule=AnimGraph.AnimGraphNode_StateMachine_1.LeftArmAdditiveHits.AnimStateNode_0.LeftArmHit_B.AnimGraphNode_StateMachine_2.LeftArmHitChooser_StartTime_DelayFloat.AnimStateNode_8.Hit_Tom2_L.AnimGraphNode_StateMachine_0.Hit_Tom2_L_SM.AnimStateTransitionNode_6.Transition
    src=Conduit
    via=AnimStateTransitionNode_6
   }
   {
    dst=Hit_Medium
    rule=AnimGraph.AnimGraphNode_StateMachine_1.LeftArmAdditiveHits.AnimStateNode_0.LeftArmHit_B.AnimGraphNode_StateMachine_2.LeftArmHitChooser_StartTime_DelayFloat.AnimStateNode_8.Hit_Tom2_L.AnimGraphNode_StateMachine_0.Hit_Tom2_L_SM.AnimStateTransitionNode_5.Transition
    src=Conduit
    via=AnimStateTransitionNode_5
   }
   {
    dst=Hit_Short
    rule=AnimGraph.AnimGraphNode_StateMachine_1.LeftArmAdditiveHits.AnimStateNode_0.LeftArmHit_B.AnimGraphNode_StateMachine_2.LeftArmHitChooser_StartTime_DelayFloat.AnimStateNode_8.Hit_Tom2_L.AnimGraphNode_StateMachine_0.Hit_Tom2_L_SM.AnimStateTransitionNode_4.Transition
    src=Conduit
    via=AnimStateTransitionNode_4
   }
  ]
 }
 Conduit {
  path=AnimGraph.AnimGraphNode_StateMachine_1.LeftArmAdditiveHits.AnimStateNode_0.LeftArmHit_B.AnimGraphNode_StateMachine_2.LeftArmHitChooser_StartTime_DelayFloat.AnimStateNode_8.Hit_Tom2_L.AnimGraphNode_StateMachine_0.Hit_Tom2_L_SM.AnimStateConduitNode_0.Conduit
  nodes [Result(bCanEnterTransition=true)]
 }
 Hit_Short {
  path=AnimGraph.AnimGraphNode_StateMachine_1.LeftArmAdditiveHits.AnimStateNode_0.LeftArmHit_B.AnimGraphNode_StateMachine_2.LeftArmHitChooser_StartTime_DelayFloat.AnimStateNode_8.Hit_Tom2_L.AnimGraphNode_StateMachine_0.Hit_Tom2_L_SM.AnimStateNode_0.Hit_Short
  nodes [Output Animation Pose(Result="Sequence Player (additive) 'Drum_Hit_Tom2_Tap_Fast'"(StartPosition=Get LeftArmMaxDelayShortenedBy(), bLoopAnimation=false) {sequence=&Drum_Hit_Tom2_Tap_Fast, startPosition=0.05})]
 }
 Hit_Medium {
  path=AnimGraph.AnimGraphNode_StateMachine_1.LeftArmAdditiveHits.AnimStateNode_0.LeftArmHit_B.AnimGraphNode_StateMachine_2.LeftArmHitChooser_StartTime_DelayFloat.AnimStateNode_8.Hit_Tom2_L.AnimGraphNode_StateMachine_0.Hit_Tom2_L_SM.AnimStateNode_1.Hit_Medium
  nodes [
   n1 = "Sequence Player (additive) 'Drum_Hit_Tom2_Normal_Medium'"(StartPosition=Get LeftArmMaxDelayShortenedBy(), bLoopAnimation=false) {sequence=&Drum_Hit_Tom2_Normal_Medium, startPosition=0.05}
   Output Animation Pose(Result=@n1)
  ]
 }
 Hit_Long {
  path=AnimGraph.AnimGraphNode_StateMachine_1.LeftArmAdditiveHits.AnimStateNode_0.LeftArmHit_B.AnimGraphNode_StateMachine_2.LeftArmHitChooser_StartTime_DelayFloat.AnimStateNode_8.Hit_Tom2_L.AnimGraphNode_StateMachine_0.Hit_Tom2_L_SM.AnimStateNode_2.Hit_Long
  nodes [Output Animation Pose(Result="Sequence Player (additive) 'Drum_Hit_Tom2_Slow_Big'"(StartPosition=Get LeftArmMaxDelayShortenedBy(), bLoopAnimation=false) {sequence=&Drum_Hit_Tom2_Slow_Big, startPosition=0.05})]
 }
 Transition {
  path=AnimGraph.AnimGraphNode_StateMachine_1.LeftArmAdditiveHits.AnimStateNode_0.LeftArmHit_B.AnimGraphNode_StateMachine_2.LeftArmHitChooser_StartTime_DelayFloat.AnimStateNode_8.Hit_Tom2_L.AnimGraphNode_StateMachine_0.Hit_Tom2_L_SM.AnimStateTransitionNode_4.Transition
  nodes [Result(bCanEnterTransition=Get LeftArmIsShortDelay())]
 }
 Transition {
  path=AnimGraph.AnimGraphNode_StateMachine_1.LeftArmAdditiveHits.AnimStateNode_0.LeftArmHit_B.AnimGraphNode_StateMachine_2.LeftArmHitChooser_StartTime_DelayFloat.AnimStateNode_8.Hit_Tom2_L.AnimGraphNode_StateMachine_0.Hit_Tom2_L_SM.AnimStateTransitionNode_5.Transition
  nodes [
   n1 = Get LeftArmIsLongDelay()
   n2 = Get LeftArmIsShortDelay()
   Result(bCanEnterTransition=AND Boolean(A=NOT Boolean(A=@n2), B=NOT Boolean(A=@n1)))
  ]
 }
 Transition {
  path=AnimGraph.AnimGraphNode_StateMachine_1.LeftArmAdditiveHits.AnimStateNode_0.LeftArmHit_B.AnimGraphNode_StateMachine_2.LeftArmHitChooser_StartTime_DelayFloat.AnimStateNode_8.Hit_Tom2_L.AnimGraphNode_StateMachine_0.Hit_Tom2_L_SM.AnimStateTransitionNode_6.Transition
  nodes [Result(bCanEnterTransition=Get LeftArmIsLongDelay())]
 }
 Hit_FloorTom_L {
  path=AnimGraph.AnimGraphNode_StateMachine_1.LeftArmAdditiveHits.AnimStateNode_0.LeftArmHit_B.AnimGraphNode_StateMachine_2.LeftArmHitChooser_StartTime_DelayFloat.AnimStateNode_9.Hit_FloorTom_L
  nodes [
   n1 = Get LeftArmMaxDelayShortenedBy()
   n2 = "Sequence Player (additive) 'Drum_Hit_FloorTom_Normal_Medium'"!(StartPosition:float=Get LeftArmMaxDelayShortenedBy(), bLoopAnimation:bool=false) -> (Pose) {sequence=/SparksCharacterCommon/Animation/MainPlayer/BandPlay/Drum/Hits/Normal_Medium/Drum_Hit_FloorTom_Normal_Medium, startPosition=0.05}
   Blend Poses by bool(
    BlendPose_0="Sequence Player (additive) 'Drum_Hit_FloorTom_Slow_Big'"!(StartPosition:float=Get LeftArmMaxDelayShortenedBy(), bLoopAnimation:bool=false) -> (Pose) {sequence=/SparksCharacterFull/Animation/MainPlayer/BandPlay/Drum/Hits/Slow_Big/Drum_Hit_FloorTom_Slow_Big, startPosition=0.05}
    BlendPose_1=Blend Poses by bool(
     BlendPose_0="Sequence Player (additive) 'Drum_Hit_FloorTom_Tap_Fast'"!(StartPosition:float=@n1, bLoopAnimation:bool=false) -> (Pose) {sequence=/SparksCharacterCommon/Animation/MainPlayer/BandPlay/Drum/Hits/Tap_Fast/Drum_Hit_FloorTom_Tap_Fast, startPosition=0.05}
     BlendPose_1=@n2
     BlendTime_0=0
     BlendTime_1=0
     bActiveValue=Get LeftArmIsShortDelay()
    ) {blendType=Linear}
    BlendTime_0=0
    BlendTime_1=0
    bActiveValue=Get LeftArmIsLongDelay()
   ) {blendType=Linear}
   AnimGraphNode_StateMachine_0 = Hit_FloorTom_R_SM() {is=title AnimGraphNode_StateMachine, bAllowConduitEntryStates=true}
   Output Animation Pose(Result=@AnimGraphNode_StateMachine_0)
  ]
 }
 Hit_FloorTom_R_SM {
  entry=Conduit
  path=AnimGraph.AnimGraphNode_StateMachine_1.LeftArmAdditiveHits.AnimStateNode_0.LeftArmHit_B.AnimGraphNode_StateMachine_2.LeftArmHitChooser_StartTime_DelayFloat.AnimStateNode_9.Hit_FloorTom_L.AnimGraphNode_StateMachine_0.Hit_FloorTom_R_SM
  ids=id
  nodes [
   AnimStateConduitNode_0 = Conduit()
   AnimStateNode_0 = Hit_Short()
   AnimStateNode_1 = Hit_Medium()
   AnimStateNode_2 = Hit_Long()
   AnimStateTransitionNode_4 = AnimStateTransitionNode() {crossfadeDuration=0}
   AnimStateTransitionNode_5 = AnimStateTransitionNode() {crossfadeDuration=0}
   AnimStateTransitionNode_6 = AnimStateTransitionNode() {crossfadeDuration=0}
  ]
  links [
   {
    dst=Hit_Long
    rule=AnimGraph.AnimGraphNode_StateMachine_1.LeftArmAdditiveHits.AnimStateNode_0.LeftArmHit_B.AnimGraphNode_StateMachine_2.LeftArmHitChooser_StartTime_DelayFloat.AnimStateNode_9.Hit_FloorTom_L.AnimGraphNode_StateMachine_0.Hit_FloorTom_R_SM.AnimStateTransitionNode_6.Transition
    src=Conduit
    via=AnimStateTransitionNode_6
   }
   {
    dst=Hit_Medium
    rule=AnimGraph.AnimGraphNode_StateMachine_1.LeftArmAdditiveHits.AnimStateNode_0.LeftArmHit_B.AnimGraphNode_StateMachine_2.LeftArmHitChooser_StartTime_DelayFloat.AnimStateNode_9.Hit_FloorTom_L.AnimGraphNode_StateMachine_0.Hit_FloorTom_R_SM.AnimStateTransitionNode_5.Transition
    src=Conduit
    via=AnimStateTransitionNode_5
   }
   {
    dst=Hit_Short
    rule=AnimGraph.AnimGraphNode_StateMachine_1.LeftArmAdditiveHits.AnimStateNode_0.LeftArmHit_B.AnimGraphNode_StateMachine_2.LeftArmHitChooser_StartTime_DelayFloat.AnimStateNode_9.Hit_FloorTom_L.AnimGraphNode_StateMachine_0.Hit_FloorTom_R_SM.AnimStateTransitionNode_4.Transition
    src=Conduit
    via=AnimStateTransitionNode_4
   }
  ]
 }
 Conduit {
  path=AnimGraph.AnimGraphNode_StateMachine_1.LeftArmAdditiveHits.AnimStateNode_0.LeftArmHit_B.AnimGraphNode_StateMachine_2.LeftArmHitChooser_StartTime_DelayFloat.AnimStateNode_9.Hit_FloorTom_L.AnimGraphNode_StateMachine_0.Hit_FloorTom_R_SM.AnimStateConduitNode_0.Conduit
  nodes [Result(bCanEnterTransition=true)]
 }
 Hit_Short {
  path=AnimGraph.AnimGraphNode_StateMachine_1.LeftArmAdditiveHits.AnimStateNode_0.LeftArmHit_B.AnimGraphNode_StateMachine_2.LeftArmHitChooser_StartTime_DelayFloat.AnimStateNode_9.Hit_FloorTom_L.AnimGraphNode_StateMachine_0.Hit_FloorTom_R_SM.AnimStateNode_0.Hit_Short
  nodes [Output Animation Pose(Result="Sequence Player (additive) 'Drum_Hit_FloorTom_Tap_Fast'"(StartPosition=Get LeftArmMaxDelayShortenedBy(), bLoopAnimation=false) {sequence=&Drum_Hit_FloorTom_Tap_Fast, startPosition=0.05})]
 }
 Hit_Medium {
  path=AnimGraph.AnimGraphNode_StateMachine_1.LeftArmAdditiveHits.AnimStateNode_0.LeftArmHit_B.AnimGraphNode_StateMachine_2.LeftArmHitChooser_StartTime_DelayFloat.AnimStateNode_9.Hit_FloorTom_L.AnimGraphNode_StateMachine_0.Hit_FloorTom_R_SM.AnimStateNode_1.Hit_Medium
  nodes [
   n1 = "Sequence Player (additive) 'Drum_Hit_FloorTom_Normal_Medium'"(StartPosition=Get LeftArmMaxDelayShortenedBy(), bLoopAnimation=false) {sequence=&Drum_Hit_FloorTom_Normal_Medium, startPosition=0.05}
   Output Animation Pose(Result=@n1)
  ]
 }
 Hit_Long {
  path=AnimGraph.AnimGraphNode_StateMachine_1.LeftArmAdditiveHits.AnimStateNode_0.LeftArmHit_B.AnimGraphNode_StateMachine_2.LeftArmHitChooser_StartTime_DelayFloat.AnimStateNode_9.Hit_FloorTom_L.AnimGraphNode_StateMachine_0.Hit_FloorTom_R_SM.AnimStateNode_2.Hit_Long
  nodes [Output Animation Pose(Result="Sequence Player (additive) 'Drum_Hit_FloorTom_Slow_Big'"(StartPosition=Get LeftArmMaxDelayShortenedBy(), bLoopAnimation=false) {sequence=&Drum_Hit_FloorTom_Slow_Big, startPosition=0.05})]
 }
 Transition {
  path=AnimGraph.AnimGraphNode_StateMachine_1.LeftArmAdditiveHits.AnimStateNode_0.LeftArmHit_B.AnimGraphNode_StateMachine_2.LeftArmHitChooser_StartTime_DelayFloat.AnimStateNode_9.Hit_FloorTom_L.AnimGraphNode_StateMachine_0.Hit_FloorTom_R_SM.AnimStateTransitionNode_4.Transition
  nodes [Result(bCanEnterTransition=Get LeftArmIsShortDelay())]
 }
 Transition {
  path=AnimGraph.AnimGraphNode_StateMachine_1.LeftArmAdditiveHits.AnimStateNode_0.LeftArmHit_B.AnimGraphNode_StateMachine_2.LeftArmHitChooser_StartTime_DelayFloat.AnimStateNode_9.Hit_FloorTom_L.AnimGraphNode_StateMachine_0.Hit_FloorTom_R_SM.AnimStateTransitionNode_5.Transition
  nodes [
   n1 = Get LeftArmIsLongDelay()
   n2 = Get LeftArmIsShortDelay()
   Result(bCanEnterTransition=AND Boolean(A=NOT Boolean(A=@n2), B=NOT Boolean(A=@n1)))
  ]
 }
 Transition {
  path=AnimGraph.AnimGraphNode_StateMachine_1.LeftArmAdditiveHits.AnimStateNode_0.LeftArmHit_B.AnimGraphNode_StateMachine_2.LeftArmHitChooser_StartTime_DelayFloat.AnimStateNode_9.Hit_FloorTom_L.AnimGraphNode_StateMachine_0.Hit_FloorTom_R_SM.AnimStateTransitionNode_6.Transition
  nodes [Result(bCanEnterTransition=Get LeftArmIsLongDelay())]
 }
 Transition {
  path=AnimGraph.AnimGraphNode_StateMachine_1.LeftArmAdditiveHits.AnimStateNode_0.LeftArmHit_B.AnimGraphNode_StateMachine_2.LeftArmHitChooser_StartTime_DelayFloat.AnimStateTransitionNode_10.Transition
  nodes [Result(bCanEnterTransition=Get Hit_Snare_L() -> (Hit_Snare_L:bool))]
 }
 Transition {
  path=AnimGraph.AnimGraphNode_StateMachine_1.LeftArmAdditiveHits.AnimStateNode_0.LeftArmHit_B.AnimGraphNode_StateMachine_2.LeftArmHitChooser_StartTime_DelayFloat.AnimStateTransitionNode_12.Transition
  nodes [Result(bCanEnterTransition=Get Hit_HiHat_L() -> (Hit_HiHat_L:bool))]
 }
 Transition {
  path=AnimGraph.AnimGraphNode_StateMachine_1.LeftArmAdditiveHits.AnimStateNode_0.LeftArmHit_B.AnimGraphNode_StateMachine_2.LeftArmHitChooser_StartTime_DelayFloat.AnimStateTransitionNode_14.Transition
  nodes [Result(bCanEnterTransition=Get Hit_Crash1_Hard_L() -> (Hit_Crash1_Hard_L:bool))]
 }
 Transition {
  path=AnimGraph.AnimGraphNode_StateMachine_1.LeftArmAdditiveHits.AnimStateNode_0.LeftArmHit_B.AnimGraphNode_StateMachine_2.LeftArmHitChooser_StartTime_DelayFloat.AnimStateTransitionNode_16.Transition
  nodes [Result(bCanEnterTransition=Get Hit_Crash1_Soft_L() -> (Hit_Crash1_Soft_L:bool))]
 }
 Transition {
  path=AnimGraph.AnimGraphNode_StateMachine_1.LeftArmAdditiveHits.AnimStateNode_0.LeftArmHit_B.AnimGraphNode_StateMachine_2.LeftArmHitChooser_StartTime_DelayFloat.AnimStateTransitionNode_18.Transition
  nodes [Result(bCanEnterTransition=Get Hit_Crash2_Hard_L() -> (Hit_Crash2_Hard_L:bool))]
 }
 Transition {
  path=AnimGraph.AnimGraphNode_StateMachine_1.LeftArmAdditiveHits.AnimStateNode_0.LeftArmHit_B.AnimGraphNode_StateMachine_2.LeftArmHitChooser_StartTime_DelayFloat.AnimStateTransitionNode_20.Transition
  nodes [Result(bCanEnterTransition=Get Hit_Crash2_Soft_L() -> (Hit_Crash2_Soft_L:bool))]
 }
 Transition {
  path=AnimGraph.AnimGraphNode_StateMachine_1.LeftArmAdditiveHits.AnimStateNode_0.LeftArmHit_B.AnimGraphNode_StateMachine_2.LeftArmHitChooser_StartTime_DelayFloat.AnimStateTransitionNode_22.Transition
  nodes [Result(bCanEnterTransition=Get Hit_Ride_L() -> (Hit_Ride_L:bool))]
 }
 Transition {
  path=AnimGraph.AnimGraphNode_StateMachine_1.LeftArmAdditiveHits.AnimStateNode_0.LeftArmHit_B.AnimGraphNode_StateMachine_2.LeftArmHitChooser_StartTime_DelayFloat.AnimStateTransitionNode_24.Transition
  nodes [Result(bCanEnterTransition=Get Hit_Tom1_L() -> (Hit_Tom1_L:bool))]
 }
 Transition {
  path=AnimGraph.AnimGraphNode_StateMachine_1.LeftArmAdditiveHits.AnimStateNode_0.LeftArmHit_B.AnimGraphNode_StateMachine_2.LeftArmHitChooser_StartTime_DelayFloat.AnimStateTransitionNode_26.Transition
  nodes [Result(bCanEnterTransition=Get Hit_Tom2_L() -> (Hit_Tom2_L:bool))]
 }
 Transition {
  path=AnimGraph.AnimGraphNode_StateMachine_1.LeftArmAdditiveHits.AnimStateNode_0.LeftArmHit_B.AnimGraphNode_StateMachine_2.LeftArmHitChooser_StartTime_DelayFloat.AnimStateTransitionNode_28.Transition
  nodes [Result(bCanEnterTransition=Get Hit_FloorTom_L() -> (Hit_FloorTom_L:bool))]
 }
 Transition {
  path=AnimGraph.AnimGraphNode_StateMachine_1.LeftArmAdditiveHits.AnimStateNode_0.LeftArmHit_B.AnimGraphNode_StateMachine_2.LeftArmHitChooser_StartTime_DelayFloat.AnimStateTransitionNode_5.Transition
  nodes [Result(bCanEnterTransition=Get Hit_Snare_Soft_L() -> (Hit_Snare_Soft_L:bool))]
 }
 LeftArmHit_A {
  path=AnimGraph.AnimGraphNode_StateMachine_1.LeftArmAdditiveHits.AnimStateNode_1.LeftArmHit_A
  nodes [
   AnimGraphNode_StateMachine_2 = LeftArmHitChooser_StartTime_DelayFloat() {bSkipFirstUpdateTransition=false}
   Output Animation Pose(Result=@AnimGraphNode_StateMachine_2)
  ]
 }
 LeftArmHitChooser_StartTime_DelayFloat {
  entry=Identity
  path=AnimGraph.AnimGraphNode_StateMachine_1.LeftArmAdditiveHits.AnimStateNode_1.LeftArmHit_A.AnimGraphNode_StateMachine_2.LeftArmHitChooser_StartTime_DelayFloat
  ids=id
  nodes [
   AnimStateNode_0 = Hit_Snare_L() {
    stateEntered {guid=#4}
    stateFullyBlended {guid=#5}
    stateLeft {guid=#6}
   }
   AnimStateNode_1 = Hit_HiHat_L() {
    stateEntered {guid=#7}
    stateFullyBlended {guid=#8}
    stateLeft {guid=#9}
   }
   AnimStateNode_10 = Identity()
   AnimStateNode_11 = Hit_Crash2_Hard_L() {
    stateEntered {guid=#10}
    stateFullyBlended {guid=#11}
    stateLeft {guid=#12}
   }
   AnimStateNode_2 = Hit_Crash1_Hard_L() {
    stateEntered {guid=#13}
    stateFullyBlended {guid=#14}
    stateLeft {guid=#15}
   }
   AnimStateNode_3 = Hit_Crash1_Soft_L() {
    stateEntered {guid=#16}
    stateFullyBlended {guid=#17}
    stateLeft {guid=#18}
   }
   AnimStateNode_4 = Hit_Soft()
   AnimStateNode_5 = Hit_Crash2_Soft_L() {
    stateEntered {guid=#19}
    stateFullyBlended {guid=#20}
    stateLeft {guid=#21}
   }
   AnimStateNode_6 = Hit_Ride_L() {
    stateEntered {guid=#22}
    stateFullyBlended {guid=#23}
    stateLeft {guid=#24}
   }
   AnimStateNode_7 = Hit_Tom1_L() {
    stateEntered {guid=#25}
    stateFullyBlended {guid=#26}
    stateLeft {guid=#27}
   }
   AnimStateNode_8 = Hit_Tom2_L() {
    stateEntered {guid=#28}
    stateFullyBlended {guid=#29}
    stateLeft {guid=#30}
   }
   AnimStateNode_9 = Hit_FloorTom_L() {
    stateEntered {guid=#31}
    stateFullyBlended {guid=#32}
    stateLeft {guid=#33}
   }
   AnimStateTransitionNode_10 = AnimStateTransitionNode() {crossfadeDuration=0}
   AnimStateTransitionNode_12 = AnimStateTransitionNode() {crossfadeDuration=0}
   AnimStateTransitionNode_14 = AnimStateTransitionNode() {crossfadeDuration=0}
   AnimStateTransitionNode_16 = AnimStateTransitionNode() {crossfadeDuration=0}
   AnimStateTransitionNode_18 = AnimStateTransitionNode() {crossfadeDuration=0}
   AnimStateTransitionNode_20 = AnimStateTransitionNode() {crossfadeDuration=0}
   AnimStateTransitionNode_22 = AnimStateTransitionNode() {crossfadeDuration=0}
   AnimStateTransitionNode_24 = AnimStateTransitionNode() {crossfadeDuration=0}
   AnimStateTransitionNode_26 = AnimStateTransitionNode() {crossfadeDuration=0}
   AnimStateTransitionNode_28 = AnimStateTransitionNode() {crossfadeDuration=0}
   AnimStateTransitionNode_4 = AnimStateTransitionNode() {crossfadeDuration=0}
  ]
  links [
   {
    dst=Hit_Crash1_Hard_L
    rule=AnimGraph.AnimGraphNode_StateMachine_1.LeftArmAdditiveHits.AnimStateNode_1.LeftArmHit_A.AnimGraphNode_StateMachine_2.LeftArmHitChooser_StartTime_DelayFloat.AnimStateTransitionNode_14.Transition
    src=Identity
    via=AnimStateTransitionNode_14
   }
   {
    dst=Hit_Crash1_Soft_L
    rule=AnimGraph.AnimGraphNode_StateMachine_1.LeftArmAdditiveHits.AnimStateNode_1.LeftArmHit_A.AnimGraphNode_StateMachine_2.LeftArmHitChooser_StartTime_DelayFloat.AnimStateTransitionNode_16.Transition
    src=Identity
    via=AnimStateTransitionNode_16
   }
   {
    dst=Hit_Crash2_Hard_L
    rule=AnimGraph.AnimGraphNode_StateMachine_1.LeftArmAdditiveHits.AnimStateNode_1.LeftArmHit_A.AnimGraphNode_StateMachine_2.LeftArmHitChooser_StartTime_DelayFloat.AnimStateTransitionNode_18.Transition
    src=Identity
    via=AnimStateTransitionNode_18
   }
   {
    dst=Hit_Crash2_Soft_L
    rule=AnimGraph.AnimGraphNode_StateMachine_1.LeftArmAdditiveHits.AnimStateNode_1.LeftArmHit_A.AnimGraphNode_StateMachine_2.LeftArmHitChooser_StartTime_DelayFloat.AnimStateTransitionNode_20.Transition
    src=Identity
    via=AnimStateTransitionNode_20
   }
   {
    dst=Hit_FloorTom_L
    rule=AnimGraph.AnimGraphNode_StateMachine_1.LeftArmAdditiveHits.AnimStateNode_1.LeftArmHit_A.AnimGraphNode_StateMachine_2.LeftArmHitChooser_StartTime_DelayFloat.AnimStateTransitionNode_28.Transition
    src=Identity
    via=AnimStateTransitionNode_28
   }
   {
    dst=Hit_HiHat_L
    rule=AnimGraph.AnimGraphNode_StateMachine_1.LeftArmAdditiveHits.AnimStateNode_1.LeftArmHit_A.AnimGraphNode_StateMachine_2.LeftArmHitChooser_StartTime_DelayFloat.AnimStateTransitionNode_12.Transition
    src=Identity
    via=AnimStateTransitionNode_12
   }
   {
    dst=Hit_Ride_L
    rule=AnimGraph.AnimGraphNode_StateMachine_1.LeftArmAdditiveHits.AnimStateNode_1.LeftArmHit_A.AnimGraphNode_StateMachine_2.LeftArmHitChooser_StartTime_DelayFloat.AnimStateTransitionNode_22.Transition
    src=Identity
    via=AnimStateTransitionNode_22
   }
   {
    dst=Hit_Snare_L
    rule=AnimGraph.AnimGraphNode_StateMachine_1.LeftArmAdditiveHits.AnimStateNode_1.LeftArmHit_A.AnimGraphNode_StateMachine_2.LeftArmHitChooser_StartTime_DelayFloat.AnimStateTransitionNode_10.Transition
    src=Identity
    via=AnimStateTransitionNode_10
   }
   {
    dst=Hit_Soft
    rule=AnimGraph.AnimGraphNode_StateMachine_1.LeftArmAdditiveHits.AnimStateNode_1.LeftArmHit_A.AnimGraphNode_StateMachine_2.LeftArmHitChooser_StartTime_DelayFloat.AnimStateTransitionNode_4.Transition
    src=Identity
    via=AnimStateTransitionNode_4
   }
   {
    dst=Hit_Tom1_L
    rule=AnimGraph.AnimGraphNode_StateMachine_1.LeftArmAdditiveHits.AnimStateNode_1.LeftArmHit_A.AnimGraphNode_StateMachine_2.LeftArmHitChooser_StartTime_DelayFloat.AnimStateTransitionNode_24.Transition
    src=Identity
    via=AnimStateTransitionNode_24
   }
   {
    dst=Hit_Tom2_L
    rule=AnimGraph.AnimGraphNode_StateMachine_1.LeftArmAdditiveHits.AnimStateNode_1.LeftArmHit_A.AnimGraphNode_StateMachine_2.LeftArmHitChooser_StartTime_DelayFloat.AnimStateTransitionNode_26.Transition
    src=Identity
    via=AnimStateTransitionNode_26
   }
  ]
 }
 Hit_Snare_L {
  path=AnimGraph.AnimGraphNode_StateMachine_1.LeftArmAdditiveHits.AnimStateNode_1.LeftArmHit_A.AnimGraphNode_StateMachine_2.LeftArmHitChooser_StartTime_DelayFloat.AnimStateNode_0.Hit_Snare_L
  nodes [
   AnimGraphNode_StateMachine_0 = Hit_Snare_L_SM() {bAllowConduitEntryStates=true}
   Output Animation Pose(Result=@AnimGraphNode_StateMachine_0)
  ]
 }
 Hit_Snare_L_SM {
  entry=Conduit
  path=AnimGraph.AnimGraphNode_StateMachine_1.LeftArmAdditiveHits.AnimStateNode_1.LeftArmHit_A.AnimGraphNode_StateMachine_2.LeftArmHitChooser_StartTime_DelayFloat.AnimStateNode_0.Hit_Snare_L.AnimGraphNode_StateMachine_0.Hit_Snare_L_SM
  ids=id
  nodes [
   AnimStateConduitNode_0 = Conduit()
   AnimStateNode_2 = Hit_Long()
   AnimStateNode_3 = Hit_Medium()
   AnimStateNode_4 = Hit_Short()
   AnimStateTransitionNode_4 = AnimStateTransitionNode() {crossfadeDuration=0, priorityOrder=2}
   AnimStateTransitionNode_5 = AnimStateTransitionNode() {crossfadeDuration=0, priorityOrder=2}
   AnimStateTransitionNode_6 = AnimStateTransitionNode() {crossfadeDuration=0, priorityOrder=2}
  ]
  links [
   {
    dst=Hit_Long
    rule=AnimGraph.AnimGraphNode_StateMachine_1.LeftArmAdditiveHits.AnimStateNode_1.LeftArmHit_A.AnimGraphNode_StateMachine_2.LeftArmHitChooser_StartTime_DelayFloat.AnimStateNode_0.Hit_Snare_L.AnimGraphNode_StateMachine_0.Hit_Snare_L_SM.AnimStateTransitionNode_6.Transition
    src=Conduit
    via=AnimStateTransitionNode_6
   }
   {
    dst=Hit_Medium
    rule=AnimGraph.AnimGraphNode_StateMachine_1.LeftArmAdditiveHits.AnimStateNode_1.LeftArmHit_A.AnimGraphNode_StateMachine_2.LeftArmHitChooser_StartTime_DelayFloat.AnimStateNode_0.Hit_Snare_L.AnimGraphNode_StateMachine_0.Hit_Snare_L_SM.AnimStateTransitionNode_5.Transition
    src=Conduit
    via=AnimStateTransitionNode_5
   }
   {
    dst=Hit_Short
    rule=AnimGraph.AnimGraphNode_StateMachine_1.LeftArmAdditiveHits.AnimStateNode_1.LeftArmHit_A.AnimGraphNode_StateMachine_2.LeftArmHitChooser_StartTime_DelayFloat.AnimStateNode_0.Hit_Snare_L.AnimGraphNode_StateMachine_0.Hit_Snare_L_SM.AnimStateTransitionNode_4.Transition
    src=Conduit
    via=AnimStateTransitionNode_4
   }
  ]
 }
 Conduit {
  path=AnimGraph.AnimGraphNode_StateMachine_1.LeftArmAdditiveHits.AnimStateNode_1.LeftArmHit_A.AnimGraphNode_StateMachine_2.LeftArmHitChooser_StartTime_DelayFloat.AnimStateNode_0.Hit_Snare_L.AnimGraphNode_StateMachine_0.Hit_Snare_L_SM.AnimStateConduitNode_0.Conduit
  nodes [Result(bCanEnterTransition=true)]
 }
 Hit_Long {
  path=AnimGraph.AnimGraphNode_StateMachine_1.LeftArmAdditiveHits.AnimStateNode_1.LeftArmHit_A.AnimGraphNode_StateMachine_2.LeftArmHitChooser_StartTime_DelayFloat.AnimStateNode_0.Hit_Snare_L.AnimGraphNode_StateMachine_0.Hit_Snare_L_SM.AnimStateNode_2.Hit_Long
  nodes [Output Animation Pose(Result="Sequence Player (additive) 'Drum_Hit_Snare_Slow_Big'"(StartPosition=Get LeftArmMaxDelayShortenedBy(), bLoopAnimation=false) {sequence=&Drum_Hit_Snare_Slow_Big, startPosition=0.05})]
 }
 Hit_Medium {
  path=AnimGraph.AnimGraphNode_StateMachine_1.LeftArmAdditiveHits.AnimStateNode_1.LeftArmHit_A.AnimGraphNode_StateMachine_2.LeftArmHitChooser_StartTime_DelayFloat.AnimStateNode_0.Hit_Snare_L.AnimGraphNode_StateMachine_0.Hit_Snare_L_SM.AnimStateNode_3.Hit_Medium
  nodes [
   n1 = Get LeftArmMaxDelayShortenedBy()
   n2 = "Sequence Player (additive) 'Drum_Hit_CrossHiHatSnare_Medium'"(StartPosition=@n1, bLoopAnimation=False) {bLoopAnimation=false, sequence=&Drum_Hit_CrossHiHatSnare_Medium}
   n3 = "Sequence Player (additive) 'Drum_Hit_Snare_Normal_Medium'"(StartPosition=@n1, bLoopAnimation=false) {sequence=&Drum_Hit_Snare_Normal_Medium, startPosition=0.05}
   Output Animation Pose(Result=Blend Poses by bool(BlendPose_0=@n2, BlendPose_1=@n3, bActiveValue=Get Hit_HiHat_R()) {blendType=Linear})
  ]
 }
 Hit_Short {
  path=AnimGraph.AnimGraphNode_StateMachine_1.LeftArmAdditiveHits.AnimStateNode_1.LeftArmHit_A.AnimGraphNode_StateMachine_2.LeftArmHitChooser_StartTime_DelayFloat.AnimStateNode_0.Hit_Snare_L.AnimGraphNode_StateMachine_0.Hit_Snare_L_SM.AnimStateNode_4.Hit_Short
  nodes [
   n1 = "Sequence Player (additive) 'Drum_Hit_CrossHiHatSnare_Tap_Fast'"(StartPosition=Get LeftArmMaxDelayShortenedBy(), bLoopAnimation=False) {bLoopAnimation=false, sequence=&Drum_Hit_CrossHiHatSnare_Tap_Fast}
   n2 = Get LeftArmMaxDelayShortenedBy()
   Output Animation Pose(Result=Blend Poses by bool(
     BlendPose_0=@n1
     BlendPose_1="Sequence Player (additive) 'Drum_Hit_Snare_Tap_Fast'"(StartPosition=@n2, bLoopAnimation=false) {sequence=&Drum_Hit_Snare_Tap_Fast, startPosition=0.05}
     bActiveValue=Get Hit_HiHat_R()
    ) {blendType=Linear})
  ]
 }
 Transition {
  path=AnimGraph.AnimGraphNode_StateMachine_1.LeftArmAdditiveHits.AnimStateNode_1.LeftArmHit_A.AnimGraphNode_StateMachine_2.LeftArmHitChooser_StartTime_DelayFloat.AnimStateNode_0.Hit_Snare_L.AnimGraphNode_StateMachine_0.Hit_Snare_L_SM.AnimStateTransitionNode_4.Transition
  nodes [Result(bCanEnterTransition=Get LeftArmIsShortDelay())]
 }
 Transition {
  path=AnimGraph.AnimGraphNode_StateMachine_1.LeftArmAdditiveHits.AnimStateNode_1.LeftArmHit_A.AnimGraphNode_StateMachine_2.LeftArmHitChooser_StartTime_DelayFloat.AnimStateNode_0.Hit_Snare_L.AnimGraphNode_StateMachine_0.Hit_Snare_L_SM.AnimStateTransitionNode_5.Transition
  nodes [
   n1 = Get LeftArmIsLongDelay()
   n2 = Get LeftArmIsShortDelay()
   Result(bCanEnterTransition=AND Boolean(A=NOT Boolean(A=@n2), B=NOT Boolean(A=@n1)))
  ]
 }
 Transition {
  path=AnimGraph.AnimGraphNode_StateMachine_1.LeftArmAdditiveHits.AnimStateNode_1.LeftArmHit_A.AnimGraphNode_StateMachine_2.LeftArmHitChooser_StartTime_DelayFloat.AnimStateNode_0.Hit_Snare_L.AnimGraphNode_StateMachine_0.Hit_Snare_L_SM.AnimStateTransitionNode_6.Transition
  nodes [Result(bCanEnterTransition=Get LeftArmIsLongDelay())]
 }
 Hit_HiHat_L {
  path=AnimGraph.AnimGraphNode_StateMachine_1.LeftArmAdditiveHits.AnimStateNode_1.LeftArmHit_A.AnimGraphNode_StateMachine_2.LeftArmHitChooser_StartTime_DelayFloat.AnimStateNode_1.Hit_HiHat_L
  nodes [
   n1 = Get LeftArmMaxDelayShortenedBy()
   n2 = "Sequence Player (additive) 'Drum_Hit_HiHat_Normal_Medium'"(StartPosition=Get LeftArmMaxDelayShortenedBy(), bLoopAnimation=false) {sequence=&Drum_Hit_HiHat_Normal_Medium, startPosition=0.05}
   Blend Poses by bool(
    BlendPose_0="Sequence Player (additive) 'Drum_Hit_HiHat_Slow_Big'"(StartPosition=Get LeftArmMaxDelayShortenedBy(), bLoopAnimation=false) {sequence=&Drum_Hit_HiHat_Slow_Big, startPosition=0.05}
    BlendPose_1=Blend Poses by bool(
     BlendPose_0="Sequence Player (additive) 'Drum_Hit_HiHat_Tap_Fast'"(StartPosition=@n1, bLoopAnimation=false) {sequence=&Drum_Hit_HiHat_Tap_Fast, startPosition=0.05}
     BlendPose_1=@n2
     BlendTime_0=0
     BlendTime_1=0
     bActiveValue=Get LeftArmIsShortDelay()
    ) {blendType=Linear}
    BlendTime_0=0
    BlendTime_1=0
    bActiveValue=Get LeftArmIsLongDelay()
   ) {blendType=Linear}
   AnimGraphNode_StateMachine_0 = Hit_HiHat_L_SM() {bAllowConduitEntryStates=true}
   Output Animation Pose(Result=@AnimGraphNode_StateMachine_0)
  ]
 }
 Hit_HiHat_L_SM {
  entry=Conduit
  path=AnimGraph.AnimGraphNode_StateMachine_1.LeftArmAdditiveHits.AnimStateNode_1.LeftArmHit_A.AnimGraphNode_StateMachine_2.LeftArmHitChooser_StartTime_DelayFloat.AnimStateNode_1.Hit_HiHat_L.AnimGraphNode_StateMachine_0.Hit_HiHat_L_SM
  ids=id
  nodes [
   AnimStateConduitNode_0 = Conduit()
   AnimStateNode_0 = Hit_Short()
   AnimStateNode_1 = Hit_Medium()
   AnimStateNode_2 = Hit_Long()
   AnimStateTransitionNode_4 = AnimStateTransitionNode() {crossfadeDuration=0}
   AnimStateTransitionNode_5 = AnimStateTransitionNode() {crossfadeDuration=0}
   AnimStateTransitionNode_6 = AnimStateTransitionNode() {crossfadeDuration=0}
  ]
  links [
   {
    dst=Hit_Long
    rule=AnimGraph.AnimGraphNode_StateMachine_1.LeftArmAdditiveHits.AnimStateNode_1.LeftArmHit_A.AnimGraphNode_StateMachine_2.LeftArmHitChooser_StartTime_DelayFloat.AnimStateNode_1.Hit_HiHat_L.AnimGraphNode_StateMachine_0.Hit_HiHat_L_SM.AnimStateTransitionNode_6.Transition
    src=Conduit
    via=AnimStateTransitionNode_6
   }
   {
    dst=Hit_Medium
    rule=AnimGraph.AnimGraphNode_StateMachine_1.LeftArmAdditiveHits.AnimStateNode_1.LeftArmHit_A.AnimGraphNode_StateMachine_2.LeftArmHitChooser_StartTime_DelayFloat.AnimStateNode_1.Hit_HiHat_L.AnimGraphNode_StateMachine_0.Hit_HiHat_L_SM.AnimStateTransitionNode_5.Transition
    src=Conduit
    via=AnimStateTransitionNode_5
   }
   {
    dst=Hit_Short
    rule=AnimGraph.AnimGraphNode_StateMachine_1.LeftArmAdditiveHits.AnimStateNode_1.LeftArmHit_A.AnimGraphNode_StateMachine_2.LeftArmHitChooser_StartTime_DelayFloat.AnimStateNode_1.Hit_HiHat_L.AnimGraphNode_StateMachine_0.Hit_HiHat_L_SM.AnimStateTransitionNode_4.Transition
    src=Conduit
    via=AnimStateTransitionNode_4
   }
  ]
 }
 Conduit {
  path=AnimGraph.AnimGraphNode_StateMachine_1.LeftArmAdditiveHits.AnimStateNode_1.LeftArmHit_A.AnimGraphNode_StateMachine_2.LeftArmHitChooser_StartTime_DelayFloat.AnimStateNode_1.Hit_HiHat_L.AnimGraphNode_StateMachine_0.Hit_HiHat_L_SM.AnimStateConduitNode_0.Conduit
  nodes [Result(bCanEnterTransition=true)]
 }
 Hit_Short {
  path=AnimGraph.AnimGraphNode_StateMachine_1.LeftArmAdditiveHits.AnimStateNode_1.LeftArmHit_A.AnimGraphNode_StateMachine_2.LeftArmHitChooser_StartTime_DelayFloat.AnimStateNode_1.Hit_HiHat_L.AnimGraphNode_StateMachine_0.Hit_HiHat_L_SM.AnimStateNode_0.Hit_Short
  nodes [Output Animation Pose(Result="Sequence Player (additive) 'Drum_Hit_HiHat_Tap_Fast'"(StartPosition=Get LeftArmMaxDelayShortenedBy(), bLoopAnimation=false) {sequence=&Drum_Hit_HiHat_Tap_Fast, startPosition=0.05})]
 }
 Hit_Medium {
  path=AnimGraph.AnimGraphNode_StateMachine_1.LeftArmAdditiveHits.AnimStateNode_1.LeftArmHit_A.AnimGraphNode_StateMachine_2.LeftArmHitChooser_StartTime_DelayFloat.AnimStateNode_1.Hit_HiHat_L.AnimGraphNode_StateMachine_0.Hit_HiHat_L_SM.AnimStateNode_1.Hit_Medium
  nodes [
   n1 = "Sequence Player (additive) 'Drum_Hit_HiHat_Normal_Medium'"(StartPosition=Get LeftArmMaxDelayShortenedBy(), bLoopAnimation=false) {sequence=&Drum_Hit_HiHat_Normal_Medium, startPosition=0.05}
   Output Animation Pose(Result=@n1)
  ]
 }
 Hit_Long {
  path=AnimGraph.AnimGraphNode_StateMachine_1.LeftArmAdditiveHits.AnimStateNode_1.LeftArmHit_A.AnimGraphNode_StateMachine_2.LeftArmHitChooser_StartTime_DelayFloat.AnimStateNode_1.Hit_HiHat_L.AnimGraphNode_StateMachine_0.Hit_HiHat_L_SM.AnimStateNode_2.Hit_Long
  nodes [Output Animation Pose(Result="Sequence Player (additive) 'Drum_Hit_HiHat_Slow_Big'"(StartPosition=Get LeftArmMaxDelayShortenedBy(), bLoopAnimation=false) {sequence=&Drum_Hit_HiHat_Slow_Big, startPosition=0.05})]
 }
 Transition {
  path=AnimGraph.AnimGraphNode_StateMachine_1.LeftArmAdditiveHits.AnimStateNode_1.LeftArmHit_A.AnimGraphNode_StateMachine_2.LeftArmHitChooser_StartTime_DelayFloat.AnimStateNode_1.Hit_HiHat_L.AnimGraphNode_StateMachine_0.Hit_HiHat_L_SM.AnimStateTransitionNode_4.Transition
  nodes [Result(bCanEnterTransition=Get LeftArmIsShortDelay())]
 }
 Transition {
  path=AnimGraph.AnimGraphNode_StateMachine_1.LeftArmAdditiveHits.AnimStateNode_1.LeftArmHit_A.AnimGraphNode_StateMachine_2.LeftArmHitChooser_StartTime_DelayFloat.AnimStateNode_1.Hit_HiHat_L.AnimGraphNode_StateMachine_0.Hit_HiHat_L_SM.AnimStateTransitionNode_5.Transition
  nodes [
   n1 = Get LeftArmIsLongDelay()
   n2 = Get LeftArmIsShortDelay()
   Result(bCanEnterTransition=AND Boolean(A=NOT Boolean(A=@n2), B=NOT Boolean(A=@n1)))
  ]
 }
 Transition {
  path=AnimGraph.AnimGraphNode_StateMachine_1.LeftArmAdditiveHits.AnimStateNode_1.LeftArmHit_A.AnimGraphNode_StateMachine_2.LeftArmHitChooser_StartTime_DelayFloat.AnimStateNode_1.Hit_HiHat_L.AnimGraphNode_StateMachine_0.Hit_HiHat_L_SM.AnimStateTransitionNode_6.Transition
  nodes [Result(bCanEnterTransition=Get LeftArmIsLongDelay())]
 }
 Identity {
  path=AnimGraph.AnimGraphNode_StateMachine_1.LeftArmAdditiveHits.AnimStateNode_1.LeftArmHit_A.AnimGraphNode_StateMachine_2.LeftArmHitChooser_StartTime_DelayFloat.AnimStateNode_10.Identity
  nodes [
   "Sequence Player (additive) 'Drum_SpineTwist_Center_Pose'"() {bLoopAnimation=false, sequence=&Drum_SpineTwist_Center_Pose}
   Output Animation Pose(Result=Additive Identity Pose())
  ]
 }
 Hit_Crash2_Hard_L {
  path=AnimGraph.AnimGraphNode_StateMachine_1.LeftArmAdditiveHits.AnimStateNode_1.LeftArmHit_A.AnimGraphNode_StateMachine_2.LeftArmHitChooser_StartTime_DelayFloat.AnimStateNode_11.Hit_Crash2_Hard_L
  nodes [
   n1 = Get LeftArmMaxDelayShortenedBy()
   n2 = "Sequence Player (additive) 'Drum_Hit_Crash2_Normal_Medium'"(StartPosition=Get LeftArmMaxDelayShortenedBy(), bLoopAnimation=false) {sequence=&Drum_Hit_Crash2_Normal_Medium, startPosition=0.05}
   Blend Poses by bool(
    BlendPose_0="Sequence Player (additive) 'Drum_Hit_Crash2_Slow_Big'"(StartPosition=Get LeftArmMaxDelayShortenedBy(), bLoopAnimation=false) {sequence=&Drum_Hit_Crash2_Slow_Big, startPosition=0.05}
    BlendPose_1=Blend Poses by bool(
     BlendPose_0="Sequence Player (additive) 'Drum_Hit_Crash2_Tap_Fast'"(StartPosition=@n1, bLoopAnimation=false) {sequence=&Drum_Hit_Crash2_Tap_Fast, startPosition=0.05}
     BlendPose_1=@n2
     BlendTime_0=0
     BlendTime_1=0
     bActiveValue=Get LeftArmIsShortDelay()
    ) {blendType=Linear}
    BlendTime_0=0
    BlendTime_1=0
    bActiveValue=Get LeftArmIsLongDelay()
   ) {blendType=Linear}
   AnimGraphNode_StateMachine_0 = Hit_Crash2_Hard_L_SM() {bAllowConduitEntryStates=true}
   Output Animation Pose(Result=@AnimGraphNode_StateMachine_0)
  ]
 }
 Hit_Crash2_Hard_L_SM {
  entry=Conduit
  path=AnimGraph.AnimGraphNode_StateMachine_1.LeftArmAdditiveHits.AnimStateNode_1.LeftArmHit_A.AnimGraphNode_StateMachine_2.LeftArmHitChooser_StartTime_DelayFloat.AnimStateNode_11.Hit_Crash2_Hard_L.AnimGraphNode_StateMachine_0.Hit_Crash2_Hard_L_SM
  ids=id
  nodes [
   AnimStateConduitNode_0 = Conduit()
   AnimStateNode_0 = Hit_Short()
   AnimStateNode_1 = Hit_Medium()
   AnimStateNode_2 = Hit_Long()
   AnimStateTransitionNode_4 = AnimStateTransitionNode() {crossfadeDuration=0}
   AnimStateTransitionNode_5 = AnimStateTransitionNode() {crossfadeDuration=0}
   AnimStateTransitionNode_6 = AnimStateTransitionNode() {crossfadeDuration=0}
  ]
  links [
   {
    dst=Hit_Long
    rule=AnimGraph.AnimGraphNode_StateMachine_1.LeftArmAdditiveHits.AnimStateNode_1.LeftArmHit_A.AnimGraphNode_StateMachine_2.LeftArmHitChooser_StartTime_DelayFloat.AnimStateNode_11.Hit_Crash2_Hard_L.AnimGraphNode_StateMachine_0.Hit_Crash2_Hard_L_SM.AnimStateTransitionNode_6.Transition
    src=Conduit
    via=AnimStateTransitionNode_6
   }
   {
    dst=Hit_Medium
    rule=AnimGraph.AnimGraphNode_StateMachine_1.LeftArmAdditiveHits.AnimStateNode_1.LeftArmHit_A.AnimGraphNode_StateMachine_2.LeftArmHitChooser_StartTime_DelayFloat.AnimStateNode_11.Hit_Crash2_Hard_L.AnimGraphNode_StateMachine_0.Hit_Crash2_Hard_L_SM.AnimStateTransitionNode_5.Transition
    src=Conduit
    via=AnimStateTransitionNode_5
   }
   {
    dst=Hit_Short
    rule=AnimGraph.AnimGraphNode_StateMachine_1.LeftArmAdditiveHits.AnimStateNode_1.LeftArmHit_A.AnimGraphNode_StateMachine_2.LeftArmHitChooser_StartTime_DelayFloat.AnimStateNode_11.Hit_Crash2_Hard_L.AnimGraphNode_StateMachine_0.Hit_Crash2_Hard_L_SM.AnimStateTransitionNode_4.Transition
    src=Conduit
    via=AnimStateTransitionNode_4
   }
  ]
 }
 Conduit {
  path=AnimGraph.AnimGraphNode_StateMachine_1.LeftArmAdditiveHits.AnimStateNode_1.LeftArmHit_A.AnimGraphNode_StateMachine_2.LeftArmHitChooser_StartTime_DelayFloat.AnimStateNode_11.Hit_Crash2_Hard_L.AnimGraphNode_StateMachine_0.Hit_Crash2_Hard_L_SM.AnimStateConduitNode_0.Conduit
  nodes [Result(bCanEnterTransition=true)]
 }
 Hit_Short {
  path=AnimGraph.AnimGraphNode_StateMachine_1.LeftArmAdditiveHits.AnimStateNode_1.LeftArmHit_A.AnimGraphNode_StateMachine_2.LeftArmHitChooser_StartTime_DelayFloat.AnimStateNode_11.Hit_Crash2_Hard_L.AnimGraphNode_StateMachine_0.Hit_Crash2_Hard_L_SM.AnimStateNode_0.Hit_Short
  nodes [Output Animation Pose(Result="Sequence Player (additive) 'Drum_Hit_Crash2_Tap_Fast'"(StartPosition=Get LeftArmMaxDelayShortenedBy(), bLoopAnimation=false) {sequence=&Drum_Hit_Crash2_Tap_Fast, startPosition=0.05})]
 }
 Hit_Medium {
  path=AnimGraph.AnimGraphNode_StateMachine_1.LeftArmAdditiveHits.AnimStateNode_1.LeftArmHit_A.AnimGraphNode_StateMachine_2.LeftArmHitChooser_StartTime_DelayFloat.AnimStateNode_11.Hit_Crash2_Hard_L.AnimGraphNode_StateMachine_0.Hit_Crash2_Hard_L_SM.AnimStateNode_1.Hit_Medium
  nodes [
   n1 = "Sequence Player (additive) 'Drum_Hit_Crash2_Normal_Medium'"(StartPosition=Get LeftArmMaxDelayShortenedBy(), bLoopAnimation=false) {sequence=&Drum_Hit_Crash2_Normal_Medium, startPosition=0.05}
   Output Animation Pose(Result=@n1)
  ]
 }
 Hit_Long {
  path=AnimGraph.AnimGraphNode_StateMachine_1.LeftArmAdditiveHits.AnimStateNode_1.LeftArmHit_A.AnimGraphNode_StateMachine_2.LeftArmHitChooser_StartTime_DelayFloat.AnimStateNode_11.Hit_Crash2_Hard_L.AnimGraphNode_StateMachine_0.Hit_Crash2_Hard_L_SM.AnimStateNode_2.Hit_Long
  nodes [Output Animation Pose(Result="Sequence Player (additive) 'Drum_Hit_Crash2_Slow_Big'"(StartPosition=Get LeftArmMaxDelayShortenedBy(), bLoopAnimation=false) {sequence=&Drum_Hit_Crash2_Slow_Big, startPosition=0.05})]
 }
 Transition {
  path=AnimGraph.AnimGraphNode_StateMachine_1.LeftArmAdditiveHits.AnimStateNode_1.LeftArmHit_A.AnimGraphNode_StateMachine_2.LeftArmHitChooser_StartTime_DelayFloat.AnimStateNode_11.Hit_Crash2_Hard_L.AnimGraphNode_StateMachine_0.Hit_Crash2_Hard_L_SM.AnimStateTransitionNode_4.Transition
  nodes [Result(bCanEnterTransition=Get LeftArmIsShortDelay())]
 }
 Transition {
  path=AnimGraph.AnimGraphNode_StateMachine_1.LeftArmAdditiveHits.AnimStateNode_1.LeftArmHit_A.AnimGraphNode_StateMachine_2.LeftArmHitChooser_StartTime_DelayFloat.AnimStateNode_11.Hit_Crash2_Hard_L.AnimGraphNode_StateMachine_0.Hit_Crash2_Hard_L_SM.AnimStateTransitionNode_5.Transition
  nodes [
   n1 = Get LeftArmIsLongDelay()
   n2 = Get LeftArmIsShortDelay()
   Result(bCanEnterTransition=AND Boolean(A=NOT Boolean(A=@n2), B=NOT Boolean(A=@n1)))
  ]
 }
 Transition {
  path=AnimGraph.AnimGraphNode_StateMachine_1.LeftArmAdditiveHits.AnimStateNode_1.LeftArmHit_A.AnimGraphNode_StateMachine_2.LeftArmHitChooser_StartTime_DelayFloat.AnimStateNode_11.Hit_Crash2_Hard_L.AnimGraphNode_StateMachine_0.Hit_Crash2_Hard_L_SM.AnimStateTransitionNode_6.Transition
  nodes [Result(bCanEnterTransition=Get LeftArmIsLongDelay())]
 }
 Hit_Crash1_Hard_L {
  path=AnimGraph.AnimGraphNode_StateMachine_1.LeftArmAdditiveHits.AnimStateNode_1.LeftArmHit_A.AnimGraphNode_StateMachine_2.LeftArmHitChooser_StartTime_DelayFloat.AnimStateNode_2.Hit_Crash1_Hard_L
  nodes [
   n1 = Get LeftArmMaxDelayShortenedBy()
   n2 = "Sequence Player (additive) 'Drum_Hit_Crash1_Normal_Medium'"(StartPosition=Get LeftArmMaxDelayShortenedBy(), bLoopAnimation=false) {sequence=&Drum_Hit_Crash1_Normal_Medium, startPosition=0.05}
   Blend Poses by bool(
    BlendPose_0="Sequence Player (additive) 'Drum_Hit_Crash1_Slow_Big'"(StartPosition=Get LeftArmMaxDelayShortenedBy(), bLoopAnimation=false) {sequence=&Drum_Hit_Crash1_Slow_Big, startPosition=0.05}
    BlendPose_1=Blend Poses by bool(
     BlendPose_0="Sequence Player (additive) 'Drum_Hit_Crash1_Tap_Fast'"(StartPosition=@n1, bLoopAnimation=false) {sequence=&Drum_Hit_Crash1_Tap_Fast, startPosition=0.05}
     BlendPose_1=@n2
     BlendTime_0=0
     BlendTime_1=0
     bActiveValue=Get LeftArmIsShortDelay()
    ) {blendType=Linear}
    BlendTime_0=0
    BlendTime_1=0
    bActiveValue=Get LeftArmIsLongDelay()
   ) {blendType=Linear}
   AnimGraphNode_StateMachine_0 = Hit_Crash1_Hard_L_SM() {bAllowConduitEntryStates=true}
   Output Animation Pose(Result=@AnimGraphNode_StateMachine_0)
  ]
 }
 Hit_Crash1_Hard_L_SM {
  entry=Conduit
  path=AnimGraph.AnimGraphNode_StateMachine_1.LeftArmAdditiveHits.AnimStateNode_1.LeftArmHit_A.AnimGraphNode_StateMachine_2.LeftArmHitChooser_StartTime_DelayFloat.AnimStateNode_2.Hit_Crash1_Hard_L.AnimGraphNode_StateMachine_0.Hit_Crash1_Hard_L_SM
  ids=id
  nodes [
   AnimStateConduitNode_0 = Conduit()
   AnimStateNode_0 = Hit_Short()
   AnimStateNode_1 = Hit_Medium()
   AnimStateNode_2 = Hit_Long()
   AnimStateTransitionNode_4 = AnimStateTransitionNode() {crossfadeDuration=0}
   AnimStateTransitionNode_5 = AnimStateTransitionNode() {crossfadeDuration=0}
   AnimStateTransitionNode_6 = AnimStateTransitionNode() {crossfadeDuration=0}
  ]
  links [
   {
    dst=Hit_Long
    rule=AnimGraph.AnimGraphNode_StateMachine_1.LeftArmAdditiveHits.AnimStateNode_1.LeftArmHit_A.AnimGraphNode_StateMachine_2.LeftArmHitChooser_StartTime_DelayFloat.AnimStateNode_2.Hit_Crash1_Hard_L.AnimGraphNode_StateMachine_0.Hit_Crash1_Hard_L_SM.AnimStateTransitionNode_6.Transition
    src=Conduit
    via=AnimStateTransitionNode_6
   }
   {
    dst=Hit_Medium
    rule=AnimGraph.AnimGraphNode_StateMachine_1.LeftArmAdditiveHits.AnimStateNode_1.LeftArmHit_A.AnimGraphNode_StateMachine_2.LeftArmHitChooser_StartTime_DelayFloat.AnimStateNode_2.Hit_Crash1_Hard_L.AnimGraphNode_StateMachine_0.Hit_Crash1_Hard_L_SM.AnimStateTransitionNode_5.Transition
    src=Conduit
    via=AnimStateTransitionNode_5
   }
   {
    dst=Hit_Short
    rule=AnimGraph.AnimGraphNode_StateMachine_1.LeftArmAdditiveHits.AnimStateNode_1.LeftArmHit_A.AnimGraphNode_StateMachine_2.LeftArmHitChooser_StartTime_DelayFloat.AnimStateNode_2.Hit_Crash1_Hard_L.AnimGraphNode_StateMachine_0.Hit_Crash1_Hard_L_SM.AnimStateTransitionNode_4.Transition
    src=Conduit
    via=AnimStateTransitionNode_4
   }
  ]
 }
 Conduit {
  path=AnimGraph.AnimGraphNode_StateMachine_1.LeftArmAdditiveHits.AnimStateNode_1.LeftArmHit_A.AnimGraphNode_StateMachine_2.LeftArmHitChooser_StartTime_DelayFloat.AnimStateNode_2.Hit_Crash1_Hard_L.AnimGraphNode_StateMachine_0.Hit_Crash1_Hard_L_SM.AnimStateConduitNode_0.Conduit
  nodes [Result(bCanEnterTransition=true)]
 }
 Hit_Short {
  path=AnimGraph.AnimGraphNode_StateMachine_1.LeftArmAdditiveHits.AnimStateNode_1.LeftArmHit_A.AnimGraphNode_StateMachine_2.LeftArmHitChooser_StartTime_DelayFloat.AnimStateNode_2.Hit_Crash1_Hard_L.AnimGraphNode_StateMachine_0.Hit_Crash1_Hard_L_SM.AnimStateNode_0.Hit_Short
  nodes [Output Animation Pose(Result="Sequence Player (additive) 'Drum_Hit_Crash1_Tap_Fast'"(StartPosition=Get LeftArmMaxDelayShortenedBy(), bLoopAnimation=false) {sequence=&Drum_Hit_Crash1_Tap_Fast, startPosition=0.05})]
 }
 Hit_Medium {
  path=AnimGraph.AnimGraphNode_StateMachine_1.LeftArmAdditiveHits.AnimStateNode_1.LeftArmHit_A.AnimGraphNode_StateMachine_2.LeftArmHitChooser_StartTime_DelayFloat.AnimStateNode_2.Hit_Crash1_Hard_L.AnimGraphNode_StateMachine_0.Hit_Crash1_Hard_L_SM.AnimStateNode_1.Hit_Medium
  nodes [
   n1 = "Sequence Player (additive) 'Drum_Hit_Crash1_Normal_Medium'"(StartPosition=Get LeftArmMaxDelayShortenedBy(), bLoopAnimation=false) {sequence=&Drum_Hit_Crash1_Normal_Medium, startPosition=0.05}
   Output Animation Pose(Result=@n1)
  ]
 }
 Hit_Long {
  path=AnimGraph.AnimGraphNode_StateMachine_1.LeftArmAdditiveHits.AnimStateNode_1.LeftArmHit_A.AnimGraphNode_StateMachine_2.LeftArmHitChooser_StartTime_DelayFloat.AnimStateNode_2.Hit_Crash1_Hard_L.AnimGraphNode_StateMachine_0.Hit_Crash1_Hard_L_SM.AnimStateNode_2.Hit_Long
  nodes [Output Animation Pose(Result="Sequence Player (additive) 'Drum_Hit_Crash1_Slow_Big'"(StartPosition=Get LeftArmMaxDelayShortenedBy(), bLoopAnimation=false) {sequence=&Drum_Hit_Crash1_Slow_Big, startPosition=0.05})]
 }
 Transition {
  path=AnimGraph.AnimGraphNode_StateMachine_1.LeftArmAdditiveHits.AnimStateNode_1.LeftArmHit_A.AnimGraphNode_StateMachine_2.LeftArmHitChooser_StartTime_DelayFloat.AnimStateNode_2.Hit_Crash1_Hard_L.AnimGraphNode_StateMachine_0.Hit_Crash1_Hard_L_SM.AnimStateTransitionNode_4.Transition
  nodes [Result(bCanEnterTransition=Get LeftArmIsShortDelay())]
 }
 Transition {
  path=AnimGraph.AnimGraphNode_StateMachine_1.LeftArmAdditiveHits.AnimStateNode_1.LeftArmHit_A.AnimGraphNode_StateMachine_2.LeftArmHitChooser_StartTime_DelayFloat.AnimStateNode_2.Hit_Crash1_Hard_L.AnimGraphNode_StateMachine_0.Hit_Crash1_Hard_L_SM.AnimStateTransitionNode_5.Transition
  nodes [
   n1 = Get LeftArmIsLongDelay()
   n2 = Get LeftArmIsShortDelay()
   Result(bCanEnterTransition=AND Boolean(A=NOT Boolean(A=@n2), B=NOT Boolean(A=@n1)))
  ]
 }
 Transition {
  path=AnimGraph.AnimGraphNode_StateMachine_1.LeftArmAdditiveHits.AnimStateNode_1.LeftArmHit_A.AnimGraphNode_StateMachine_2.LeftArmHitChooser_StartTime_DelayFloat.AnimStateNode_2.Hit_Crash1_Hard_L.AnimGraphNode_StateMachine_0.Hit_Crash1_Hard_L_SM.AnimStateTransitionNode_6.Transition
  nodes [Result(bCanEnterTransition=Get LeftArmIsLongDelay())]
 }
 Hit_Crash1_Soft_L {
  path=AnimGraph.AnimGraphNode_StateMachine_1.LeftArmAdditiveHits.AnimStateNode_1.LeftArmHit_A.AnimGraphNode_StateMachine_2.LeftArmHitChooser_StartTime_DelayFloat.AnimStateNode_3.Hit_Crash1_Soft_L
  nodes [
   n1 = Get LeftArmMaxDelayShortenedBy()
   n2 = "Sequence Player (additive) 'Drum_Hit_Crash1_Normal_Medium'"(StartPosition=Get LeftArmMaxDelayShortenedBy(), bLoopAnimation=false) {sequence=&Drum_Hit_Crash1_Normal_Medium, startPosition=0.05}
   Blend Poses by bool(
    BlendPose_0="Sequence Player (additive) 'Drum_Hit_Crash1_Slow_Big'"(StartPosition=Get LeftArmMaxDelayShortenedBy(), bLoopAnimation=false) {sequence=&Drum_Hit_Crash1_Slow_Big, startPosition=0.05}
    BlendPose_1=Blend Poses by bool(
     BlendPose_0="Sequence Player (additive) 'Drum_Hit_Crash1_Tap_Fast'"(StartPosition=@n1, bLoopAnimation=false) {sequence=&Drum_Hit_Crash1_Tap_Fast, startPosition=0.05}
     BlendPose_1=@n2
     BlendTime_0=0
     BlendTime_1=0
     bActiveValue=Get LeftArmIsShortDelay()
    ) {blendType=Linear}
    BlendTime_0=0
    BlendTime_1=0
    bActiveValue=Get LeftArmIsLongDelay()
   ) {blendType=Linear}
   AnimGraphNode_StateMachine_0 = Hit_Crash1_Soft_L_SM() {bAllowConduitEntryStates=true}
   Output Animation Pose(Result=@AnimGraphNode_StateMachine_0)
  ]
 }
 Hit_Crash1_Soft_L_SM {
  entry=Conduit
  path=AnimGraph.AnimGraphNode_StateMachine_1.LeftArmAdditiveHits.AnimStateNode_1.LeftArmHit_A.AnimGraphNode_StateMachine_2.LeftArmHitChooser_StartTime_DelayFloat.AnimStateNode_3.Hit_Crash1_Soft_L.AnimGraphNode_StateMachine_0.Hit_Crash1_Soft_L_SM
  ids=id
  nodes [
   AnimStateConduitNode_0 = Conduit()
   AnimStateNode_0 = Hit_Short()
   AnimStateNode_1 = Hit_Medium()
   AnimStateNode_2 = Hit_Long()
   AnimStateTransitionNode_4 = AnimStateTransitionNode() {crossfadeDuration=0}
   AnimStateTransitionNode_5 = AnimStateTransitionNode() {crossfadeDuration=0}
   AnimStateTransitionNode_6 = AnimStateTransitionNode() {crossfadeDuration=0}
  ]
  links [
   {
    dst=Hit_Long
    rule=AnimGraph.AnimGraphNode_StateMachine_1.LeftArmAdditiveHits.AnimStateNode_1.LeftArmHit_A.AnimGraphNode_StateMachine_2.LeftArmHitChooser_StartTime_DelayFloat.AnimStateNode_3.Hit_Crash1_Soft_L.AnimGraphNode_StateMachine_0.Hit_Crash1_Soft_L_SM.AnimStateTransitionNode_6.Transition
    src=Conduit
    via=AnimStateTransitionNode_6
   }
   {
    dst=Hit_Medium
    rule=AnimGraph.AnimGraphNode_StateMachine_1.LeftArmAdditiveHits.AnimStateNode_1.LeftArmHit_A.AnimGraphNode_StateMachine_2.LeftArmHitChooser_StartTime_DelayFloat.AnimStateNode_3.Hit_Crash1_Soft_L.AnimGraphNode_StateMachine_0.Hit_Crash1_Soft_L_SM.AnimStateTransitionNode_5.Transition
    src=Conduit
    via=AnimStateTransitionNode_5
   }
   {
    dst=Hit_Short
    rule=AnimGraph.AnimGraphNode_StateMachine_1.LeftArmAdditiveHits.AnimStateNode_1.LeftArmHit_A.AnimGraphNode_StateMachine_2.LeftArmHitChooser_StartTime_DelayFloat.AnimStateNode_3.Hit_Crash1_Soft_L.AnimGraphNode_StateMachine_0.Hit_Crash1_Soft_L_SM.AnimStateTransitionNode_4.Transition
    src=Conduit
    via=AnimStateTransitionNode_4
   }
  ]
 }
 Conduit {
  path=AnimGraph.AnimGraphNode_StateMachine_1.LeftArmAdditiveHits.AnimStateNode_1.LeftArmHit_A.AnimGraphNode_StateMachine_2.LeftArmHitChooser_StartTime_DelayFloat.AnimStateNode_3.Hit_Crash1_Soft_L.AnimGraphNode_StateMachine_0.Hit_Crash1_Soft_L_SM.AnimStateConduitNode_0.Conduit
  nodes [Result(bCanEnterTransition=true)]
 }
 Hit_Short {
  path=AnimGraph.AnimGraphNode_StateMachine_1.LeftArmAdditiveHits.AnimStateNode_1.LeftArmHit_A.AnimGraphNode_StateMachine_2.LeftArmHitChooser_StartTime_DelayFloat.AnimStateNode_3.Hit_Crash1_Soft_L.AnimGraphNode_StateMachine_0.Hit_Crash1_Soft_L_SM.AnimStateNode_0.Hit_Short
  nodes [Output Animation Pose(Result="Sequence Player (additive) 'Drum_Hit_Crash1_Tap_Fast'"(StartPosition=Get LeftArmMaxDelayShortenedBy(), bLoopAnimation=false) {sequence=&Drum_Hit_Crash1_Tap_Fast, startPosition=0.05})]
 }
 Hit_Medium {
  path=AnimGraph.AnimGraphNode_StateMachine_1.LeftArmAdditiveHits.AnimStateNode_1.LeftArmHit_A.AnimGraphNode_StateMachine_2.LeftArmHitChooser_StartTime_DelayFloat.AnimStateNode_3.Hit_Crash1_Soft_L.AnimGraphNode_StateMachine_0.Hit_Crash1_Soft_L_SM.AnimStateNode_1.Hit_Medium
  nodes [
   n1 = "Sequence Player (additive) 'Drum_Hit_Crash1_Normal_Medium'"(StartPosition=Get LeftArmMaxDelayShortenedBy(), bLoopAnimation=false) {sequence=&Drum_Hit_Crash1_Normal_Medium, startPosition=0.05}
   Output Animation Pose(Result=@n1)
  ]
 }
 Hit_Long {
  path=AnimGraph.AnimGraphNode_StateMachine_1.LeftArmAdditiveHits.AnimStateNode_1.LeftArmHit_A.AnimGraphNode_StateMachine_2.LeftArmHitChooser_StartTime_DelayFloat.AnimStateNode_3.Hit_Crash1_Soft_L.AnimGraphNode_StateMachine_0.Hit_Crash1_Soft_L_SM.AnimStateNode_2.Hit_Long
  nodes [Output Animation Pose(Result="Sequence Player (additive) 'Drum_Hit_Crash1_Slow_Big'"(StartPosition=Get LeftArmMaxDelayShortenedBy(), bLoopAnimation=false) {sequence=&Drum_Hit_Crash1_Slow_Big, startPosition=0.05})]
 }
 Transition {
  path=AnimGraph.AnimGraphNode_StateMachine_1.LeftArmAdditiveHits.AnimStateNode_1.LeftArmHit_A.AnimGraphNode_StateMachine_2.LeftArmHitChooser_StartTime_DelayFloat.AnimStateNode_3.Hit_Crash1_Soft_L.AnimGraphNode_StateMachine_0.Hit_Crash1_Soft_L_SM.AnimStateTransitionNode_4.Transition
  nodes [Result(bCanEnterTransition=Get LeftArmIsShortDelay())]
 }
 Transition {
  path=AnimGraph.AnimGraphNode_StateMachine_1.LeftArmAdditiveHits.AnimStateNode_1.LeftArmHit_A.AnimGraphNode_StateMachine_2.LeftArmHitChooser_StartTime_DelayFloat.AnimStateNode_3.Hit_Crash1_Soft_L.AnimGraphNode_StateMachine_0.Hit_Crash1_Soft_L_SM.AnimStateTransitionNode_5.Transition
  nodes [
   n1 = Get LeftArmIsLongDelay()
   n2 = Get LeftArmIsShortDelay()
   Result(bCanEnterTransition=AND Boolean(A=NOT Boolean(A=@n2), B=NOT Boolean(A=@n1)))
  ]
 }
 Transition {
  path=AnimGraph.AnimGraphNode_StateMachine_1.LeftArmAdditiveHits.AnimStateNode_1.LeftArmHit_A.AnimGraphNode_StateMachine_2.LeftArmHitChooser_StartTime_DelayFloat.AnimStateNode_3.Hit_Crash1_Soft_L.AnimGraphNode_StateMachine_0.Hit_Crash1_Soft_L_SM.AnimStateTransitionNode_6.Transition
  nodes [Result(bCanEnterTransition=Get LeftArmIsLongDelay())]
 }
 Hit_Soft {
  path=AnimGraph.AnimGraphNode_StateMachine_1.LeftArmAdditiveHits.AnimStateNode_1.LeftArmHit_A.AnimGraphNode_StateMachine_2.LeftArmHitChooser_StartTime_DelayFloat.AnimStateNode_4.Hit_Soft
  nodes [Output Animation Pose(Result="Sequence Player (additive) 'Drum_Hit_SnareSoft_Tap_Fast'"(StartPosition=Get LeftArmMaxDelayShortenedBy(), bLoopAnimation=False) {bLoopAnimation=false, sequence=&Drum_Hit_SnareSoft_Tap_Fast})]
 }
 Hit_Crash2_Soft_L {
  path=AnimGraph.AnimGraphNode_StateMachine_1.LeftArmAdditiveHits.AnimStateNode_1.LeftArmHit_A.AnimGraphNode_StateMachine_2.LeftArmHitChooser_StartTime_DelayFloat.AnimStateNode_5.Hit_Crash2_Soft_L
  nodes [
   n1 = Get LeftArmMaxDelayShortenedBy()
   n2 = "Sequence Player (additive) 'Drum_Hit_Crash2_Normal_Medium'"(StartPosition=Get LeftArmMaxDelayShortenedBy(), bLoopAnimation=false) {sequence=&Drum_Hit_Crash2_Normal_Medium, startPosition=0.05}
   Blend Poses by bool(
    BlendPose_0="Sequence Player (additive) 'Drum_Hit_Crash2_Slow_Big'"(StartPosition=Get LeftArmMaxDelayShortenedBy(), bLoopAnimation=false) {sequence=&Drum_Hit_Crash2_Slow_Big, startPosition=0.05}
    BlendPose_1=Blend Poses by bool(
     BlendPose_0="Sequence Player (additive) 'Drum_Hit_Crash2_Tap_Fast'"(StartPosition=@n1, bLoopAnimation=false) {sequence=&Drum_Hit_Crash2_Tap_Fast, startPosition=0.05}
     BlendPose_1=@n2
     BlendTime_0=0
     BlendTime_1=0
     bActiveValue=Get LeftArmIsShortDelay()
    ) {blendType=Linear}
    BlendTime_0=0
    BlendTime_1=0
    bActiveValue=Get LeftArmIsLongDelay()
   ) {blendType=Linear}
   AnimGraphNode_StateMachine_0 = Hit_Crash2_Soft_L_SM() {bAllowConduitEntryStates=true}
   Output Animation Pose(Result=@AnimGraphNode_StateMachine_0)
  ]
 }
 Hit_Crash2_Soft_L_SM {
  entry=Conduit
  path=AnimGraph.AnimGraphNode_StateMachine_1.LeftArmAdditiveHits.AnimStateNode_1.LeftArmHit_A.AnimGraphNode_StateMachine_2.LeftArmHitChooser_StartTime_DelayFloat.AnimStateNode_5.Hit_Crash2_Soft_L.AnimGraphNode_StateMachine_0.Hit_Crash2_Soft_L_SM
  ids=id
  nodes [
   AnimStateConduitNode_0 = Conduit()
   AnimStateNode_0 = Hit_Short()
   AnimStateNode_1 = Hit_Medium()
   AnimStateNode_2 = Hit_Long()
   AnimStateTransitionNode_4 = AnimStateTransitionNode() {crossfadeDuration=0}
   AnimStateTransitionNode_5 = AnimStateTransitionNode() {crossfadeDuration=0}
   AnimStateTransitionNode_6 = AnimStateTransitionNode() {crossfadeDuration=0}
  ]
  links [
   {
    dst=Hit_Long
    rule=AnimGraph.AnimGraphNode_StateMachine_1.LeftArmAdditiveHits.AnimStateNode_1.LeftArmHit_A.AnimGraphNode_StateMachine_2.LeftArmHitChooser_StartTime_DelayFloat.AnimStateNode_5.Hit_Crash2_Soft_L.AnimGraphNode_StateMachine_0.Hit_Crash2_Soft_L_SM.AnimStateTransitionNode_6.Transition
    src=Conduit
    via=AnimStateTransitionNode_6
   }
   {
    dst=Hit_Medium
    rule=AnimGraph.AnimGraphNode_StateMachine_1.LeftArmAdditiveHits.AnimStateNode_1.LeftArmHit_A.AnimGraphNode_StateMachine_2.LeftArmHitChooser_StartTime_DelayFloat.AnimStateNode_5.Hit_Crash2_Soft_L.AnimGraphNode_StateMachine_0.Hit_Crash2_Soft_L_SM.AnimStateTransitionNode_5.Transition
    src=Conduit
    via=AnimStateTransitionNode_5
   }
   {
    dst=Hit_Short
    rule=AnimGraph.AnimGraphNode_StateMachine_1.LeftArmAdditiveHits.AnimStateNode_1.LeftArmHit_A.AnimGraphNode_StateMachine_2.LeftArmHitChooser_StartTime_DelayFloat.AnimStateNode_5.Hit_Crash2_Soft_L.AnimGraphNode_StateMachine_0.Hit_Crash2_Soft_L_SM.AnimStateTransitionNode_4.Transition
    src=Conduit
    via=AnimStateTransitionNode_4
   }
  ]
 }
 Conduit {
  path=AnimGraph.AnimGraphNode_StateMachine_1.LeftArmAdditiveHits.AnimStateNode_1.LeftArmHit_A.AnimGraphNode_StateMachine_2.LeftArmHitChooser_StartTime_DelayFloat.AnimStateNode_5.Hit_Crash2_Soft_L.AnimGraphNode_StateMachine_0.Hit_Crash2_Soft_L_SM.AnimStateConduitNode_0.Conduit
  nodes [Result(bCanEnterTransition=true)]
 }
 Hit_Short {
  path=AnimGraph.AnimGraphNode_StateMachine_1.LeftArmAdditiveHits.AnimStateNode_1.LeftArmHit_A.AnimGraphNode_StateMachine_2.LeftArmHitChooser_StartTime_DelayFloat.AnimStateNode_5.Hit_Crash2_Soft_L.AnimGraphNode_StateMachine_0.Hit_Crash2_Soft_L_SM.AnimStateNode_0.Hit_Short
  nodes [Output Animation Pose(Result="Sequence Player (additive) 'Drum_Hit_Crash2_Tap_Fast'"(StartPosition=Get LeftArmMaxDelayShortenedBy(), bLoopAnimation=false) {sequence=&Drum_Hit_Crash2_Tap_Fast, startPosition=0.05})]
 }
 Hit_Medium {
  path=AnimGraph.AnimGraphNode_StateMachine_1.LeftArmAdditiveHits.AnimStateNode_1.LeftArmHit_A.AnimGraphNode_StateMachine_2.LeftArmHitChooser_StartTime_DelayFloat.AnimStateNode_5.Hit_Crash2_Soft_L.AnimGraphNode_StateMachine_0.Hit_Crash2_Soft_L_SM.AnimStateNode_1.Hit_Medium
  nodes [
   n1 = "Sequence Player (additive) 'Drum_Hit_Crash2_Normal_Medium'"(StartPosition=Get LeftArmMaxDelayShortenedBy(), bLoopAnimation=false) {sequence=&Drum_Hit_Crash2_Normal_Medium, startPosition=0.05}
   Output Animation Pose(Result=@n1)
  ]
 }
 Hit_Long {
  path=AnimGraph.AnimGraphNode_StateMachine_1.LeftArmAdditiveHits.AnimStateNode_1.LeftArmHit_A.AnimGraphNode_StateMachine_2.LeftArmHitChooser_StartTime_DelayFloat.AnimStateNode_5.Hit_Crash2_Soft_L.AnimGraphNode_StateMachine_0.Hit_Crash2_Soft_L_SM.AnimStateNode_2.Hit_Long
  nodes [Output Animation Pose(Result="Sequence Player (additive) 'Drum_Hit_Crash2_Slow_Big'"(StartPosition=Get LeftArmMaxDelayShortenedBy(), bLoopAnimation=false) {sequence=&Drum_Hit_Crash2_Slow_Big, startPosition=0.05})]
 }
 Transition {
  path=AnimGraph.AnimGraphNode_StateMachine_1.LeftArmAdditiveHits.AnimStateNode_1.LeftArmHit_A.AnimGraphNode_StateMachine_2.LeftArmHitChooser_StartTime_DelayFloat.AnimStateNode_5.Hit_Crash2_Soft_L.AnimGraphNode_StateMachine_0.Hit_Crash2_Soft_L_SM.AnimStateTransitionNode_4.Transition
  nodes [Result(bCanEnterTransition=Get LeftArmIsShortDelay())]
 }
 Transition {
  path=AnimGraph.AnimGraphNode_StateMachine_1.LeftArmAdditiveHits.AnimStateNode_1.LeftArmHit_A.AnimGraphNode_StateMachine_2.LeftArmHitChooser_StartTime_DelayFloat.AnimStateNode_5.Hit_Crash2_Soft_L.AnimGraphNode_StateMachine_0.Hit_Crash2_Soft_L_SM.AnimStateTransitionNode_5.Transition
  nodes [
   n1 = Get LeftArmIsLongDelay()
   n2 = Get LeftArmIsShortDelay()
   Result(bCanEnterTransition=AND Boolean(A=NOT Boolean(A=@n2), B=NOT Boolean(A=@n1)))
  ]
 }
 Transition {
  path=AnimGraph.AnimGraphNode_StateMachine_1.LeftArmAdditiveHits.AnimStateNode_1.LeftArmHit_A.AnimGraphNode_StateMachine_2.LeftArmHitChooser_StartTime_DelayFloat.AnimStateNode_5.Hit_Crash2_Soft_L.AnimGraphNode_StateMachine_0.Hit_Crash2_Soft_L_SM.AnimStateTransitionNode_6.Transition
  nodes [Result(bCanEnterTransition=Get LeftArmIsLongDelay())]
 }
 Hit_Ride_L {
  path=AnimGraph.AnimGraphNode_StateMachine_1.LeftArmAdditiveHits.AnimStateNode_1.LeftArmHit_A.AnimGraphNode_StateMachine_2.LeftArmHitChooser_StartTime_DelayFloat.AnimStateNode_6.Hit_Ride_L
  nodes [
   n1 = Get LeftArmMaxDelayShortenedBy()
   n2 = "Sequence Player (additive) 'Drum_Hit_Ride_Normal_Medium'"(StartPosition=Get LeftArmMaxDelayShortenedBy(), bLoopAnimation=false) {sequence=&Drum_Hit_Ride_Normal_Medium, startPosition=0.05}
   Blend Poses by bool(
    BlendPose_0="Sequence Player (additive) 'Drum_Hit_Ride_Slow_Big'"(StartPosition=Get LeftArmMaxDelayShortenedBy(), bLoopAnimation=false) {sequence=&Drum_Hit_Ride_Slow_Big, startPosition=0.05}
    BlendPose_1=Blend Poses by bool(
     BlendPose_0="Sequence Player (additive) 'Drum_Hit_Ride_Tap_Fast'"(StartPosition=@n1, bLoopAnimation=false) {sequence=&Drum_Hit_Ride_Tap_Fast, startPosition=0.05}
     BlendPose_1=@n2
     BlendTime_0=0
     BlendTime_1=0
     bActiveValue=Get LeftArmIsShortDelay()
    ) {blendType=Linear}
    BlendTime_0=0
    BlendTime_1=0
    bActiveValue=Get LeftArmIsLongDelay()
   ) {blendType=Linear}
   AnimGraphNode_StateMachine_0 = Hit_Ride_L_SM() {bAllowConduitEntryStates=true}
   Output Animation Pose(Result=@AnimGraphNode_StateMachine_0)
  ]
 }
 Hit_Ride_L_SM {
  entry=Conduit
  path=AnimGraph.AnimGraphNode_StateMachine_1.LeftArmAdditiveHits.AnimStateNode_1.LeftArmHit_A.AnimGraphNode_StateMachine_2.LeftArmHitChooser_StartTime_DelayFloat.AnimStateNode_6.Hit_Ride_L.AnimGraphNode_StateMachine_0.Hit_Ride_L_SM
  ids=id
  nodes [
   AnimStateConduitNode_0 = Conduit()
   AnimStateNode_0 = Hit_Short()
   AnimStateNode_1 = Hit_Medium()
   AnimStateNode_2 = Hit_Long()
   AnimStateTransitionNode_4 = AnimStateTransitionNode() {crossfadeDuration=0}
   AnimStateTransitionNode_5 = AnimStateTransitionNode() {crossfadeDuration=0}
   AnimStateTransitionNode_6 = AnimStateTransitionNode() {crossfadeDuration=0}
  ]
  links [
   {
    dst=Hit_Long
    rule=AnimGraph.AnimGraphNode_StateMachine_1.LeftArmAdditiveHits.AnimStateNode_1.LeftArmHit_A.AnimGraphNode_StateMachine_2.LeftArmHitChooser_StartTime_DelayFloat.AnimStateNode_6.Hit_Ride_L.AnimGraphNode_StateMachine_0.Hit_Ride_L_SM.AnimStateTransitionNode_6.Transition
    src=Conduit
    via=AnimStateTransitionNode_6
   }
   {
    dst=Hit_Medium
    rule=AnimGraph.AnimGraphNode_StateMachine_1.LeftArmAdditiveHits.AnimStateNode_1.LeftArmHit_A.AnimGraphNode_StateMachine_2.LeftArmHitChooser_StartTime_DelayFloat.AnimStateNode_6.Hit_Ride_L.AnimGraphNode_StateMachine_0.Hit_Ride_L_SM.AnimStateTransitionNode_5.Transition
    src=Conduit
    via=AnimStateTransitionNode_5
   }
   {
    dst=Hit_Short
    rule=AnimGraph.AnimGraphNode_StateMachine_1.LeftArmAdditiveHits.AnimStateNode_1.LeftArmHit_A.AnimGraphNode_StateMachine_2.LeftArmHitChooser_StartTime_DelayFloat.AnimStateNode_6.Hit_Ride_L.AnimGraphNode_StateMachine_0.Hit_Ride_L_SM.AnimStateTransitionNode_4.Transition
    src=Conduit
    via=AnimStateTransitionNode_4
   }
  ]
 }
 Conduit {
  path=AnimGraph.AnimGraphNode_StateMachine_1.LeftArmAdditiveHits.AnimStateNode_1.LeftArmHit_A.AnimGraphNode_StateMachine_2.LeftArmHitChooser_StartTime_DelayFloat.AnimStateNode_6.Hit_Ride_L.AnimGraphNode_StateMachine_0.Hit_Ride_L_SM.AnimStateConduitNode_0.Conduit
  nodes [Result(bCanEnterTransition=true)]
 }
 Hit_Short {
  path=AnimGraph.AnimGraphNode_StateMachine_1.LeftArmAdditiveHits.AnimStateNode_1.LeftArmHit_A.AnimGraphNode_StateMachine_2.LeftArmHitChooser_StartTime_DelayFloat.AnimStateNode_6.Hit_Ride_L.AnimGraphNode_StateMachine_0.Hit_Ride_L_SM.AnimStateNode_0.Hit_Short
  nodes [Output Animation Pose(Result="Sequence Player (additive) 'Drum_Hit_Ride_Tap_Fast'"(StartPosition=Get LeftArmMaxDelayShortenedBy(), bLoopAnimation=false) {sequence=&Drum_Hit_Ride_Tap_Fast, startPosition=0.05})]
 }
 Hit_Medium {
  path=AnimGraph.AnimGraphNode_StateMachine_1.LeftArmAdditiveHits.AnimStateNode_1.LeftArmHit_A.AnimGraphNode_StateMachine_2.LeftArmHitChooser_StartTime_DelayFloat.AnimStateNode_6.Hit_Ride_L.AnimGraphNode_StateMachine_0.Hit_Ride_L_SM.AnimStateNode_1.Hit_Medium
  nodes [
   n1 = "Sequence Player (additive) 'Drum_Hit_Ride_Normal_Medium'"(StartPosition=Get LeftArmMaxDelayShortenedBy(), bLoopAnimation=false) {sequence=&Drum_Hit_Ride_Normal_Medium, startPosition=0.05}
   Output Animation Pose(Result=@n1)
  ]
 }
 Hit_Long {
  path=AnimGraph.AnimGraphNode_StateMachine_1.LeftArmAdditiveHits.AnimStateNode_1.LeftArmHit_A.AnimGraphNode_StateMachine_2.LeftArmHitChooser_StartTime_DelayFloat.AnimStateNode_6.Hit_Ride_L.AnimGraphNode_StateMachine_0.Hit_Ride_L_SM.AnimStateNode_2.Hit_Long
  nodes [Output Animation Pose(Result="Sequence Player (additive) 'Drum_Hit_Ride_Slow_Big'"(StartPosition=Get LeftArmMaxDelayShortenedBy(), bLoopAnimation=false) {sequence=&Drum_Hit_Ride_Slow_Big, startPosition=0.05})]
 }
 Transition {
  path=AnimGraph.AnimGraphNode_StateMachine_1.LeftArmAdditiveHits.AnimStateNode_1.LeftArmHit_A.AnimGraphNode_StateMachine_2.LeftArmHitChooser_StartTime_DelayFloat.AnimStateNode_6.Hit_Ride_L.AnimGraphNode_StateMachine_0.Hit_Ride_L_SM.AnimStateTransitionNode_4.Transition
  nodes [Result(bCanEnterTransition=Get LeftArmIsShortDelay())]
 }
 Transition {
  path=AnimGraph.AnimGraphNode_StateMachine_1.LeftArmAdditiveHits.AnimStateNode_1.LeftArmHit_A.AnimGraphNode_StateMachine_2.LeftArmHitChooser_StartTime_DelayFloat.AnimStateNode_6.Hit_Ride_L.AnimGraphNode_StateMachine_0.Hit_Ride_L_SM.AnimStateTransitionNode_5.Transition
  nodes [
   n1 = Get LeftArmIsLongDelay()
   n2 = Get LeftArmIsShortDelay()
   Result(bCanEnterTransition=AND Boolean(A=NOT Boolean(A=@n2), B=NOT Boolean(A=@n1)))
  ]
 }
 Transition {
  path=AnimGraph.AnimGraphNode_StateMachine_1.LeftArmAdditiveHits.AnimStateNode_1.LeftArmHit_A.AnimGraphNode_StateMachine_2.LeftArmHitChooser_StartTime_DelayFloat.AnimStateNode_6.Hit_Ride_L.AnimGraphNode_StateMachine_0.Hit_Ride_L_SM.AnimStateTransitionNode_6.Transition
  nodes [Result(bCanEnterTransition=Get LeftArmIsLongDelay())]
 }
 Hit_Tom1_L {
  path=AnimGraph.AnimGraphNode_StateMachine_1.LeftArmAdditiveHits.AnimStateNode_1.LeftArmHit_A.AnimGraphNode_StateMachine_2.LeftArmHitChooser_StartTime_DelayFloat.AnimStateNode_7.Hit_Tom1_L
  nodes [
   n1 = Get LeftArmMaxDelayShortenedBy()
   n2 = "Sequence Player (additive) 'Drum_Hit_Tom1_Normal_Medium'"(StartPosition=Get LeftArmMaxDelayShortenedBy(), bLoopAnimation=false) {sequence=&Drum_Hit_Tom1_Normal_Medium, startPosition=0.05}
   Blend Poses by bool(
    BlendPose_0="Sequence Player (additive) 'Drum_Hit_Tom1_Slow_Big'"(StartPosition=Get LeftArmMaxDelayShortenedBy(), bLoopAnimation=false) {sequence=&Drum_Hit_Tom1_Slow_Big, startPosition=0.05}
    BlendPose_1=Blend Poses by bool(
     BlendPose_0="Sequence Player (additive) 'Drum_Hit_Tom1_Tap_Fast'"(StartPosition=@n1, bLoopAnimation=false) {sequence=&Drum_Hit_Tom1_Tap_Fast, startPosition=0.05}
     BlendPose_1=@n2
     BlendTime_0=0
     BlendTime_1=0
     bActiveValue=Get LeftArmIsShortDelay()
    ) {blendType=Linear}
    BlendTime_0=0
    BlendTime_1=0
    bActiveValue=Get LeftArmIsLongDelay()
   ) {blendType=Linear}
   AnimGraphNode_StateMachine_0 = Hit_Tom1_L_SM() {bAllowConduitEntryStates=true}
   Output Animation Pose(Result=@AnimGraphNode_StateMachine_0)
  ]
 }
 Hit_Tom1_L_SM {
  entry=Conduit
  path=AnimGraph.AnimGraphNode_StateMachine_1.LeftArmAdditiveHits.AnimStateNode_1.LeftArmHit_A.AnimGraphNode_StateMachine_2.LeftArmHitChooser_StartTime_DelayFloat.AnimStateNode_7.Hit_Tom1_L.AnimGraphNode_StateMachine_0.Hit_Tom1_L_SM
  ids=id
  nodes [
   AnimStateConduitNode_0 = Conduit()
   AnimStateNode_0 = Hit_Short()
   AnimStateNode_1 = Hit_Medium()
   AnimStateNode_2 = Hit_Long()
   AnimStateTransitionNode_4 = AnimStateTransitionNode() {crossfadeDuration=0}
   AnimStateTransitionNode_5 = AnimStateTransitionNode() {crossfadeDuration=0}
   AnimStateTransitionNode_6 = AnimStateTransitionNode() {crossfadeDuration=0}
  ]
  links [
   {
    dst=Hit_Long
    rule=AnimGraph.AnimGraphNode_StateMachine_1.LeftArmAdditiveHits.AnimStateNode_1.LeftArmHit_A.AnimGraphNode_StateMachine_2.LeftArmHitChooser_StartTime_DelayFloat.AnimStateNode_7.Hit_Tom1_L.AnimGraphNode_StateMachine_0.Hit_Tom1_L_SM.AnimStateTransitionNode_6.Transition
    src=Conduit
    via=AnimStateTransitionNode_6
   }
   {
    dst=Hit_Medium
    rule=AnimGraph.AnimGraphNode_StateMachine_1.LeftArmAdditiveHits.AnimStateNode_1.LeftArmHit_A.AnimGraphNode_StateMachine_2.LeftArmHitChooser_StartTime_DelayFloat.AnimStateNode_7.Hit_Tom1_L.AnimGraphNode_StateMachine_0.Hit_Tom1_L_SM.AnimStateTransitionNode_5.Transition
    src=Conduit
    via=AnimStateTransitionNode_5
   }
   {
    dst=Hit_Short
    rule=AnimGraph.AnimGraphNode_StateMachine_1.LeftArmAdditiveHits.AnimStateNode_1.LeftArmHit_A.AnimGraphNode_StateMachine_2.LeftArmHitChooser_StartTime_DelayFloat.AnimStateNode_7.Hit_Tom1_L.AnimGraphNode_StateMachine_0.Hit_Tom1_L_SM.AnimStateTransitionNode_4.Transition
    src=Conduit
    via=AnimStateTransitionNode_4
   }
  ]
 }
 Conduit {
  path=AnimGraph.AnimGraphNode_StateMachine_1.LeftArmAdditiveHits.AnimStateNode_1.LeftArmHit_A.AnimGraphNode_StateMachine_2.LeftArmHitChooser_StartTime_DelayFloat.AnimStateNode_7.Hit_Tom1_L.AnimGraphNode_StateMachine_0.Hit_Tom1_L_SM.AnimStateConduitNode_0.Conduit
  nodes [Result(bCanEnterTransition=true)]
 }
 Hit_Short {
  path=AnimGraph.AnimGraphNode_StateMachine_1.LeftArmAdditiveHits.AnimStateNode_1.LeftArmHit_A.AnimGraphNode_StateMachine_2.LeftArmHitChooser_StartTime_DelayFloat.AnimStateNode_7.Hit_Tom1_L.AnimGraphNode_StateMachine_0.Hit_Tom1_L_SM.AnimStateNode_0.Hit_Short
  nodes [Output Animation Pose(Result="Sequence Player (additive) 'Drum_Hit_Tom1_Tap_Fast'"(StartPosition=Get LeftArmMaxDelayShortenedBy(), bLoopAnimation=false) {sequence=&Drum_Hit_Tom1_Tap_Fast, startPosition=0.05})]
 }
 Hit_Medium {
  path=AnimGraph.AnimGraphNode_StateMachine_1.LeftArmAdditiveHits.AnimStateNode_1.LeftArmHit_A.AnimGraphNode_StateMachine_2.LeftArmHitChooser_StartTime_DelayFloat.AnimStateNode_7.Hit_Tom1_L.AnimGraphNode_StateMachine_0.Hit_Tom1_L_SM.AnimStateNode_1.Hit_Medium
  nodes [
   n1 = "Sequence Player (additive) 'Drum_Hit_Tom1_Normal_Medium'"(StartPosition=Get LeftArmMaxDelayShortenedBy(), bLoopAnimation=false) {sequence=&Drum_Hit_Tom1_Normal_Medium, startPosition=0.05}
   Output Animation Pose(Result=@n1)
  ]
 }
 Hit_Long {
  path=AnimGraph.AnimGraphNode_StateMachine_1.LeftArmAdditiveHits.AnimStateNode_1.LeftArmHit_A.AnimGraphNode_StateMachine_2.LeftArmHitChooser_StartTime_DelayFloat.AnimStateNode_7.Hit_Tom1_L.AnimGraphNode_StateMachine_0.Hit_Tom1_L_SM.AnimStateNode_2.Hit_Long
  nodes [Output Animation Pose(Result="Sequence Player (additive) 'Drum_Hit_Tom1_Slow_Big'"(StartPosition=Get LeftArmMaxDelayShortenedBy(), bLoopAnimation=false) {sequence=&Drum_Hit_Tom1_Slow_Big, startPosition=0.05})]
 }
 Transition {
  path=AnimGraph.AnimGraphNode_StateMachine_1.LeftArmAdditiveHits.AnimStateNode_1.LeftArmHit_A.AnimGraphNode_StateMachine_2.LeftArmHitChooser_StartTime_DelayFloat.AnimStateNode_7.Hit_Tom1_L.AnimGraphNode_StateMachine_0.Hit_Tom1_L_SM.AnimStateTransitionNode_4.Transition
  nodes [Result(bCanEnterTransition=Get LeftArmIsShortDelay())]
 }
 Transition {
  path=AnimGraph.AnimGraphNode_StateMachine_1.LeftArmAdditiveHits.AnimStateNode_1.LeftArmHit_A.AnimGraphNode_StateMachine_2.LeftArmHitChooser_StartTime_DelayFloat.AnimStateNode_7.Hit_Tom1_L.AnimGraphNode_StateMachine_0.Hit_Tom1_L_SM.AnimStateTransitionNode_5.Transition
  nodes [
   n1 = Get LeftArmIsLongDelay()
   n2 = Get LeftArmIsShortDelay()
   Result(bCanEnterTransition=AND Boolean(A=NOT Boolean(A=@n2), B=NOT Boolean(A=@n1)))
  ]
 }
 Transition {
  path=AnimGraph.AnimGraphNode_StateMachine_1.LeftArmAdditiveHits.AnimStateNode_1.LeftArmHit_A.AnimGraphNode_StateMachine_2.LeftArmHitChooser_StartTime_DelayFloat.AnimStateNode_7.Hit_Tom1_L.AnimGraphNode_StateMachine_0.Hit_Tom1_L_SM.AnimStateTransitionNode_6.Transition
  nodes [Result(bCanEnterTransition=Get LeftArmIsLongDelay())]
 }
 Hit_Tom2_L {
  path=AnimGraph.AnimGraphNode_StateMachine_1.LeftArmAdditiveHits.AnimStateNode_1.LeftArmHit_A.AnimGraphNode_StateMachine_2.LeftArmHitChooser_StartTime_DelayFloat.AnimStateNode_8.Hit_Tom2_L
  nodes [
   n1 = Get LeftArmMaxDelayShortenedBy()
   n2 = "Sequence Player (additive) 'Drum_Hit_Tom2_Normal_Medium'"(StartPosition=Get LeftArmMaxDelayShortenedBy(), bLoopAnimation=false) {sequence=&Drum_Hit_Tom2_Normal_Medium, startPosition=0.05}
   Blend Poses by bool(
    BlendPose_0="Sequence Player (additive) 'Drum_Hit_Tom2_Slow_Big'"(StartPosition=Get LeftArmMaxDelayShortenedBy(), bLoopAnimation=false) {sequence=&Drum_Hit_Tom2_Slow_Big, startPosition=0.05}
    BlendPose_1=Blend Poses by bool(
     BlendPose_0="Sequence Player (additive) 'Drum_Hit_Tom2_Tap_Fast'"(StartPosition=@n1, bLoopAnimation=false) {sequence=&Drum_Hit_Tom2_Tap_Fast, startPosition=0.05}
     BlendPose_1=@n2
     BlendTime_0=0
     BlendTime_1=0
     bActiveValue=Get LeftArmIsShortDelay()
    ) {blendType=Linear}
    BlendTime_0=0
    BlendTime_1=0
    bActiveValue=Get LeftArmIsLongDelay()
   ) {blendType=Linear}
   AnimGraphNode_StateMachine_0 = Hit_Tom2_L_SM() {bAllowConduitEntryStates=true}
   Output Animation Pose(Result=@AnimGraphNode_StateMachine_0)
  ]
 }
 Hit_Tom2_L_SM {
  entry=Conduit
  path=AnimGraph.AnimGraphNode_StateMachine_1.LeftArmAdditiveHits.AnimStateNode_1.LeftArmHit_A.AnimGraphNode_StateMachine_2.LeftArmHitChooser_StartTime_DelayFloat.AnimStateNode_8.Hit_Tom2_L.AnimGraphNode_StateMachine_0.Hit_Tom2_L_SM
  ids=id
  nodes [
   AnimStateConduitNode_0 = Conduit()
   AnimStateNode_0 = Hit_Short()
   AnimStateNode_1 = Hit_Medium()
   AnimStateNode_2 = Hit_Long()
   AnimStateTransitionNode_4 = AnimStateTransitionNode() {crossfadeDuration=0}
   AnimStateTransitionNode_5 = AnimStateTransitionNode() {crossfadeDuration=0}
   AnimStateTransitionNode_6 = AnimStateTransitionNode() {crossfadeDuration=0}
  ]
  links [
   {
    dst=Hit_Long
    rule=AnimGraph.AnimGraphNode_StateMachine_1.LeftArmAdditiveHits.AnimStateNode_1.LeftArmHit_A.AnimGraphNode_StateMachine_2.LeftArmHitChooser_StartTime_DelayFloat.AnimStateNode_8.Hit_Tom2_L.AnimGraphNode_StateMachine_0.Hit_Tom2_L_SM.AnimStateTransitionNode_6.Transition
    src=Conduit
    via=AnimStateTransitionNode_6
   }
   {
    dst=Hit_Medium
    rule=AnimGraph.AnimGraphNode_StateMachine_1.LeftArmAdditiveHits.AnimStateNode_1.LeftArmHit_A.AnimGraphNode_StateMachine_2.LeftArmHitChooser_StartTime_DelayFloat.AnimStateNode_8.Hit_Tom2_L.AnimGraphNode_StateMachine_0.Hit_Tom2_L_SM.AnimStateTransitionNode_5.Transition
    src=Conduit
    via=AnimStateTransitionNode_5
   }
   {
    dst=Hit_Short
    rule=AnimGraph.AnimGraphNode_StateMachine_1.LeftArmAdditiveHits.AnimStateNode_1.LeftArmHit_A.AnimGraphNode_StateMachine_2.LeftArmHitChooser_StartTime_DelayFloat.AnimStateNode_8.Hit_Tom2_L.AnimGraphNode_StateMachine_0.Hit_Tom2_L_SM.AnimStateTransitionNode_4.Transition
    src=Conduit
    via=AnimStateTransitionNode_4
   }
  ]
 }
 Conduit {
  path=AnimGraph.AnimGraphNode_StateMachine_1.LeftArmAdditiveHits.AnimStateNode_1.LeftArmHit_A.AnimGraphNode_StateMachine_2.LeftArmHitChooser_StartTime_DelayFloat.AnimStateNode_8.Hit_Tom2_L.AnimGraphNode_StateMachine_0.Hit_Tom2_L_SM.AnimStateConduitNode_0.Conduit
  nodes [Result(bCanEnterTransition=true)]
 }
 Hit_Short {
  path=AnimGraph.AnimGraphNode_StateMachine_1.LeftArmAdditiveHits.AnimStateNode_1.LeftArmHit_A.AnimGraphNode_StateMachine_2.LeftArmHitChooser_StartTime_DelayFloat.AnimStateNode_8.Hit_Tom2_L.AnimGraphNode_StateMachine_0.Hit_Tom2_L_SM.AnimStateNode_0.Hit_Short
  nodes [Output Animation Pose(Result="Sequence Player (additive) 'Drum_Hit_Tom2_Tap_Fast'"(StartPosition=Get LeftArmMaxDelayShortenedBy(), bLoopAnimation=false) {sequence=&Drum_Hit_Tom2_Tap_Fast, startPosition=0.05})]
 }
 Hit_Medium {
  path=AnimGraph.AnimGraphNode_StateMachine_1.LeftArmAdditiveHits.AnimStateNode_1.LeftArmHit_A.AnimGraphNode_StateMachine_2.LeftArmHitChooser_StartTime_DelayFloat.AnimStateNode_8.Hit_Tom2_L.AnimGraphNode_StateMachine_0.Hit_Tom2_L_SM.AnimStateNode_1.Hit_Medium
  nodes [
   n1 = "Sequence Player (additive) 'Drum_Hit_Tom2_Normal_Medium'"(StartPosition=Get LeftArmMaxDelayShortenedBy(), bLoopAnimation=false) {sequence=&Drum_Hit_Tom2_Normal_Medium, startPosition=0.05}
   Output Animation Pose(Result=@n1)
  ]
 }
 Hit_Long {
  path=AnimGraph.AnimGraphNode_StateMachine_1.LeftArmAdditiveHits.AnimStateNode_1.LeftArmHit_A.AnimGraphNode_StateMachine_2.LeftArmHitChooser_StartTime_DelayFloat.AnimStateNode_8.Hit_Tom2_L.AnimGraphNode_StateMachine_0.Hit_Tom2_L_SM.AnimStateNode_2.Hit_Long
  nodes [Output Animation Pose(Result="Sequence Player (additive) 'Drum_Hit_Tom2_Slow_Big'"(StartPosition=Get LeftArmMaxDelayShortenedBy(), bLoopAnimation=false) {sequence=&Drum_Hit_Tom2_Slow_Big, startPosition=0.05})]
 }
 Transition {
  path=AnimGraph.AnimGraphNode_StateMachine_1.LeftArmAdditiveHits.AnimStateNode_1.LeftArmHit_A.AnimGraphNode_StateMachine_2.LeftArmHitChooser_StartTime_DelayFloat.AnimStateNode_8.Hit_Tom2_L.AnimGraphNode_StateMachine_0.Hit_Tom2_L_SM.AnimStateTransitionNode_4.Transition
  nodes [Result(bCanEnterTransition=Get LeftArmIsShortDelay())]
 }
 Transition {
  path=AnimGraph.AnimGraphNode_StateMachine_1.LeftArmAdditiveHits.AnimStateNode_1.LeftArmHit_A.AnimGraphNode_StateMachine_2.LeftArmHitChooser_StartTime_DelayFloat.AnimStateNode_8.Hit_Tom2_L.AnimGraphNode_StateMachine_0.Hit_Tom2_L_SM.AnimStateTransitionNode_5.Transition
  nodes [
   n1 = Get LeftArmIsLongDelay()
   n2 = Get LeftArmIsShortDelay()
   Result(bCanEnterTransition=AND Boolean(A=NOT Boolean(A=@n2), B=NOT Boolean(A=@n1)))
  ]
 }
 Transition {
  path=AnimGraph.AnimGraphNode_StateMachine_1.LeftArmAdditiveHits.AnimStateNode_1.LeftArmHit_A.AnimGraphNode_StateMachine_2.LeftArmHitChooser_StartTime_DelayFloat.AnimStateNode_8.Hit_Tom2_L.AnimGraphNode_StateMachine_0.Hit_Tom2_L_SM.AnimStateTransitionNode_6.Transition
  nodes [Result(bCanEnterTransition=Get LeftArmIsLongDelay())]
 }
 Hit_FloorTom_L {
  path=AnimGraph.AnimGraphNode_StateMachine_1.LeftArmAdditiveHits.AnimStateNode_1.LeftArmHit_A.AnimGraphNode_StateMachine_2.LeftArmHitChooser_StartTime_DelayFloat.AnimStateNode_9.Hit_FloorTom_L
  nodes [
   n1 = Get LeftArmMaxDelayShortenedBy()
   n2 = "Sequence Player (additive) 'Drum_Hit_FloorTom_Normal_Medium'"(StartPosition=Get LeftArmMaxDelayShortenedBy(), bLoopAnimation=false) {sequence=&Drum_Hit_FloorTom_Normal_Medium, startPosition=0.05}
   Blend Poses by bool(
    BlendPose_0="Sequence Player (additive) 'Drum_Hit_FloorTom_Slow_Big'"(StartPosition=Get LeftArmMaxDelayShortenedBy(), bLoopAnimation=false) {sequence=&Drum_Hit_FloorTom_Slow_Big, startPosition=0.05}
    BlendPose_1=Blend Poses by bool(
     BlendPose_0="Sequence Player (additive) 'Drum_Hit_FloorTom_Tap_Fast'"(StartPosition=@n1, bLoopAnimation=false) {sequence=&Drum_Hit_FloorTom_Tap_Fast, startPosition=0.05}
     BlendPose_1=@n2
     BlendTime_0=0
     BlendTime_1=0
     bActiveValue=Get LeftArmIsShortDelay()
    ) {blendType=Linear}
    BlendTime_0=0
    BlendTime_1=0
    bActiveValue=Get LeftArmIsLongDelay()
   ) {blendType=Linear}
   AnimGraphNode_StateMachine_0 = Hit_FloorTom_R_SM() {bAllowConduitEntryStates=true}
   Output Animation Pose(Result=@AnimGraphNode_StateMachine_0)
  ]
 }
 Hit_FloorTom_R_SM {
  entry=Conduit
  path=AnimGraph.AnimGraphNode_StateMachine_1.LeftArmAdditiveHits.AnimStateNode_1.LeftArmHit_A.AnimGraphNode_StateMachine_2.LeftArmHitChooser_StartTime_DelayFloat.AnimStateNode_9.Hit_FloorTom_L.AnimGraphNode_StateMachine_0.Hit_FloorTom_R_SM
  ids=id
  nodes [
   AnimStateConduitNode_0 = Conduit()
   AnimStateNode_0 = Hit_Short()
   AnimStateNode_1 = Hit_Medium()
   AnimStateNode_2 = Hit_Long()
   AnimStateTransitionNode_4 = AnimStateTransitionNode() {crossfadeDuration=0}
   AnimStateTransitionNode_5 = AnimStateTransitionNode() {crossfadeDuration=0}
   AnimStateTransitionNode_6 = AnimStateTransitionNode() {crossfadeDuration=0}
  ]
  links [
   {
    dst=Hit_Long
    rule=AnimGraph.AnimGraphNode_StateMachine_1.LeftArmAdditiveHits.AnimStateNode_1.LeftArmHit_A.AnimGraphNode_StateMachine_2.LeftArmHitChooser_StartTime_DelayFloat.AnimStateNode_9.Hit_FloorTom_L.AnimGraphNode_StateMachine_0.Hit_FloorTom_R_SM.AnimStateTransitionNode_6.Transition
    src=Conduit
    via=AnimStateTransitionNode_6
   }
   {
    dst=Hit_Medium
    rule=AnimGraph.AnimGraphNode_StateMachine_1.LeftArmAdditiveHits.AnimStateNode_1.LeftArmHit_A.AnimGraphNode_StateMachine_2.LeftArmHitChooser_StartTime_DelayFloat.AnimStateNode_9.Hit_FloorTom_L.AnimGraphNode_StateMachine_0.Hit_FloorTom_R_SM.AnimStateTransitionNode_5.Transition
    src=Conduit
    via=AnimStateTransitionNode_5
   }
   {
    dst=Hit_Short
    rule=AnimGraph.AnimGraphNode_StateMachine_1.LeftArmAdditiveHits.AnimStateNode_1.LeftArmHit_A.AnimGraphNode_StateMachine_2.LeftArmHitChooser_StartTime_DelayFloat.AnimStateNode_9.Hit_FloorTom_L.AnimGraphNode_StateMachine_0.Hit_FloorTom_R_SM.AnimStateTransitionNode_4.Transition
    src=Conduit
    via=AnimStateTransitionNode_4
   }
  ]
 }
 Conduit {
  path=AnimGraph.AnimGraphNode_StateMachine_1.LeftArmAdditiveHits.AnimStateNode_1.LeftArmHit_A.AnimGraphNode_StateMachine_2.LeftArmHitChooser_StartTime_DelayFloat.AnimStateNode_9.Hit_FloorTom_L.AnimGraphNode_StateMachine_0.Hit_FloorTom_R_SM.AnimStateConduitNode_0.Conduit
  nodes [Result(bCanEnterTransition=true)]
 }
 Hit_Short {
  path=AnimGraph.AnimGraphNode_StateMachine_1.LeftArmAdditiveHits.AnimStateNode_1.LeftArmHit_A.AnimGraphNode_StateMachine_2.LeftArmHitChooser_StartTime_DelayFloat.AnimStateNode_9.Hit_FloorTom_L.AnimGraphNode_StateMachine_0.Hit_FloorTom_R_SM.AnimStateNode_0.Hit_Short
  nodes [Output Animation Pose(Result="Sequence Player (additive) 'Drum_Hit_FloorTom_Tap_Fast'"(StartPosition=Get LeftArmMaxDelayShortenedBy(), bLoopAnimation=false) {sequence=&Drum_Hit_FloorTom_Tap_Fast, startPosition=0.05})]
 }
 Hit_Medium {
  path=AnimGraph.AnimGraphNode_StateMachine_1.LeftArmAdditiveHits.AnimStateNode_1.LeftArmHit_A.AnimGraphNode_StateMachine_2.LeftArmHitChooser_StartTime_DelayFloat.AnimStateNode_9.Hit_FloorTom_L.AnimGraphNode_StateMachine_0.Hit_FloorTom_R_SM.AnimStateNode_1.Hit_Medium
  nodes [
   n1 = "Sequence Player (additive) 'Drum_Hit_FloorTom_Normal_Medium'"(StartPosition=Get LeftArmMaxDelayShortenedBy(), bLoopAnimation=false) {sequence=&Drum_Hit_FloorTom_Normal_Medium, startPosition=0.05}
   Output Animation Pose(Result=@n1)
  ]
 }
 Hit_Long {
  path=AnimGraph.AnimGraphNode_StateMachine_1.LeftArmAdditiveHits.AnimStateNode_1.LeftArmHit_A.AnimGraphNode_StateMachine_2.LeftArmHitChooser_StartTime_DelayFloat.AnimStateNode_9.Hit_FloorTom_L.AnimGraphNode_StateMachine_0.Hit_FloorTom_R_SM.AnimStateNode_2.Hit_Long
  nodes [Output Animation Pose(Result="Sequence Player (additive) 'Drum_Hit_FloorTom_Slow_Big'"(StartPosition=Get LeftArmMaxDelayShortenedBy(), bLoopAnimation=false) {sequence=&Drum_Hit_FloorTom_Slow_Big, startPosition=0.05})]
 }
 Transition {
  path=AnimGraph.AnimGraphNode_StateMachine_1.LeftArmAdditiveHits.AnimStateNode_1.LeftArmHit_A.AnimGraphNode_StateMachine_2.LeftArmHitChooser_StartTime_DelayFloat.AnimStateNode_9.Hit_FloorTom_L.AnimGraphNode_StateMachine_0.Hit_FloorTom_R_SM.AnimStateTransitionNode_4.Transition
  nodes [Result(bCanEnterTransition=Get LeftArmIsShortDelay())]
 }
 Transition {
  path=AnimGraph.AnimGraphNode_StateMachine_1.LeftArmAdditiveHits.AnimStateNode_1.LeftArmHit_A.AnimGraphNode_StateMachine_2.LeftArmHitChooser_StartTime_DelayFloat.AnimStateNode_9.Hit_FloorTom_L.AnimGraphNode_StateMachine_0.Hit_FloorTom_R_SM.AnimStateTransitionNode_5.Transition
  nodes [
   n1 = Get LeftArmIsLongDelay()
   n2 = Get LeftArmIsShortDelay()
   Result(bCanEnterTransition=AND Boolean(A=NOT Boolean(A=@n2), B=NOT Boolean(A=@n1)))
  ]
 }
 Transition {
  path=AnimGraph.AnimGraphNode_StateMachine_1.LeftArmAdditiveHits.AnimStateNode_1.LeftArmHit_A.AnimGraphNode_StateMachine_2.LeftArmHitChooser_StartTime_DelayFloat.AnimStateNode_9.Hit_FloorTom_L.AnimGraphNode_StateMachine_0.Hit_FloorTom_R_SM.AnimStateTransitionNode_6.Transition
  nodes [Result(bCanEnterTransition=Get LeftArmIsLongDelay())]
 }
 Transition {
  path=AnimGraph.AnimGraphNode_StateMachine_1.LeftArmAdditiveHits.AnimStateNode_1.LeftArmHit_A.AnimGraphNode_StateMachine_2.LeftArmHitChooser_StartTime_DelayFloat.AnimStateTransitionNode_10.Transition
  nodes [Result(bCanEnterTransition=Get Hit_Snare_L())]
 }
 Transition {
  path=AnimGraph.AnimGraphNode_StateMachine_1.LeftArmAdditiveHits.AnimStateNode_1.LeftArmHit_A.AnimGraphNode_StateMachine_2.LeftArmHitChooser_StartTime_DelayFloat.AnimStateTransitionNode_12.Transition
  nodes [Result(bCanEnterTransition=Get Hit_HiHat_L())]
 }
 Transition {
  path=AnimGraph.AnimGraphNode_StateMachine_1.LeftArmAdditiveHits.AnimStateNode_1.LeftArmHit_A.AnimGraphNode_StateMachine_2.LeftArmHitChooser_StartTime_DelayFloat.AnimStateTransitionNode_14.Transition
  nodes [Result(bCanEnterTransition=Get Hit_Crash1_Hard_L())]
 }
 Transition {
  path=AnimGraph.AnimGraphNode_StateMachine_1.LeftArmAdditiveHits.AnimStateNode_1.LeftArmHit_A.AnimGraphNode_StateMachine_2.LeftArmHitChooser_StartTime_DelayFloat.AnimStateTransitionNode_16.Transition
  nodes [Result(bCanEnterTransition=Get Hit_Crash1_Soft_L())]
 }
 Transition {
  path=AnimGraph.AnimGraphNode_StateMachine_1.LeftArmAdditiveHits.AnimStateNode_1.LeftArmHit_A.AnimGraphNode_StateMachine_2.LeftArmHitChooser_StartTime_DelayFloat.AnimStateTransitionNode_18.Transition
  nodes [Result(bCanEnterTransition=Get Hit_Crash2_Hard_L())]
 }
 Transition {
  path=AnimGraph.AnimGraphNode_StateMachine_1.LeftArmAdditiveHits.AnimStateNode_1.LeftArmHit_A.AnimGraphNode_StateMachine_2.LeftArmHitChooser_StartTime_DelayFloat.AnimStateTransitionNode_20.Transition
  nodes [Result(bCanEnterTransition=Get Hit_Crash2_Soft_L())]
 }
 Transition {
  path=AnimGraph.AnimGraphNode_StateMachine_1.LeftArmAdditiveHits.AnimStateNode_1.LeftArmHit_A.AnimGraphNode_StateMachine_2.LeftArmHitChooser_StartTime_DelayFloat.AnimStateTransitionNode_22.Transition
  nodes [Result(bCanEnterTransition=Get Hit_Ride_L())]
 }
 Transition {
  path=AnimGraph.AnimGraphNode_StateMachine_1.LeftArmAdditiveHits.AnimStateNode_1.LeftArmHit_A.AnimGraphNode_StateMachine_2.LeftArmHitChooser_StartTime_DelayFloat.AnimStateTransitionNode_24.Transition
  nodes [Result(bCanEnterTransition=Get Hit_Tom1_L())]
 }
 Transition {
  path=AnimGraph.AnimGraphNode_StateMachine_1.LeftArmAdditiveHits.AnimStateNode_1.LeftArmHit_A.AnimGraphNode_StateMachine_2.LeftArmHitChooser_StartTime_DelayFloat.AnimStateTransitionNode_26.Transition
  nodes [Result(bCanEnterTransition=Get Hit_Tom2_L())]
 }
 Transition {
  path=AnimGraph.AnimGraphNode_StateMachine_1.LeftArmAdditiveHits.AnimStateNode_1.LeftArmHit_A.AnimGraphNode_StateMachine_2.LeftArmHitChooser_StartTime_DelayFloat.AnimStateTransitionNode_28.Transition
  nodes [Result(bCanEnterTransition=Get Hit_FloorTom_L())]
 }
 Transition {
  path=AnimGraph.AnimGraphNode_StateMachine_1.LeftArmAdditiveHits.AnimStateNode_1.LeftArmHit_A.AnimGraphNode_StateMachine_2.LeftArmHitChooser_StartTime_DelayFloat.AnimStateTransitionNode_4.Transition
  nodes [Result(bCanEnterTransition=Get Hit_Snare_Soft_L())]
 }
 Neutral {
  path=AnimGraph.AnimGraphNode_StateMachine_1.LeftArmAdditiveHits.AnimStateNode_2.Neutral
  nodes [
   "Sequence Player (additive) 'Drum_SpineTwist_Center_Pose'"() {bLoopAnimation=false, sequence=&Drum_SpineTwist_Center_Pose}
   Output Animation Pose(Result=Additive Identity Pose())
   "Use cached pose 'Body (Cache Pose)'"()
  ]
 }
 Transition {
  path=AnimGraph.AnimGraphNode_StateMachine_1.LeftArmAdditiveHits.AnimStateTransitionNode_0.Transition
  nodes [
   n1 = Get LeftArmHitLongBlend() -> (LeftArmHitLongBlend:bool)
   n2 = Get LeftArmHitMediumBlend() -> (LeftArmHitMediumBlend:bool)
   Result(bCanEnterTransition=AND Boolean!(A=Get TriggerLeftArmHit_B() -> (TriggerLeftArmHit_B:bool), B=NOT Boolean(A=@n1), C:bool=NOT Boolean(A=@n2)) -> (ReturnValue))
   Get LeftArmIsShortDelay()
  ]
 }
 Transition {
  path=AnimGraph.AnimGraphNode_StateMachine_1.LeftArmAdditiveHits.AnimStateTransitionNode_1.Transition
  nodes [
   n1 = Get LeftArmHitMediumBlend()
   n2 = Get LeftArmHitLongBlend()
   Result(bCanEnterTransition=AND Boolean(A=Get TriggerLeftArmHit_A() -> (TriggerLeftArmHit_A:bool), B=NOT Boolean(A=@n2), C=NOT Boolean(A=@n1)))
   Get LeftArmIsShortDelay()
  ]
 }
 Transition {
  path=AnimGraph.AnimGraphNode_StateMachine_1.LeftArmAdditiveHits.AnimStateTransitionNode_10.Transition
  nodes [
   n1 = Get TriggerLeftArmHit_A()
   Result(bCanEnterTransition=@n1)
   OR Boolean(A=@n1, B=Get TriggerLeftArmHit_B())
  ]
 }
 Transition {
  path=AnimGraph.AnimGraphNode_StateMachine_1.LeftArmAdditiveHits.AnimStateTransitionNode_2.Transition
  nodes [
   n1 = Get LeftArmHitLongBlend()
   Result(bCanEnterTransition=AND Boolean(A=Get TriggerLeftArmHit_B(), B=NOT Boolean(A=@n1), C=Get LeftArmHitMediumBlend()))
   NOT Boolean(A=Get LeftArmIsShortDelay())
  ]
 }
 Transition {
  path=AnimGraph.AnimGraphNode_StateMachine_1.LeftArmAdditiveHits.AnimStateTransitionNode_3.Transition
  nodes [
   n1 = Get LeftArmHitLongBlend()
   Result(bCanEnterTransition=AND Boolean(A=Get TriggerLeftArmHit_A(), B=NOT Boolean(A=@n1), C=Get LeftArmHitMediumBlend()))
   NOT Boolean(A=Get LeftArmIsShortDelay())
  ]
 }
 Transition {path=AnimGraph.AnimGraphNode_StateMachine_1.LeftArmAdditiveHits.AnimStateTransitionNode_4.Transition, nodes [Result()]}
 Transition {path=AnimGraph.AnimGraphNode_StateMachine_1.LeftArmAdditiveHits.AnimStateTransitionNode_5.Transition, nodes [Result()]}
 Transition {
  path=AnimGraph.AnimGraphNode_StateMachine_1.LeftArmAdditiveHits.AnimStateTransitionNode_6.Transition
  nodes [Result(bCanEnterTransition=AND Boolean(A=Get TriggerLeftArmHit_A(), B=Get LeftArmHitLongBlend(), C=Get LeftArmHitMediumBlend()))]
 }
 Transition {
  path=AnimGraph.AnimGraphNode_StateMachine_1.LeftArmAdditiveHits.AnimStateTransitionNode_7.Transition
  nodes [
   n1 = Get TriggerLeftArmHit_B()
   Result(bCanEnterTransition=@n1)
   OR Boolean(A=Get TriggerLeftArmHit_A(), B=@n1)
  ]
 }
 Transition {
  path=AnimGraph.AnimGraphNode_StateMachine_1.LeftArmAdditiveHits.AnimStateTransitionNode_8.Transition
  nodes [Result(bCanEnterTransition=AND Boolean(A=Get TriggerLeftArmHit_B(), B=Get LeftArmHitLongBlend(), C=Get LeftArmHitMediumBlend()))]
 }
 Body {
  entry=Conduit
  path=AnimGraph.AnimGraphNode_StateMachine_2.Body
  ids=id
  nodes [
   AnimStateConduitNode_0 = Conduit()
   AnimStateNode_10 = Play 0() {
    is=title AnimStateNode
    stateEntered {guid=#34}
    stateFullyBlended {guid=#35}
    stateLeft {guid=#36}
   }
   AnimStateNode_32 = Play 1() {
    is=title AnimStateNode
    stateEntered {guid=#34}
    stateFullyBlended {guid=#35}
    stateLeft {guid=#36}
   }
   AnimStateTransitionNode_44 = AnimStateTransitionNode() {
    crossfadeDuration=0.4
    transitionEnd {guid=#37}
    transitionInterrupt {guid=#38}
    transitionStart {guid=#39}
   }
   AnimStateTransitionNode_53 = AnimStateTransitionNode() {crossfadeDuration=0.5}
   AnimStateTransitionNode_81 = AnimStateTransitionNode() {
    crossfadeDuration=0.5
    transitionEnd {guid=#40}
    transitionInterrupt {guid=#41}
    transitionStart {guid=#42}
   }
  ]
  links [
   {
    dst=Play 0
    rule=AnimGraph.AnimGraphNode_StateMachine_2.Body.AnimStateTransitionNode_44.AnimationTransitionGraph_8
    src=Conduit
    via=AnimStateTransitionNode_44
   }
   {
    dst=Play 1
    rule=AnimGraph.AnimGraphNode_StateMachine_2.Body.AnimStateTransitionNode_81.AnimationTransitionGraph_8
    src=Play 0
    via=AnimStateTransitionNode_81
   }
   {
    dst=Play 0
    rule=AnimGraph.AnimGraphNode_StateMachine_2.Body.AnimStateTransitionNode_53.AnimationTransitionGraph_8
    src=Play 1
    via=AnimStateTransitionNode_53
   }
  ]
 }
 Conduit {path=AnimGraph.AnimGraphNode_StateMachine_2.Body.AnimStateConduitNode_0.Conduit, nodes [Result(bCanEnterTransition=true)]}
 Play 0 {
  path=AnimGraph.AnimGraphNode_StateMachine_2.Body.AnimStateNode_10.Play 0
  nodes [
   n1 = Get PlaySequences() -> (PlaySequences:Array of Anim Sequence Object References)
   n2 = Get TempoMultipliers() -> (TempoMultipliers:"Array of Float (double-precision)s")
   Output Animation Pose(Result=Play Beat Synced Animation!(
     ExplicitTime:float=0
     bShouldLoop:bool=True
     InSequence:Anim Sequence Base Object Reference="Get (a copy)"(Array:Array of Anim Sequence Object References=@n1) -> (Output:Anim Sequence Object Reference)
     MusicClock:Music Clock Object Reference
     TempoMultiplier:float="Get (a copy)"(Array:"Array of Float (double-precision)s"=@n2) -> (Output:double)
     bEnableTempoVariants:bool=False
     TempoVariantHysteresis:float=0.05
     bShowDebug:bool=False
    ) -> (Pose:Pose Link Structure) {is=title AnimGraphNode_PlayBeatSyncedAnim}) {
    node {stateEntryFunction {functionName=OnStateEntry}}
    stateEntryFunction {bSelfContext=true, memberGuid=#43, memberName=OnPlay0Entry}
   }
  ]
 }
 Play 1 {
  path=AnimGraph.AnimGraphNode_StateMachine_2.Body.AnimStateNode_32.Play 1
  nodes [
   n1 = Get PlaySequences()
   n2 = Get TempoMultipliers()
   Output Animation Pose(Result=Play Beat Synced Animation(
     ExplicitTime=0
     bShouldLoop=True
     InSequence="Get (a copy)"(Array:Array of Anim Sequence Object References=@n1, Dimension 1=1) -> (Output:Anim Sequence Object Reference)
     TempoMultiplier="Get (a copy)"(Array:"Array of Float (double-precision)s"=@n2, Dimension 1=1) -> (Output:double)
     bEnableTempoVariants=False
     TempoVariantHysteresis=0.05
     bShowDebug=False
    )) {
    node {stateEntryFunction {functionName=OnStateEntry}}
    stateEntryFunction {bSelfContext=true, memberGuid=#44, memberName=OnPlay1Entry}
   }
  ]
 }
 AnimationTransitionGraph_8 {
  path=AnimGraph.AnimGraphNode_StateMachine_2.Body.AnimStateTransitionNode_44.AnimationTransitionGraph_8
  nodes [Result(bCanEnterTransition=true)]
 }
 AnimationTransitionGraph_8 {
  path=AnimGraph.AnimGraphNode_StateMachine_2.Body.AnimStateTransitionNode_53.AnimationTransitionGraph_8
  nodes [Result(bCanEnterTransition=ShouldTransition!(self) -> (Transition:bool))]
 }
 AnimationTransitionGraph_8 {
  path=AnimGraph.AnimGraphNode_StateMachine_2.Body.AnimStateTransitionNode_81.AnimationTransitionGraph_8
  nodes [Result(bCanEnterTransition=ShouldTransition())]
 }
 Left Leg {
  entry=HiHat_ClosedHold
  path=AnimGraph.AnimGraphNode_StateMachine_4.Left Leg
  ids=id
  nodes [
   AnimStateNode_0 = HiHat_ClosedHold() {is=title AnimStateNode}
   AnimStateNode_1 = HitHat_Open() {is=title AnimStateNode}
   AnimStateNode_2 = HiHat_OpenHold() {is=title AnimStateNode}
   AnimStateNode_3 = HiHat_Close() {is=title AnimStateNode}
   AnimStateTransitionNode_0 = AnimStateTransitionNode() {crossfadeDuration=0}
   AnimStateTransitionNode_1 = AnimStateTransitionNode() {bAutomaticRuleBasedOnSequencePlayerInState=true, crossfadeDuration=0}
   AnimStateTransitionNode_2 = AnimStateTransitionNode() {crossfadeDuration=0}
   AnimStateTransitionNode_3 = AnimStateTransitionNode() {bAutomaticRuleBasedOnSequencePlayerInState=true, crossfadeDuration=0}
   AnimStateTransitionNode_4 = AnimStateTransitionNode() {crossfadeDuration=0.1}
   AnimStateTransitionNode_6 = AnimStateTransitionNode() {crossfadeDuration=0.1}
  ]
  links [
   {
    dst=HiHat_ClosedHold
    rule=AnimGraph.AnimGraphNode_StateMachine_4.Left Leg.AnimStateTransitionNode_3.Transition
    src=HiHat_Close
    via=AnimStateTransitionNode_3
   }
   {
    dst=HitHat_Open
    rule=AnimGraph.AnimGraphNode_StateMachine_4.Left Leg.AnimStateTransitionNode_6.Transition
    src=HiHat_Close
    via=AnimStateTransitionNode_6
   }
   {
    dst=HitHat_Open
    rule=AnimGraph.AnimGraphNode_StateMachine_4.Left Leg.AnimStateTransitionNode_0.Transition
    src=HiHat_ClosedHold
    via=AnimStateTransitionNode_0
   }
   {
    dst=HiHat_Close
    rule=AnimGraph.AnimGraphNode_StateMachine_4.Left Leg.AnimStateTransitionNode_2.Transition
    src=HiHat_OpenHold
    via=AnimStateTransitionNode_2
   }
   {
    dst=HiHat_Close
    rule=AnimGraph.AnimGraphNode_StateMachine_4.Left Leg.AnimStateTransitionNode_4.Transition
    src=HitHat_Open
    via=AnimStateTransitionNode_4
   }
   {
    dst=HiHat_OpenHold
    rule=AnimGraph.AnimGraphNode_StateMachine_4.Left Leg.AnimStateTransitionNode_1.Transition
    src=HitHat_Open
    via=AnimStateTransitionNode_1
   }
  ]
 }
 HiHat_ClosedHold {
  path=AnimGraph.AnimGraphNode_StateMachine_4.Left Leg.AnimStateNode_0.HiHat_ClosedHold
  nodes [Output Animation Pose(Result=Sequence Player 'Drum_Hit_HiHat_ClosedHold'() {bLoopAnimation=false, sequence=/SparksCharacterCommon/Animation/MainPlayer/BandPlay/Drum/Hits/Drum_Hit_HiHat_ClosedHold})]
 }
 HitHat_Open {
  path=AnimGraph.AnimGraphNode_StateMachine_4.Left Leg.AnimStateNode_1.HitHat_Open
  nodes [Output Animation Pose(Result=Sequence Player 'Drum_Hit_HiHat_Open'() {bLoopAnimation=false, sequence=/SparksCharacterCommon/Animation/MainPlayer/BandPlay/Drum/Hits/Drum_Hit_HiHat_Open})]
 }
 HiHat_OpenHold {
  path=AnimGraph.AnimGraphNode_StateMachine_4.Left Leg.AnimStateNode_2.HiHat_OpenHold
  nodes [Output Animation Pose(Result=Sequence Player 'Drum_Hit_HiHat_OpenHold'() {bLoopAnimation=false, sequence=/SparksCharacterCommon/Animation/MainPlayer/BandPlay/Drum/Hits/Drum_Hit_HiHat_OpenHold})]
 }
 HiHat_Close {
  path=AnimGraph.AnimGraphNode_StateMachine_4.Left Leg.AnimStateNode_3.HiHat_Close
  nodes [Output Animation Pose(Result=Sequence Player 'Drum_Hit_HiHat_Close'() {bLoopAnimation=false, sequence=/SparksCharacterCommon/Animation/MainPlayer/BandPlay/Drum/Hits/Drum_Hit_HiHat_Close})]
 }
 Transition {
  path=AnimGraph.AnimGraphNode_StateMachine_4.Left Leg.AnimStateTransitionNode_0.Transition
  nodes [Result(bCanEnterTransition=Get Trigger_HiHat_Open() -> (Trigger_HiHat_Open:bool))]
 }
 Transition {path=AnimGraph.AnimGraphNode_StateMachine_4.Left Leg.AnimStateTransitionNode_1.Transition, nodes [Result()]}
 Transition {
  path=AnimGraph.AnimGraphNode_StateMachine_4.Left Leg.AnimStateTransitionNode_2.Transition
  nodes [Result(bCanEnterTransition=Get Trigger_HiHat_Close() -> (Trigger_HiHat_Close:bool))]
 }
 Transition {path=AnimGraph.AnimGraphNode_StateMachine_4.Left Leg.AnimStateTransitionNode_3.Transition, nodes [Result()]}
 Transition {
  path=AnimGraph.AnimGraphNode_StateMachine_4.Left Leg.AnimStateTransitionNode_4.Transition
  nodes [Result(bCanEnterTransition=Get Trigger_HiHat_Close())]
 }
 Transition {
  path=AnimGraph.AnimGraphNode_StateMachine_4.Left Leg.AnimStateTransitionNode_6.Transition
  nodes [Result(bCanEnterTransition=Get Trigger_HiHat_Open())]
 }
 Right Leg {
  entry=Bass_Hit_Ready
  path=AnimGraph.AnimGraphNode_StateMachine_5.Right Leg
  ids=id
  nodes [
   AnimStateNode_0 = Bass_Hit_Ready() {is=title AnimStateNode}
   AnimStateNode_1 = Bass_Hit_B() {is=title AnimStateNode, bAlwaysResetOnEntry=true}
   AnimStateNode_2 = Bass_Hit_A() {is=title AnimStateNode, bAlwaysResetOnEntry=true}
   AnimStateTransitionNode_14 = AnimStateTransitionNode() {bAutomaticRuleBasedOnSequencePlayerInState=true, crossfadeDuration=0}
   AnimStateTransitionNode_15 = AnimStateTransitionNode() {crossfadeDuration=0}
   AnimStateTransitionNode_16 = AnimStateTransitionNode() {bAutomaticRuleBasedOnSequencePlayerInState=true, crossfadeDuration=0}
   AnimStateTransitionNode_17 = AnimStateTransitionNode() {crossfadeDuration=0.1}
   AnimStateTransitionNode_18 = AnimStateTransitionNode() {crossfadeDuration=0.1}
  ]
  links [
   {
    dst=Bass_Hit_B
    rule=AnimGraph.AnimGraphNode_StateMachine_5.Right Leg.AnimStateTransitionNode_17.Transition
    src=Bass_Hit_A
    via=AnimStateTransitionNode_17
   }
   {
    dst=Bass_Hit_Ready
    rule=AnimGraph.AnimGraphNode_StateMachine_5.Right Leg.AnimStateTransitionNode_14.Transition
    src=Bass_Hit_A
    via=AnimStateTransitionNode_14
   }
   {
    dst=Bass_Hit_A
    rule=AnimGraph.AnimGraphNode_StateMachine_5.Right Leg.AnimStateTransitionNode_18.Transition
    src=Bass_Hit_B
    via=AnimStateTransitionNode_18
   }
   {
    dst=Bass_Hit_Ready
    rule=AnimGraph.AnimGraphNode_StateMachine_5.Right Leg.AnimStateTransitionNode_16.Transition
    src=Bass_Hit_B
    via=AnimStateTransitionNode_16
   }
   {
    dst=Bass_Hit_B
    rule=AnimGraph.AnimGraphNode_StateMachine_5.Right Leg.AnimStateTransitionNode_15.Transition
    src=Bass_Hit_Ready
    via=AnimStateTransitionNode_15
   }
  ]
 }
 Bass_Hit_Ready {
  path=AnimGraph.AnimGraphNode_StateMachine_5.Right Leg.AnimStateNode_0.Bass_Hit_Ready
  nodes [Output Animation Pose(Result=Sequence Player 'Drum_Hit_BassDrum_Ready'() {bLoopAnimation=false, sequence=/SparksCharacterCommon/Animation/MainPlayer/BandPlay/Drum/Hits/Drum_Hit_BassDrum_Ready})]
 }
 Bass_Hit_B {
  path=AnimGraph.AnimGraphNode_StateMachine_5.Right Leg.AnimStateNode_1.Bass_Hit_B
  nodes [Output Animation Pose(Result=Sequence Player 'Drum_Hit_BassDrum_Kick'!(StartPosition:float=Get RightLegMaxDelayShortenedBy() -> (RightLegMaxDelayShortenedBy:float)) -> (Pose) {bLoopAnimation=false, sequence=/SparksCharacterCommon/Animation/MainPlayer/BandPlay/Drum/Hits/Drum_Hit_BassDrum_Kick}) {
    node {stateEntryFunction {functionName=OnHitBassEntry}}
    stateEntryFunction {bSelfContext=true, memberGuid=#45, memberName=OnHitBassEntry}
   }]
 }
 Bass_Hit_A {
  path=AnimGraph.AnimGraphNode_StateMachine_5.Right Leg.AnimStateNode_2.Bass_Hit_A
  nodes [Output Animation Pose(Result=Sequence Player 'Drum_Hit_BassDrum_Kick'(StartPosition=Get RightLegMaxDelayShortenedBy()) {bLoopAnimation=false, sequence=&Drum_Hit_BassDrum_Kick}) {
    node {stateEntryFunction {functionName=OnHitBassEntry}}
    stateEntryFunction {bSelfContext=true, memberGuid=#45, memberName=OnHitBassEntry}
   }]
 }
 Transition {path=AnimGraph.AnimGraphNode_StateMachine_5.Right Leg.AnimStateTransitionNode_14.Transition, nodes [Result()]}
 Transition {
  path=AnimGraph.AnimGraphNode_StateMachine_5.Right Leg.AnimStateTransitionNode_15.Transition
  nodes [Result(bCanEnterTransition=ShouldHitBassTransition!(self) -> (Transition:bool))]
 }
 Transition {path=AnimGraph.AnimGraphNode_StateMachine_5.Right Leg.AnimStateTransitionNode_16.Transition, nodes [Result()]}
 Transition {
  path=AnimGraph.AnimGraphNode_StateMachine_5.Right Leg.AnimStateTransitionNode_17.Transition
  nodes [Result(bCanEnterTransition=ShouldHitBassTransition())]
 }
 Transition {
  path=AnimGraph.AnimGraphNode_StateMachine_5.Right Leg.AnimStateTransitionNode_18.Transition
  nodes [Result(bCanEnterTransition=ShouldHitBassTransition())]
 }
 RightArmAdditiveHits {
  entry=Neutral
  path=AnimGraph.AnimGraphNode_StateMachine_6.RightArmAdditiveHits
  ids=id
  nodes [
   AnimStateNode_0 = RightArmHit_B() {is=title AnimStateNode, bAlwaysResetOnEntry=true, stateEntered {guid=#46, notifyName=RightArmHit_B_Entered}}
   AnimStateNode_1 = RightArmHit_A() {is=title AnimStateNode, bAlwaysResetOnEntry=true, stateEntered {guid=#47, notifyName=RightArmHit_A_Entered}}
   AnimStateNode_2 = Neutral() {stateEntered {guid=#48, notifyName=RightArm_Neutral_Entered}}
   AnimStateTransitionNode_0 = AnimStateTransitionNode() {crossfadeDuration=0.05, priorityOrder=3}
   AnimStateTransitionNode_1 = AnimStateTransitionNode() {crossfadeDuration=0.05, priorityOrder=3}
   AnimStateTransitionNode_11 = AnimStateTransitionNode() {crossfadeDuration=0.4}
   AnimStateTransitionNode_2 = AnimStateTransitionNode() {bAutomaticRuleBasedOnSequencePlayerInState=true, crossfadeDuration=1.2}
   AnimStateTransitionNode_3 = AnimStateTransitionNode() {bAutomaticRuleBasedOnSequencePlayerInState=true, crossfadeDuration=1.2}
   AnimStateTransitionNode_4 = AnimStateTransitionNode() {crossfadeDuration=0.4}
   AnimStateTransitionNode_5 = AnimStateTransitionNode() {crossfadeDuration=0.15, priorityOrder=2}
   AnimStateTransitionNode_6 = AnimStateTransitionNode() {crossfadeDuration=0.15, priorityOrder=2}
   AnimStateTransitionNode_7 = AnimStateTransitionNode() {crossfadeDuration=0.25}
   AnimStateTransitionNode_8 = AnimStateTransitionNode() {crossfadeDuration=0.25}
  ]
  links [
   {
    dst=RightArmHit_A
    rule=AnimGraph.AnimGraphNode_StateMachine_6.RightArmAdditiveHits.AnimStateTransitionNode_11.Transition
    src=Neutral
    via=AnimStateTransitionNode_11
   }
   {
    dst=RightArmHit_B
    rule=AnimGraph.AnimGraphNode_StateMachine_6.RightArmAdditiveHits.AnimStateTransitionNode_4.Transition
    src=Neutral
    via=AnimStateTransitionNode_4
   }
   {
    dst=Neutral
    rule=AnimGraph.AnimGraphNode_StateMachine_6.RightArmAdditiveHits.AnimStateTransitionNode_2.Transition
    src=RightArmHit_A
    via=AnimStateTransitionNode_2
   }
   {
    dst=RightArmHit_B
    rule=AnimGraph.AnimGraphNode_StateMachine_6.RightArmAdditiveHits.AnimStateTransitionNode_0.Transition
    src=RightArmHit_A
    via=AnimStateTransitionNode_0
   }
   {
    dst=RightArmHit_B
    rule=AnimGraph.AnimGraphNode_StateMachine_6.RightArmAdditiveHits.AnimStateTransitionNode_6.Transition
    src=RightArmHit_A
    via=AnimStateTransitionNode_6
   }
   {
    dst=RightArmHit_B
    rule=AnimGraph.AnimGraphNode_StateMachine_6.RightArmAdditiveHits.AnimStateTransitionNode_8.Transition
    src=RightArmHit_A
    via=AnimStateTransitionNode_8
   }
   {
    dst=Neutral
    rule=AnimGraph.AnimGraphNode_StateMachine_6.RightArmAdditiveHits.AnimStateTransitionNode_3.Transition
    src=RightArmHit_B
    via=AnimStateTransitionNode_3
   }
   {
    dst=RightArmHit_A
    rule=AnimGraph.AnimGraphNode_StateMachine_6.RightArmAdditiveHits.AnimStateTransitionNode_1.Transition
    src=RightArmHit_B
    via=AnimStateTransitionNode_1
   }
   {
    dst=RightArmHit_A
    rule=AnimGraph.AnimGraphNode_StateMachine_6.RightArmAdditiveHits.AnimStateTransitionNode_5.Transition
    src=RightArmHit_B
    via=AnimStateTransitionNode_5
   }
   {
    dst=RightArmHit_A
    rule=AnimGraph.AnimGraphNode_StateMachine_6.RightArmAdditiveHits.AnimStateTransitionNode_7.Transition
    src=RightArmHit_B
    via=AnimStateTransitionNode_7
   }
  ]
 }
 RightArmHit_B {
  path=AnimGraph.AnimGraphNode_StateMachine_6.RightArmAdditiveHits.AnimStateNode_0.RightArmHit_B
  nodes [
   AnimGraphNode_StateMachine_1 = RightArmHitChooser_StartTime_DelayFloat() {is=title AnimGraphNode_StateMachine, bSkipFirstUpdateTransition=false}
   Output Animation Pose(Result=@AnimGraphNode_StateMachine_1)
  ]
 }
 RightArmHitChooser_StartTime_DelayFloat {
  entry=Identity
  path=AnimGraph.AnimGraphNode_StateMachine_6.RightArmAdditiveHits.AnimStateNode_0.RightArmHit_B.AnimGraphNode_StateMachine_1.RightArmHitChooser_StartTime_DelayFloat
  ids=id
  nodes [
   AnimStateNode_0 = Hit_Snare_R() {is=title AnimStateNode}
   AnimStateNode_1 = Hit_HiHat_R() {is=title AnimStateNode}
   AnimStateNode_10 = Identity()
   AnimStateNode_11 = Hit_FloorTom_R() {is=title AnimStateNode}
   AnimStateNode_12 = Hit_Crash2_Choke_R() {is=title AnimStateNode}
   AnimStateNode_13 = Hit_Crash1_Hard_R() {is=title AnimStateNode}
   AnimStateNode_14 = Hit_Crash2_Hard_R() {is=title AnimStateNode}
   AnimStateNode_2 = Hit_Crash1_Choke_R() {is=title AnimStateNode}
   AnimStateNode_3 = Hit_Crash1_Soft_R() {is=title AnimStateNode}
   AnimStateNode_4 = Hit_Snare_Soft_R() {is=title AnimStateNode}
   AnimStateNode_5 = Hit_Crash2_Soft_R() {is=title AnimStateNode}
   AnimStateNode_6 = Hit_Ride_R() {is=title AnimStateNode}
   AnimStateNode_7 = Hit_Tom1_R() {is=title AnimStateNode}
   AnimStateNode_8 = Hit_Tom2_R() {is=title AnimStateNode}
   AnimStateNode_9 = Hit_Percussion_R() {is=title AnimStateNode}
   AnimStateTransitionNode_2 = AnimStateTransitionNode()
   AnimStateTransitionNode_3 = AnimStateTransitionNode()
   AnimStateTransitionNode_30 = AnimStateTransitionNode() {crossfadeDuration=0}
   AnimStateTransitionNode_32 = AnimStateTransitionNode() {crossfadeDuration=0}
   AnimStateTransitionNode_34 = AnimStateTransitionNode() {crossfadeDuration=0}
   AnimStateTransitionNode_36 = AnimStateTransitionNode() {crossfadeDuration=0}
   AnimStateTransitionNode_38 = AnimStateTransitionNode() {crossfadeDuration=0}
   AnimStateTransitionNode_40 = AnimStateTransitionNode() {crossfadeDuration=0}
   AnimStateTransitionNode_42 = AnimStateTransitionNode() {crossfadeDuration=0}
   AnimStateTransitionNode_44 = AnimStateTransitionNode() {crossfadeDuration=0}
   AnimStateTransitionNode_46 = AnimStateTransitionNode() {crossfadeDuration=0}
   AnimStateTransitionNode_48 = AnimStateTransitionNode() {crossfadeDuration=0}
   AnimStateTransitionNode_50 = AnimStateTransitionNode() {crossfadeDuration=0}
   AnimStateTransitionNode_6 = AnimStateTransitionNode() {crossfadeDuration=0}
  ]
  links [
   {
    dst=Hit_Crash1_Choke_R
    rule=AnimGraph.AnimGraphNode_StateMachine_6.RightArmAdditiveHits.AnimStateNode_0.RightArmHit_B.AnimGraphNode_StateMachine_1.RightArmHitChooser_StartTime_DelayFloat.AnimStateTransitionNode_2.Transition
    src=Identity
    via=AnimStateTransitionNode_2
   }
   {
    dst=Hit_Crash1_Hard_R
    rule=AnimGraph.AnimGraphNode_StateMachine_6.RightArmAdditiveHits.AnimStateNode_0.RightArmHit_B.AnimGraphNode_StateMachine_1.RightArmHitChooser_StartTime_DelayFloat.AnimStateTransitionNode_34.Transition
    src=Identity
    via=AnimStateTransitionNode_34
   }
   {
    dst=Hit_Crash1_Soft_R
    rule=AnimGraph.AnimGraphNode_StateMachine_6.RightArmAdditiveHits.AnimStateNode_0.RightArmHit_B.AnimGraphNode_StateMachine_1.RightArmHitChooser_StartTime_DelayFloat.AnimStateTransitionNode_36.Transition
    src=Identity
    via=AnimStateTransitionNode_36
   }
   {
    dst=Hit_Crash2_Choke_R
    rule=AnimGraph.AnimGraphNode_StateMachine_6.RightArmAdditiveHits.AnimStateNode_0.RightArmHit_B.AnimGraphNode_StateMachine_1.RightArmHitChooser_StartTime_DelayFloat.AnimStateTransitionNode_3.Transition
    src=Identity
    via=AnimStateTransitionNode_3
   }
   {
    dst=Hit_Crash2_Hard_R
    rule=AnimGraph.AnimGraphNode_StateMachine_6.RightArmAdditiveHits.AnimStateNode_0.RightArmHit_B.AnimGraphNode_StateMachine_1.RightArmHitChooser_StartTime_DelayFloat.AnimStateTransitionNode_38.Transition
    src=Identity
    via=AnimStateTransitionNode_38
   }
   {
    dst=Hit_Crash2_Soft_R
    rule=AnimGraph.AnimGraphNode_StateMachine_6.RightArmAdditiveHits.AnimStateNode_0.RightArmHit_B.AnimGraphNode_StateMachine_1.RightArmHitChooser_StartTime_DelayFloat.AnimStateTransitionNode_40.Transition
    src=Identity
    via=AnimStateTransitionNode_40
   }
   {
    dst=Hit_FloorTom_R
    rule=AnimGraph.AnimGraphNode_StateMachine_6.RightArmAdditiveHits.AnimStateNode_0.RightArmHit_B.AnimGraphNode_StateMachine_1.RightArmHitChooser_StartTime_DelayFloat.AnimStateTransitionNode_48.Transition
    src=Identity
    via=AnimStateTransitionNode_48
   }
   {
    dst=Hit_HiHat_R
    rule=AnimGraph.AnimGraphNode_StateMachine_6.RightArmAdditiveHits.AnimStateNode_0.RightArmHit_B.AnimGraphNode_StateMachine_1.RightArmHitChooser_StartTime_DelayFloat.AnimStateTransitionNode_32.Transition
    src=Identity
    via=AnimStateTransitionNode_32
   }
   {
    dst=Hit_Percussion_R
    rule=AnimGraph.AnimGraphNode_StateMachine_6.RightArmAdditiveHits.AnimStateNode_0.RightArmHit_B.AnimGraphNode_StateMachine_1.RightArmHitChooser_StartTime_DelayFloat.AnimStateTransitionNode_50.Transition
    src=Identity
    via=AnimStateTransitionNode_50
   }
   {
    dst=Hit_Ride_R
    rule=AnimGraph.AnimGraphNode_StateMachine_6.RightArmAdditiveHits.AnimStateNode_0.RightArmHit_B.AnimGraphNode_StateMachine_1.RightArmHitChooser_StartTime_DelayFloat.AnimStateTransitionNode_42.Transition
    src=Identity
    via=AnimStateTransitionNode_42
   }
   {
    dst=Hit_Snare_R
    rule=AnimGraph.AnimGraphNode_StateMachine_6.RightArmAdditiveHits.AnimStateNode_0.RightArmHit_B.AnimGraphNode_StateMachine_1.RightArmHitChooser_StartTime_DelayFloat.AnimStateTransitionNode_30.Transition
    src=Identity
    via=AnimStateTransitionNode_30
   }
   {
    dst=Hit_Snare_Soft_R
    rule=AnimGraph.AnimGraphNode_StateMachine_6.RightArmAdditiveHits.AnimStateNode_0.RightArmHit_B.AnimGraphNode_StateMachine_1.RightArmHitChooser_StartTime_DelayFloat.AnimStateTransitionNode_6.Transition
    src=Identity
    via=AnimStateTransitionNode_6
   }
   {
    dst=Hit_Tom1_R
    rule=AnimGraph.AnimGraphNode_StateMachine_6.RightArmAdditiveHits.AnimStateNode_0.RightArmHit_B.AnimGraphNode_StateMachine_1.RightArmHitChooser_StartTime_DelayFloat.AnimStateTransitionNode_44.Transition
    src=Identity
    via=AnimStateTransitionNode_44
   }
   {
    dst=Hit_Tom2_R
    rule=AnimGraph.AnimGraphNode_StateMachine_6.RightArmAdditiveHits.AnimStateNode_0.RightArmHit_B.AnimGraphNode_StateMachine_1.RightArmHitChooser_StartTime_DelayFloat.AnimStateTransitionNode_46.Transition
    src=Identity
    via=AnimStateTransitionNode_46
   }
  ]
 }
 Hit_Snare_R {
  path=AnimGraph.AnimGraphNode_StateMachine_6.RightArmAdditiveHits.AnimStateNode_0.RightArmHit_B.AnimGraphNode_StateMachine_1.RightArmHitChooser_StartTime_DelayFloat.AnimStateNode_0.Hit_Snare_R
  nodes [
   n1 = Get RightArmMaxDelayShortenedBy() -> (RightArmMaxDelayShortenedBy:float)
   n2 = "Sequence Player (additive) 'Drum_Hit_Snare_Normal_Medium'"(StartPosition=Get RightArmMaxDelayShortenedBy(), bLoopAnimation=false) {sequence=&Drum_Hit_Snare_Normal_Medium, startPosition=0.05}
   Blend Poses by bool(
    BlendPose_0="Sequence Player (additive) 'Drum_Hit_Snare_Slow_Big'"(StartPosition=Get RightArmMaxDelayShortenedBy(), bLoopAnimation=false) {sequence=&Drum_Hit_Snare_Slow_Big, startPosition=0.05}
    BlendPose_1=Blend Poses by bool(
     BlendPose_0="Sequence Player (additive) 'Drum_Hit_Snare_Tap_Fast'"(StartPosition=@n1, bLoopAnimation=false) {sequence=&Drum_Hit_Snare_Tap_Fast, startPosition=0.05}
     BlendPose_1=@n2
     BlendTime_0=0
     BlendTime_1=0
     bActiveValue=Get RightArmIsShortDelay() -> (RightArmIsShortDelay:bool)
    ) {blendType=Linear}
    BlendTime_0=0
    BlendTime_1=0
    bActiveValue=Get RightArmIsLongDelay() -> (RightArmIsLongDelay:bool)
   ) {blendType=Linear}
   AnimGraphNode_StateMachine_0 = Hit_Snare_R_SM() {is=title AnimGraphNode_StateMachine, bAllowConduitEntryStates=true}
   Output Animation Pose(Result=@AnimGraphNode_StateMachine_0)
  ]
 }
 Hit_Snare_R_SM {
  entry=Conduit
  path=AnimGraph.AnimGraphNode_StateMachine_6.RightArmAdditiveHits.AnimStateNode_0.RightArmHit_B.AnimGraphNode_StateMachine_1.RightArmHitChooser_StartTime_DelayFloat.AnimStateNode_0.Hit_Snare_R.AnimGraphNode_StateMachine_0.Hit_Snare_R_SM
  ids=id
  nodes [
   AnimStateConduitNode_0 = Conduit()
   AnimStateNode_0 = Hit_Short()
   AnimStateNode_1 = Hit_Medium()
   AnimStateNode_2 = Hit_Long()
   AnimStateTransitionNode_4 = AnimStateTransitionNode() {crossfadeDuration=0, priorityOrder=2}
   AnimStateTransitionNode_5 = AnimStateTransitionNode() {crossfadeDuration=0, priorityOrder=2}
   AnimStateTransitionNode_6 = AnimStateTransitionNode() {crossfadeDuration=0, priorityOrder=2}
  ]
  links [
   {
    dst=Hit_Long
    rule=AnimGraph.AnimGraphNode_StateMachine_6.RightArmAdditiveHits.AnimStateNode_0.RightArmHit_B.AnimGraphNode_StateMachine_1.RightArmHitChooser_StartTime_DelayFloat.AnimStateNode_0.Hit_Snare_R.AnimGraphNode_StateMachine_0.Hit_Snare_R_SM.AnimStateTransitionNode_6.Transition
    src=Conduit
    via=AnimStateTransitionNode_6
   }
   {
    dst=Hit_Medium
    rule=AnimGraph.AnimGraphNode_StateMachine_6.RightArmAdditiveHits.AnimStateNode_0.RightArmHit_B.AnimGraphNode_StateMachine_1.RightArmHitChooser_StartTime_DelayFloat.AnimStateNode_0.Hit_Snare_R.AnimGraphNode_StateMachine_0.Hit_Snare_R_SM.AnimStateTransitionNode_5.Transition
    src=Conduit
    via=AnimStateTransitionNode_5
   }
   {
    dst=Hit_Short
    rule=AnimGraph.AnimGraphNode_StateMachine_6.RightArmAdditiveHits.AnimStateNode_0.RightArmHit_B.AnimGraphNode_StateMachine_1.RightArmHitChooser_StartTime_DelayFloat.AnimStateNode_0.Hit_Snare_R.AnimGraphNode_StateMachine_0.Hit_Snare_R_SM.AnimStateTransitionNode_4.Transition
    src=Conduit
    via=AnimStateTransitionNode_4
   }
  ]
 }
 Conduit {
  path=AnimGraph.AnimGraphNode_StateMachine_6.RightArmAdditiveHits.AnimStateNode_0.RightArmHit_B.AnimGraphNode_StateMachine_1.RightArmHitChooser_StartTime_DelayFloat.AnimStateNode_0.Hit_Snare_R.AnimGraphNode_StateMachine_0.Hit_Snare_R_SM.AnimStateConduitNode_0.Conduit
  nodes [Result(bCanEnterTransition=true)]
 }
 Hit_Short {
  path=AnimGraph.AnimGraphNode_StateMachine_6.RightArmAdditiveHits.AnimStateNode_0.RightArmHit_B.AnimGraphNode_StateMachine_1.RightArmHitChooser_StartTime_DelayFloat.AnimStateNode_0.Hit_Snare_R.AnimGraphNode_StateMachine_0.Hit_Snare_R_SM.AnimStateNode_0.Hit_Short
  nodes [Output Animation Pose(Result="Sequence Player (additive) 'Drum_Hit_Snare_Tap_Fast'"(StartPosition=Get RightArmMaxDelayShortenedBy(), bLoopAnimation=false) {sequence=&Drum_Hit_Snare_Tap_Fast, startPosition=0.05})]
 }
 Hit_Medium {
  path=AnimGraph.AnimGraphNode_StateMachine_6.RightArmAdditiveHits.AnimStateNode_0.RightArmHit_B.AnimGraphNode_StateMachine_1.RightArmHitChooser_StartTime_DelayFloat.AnimStateNode_0.Hit_Snare_R.AnimGraphNode_StateMachine_0.Hit_Snare_R_SM.AnimStateNode_1.Hit_Medium
  nodes [
   n1 = "Sequence Player (additive) 'Drum_Hit_Snare_Normal_Medium'"(StartPosition=Get RightArmMaxDelayShortenedBy(), bLoopAnimation=false) {sequence=&Drum_Hit_Snare_Normal_Medium, startPosition=0.05}
   Output Animation Pose(Result=@n1)
  ]
 }
 Hit_Long {
  path=AnimGraph.AnimGraphNode_StateMachine_6.RightArmAdditiveHits.AnimStateNode_0.RightArmHit_B.AnimGraphNode_StateMachine_1.RightArmHitChooser_StartTime_DelayFloat.AnimStateNode_0.Hit_Snare_R.AnimGraphNode_StateMachine_0.Hit_Snare_R_SM.AnimStateNode_2.Hit_Long
  nodes [Output Animation Pose(Result="Sequence Player (additive) 'Drum_Hit_Snare_Slow_Big'"(StartPosition=Get RightArmMaxDelayShortenedBy(), bLoopAnimation=false) {sequence=&Drum_Hit_Snare_Slow_Big, startPosition=0.05})]
 }
 Transition {
  path=AnimGraph.AnimGraphNode_StateMachine_6.RightArmAdditiveHits.AnimStateNode_0.RightArmHit_B.AnimGraphNode_StateMachine_1.RightArmHitChooser_StartTime_DelayFloat.AnimStateNode_0.Hit_Snare_R.AnimGraphNode_StateMachine_0.Hit_Snare_R_SM.AnimStateTransitionNode_4.Transition
  nodes [Result(bCanEnterTransition=Get RightArmIsShortDelay())]
 }
 Transition {
  path=AnimGraph.AnimGraphNode_StateMachine_6.RightArmAdditiveHits.AnimStateNode_0.RightArmHit_B.AnimGraphNode_StateMachine_1.RightArmHitChooser_StartTime_DelayFloat.AnimStateNode_0.Hit_Snare_R.AnimGraphNode_StateMachine_0.Hit_Snare_R_SM.AnimStateTransitionNode_5.Transition
  nodes [
   n1 = Get RightArmIsLongDelay()
   n2 = Get RightArmIsShortDelay()
   Result(bCanEnterTransition=AND Boolean!(A=NOT Boolean(A=@n2), B=NOT Boolean(A=@n1)) -> (ReturnValue))
  ]
 }
 Transition {
  path=AnimGraph.AnimGraphNode_StateMachine_6.RightArmAdditiveHits.AnimStateNode_0.RightArmHit_B.AnimGraphNode_StateMachine_1.RightArmHitChooser_StartTime_DelayFloat.AnimStateNode_0.Hit_Snare_R.AnimGraphNode_StateMachine_0.Hit_Snare_R_SM.AnimStateTransitionNode_6.Transition
  nodes [Result(bCanEnterTransition=Get RightArmIsLongDelay())]
 }
 Hit_HiHat_R {
  path=AnimGraph.AnimGraphNode_StateMachine_6.RightArmAdditiveHits.AnimStateNode_0.RightArmHit_B.AnimGraphNode_StateMachine_1.RightArmHitChooser_StartTime_DelayFloat.AnimStateNode_1.Hit_HiHat_R
  nodes [
   AnimGraphNode_StateMachine_0 = Hit_HiHat_R_SM() {is=title AnimGraphNode_StateMachine, bAllowConduitEntryStates=true}
   Output Animation Pose(Result=@AnimGraphNode_StateMachine_0)
  ]
 }
 Hit_HiHat_R_SM {
  entry=Conduit
  path=AnimGraph.AnimGraphNode_StateMachine_6.RightArmAdditiveHits.AnimStateNode_0.RightArmHit_B.AnimGraphNode_StateMachine_1.RightArmHitChooser_StartTime_DelayFloat.AnimStateNode_1.Hit_HiHat_R.AnimGraphNode_StateMachine_0.Hit_HiHat_R_SM
  ids=id
  nodes [
   AnimStateConduitNode_0 = Conduit()
   AnimStateNode_0 = Hit_Short()
   AnimStateNode_2 = Hit_Long()
   AnimStateNode_3 = Hit_Medium()
   AnimStateTransitionNode_4 = AnimStateTransitionNode() {crossfadeDuration=0}
   AnimStateTransitionNode_5 = AnimStateTransitionNode() {crossfadeDuration=0}
   AnimStateTransitionNode_6 = AnimStateTransitionNode() {crossfadeDuration=0}
  ]
  links [
   {
    dst=Hit_Long
    rule=AnimGraph.AnimGraphNode_StateMachine_6.RightArmAdditiveHits.AnimStateNode_0.RightArmHit_B.AnimGraphNode_StateMachine_1.RightArmHitChooser_StartTime_DelayFloat.AnimStateNode_1.Hit_HiHat_R.AnimGraphNode_StateMachine_0.Hit_HiHat_R_SM.AnimStateTransitionNode_6.Transition
    src=Conduit
    via=AnimStateTransitionNode_6
   }
   {
    dst=Hit_Medium
    rule=AnimGraph.AnimGraphNode_StateMachine_6.RightArmAdditiveHits.AnimStateNode_0.RightArmHit_B.AnimGraphNode_StateMachine_1.RightArmHitChooser_StartTime_DelayFloat.AnimStateNode_1.Hit_HiHat_R.AnimGraphNode_StateMachine_0.Hit_HiHat_R_SM.AnimStateTransitionNode_5.Transition
    src=Conduit
    via=AnimStateTransitionNode_5
   }
   {
    dst=Hit_Short
    rule=AnimGraph.AnimGraphNode_StateMachine_6.RightArmAdditiveHits.AnimStateNode_0.RightArmHit_B.AnimGraphNode_StateMachine_1.RightArmHitChooser_StartTime_DelayFloat.AnimStateNode_1.Hit_HiHat_R.AnimGraphNode_StateMachine_0.Hit_HiHat_R_SM.AnimStateTransitionNode_4.Transition
    src=Conduit
    via=AnimStateTransitionNode_4
   }
  ]
 }
 Conduit {
  path=AnimGraph.AnimGraphNode_StateMachine_6.RightArmAdditiveHits.AnimStateNode_0.RightArmHit_B.AnimGraphNode_StateMachine_1.RightArmHitChooser_StartTime_DelayFloat.AnimStateNode_1.Hit_HiHat_R.AnimGraphNode_StateMachine_0.Hit_HiHat_R_SM.AnimStateConduitNode_0.Conduit
  nodes [Result(bCanEnterTransition=true)]
 }
 Hit_Short {
  path=AnimGraph.AnimGraphNode_StateMachine_6.RightArmAdditiveHits.AnimStateNode_0.RightArmHit_B.AnimGraphNode_StateMachine_1.RightArmHitChooser_StartTime_DelayFloat.AnimStateNode_1.Hit_HiHat_R.AnimGraphNode_StateMachine_0.Hit_HiHat_R_SM.AnimStateNode_0.Hit_Short
  nodes [
   n1 = Get RightArmMaxDelayShortenedBy()
   n2 = "Sequence Player (additive) 'Drum_Hit_CrossHiHatSnare_Tap_Fast'"(StartPosition=@n1, bLoopAnimation=False) {bLoopAnimation=false, sequence=&Drum_Hit_CrossHiHatSnare_Tap_Fast}
   n3 = Get Hit_Snare_L()
   n4 = Get Hit_Snare_Soft_L()
   Output Animation Pose(Result=Blend Poses by bool(
     BlendPose_0=@n2
     BlendPose_1="Sequence Player (additive) 'Drum_Hit_HiHat_Tap_Fast'"(StartPosition=@n1, bLoopAnimation=false) {sequence=&Drum_Hit_HiHat_Tap_Fast, startPosition=0.05}
     bActiveValue=OR Boolean(A=@n3, B=@n4)
    ) {blendType=Linear})
  ]
 }
 Hit_Long {
  path=AnimGraph.AnimGraphNode_StateMachine_6.RightArmAdditiveHits.AnimStateNode_0.RightArmHit_B.AnimGraphNode_StateMachine_1.RightArmHitChooser_StartTime_DelayFloat.AnimStateNode_1.Hit_HiHat_R.AnimGraphNode_StateMachine_0.Hit_HiHat_R_SM.AnimStateNode_2.Hit_Long
  nodes [Output Animation Pose(Result="Sequence Player (additive) 'Drum_Hit_HiHat_Slow_Big'"(StartPosition=Get RightArmMaxDelayShortenedBy(), bLoopAnimation=false) {sequence=&Drum_Hit_HiHat_Slow_Big, startPosition=0.05})]
 }
 Hit_Medium {
  path=AnimGraph.AnimGraphNode_StateMachine_6.RightArmAdditiveHits.AnimStateNode_0.RightArmHit_B.AnimGraphNode_StateMachine_1.RightArmHitChooser_StartTime_DelayFloat.AnimStateNode_1.Hit_HiHat_R.AnimGraphNode_StateMachine_0.Hit_HiHat_R_SM.AnimStateNode_3.Hit_Medium
  nodes [
   n1 = Get RightArmMaxDelayShortenedBy()
   n2 = "Sequence Player (additive) 'Drum_Hit_CrossHiHatSnare_Medium'"(StartPosition=@n1, bLoopAnimation=False) {bLoopAnimation=false, sequence=&Drum_Hit_CrossHiHatSnare_Medium}
   n3 = "Sequence Player (additive) 'Drum_Hit_HiHat_Normal_Medium'"(StartPosition=@n1, bLoopAnimation=false) {sequence=&Drum_Hit_HiHat_Normal_Medium, startPosition=0.05}
   n4 = Get Hit_Snare_L()
   n5 = Get Hit_Snare_Soft_L()
   Output Animation Pose(Result=Blend Poses by bool(BlendPose_0=@n2, BlendPose_1=@n3, bActiveValue=OR Boolean(A=@n4, B=@n5)) {blendType=Linear})
  ]
 }
 Transition {
  path=AnimGraph.AnimGraphNode_StateMachine_6.RightArmAdditiveHits.AnimStateNode_0.RightArmHit_B.AnimGraphNode_StateMachine_1.RightArmHitChooser_StartTime_DelayFloat.AnimStateNode_1.Hit_HiHat_R.AnimGraphNode_StateMachine_0.Hit_HiHat_R_SM.AnimStateTransitionNode_4.Transition
  nodes [Result(bCanEnterTransition=Get RightArmIsShortDelay())]
 }
 Transition {
  path=AnimGraph.AnimGraphNode_StateMachine_6.RightArmAdditiveHits.AnimStateNode_0.RightArmHit_B.AnimGraphNode_StateMachine_1.RightArmHitChooser_StartTime_DelayFloat.AnimStateNode_1.Hit_HiHat_R.AnimGraphNode_StateMachine_0.Hit_HiHat_R_SM.AnimStateTransitionNode_5.Transition
  nodes [
   n1 = Get RightArmIsLongDelay()
   n2 = Get RightArmIsShortDelay()
   Result(bCanEnterTransition=AND Boolean(A=NOT Boolean(A=@n2), B=NOT Boolean(A=@n1)))
  ]
 }
 Transition {
  path=AnimGraph.AnimGraphNode_StateMachine_6.RightArmAdditiveHits.AnimStateNode_0.RightArmHit_B.AnimGraphNode_StateMachine_1.RightArmHitChooser_StartTime_DelayFloat.AnimStateNode_1.Hit_HiHat_R.AnimGraphNode_StateMachine_0.Hit_HiHat_R_SM.AnimStateTransitionNode_6.Transition
  nodes [Result(bCanEnterTransition=Get RightArmIsLongDelay())]
 }
 Identity {
  path=AnimGraph.AnimGraphNode_StateMachine_6.RightArmAdditiveHits.AnimStateNode_0.RightArmHit_B.AnimGraphNode_StateMachine_1.RightArmHitChooser_StartTime_DelayFloat.AnimStateNode_10.Identity
  nodes [
   "Sequence Player (additive) 'Drum_SpineTwist_Center_Pose'"() {bLoopAnimation=false, sequence=&Drum_SpineTwist_Center_Pose}
   Output Animation Pose(Result=Additive Identity Pose())
  ]
 }
 Hit_FloorTom_R {
  path=AnimGraph.AnimGraphNode_StateMachine_6.RightArmAdditiveHits.AnimStateNode_0.RightArmHit_B.AnimGraphNode_StateMachine_1.RightArmHitChooser_StartTime_DelayFloat.AnimStateNode_11.Hit_FloorTom_R
  nodes [
   n1 = Get RightArmMaxDelayShortenedBy()
   n2 = "Sequence Player (additive) 'Drum_Hit_FloorTom_Normal_Medium'"(StartPosition=Get RightArmMaxDelayShortenedBy(), bLoopAnimation=false) {sequence=&Drum_Hit_FloorTom_Normal_Medium, startPosition=0.05}
   Blend Poses by bool(
    BlendPose_0="Sequence Player (additive) 'Drum_Hit_FloorTom_Slow_Big'"(StartPosition=Get RightArmMaxDelayShortenedBy(), bLoopAnimation=false) {sequence=&Drum_Hit_FloorTom_Slow_Big, startPosition=0.05}
    BlendPose_1=Blend Poses by bool(
     BlendPose_0="Sequence Player (additive) 'Drum_Hit_FloorTom_Tap_Fast'"(StartPosition=@n1, bLoopAnimation=false) {sequence=&Drum_Hit_FloorTom_Tap_Fast, startPosition=0.05}
     BlendPose_1=@n2
     BlendTime_0=0
     BlendTime_1=0
     bActiveValue=Get RightArmIsShortDelay()
    ) {blendType=Linear}
    BlendTime_0=0
    BlendTime_1=0
    bActiveValue=Get RightArmIsLongDelay()
   ) {blendType=Linear}
   AnimGraphNode_StateMachine_0 = Hit_FloorTom_R_SM() {bAllowConduitEntryStates=true}
   Output Animation Pose(Result=@AnimGraphNode_StateMachine_0)
  ]
 }
 Hit_FloorTom_R_SM {
  entry=Conduit
  path=AnimGraph.AnimGraphNode_StateMachine_6.RightArmAdditiveHits.AnimStateNode_0.RightArmHit_B.AnimGraphNode_StateMachine_1.RightArmHitChooser_StartTime_DelayFloat.AnimStateNode_11.Hit_FloorTom_R.AnimGraphNode_StateMachine_0.Hit_FloorTom_R_SM
  ids=id
  nodes [
   AnimStateConduitNode_0 = Conduit()
   AnimStateNode_0 = Hit_Short()
   AnimStateNode_1 = Hit_Medium()
   AnimStateNode_2 = Hit_Long()
   AnimStateTransitionNode_4 = AnimStateTransitionNode() {crossfadeDuration=0}
   AnimStateTransitionNode_5 = AnimStateTransitionNode() {crossfadeDuration=0}
   AnimStateTransitionNode_6 = AnimStateTransitionNode() {crossfadeDuration=0}
  ]
  links [
   {
    dst=Hit_Long
    rule=AnimGraph.AnimGraphNode_StateMachine_6.RightArmAdditiveHits.AnimStateNode_0.RightArmHit_B.AnimGraphNode_StateMachine_1.RightArmHitChooser_StartTime_DelayFloat.AnimStateNode_11.Hit_FloorTom_R.AnimGraphNode_StateMachine_0.Hit_FloorTom_R_SM.AnimStateTransitionNode_6.Transition
    src=Conduit
    via=AnimStateTransitionNode_6
   }
   {
    dst=Hit_Medium
    rule=AnimGraph.AnimGraphNode_StateMachine_6.RightArmAdditiveHits.AnimStateNode_0.RightArmHit_B.AnimGraphNode_StateMachine_1.RightArmHitChooser_StartTime_DelayFloat.AnimStateNode_11.Hit_FloorTom_R.AnimGraphNode_StateMachine_0.Hit_FloorTom_R_SM.AnimStateTransitionNode_5.Transition
    src=Conduit
    via=AnimStateTransitionNode_5
   }
   {
    dst=Hit_Short
    rule=AnimGraph.AnimGraphNode_StateMachine_6.RightArmAdditiveHits.AnimStateNode_0.RightArmHit_B.AnimGraphNode_StateMachine_1.RightArmHitChooser_StartTime_DelayFloat.AnimStateNode_11.Hit_FloorTom_R.AnimGraphNode_StateMachine_0.Hit_FloorTom_R_SM.AnimStateTransitionNode_4.Transition
    src=Conduit
    via=AnimStateTransitionNode_4
   }
  ]
 }
 Conduit {
  path=AnimGraph.AnimGraphNode_StateMachine_6.RightArmAdditiveHits.AnimStateNode_0.RightArmHit_B.AnimGraphNode_StateMachine_1.RightArmHitChooser_StartTime_DelayFloat.AnimStateNode_11.Hit_FloorTom_R.AnimGraphNode_StateMachine_0.Hit_FloorTom_R_SM.AnimStateConduitNode_0.Conduit
  nodes [Result(bCanEnterTransition=true)]
 }
 Hit_Short {
  path=AnimGraph.AnimGraphNode_StateMachine_6.RightArmAdditiveHits.AnimStateNode_0.RightArmHit_B.AnimGraphNode_StateMachine_1.RightArmHitChooser_StartTime_DelayFloat.AnimStateNode_11.Hit_FloorTom_R.AnimGraphNode_StateMachine_0.Hit_FloorTom_R_SM.AnimStateNode_0.Hit_Short
  nodes [Output Animation Pose(Result="Sequence Player (additive) 'Drum_Hit_FloorTom_Tap_Fast'"(StartPosition=Get RightArmMaxDelayShortenedBy(), bLoopAnimation=false) {sequence=&Drum_Hit_FloorTom_Tap_Fast, startPosition=0.05})]
 }
 Hit_Medium {
  path=AnimGraph.AnimGraphNode_StateMachine_6.RightArmAdditiveHits.AnimStateNode_0.RightArmHit_B.AnimGraphNode_StateMachine_1.RightArmHitChooser_StartTime_DelayFloat.AnimStateNode_11.Hit_FloorTom_R.AnimGraphNode_StateMachine_0.Hit_FloorTom_R_SM.AnimStateNode_1.Hit_Medium
  nodes [
   n1 = "Sequence Player (additive) 'Drum_Hit_FloorTom_Normal_Medium'"(StartPosition=Get RightArmMaxDelayShortenedBy(), bLoopAnimation=false) {sequence=&Drum_Hit_FloorTom_Normal_Medium, startPosition=0.05}
   Output Animation Pose(Result=@n1)
  ]
 }
 Hit_Long {
  path=AnimGraph.AnimGraphNode_StateMachine_6.RightArmAdditiveHits.AnimStateNode_0.RightArmHit_B.AnimGraphNode_StateMachine_1.RightArmHitChooser_StartTime_DelayFloat.AnimStateNode_11.Hit_FloorTom_R.AnimGraphNode_StateMachine_0.Hit_FloorTom_R_SM.AnimStateNode_2.Hit_Long
  nodes [Output Animation Pose(Result="Sequence Player (additive) 'Drum_Hit_FloorTom_Slow_Big'"(StartPosition=Get RightArmMaxDelayShortenedBy(), bLoopAnimation=false) {sequence=&Drum_Hit_FloorTom_Slow_Big, startPosition=0.05})]
 }
 Transition {
  path=AnimGraph.AnimGraphNode_StateMachine_6.RightArmAdditiveHits.AnimStateNode_0.RightArmHit_B.AnimGraphNode_StateMachine_1.RightArmHitChooser_StartTime_DelayFloat.AnimStateNode_11.Hit_FloorTom_R.AnimGraphNode_StateMachine_0.Hit_FloorTom_R_SM.AnimStateTransitionNode_4.Transition
  nodes [Result(bCanEnterTransition=Get RightArmIsShortDelay())]
 }
 Transition {
  path=AnimGraph.AnimGraphNode_StateMachine_6.RightArmAdditiveHits.AnimStateNode_0.RightArmHit_B.AnimGraphNode_StateMachine_1.RightArmHitChooser_StartTime_DelayFloat.AnimStateNode_11.Hit_FloorTom_R.AnimGraphNode_StateMachine_0.Hit_FloorTom_R_SM.AnimStateTransitionNode_5.Transition
  nodes [
   n1 = Get RightArmIsLongDelay()
   n2 = Get RightArmIsShortDelay()
   Result(bCanEnterTransition=AND Boolean(A=NOT Boolean(A=@n2), B=NOT Boolean(A=@n1)))
  ]
 }
 Transition {
  path=AnimGraph.AnimGraphNode_StateMachine_6.RightArmAdditiveHits.AnimStateNode_0.RightArmHit_B.AnimGraphNode_StateMachine_1.RightArmHitChooser_StartTime_DelayFloat.AnimStateNode_11.Hit_FloorTom_R.AnimGraphNode_StateMachine_0.Hit_FloorTom_R_SM.AnimStateTransitionNode_6.Transition
  nodes [Result(bCanEnterTransition=Get RightArmIsLongDelay())]
 }
 Hit_Crash2_Choke_R {
  path=AnimGraph.AnimGraphNode_StateMachine_6.RightArmAdditiveHits.AnimStateNode_0.RightArmHit_B.AnimGraphNode_StateMachine_1.RightArmHitChooser_StartTime_DelayFloat.AnimStateNode_12.Hit_Crash2_Choke_R
  nodes [
   n1 = Get RightArmMaxDelayShortenedBy()
   n2 = "Sequence Player (additive) 'Drum_Hit_Crash2_Normal_Medium'"(StartPosition=Get RightArmMaxDelayShortenedBy(), bLoopAnimation=false) {sequence=&Drum_Hit_Crash2_Normal_Medium, startPosition=0.05}
   Blend Poses by bool(
    BlendPose_0="Sequence Player (additive) 'Drum_Hit_Crash2_Slow_Big'"(StartPosition=Get RightArmMaxDelayShortenedBy(), bLoopAnimation=false) {sequence=&Drum_Hit_Crash2_Slow_Big, startPosition=0.05}
    BlendPose_1=Blend Poses by bool(
     BlendPose_0="Sequence Player (additive) 'Drum_Hit_Crash2_Tap_Fast'"(StartPosition=@n1, bLoopAnimation=false) {sequence=&Drum_Hit_Crash2_Tap_Fast, startPosition=0.05}
     BlendPose_1=@n2
     BlendTime_0=0
     BlendTime_1=0
     bActiveValue=Get RightArmIsShortDelay()
    ) {blendType=Linear}
    BlendTime_0=0
    BlendTime_1=0
    bActiveValue=Get RightArmIsLongDelay()
   ) {blendType=Linear}
   AnimGraphNode_StateMachine_0 = Hit_Crash2_Hard_R_SM() {is=title AnimGraphNode_StateMachine, bAllowConduitEntryStates=true}
   Output Animation Pose(Result=@AnimGraphNode_StateMachine_0)
  ]
 }
 Hit_Crash2_Hard_R_SM {
  entry=Conduit
  path=AnimGraph.AnimGraphNode_StateMachine_6.RightArmAdditiveHits.AnimStateNode_0.RightArmHit_B.AnimGraphNode_StateMachine_1.RightArmHitChooser_StartTime_DelayFloat.AnimStateNode_12.Hit_Crash2_Choke_R.AnimGraphNode_StateMachine_0.Hit_Crash2_Hard_R_SM
  ids=id
  nodes [
   AnimStateConduitNode_0 = Conduit()
   AnimStateNode_0 = Hit_Short()
   AnimStateNode_1 = Hit_Medium()
   AnimStateNode_2 = Hit_Long()
   AnimStateTransitionNode_4 = AnimStateTransitionNode() {crossfadeDuration=0}
   AnimStateTransitionNode_5 = AnimStateTransitionNode() {crossfadeDuration=0}
   AnimStateTransitionNode_6 = AnimStateTransitionNode() {crossfadeDuration=0}
  ]
  links [
   {
    dst=Hit_Long
    rule=AnimGraph.AnimGraphNode_StateMachine_6.RightArmAdditiveHits.AnimStateNode_0.RightArmHit_B.AnimGraphNode_StateMachine_1.RightArmHitChooser_StartTime_DelayFloat.AnimStateNode_12.Hit_Crash2_Choke_R.AnimGraphNode_StateMachine_0.Hit_Crash2_Hard_R_SM.AnimStateTransitionNode_6.Transition
    src=Conduit
    via=AnimStateTransitionNode_6
   }
   {
    dst=Hit_Medium
    rule=AnimGraph.AnimGraphNode_StateMachine_6.RightArmAdditiveHits.AnimStateNode_0.RightArmHit_B.AnimGraphNode_StateMachine_1.RightArmHitChooser_StartTime_DelayFloat.AnimStateNode_12.Hit_Crash2_Choke_R.AnimGraphNode_StateMachine_0.Hit_Crash2_Hard_R_SM.AnimStateTransitionNode_5.Transition
    src=Conduit
    via=AnimStateTransitionNode_5
   }
   {
    dst=Hit_Short
    rule=AnimGraph.AnimGraphNode_StateMachine_6.RightArmAdditiveHits.AnimStateNode_0.RightArmHit_B.AnimGraphNode_StateMachine_1.RightArmHitChooser_StartTime_DelayFloat.AnimStateNode_12.Hit_Crash2_Choke_R.AnimGraphNode_StateMachine_0.Hit_Crash2_Hard_R_SM.AnimStateTransitionNode_4.Transition
    src=Conduit
    via=AnimStateTransitionNode_4
   }
  ]
 }
 Conduit {
  path=AnimGraph.AnimGraphNode_StateMachine_6.RightArmAdditiveHits.AnimStateNode_0.RightArmHit_B.AnimGraphNode_StateMachine_1.RightArmHitChooser_StartTime_DelayFloat.AnimStateNode_12.Hit_Crash2_Choke_R.AnimGraphNode_StateMachine_0.Hit_Crash2_Hard_R_SM.AnimStateConduitNode_0.Conduit
  nodes [Result(bCanEnterTransition=true)]
 }
 Hit_Short {
  path=AnimGraph.AnimGraphNode_StateMachine_6.RightArmAdditiveHits.AnimStateNode_0.RightArmHit_B.AnimGraphNode_StateMachine_1.RightArmHitChooser_StartTime_DelayFloat.AnimStateNode_12.Hit_Crash2_Choke_R.AnimGraphNode_StateMachine_0.Hit_Crash2_Hard_R_SM.AnimStateNode_0.Hit_Short
  nodes [Output Animation Pose(Result="Sequence Player (additive) 'Drum_Hit_Crash2_Tap_Fast'"(StartPosition=Get RightArmMaxDelayShortenedBy(), bLoopAnimation=false) {sequence=&Drum_Hit_Crash2_Tap_Fast, startPosition=0.05})]
 }
 Hit_Medium {
  path=AnimGraph.AnimGraphNode_StateMachine_6.RightArmAdditiveHits.AnimStateNode_0.RightArmHit_B.AnimGraphNode_StateMachine_1.RightArmHitChooser_StartTime_DelayFloat.AnimStateNode_12.Hit_Crash2_Choke_R.AnimGraphNode_StateMachine_0.Hit_Crash2_Hard_R_SM.AnimStateNode_1.Hit_Medium
  nodes [
   n1 = "Sequence Player (additive) 'Drum_Hit_Crash2_Normal_Medium'"(StartPosition=Get RightArmMaxDelayShortenedBy(), bLoopAnimation=false) {sequence=&Drum_Hit_Crash2_Normal_Medium, startPosition=0.05}
   Output Animation Pose(Result=@n1)
  ]
 }
 Hit_Long {
  path=AnimGraph.AnimGraphNode_StateMachine_6.RightArmAdditiveHits.AnimStateNode_0.RightArmHit_B.AnimGraphNode_StateMachine_1.RightArmHitChooser_StartTime_DelayFloat.AnimStateNode_12.Hit_Crash2_Choke_R.AnimGraphNode_StateMachine_0.Hit_Crash2_Hard_R_SM.AnimStateNode_2.Hit_Long
  nodes [Output Animation Pose(Result="Sequence Player (additive) 'Drum_Hit_Crash2_Slow_Big'"(StartPosition=Get RightArmMaxDelayShortenedBy(), bLoopAnimation=false) {sequence=&Drum_Hit_Crash2_Slow_Big, startPosition=0.05})]
 }
 Transition {
  path=AnimGraph.AnimGraphNode_StateMachine_6.RightArmAdditiveHits.AnimStateNode_0.RightArmHit_B.AnimGraphNode_StateMachine_1.RightArmHitChooser_StartTime_DelayFloat.AnimStateNode_12.Hit_Crash2_Choke_R.AnimGraphNode_StateMachine_0.Hit_Crash2_Hard_R_SM.AnimStateTransitionNode_4.Transition
  nodes [Result(bCanEnterTransition=Get RightArmIsShortDelay())]
 }
 Transition {
  path=AnimGraph.AnimGraphNode_StateMachine_6.RightArmAdditiveHits.AnimStateNode_0.RightArmHit_B.AnimGraphNode_StateMachine_1.RightArmHitChooser_StartTime_DelayFloat.AnimStateNode_12.Hit_Crash2_Choke_R.AnimGraphNode_StateMachine_0.Hit_Crash2_Hard_R_SM.AnimStateTransitionNode_5.Transition
  nodes [
   n1 = Get RightArmIsLongDelay()
   n2 = Get RightArmIsShortDelay()
   Result(bCanEnterTransition=AND Boolean(A=NOT Boolean(A=@n2), B=NOT Boolean(A=@n1)))
  ]
 }
 Transition {
  path=AnimGraph.AnimGraphNode_StateMachine_6.RightArmAdditiveHits.AnimStateNode_0.RightArmHit_B.AnimGraphNode_StateMachine_1.RightArmHitChooser_StartTime_DelayFloat.AnimStateNode_12.Hit_Crash2_Choke_R.AnimGraphNode_StateMachine_0.Hit_Crash2_Hard_R_SM.AnimStateTransitionNode_6.Transition
  nodes [Result(bCanEnterTransition=Get RightArmIsLongDelay())]
 }
 Hit_Crash1_Hard_R {
  path=AnimGraph.AnimGraphNode_StateMachine_6.RightArmAdditiveHits.AnimStateNode_0.RightArmHit_B.AnimGraphNode_StateMachine_1.RightArmHitChooser_StartTime_DelayFloat.AnimStateNode_13.Hit_Crash1_Hard_R
  nodes [
   n1 = Get RightArmMaxDelayShortenedBy()
   n2 = "Sequence Player (additive) 'Drum_Hit_Crash1_Normal_Medium'"(StartPosition=Get RightArmMaxDelayShortenedBy(), bLoopAnimation=false) {sequence=&Drum_Hit_Crash1_Normal_Medium, startPosition=0.05}
   Blend Poses by bool(
    BlendPose_0="Sequence Player (additive) 'Drum_Hit_Crash1_Slow_Big'"(StartPosition=Get RightArmMaxDelayShortenedBy(), bLoopAnimation=false) {sequence=&Drum_Hit_Crash1_Slow_Big, startPosition=0.05}
    BlendPose_1=Blend Poses by bool(
     BlendPose_0="Sequence Player (additive) 'Drum_Hit_Crash1_Tap_Fast'"(StartPosition=@n1, bLoopAnimation=false) {sequence=&Drum_Hit_Crash1_Tap_Fast, startPosition=0.05}
     BlendPose_1=@n2
     BlendTime_0=0
     BlendTime_1=0
     bActiveValue=Get RightArmIsShortDelay()
    ) {blendType=Linear}
    BlendTime_0=0
    BlendTime_1=0
    bActiveValue=Get RightArmIsLongDelay()
   ) {blendType=Linear}
   AnimGraphNode_StateMachine_0 = Hit_Crash1_Hard_R_SM() {is=title AnimGraphNode_StateMachine, bAllowConduitEntryStates=true}
   Output Animation Pose(Result=@AnimGraphNode_StateMachine_0)
  ]
 }
 Hit_Crash1_Hard_R_SM {
  entry=Conduit
  path=AnimGraph.AnimGraphNode_StateMachine_6.RightArmAdditiveHits.AnimStateNode_0.RightArmHit_B.AnimGraphNode_StateMachine_1.RightArmHitChooser_StartTime_DelayFloat.AnimStateNode_13.Hit_Crash1_Hard_R.AnimGraphNode_StateMachine_0.Hit_Crash1_Hard_R_SM
  ids=id
  nodes [
   AnimStateConduitNode_0 = Conduit()
   AnimStateNode_0 = Hit_Short()
   AnimStateNode_1 = Hit_Medium()
   AnimStateNode_2 = Hit_Long()
   AnimStateTransitionNode_4 = AnimStateTransitionNode() {crossfadeDuration=0}
   AnimStateTransitionNode_5 = AnimStateTransitionNode() {crossfadeDuration=0}
   AnimStateTransitionNode_6 = AnimStateTransitionNode() {crossfadeDuration=0}
  ]
  links [
   {
    dst=Hit_Long
    rule=AnimGraph.AnimGraphNode_StateMachine_6.RightArmAdditiveHits.AnimStateNode_0.RightArmHit_B.AnimGraphNode_StateMachine_1.RightArmHitChooser_StartTime_DelayFloat.AnimStateNode_13.Hit_Crash1_Hard_R.AnimGraphNode_StateMachine_0.Hit_Crash1_Hard_R_SM.AnimStateTransitionNode_6.Transition
    src=Conduit
    via=AnimStateTransitionNode_6
   }
   {
    dst=Hit_Medium
    rule=AnimGraph.AnimGraphNode_StateMachine_6.RightArmAdditiveHits.AnimStateNode_0.RightArmHit_B.AnimGraphNode_StateMachine_1.RightArmHitChooser_StartTime_DelayFloat.AnimStateNode_13.Hit_Crash1_Hard_R.AnimGraphNode_StateMachine_0.Hit_Crash1_Hard_R_SM.AnimStateTransitionNode_5.Transition
    src=Conduit
    via=AnimStateTransitionNode_5
   }
   {
    dst=Hit_Short
    rule=AnimGraph.AnimGraphNode_StateMachine_6.RightArmAdditiveHits.AnimStateNode_0.RightArmHit_B.AnimGraphNode_StateMachine_1.RightArmHitChooser_StartTime_DelayFloat.AnimStateNode_13.Hit_Crash1_Hard_R.AnimGraphNode_StateMachine_0.Hit_Crash1_Hard_R_SM.AnimStateTransitionNode_4.Transition
    src=Conduit
    via=AnimStateTransitionNode_4
   }
  ]
 }
 Conduit {
  path=AnimGraph.AnimGraphNode_StateMachine_6.RightArmAdditiveHits.AnimStateNode_0.RightArmHit_B.AnimGraphNode_StateMachine_1.RightArmHitChooser_StartTime_DelayFloat.AnimStateNode_13.Hit_Crash1_Hard_R.AnimGraphNode_StateMachine_0.Hit_Crash1_Hard_R_SM.AnimStateConduitNode_0.Conduit
  nodes [Result(bCanEnterTransition=true)]
 }
 Hit_Short {
  path=AnimGraph.AnimGraphNode_StateMachine_6.RightArmAdditiveHits.AnimStateNode_0.RightArmHit_B.AnimGraphNode_StateMachine_1.RightArmHitChooser_StartTime_DelayFloat.AnimStateNode_13.Hit_Crash1_Hard_R.AnimGraphNode_StateMachine_0.Hit_Crash1_Hard_R_SM.AnimStateNode_0.Hit_Short
  nodes [Output Animation Pose(Result="Sequence Player (additive) 'Drum_Hit_Crash1_Tap_Fast'"(StartPosition=Get RightArmMaxDelayShortenedBy(), bLoopAnimation=false) {sequence=&Drum_Hit_Crash1_Tap_Fast, startPosition=0.05})]
 }
 Hit_Medium {
  path=AnimGraph.AnimGraphNode_StateMachine_6.RightArmAdditiveHits.AnimStateNode_0.RightArmHit_B.AnimGraphNode_StateMachine_1.RightArmHitChooser_StartTime_DelayFloat.AnimStateNode_13.Hit_Crash1_Hard_R.AnimGraphNode_StateMachine_0.Hit_Crash1_Hard_R_SM.AnimStateNode_1.Hit_Medium
  nodes [
   n1 = "Sequence Player (additive) 'Drum_Hit_Crash1_Normal_Medium'"(StartPosition=Get RightArmMaxDelayShortenedBy(), bLoopAnimation=false) {sequence=&Drum_Hit_Crash1_Normal_Medium, startPosition=0.05}
   Output Animation Pose(Result=@n1)
  ]
 }
 Hit_Long {
  path=AnimGraph.AnimGraphNode_StateMachine_6.RightArmAdditiveHits.AnimStateNode_0.RightArmHit_B.AnimGraphNode_StateMachine_1.RightArmHitChooser_StartTime_DelayFloat.AnimStateNode_13.Hit_Crash1_Hard_R.AnimGraphNode_StateMachine_0.Hit_Crash1_Hard_R_SM.AnimStateNode_2.Hit_Long
  nodes [Output Animation Pose(Result="Sequence Player (additive) 'Drum_Hit_Crash1_Slow_Big'"(StartPosition=Get RightArmMaxDelayShortenedBy(), bLoopAnimation=false) {sequence=&Drum_Hit_Crash1_Slow_Big, startPosition=0.05})]
 }
 Transition {
  path=AnimGraph.AnimGraphNode_StateMachine_6.RightArmAdditiveHits.AnimStateNode_0.RightArmHit_B.AnimGraphNode_StateMachine_1.RightArmHitChooser_StartTime_DelayFloat.AnimStateNode_13.Hit_Crash1_Hard_R.AnimGraphNode_StateMachine_0.Hit_Crash1_Hard_R_SM.AnimStateTransitionNode_4.Transition
  nodes [Result(bCanEnterTransition=Get RightArmIsShortDelay())]
 }
 Transition {
  path=AnimGraph.AnimGraphNode_StateMachine_6.RightArmAdditiveHits.AnimStateNode_0.RightArmHit_B.AnimGraphNode_StateMachine_1.RightArmHitChooser_StartTime_DelayFloat.AnimStateNode_13.Hit_Crash1_Hard_R.AnimGraphNode_StateMachine_0.Hit_Crash1_Hard_R_SM.AnimStateTransitionNode_5.Transition
  nodes [
   n1 = Get RightArmIsLongDelay()
   n2 = Get RightArmIsShortDelay()
   Result(bCanEnterTransition=AND Boolean(A=NOT Boolean(A=@n2), B=NOT Boolean(A=@n1)))
  ]
 }
 Transition {
  path=AnimGraph.AnimGraphNode_StateMachine_6.RightArmAdditiveHits.AnimStateNode_0.RightArmHit_B.AnimGraphNode_StateMachine_1.RightArmHitChooser_StartTime_DelayFloat.AnimStateNode_13.Hit_Crash1_Hard_R.AnimGraphNode_StateMachine_0.Hit_Crash1_Hard_R_SM.AnimStateTransitionNode_6.Transition
  nodes [Result(bCanEnterTransition=Get RightArmIsLongDelay())]
 }
 Hit_Crash2_Hard_R {
  path=AnimGraph.AnimGraphNode_StateMachine_6.RightArmAdditiveHits.AnimStateNode_0.RightArmHit_B.AnimGraphNode_StateMachine_1.RightArmHitChooser_StartTime_DelayFloat.AnimStateNode_14.Hit_Crash2_Hard_R
  nodes [
   n1 = Get RightArmMaxDelayShortenedBy()
   n2 = "Sequence Player (additive) 'Drum_Hit_Crash2_Normal_Medium'"(StartPosition=Get RightArmMaxDelayShortenedBy(), bLoopAnimation=false) {sequence=&Drum_Hit_Crash2_Normal_Medium, startPosition=0.05}
   Blend Poses by bool(
    BlendPose_0="Sequence Player (additive) 'Drum_Hit_Crash2_Slow_Big'"(StartPosition=Get RightArmMaxDelayShortenedBy(), bLoopAnimation=false) {sequence=&Drum_Hit_Crash2_Slow_Big, startPosition=0.05}
    BlendPose_1=Blend Poses by bool(
     BlendPose_0="Sequence Player (additive) 'Drum_Hit_Crash2_Tap_Fast'"(StartPosition=@n1, bLoopAnimation=false) {sequence=&Drum_Hit_Crash2_Tap_Fast, startPosition=0.05}
     BlendPose_1=@n2
     BlendTime_0=0
     BlendTime_1=0
     bActiveValue=Get RightArmIsShortDelay()
    ) {blendType=Linear}
    BlendTime_0=0
    BlendTime_1=0
    bActiveValue=Get RightArmIsLongDelay()
   ) {blendType=Linear}
   AnimGraphNode_StateMachine_0 = Hit_Crash2_Hard_R_SM() {bAllowConduitEntryStates=true}
   Output Animation Pose(Result=@AnimGraphNode_StateMachine_0)
  ]
 }
 Hit_Crash2_Hard_R_SM {
  entry=Conduit
  path=AnimGraph.AnimGraphNode_StateMachine_6.RightArmAdditiveHits.AnimStateNode_0.RightArmHit_B.AnimGraphNode_StateMachine_1.RightArmHitChooser_StartTime_DelayFloat.AnimStateNode_14.Hit_Crash2_Hard_R.AnimGraphNode_StateMachine_0.Hit_Crash2_Hard_R_SM
  ids=id
  nodes [
   AnimStateConduitNode_0 = Conduit()
   AnimStateNode_0 = Hit_Short()
   AnimStateNode_1 = Hit_Medium()
   AnimStateNode_2 = Hit_Long()
   AnimStateTransitionNode_4 = AnimStateTransitionNode() {crossfadeDuration=0}
   AnimStateTransitionNode_5 = AnimStateTransitionNode() {crossfadeDuration=0}
   AnimStateTransitionNode_6 = AnimStateTransitionNode() {crossfadeDuration=0}
  ]
  links [
   {
    dst=Hit_Long
    rule=AnimGraph.AnimGraphNode_StateMachine_6.RightArmAdditiveHits.AnimStateNode_0.RightArmHit_B.AnimGraphNode_StateMachine_1.RightArmHitChooser_StartTime_DelayFloat.AnimStateNode_14.Hit_Crash2_Hard_R.AnimGraphNode_StateMachine_0.Hit_Crash2_Hard_R_SM.AnimStateTransitionNode_6.Transition
    src=Conduit
    via=AnimStateTransitionNode_6
   }
   {
    dst=Hit_Medium
    rule=AnimGraph.AnimGraphNode_StateMachine_6.RightArmAdditiveHits.AnimStateNode_0.RightArmHit_B.AnimGraphNode_StateMachine_1.RightArmHitChooser_StartTime_DelayFloat.AnimStateNode_14.Hit_Crash2_Hard_R.AnimGraphNode_StateMachine_0.Hit_Crash2_Hard_R_SM.AnimStateTransitionNode_5.Transition
    src=Conduit
    via=AnimStateTransitionNode_5
   }
   {
    dst=Hit_Short
    rule=AnimGraph.AnimGraphNode_StateMachine_6.RightArmAdditiveHits.AnimStateNode_0.RightArmHit_B.AnimGraphNode_StateMachine_1.RightArmHitChooser_StartTime_DelayFloat.AnimStateNode_14.Hit_Crash2_Hard_R.AnimGraphNode_StateMachine_0.Hit_Crash2_Hard_R_SM.AnimStateTransitionNode_4.Transition
    src=Conduit
    via=AnimStateTransitionNode_4
   }
  ]
 }
 Conduit {
  path=AnimGraph.AnimGraphNode_StateMachine_6.RightArmAdditiveHits.AnimStateNode_0.RightArmHit_B.AnimGraphNode_StateMachine_1.RightArmHitChooser_StartTime_DelayFloat.AnimStateNode_14.Hit_Crash2_Hard_R.AnimGraphNode_StateMachine_0.Hit_Crash2_Hard_R_SM.AnimStateConduitNode_0.Conduit
  nodes [Result(bCanEnterTransition=true)]
 }
 Hit_Short {
  path=AnimGraph.AnimGraphNode_StateMachine_6.RightArmAdditiveHits.AnimStateNode_0.RightArmHit_B.AnimGraphNode_StateMachine_1.RightArmHitChooser_StartTime_DelayFloat.AnimStateNode_14.Hit_Crash2_Hard_R.AnimGraphNode_StateMachine_0.Hit_Crash2_Hard_R_SM.AnimStateNode_0.Hit_Short
  nodes [Output Animation Pose(Result="Sequence Player (additive) 'Drum_Hit_Crash2_Tap_Fast'"(StartPosition=Get RightArmMaxDelayShortenedBy(), bLoopAnimation=false) {sequence=&Drum_Hit_Crash2_Tap_Fast, startPosition=0.05})]
 }
 Hit_Medium {
  path=AnimGraph.AnimGraphNode_StateMachine_6.RightArmAdditiveHits.AnimStateNode_0.RightArmHit_B.AnimGraphNode_StateMachine_1.RightArmHitChooser_StartTime_DelayFloat.AnimStateNode_14.Hit_Crash2_Hard_R.AnimGraphNode_StateMachine_0.Hit_Crash2_Hard_R_SM.AnimStateNode_1.Hit_Medium
  nodes [
   n1 = "Sequence Player (additive) 'Drum_Hit_Crash2_Normal_Medium'"(StartPosition=Get RightArmMaxDelayShortenedBy(), bLoopAnimation=false) {sequence=&Drum_Hit_Crash2_Normal_Medium, startPosition=0.05}
   Output Animation Pose(Result=@n1)
  ]
 }
 Hit_Long {
  path=AnimGraph.AnimGraphNode_StateMachine_6.RightArmAdditiveHits.AnimStateNode_0.RightArmHit_B.AnimGraphNode_StateMachine_1.RightArmHitChooser_StartTime_DelayFloat.AnimStateNode_14.Hit_Crash2_Hard_R.AnimGraphNode_StateMachine_0.Hit_Crash2_Hard_R_SM.AnimStateNode_2.Hit_Long
  nodes [Output Animation Pose(Result="Sequence Player (additive) 'Drum_Hit_Crash2_Slow_Big'"(StartPosition=Get RightArmMaxDelayShortenedBy(), bLoopAnimation=false) {sequence=&Drum_Hit_Crash2_Slow_Big, startPosition=0.05})]
 }
 Transition {
  path=AnimGraph.AnimGraphNode_StateMachine_6.RightArmAdditiveHits.AnimStateNode_0.RightArmHit_B.AnimGraphNode_StateMachine_1.RightArmHitChooser_StartTime_DelayFloat.AnimStateNode_14.Hit_Crash2_Hard_R.AnimGraphNode_StateMachine_0.Hit_Crash2_Hard_R_SM.AnimStateTransitionNode_4.Transition
  nodes [Result(bCanEnterTransition=Get RightArmIsShortDelay())]
 }
 Transition {
  path=AnimGraph.AnimGraphNode_StateMachine_6.RightArmAdditiveHits.AnimStateNode_0.RightArmHit_B.AnimGraphNode_StateMachine_1.RightArmHitChooser_StartTime_DelayFloat.AnimStateNode_14.Hit_Crash2_Hard_R.AnimGraphNode_StateMachine_0.Hit_Crash2_Hard_R_SM.AnimStateTransitionNode_5.Transition
  nodes [
   n1 = Get RightArmIsLongDelay()
   n2 = Get RightArmIsShortDelay()
   Result(bCanEnterTransition=AND Boolean(A=NOT Boolean(A=@n2), B=NOT Boolean(A=@n1)))
  ]
 }
 Transition {
  path=AnimGraph.AnimGraphNode_StateMachine_6.RightArmAdditiveHits.AnimStateNode_0.RightArmHit_B.AnimGraphNode_StateMachine_1.RightArmHitChooser_StartTime_DelayFloat.AnimStateNode_14.Hit_Crash2_Hard_R.AnimGraphNode_StateMachine_0.Hit_Crash2_Hard_R_SM.AnimStateTransitionNode_6.Transition
  nodes [Result(bCanEnterTransition=Get RightArmIsLongDelay())]
 }
 Hit_Crash1_Choke_R {
  path=AnimGraph.AnimGraphNode_StateMachine_6.RightArmAdditiveHits.AnimStateNode_0.RightArmHit_B.AnimGraphNode_StateMachine_1.RightArmHitChooser_StartTime_DelayFloat.AnimStateNode_2.Hit_Crash1_Choke_R
  nodes [
   n1 = Get RightArmMaxDelayShortenedBy()
   n2 = "Sequence Player (additive) 'Drum_Hit_Crash1_Normal_Medium'"(StartPosition=Get RightArmMaxDelayShortenedBy(), bLoopAnimation=false) {sequence=&Drum_Hit_Crash1_Normal_Medium, startPosition=0.05}
   Blend Poses by bool(
    BlendPose_0="Sequence Player (additive) 'Drum_Hit_Crash1_Slow_Big'"(StartPosition=Get RightArmMaxDelayShortenedBy(), bLoopAnimation=false) {sequence=&Drum_Hit_Crash1_Slow_Big, startPosition=0.05}
    BlendPose_1=Blend Poses by bool(
     BlendPose_0="Sequence Player (additive) 'Drum_Hit_Crash1_Tap_Fast'"(StartPosition=@n1, bLoopAnimation=false) {sequence=&Drum_Hit_Crash1_Tap_Fast, startPosition=0.05}
     BlendPose_1=@n2
     BlendTime_0=0
     BlendTime_1=0
     bActiveValue=Get RightArmIsShortDelay()
    ) {blendType=Linear}
    BlendTime_0=0
    BlendTime_1=0
    bActiveValue=Get RightArmIsLongDelay()
   ) {blendType=Linear}
   AnimGraphNode_StateMachine_0 = Hit_Crash1_Hard_R_SM() {bAllowConduitEntryStates=true}
   Output Animation Pose(Result=@AnimGraphNode_StateMachine_0)
  ]
 }
 Hit_Crash1_Hard_R_SM {
  entry=Conduit
  path=AnimGraph.AnimGraphNode_StateMachine_6.RightArmAdditiveHits.AnimStateNode_0.RightArmHit_B.AnimGraphNode_StateMachine_1.RightArmHitChooser_StartTime_DelayFloat.AnimStateNode_2.Hit_Crash1_Choke_R.AnimGraphNode_StateMachine_0.Hit_Crash1_Hard_R_SM
  ids=id
  nodes [
   AnimStateConduitNode_0 = Conduit()
   AnimStateNode_0 = Hit_Short()
   AnimStateNode_1 = Hit_Medium()
   AnimStateNode_2 = Hit_Long()
   AnimStateTransitionNode_4 = AnimStateTransitionNode() {crossfadeDuration=0}
   AnimStateTransitionNode_5 = AnimStateTransitionNode() {crossfadeDuration=0}
   AnimStateTransitionNode_6 = AnimStateTransitionNode() {crossfadeDuration=0}
  ]
  links [
   {
    dst=Hit_Long
    rule=AnimGraph.AnimGraphNode_StateMachine_6.RightArmAdditiveHits.AnimStateNode_0.RightArmHit_B.AnimGraphNode_StateMachine_1.RightArmHitChooser_StartTime_DelayFloat.AnimStateNode_2.Hit_Crash1_Choke_R.AnimGraphNode_StateMachine_0.Hit_Crash1_Hard_R_SM.AnimStateTransitionNode_6.Transition
    src=Conduit
    via=AnimStateTransitionNode_6
   }
   {
    dst=Hit_Medium
    rule=AnimGraph.AnimGraphNode_StateMachine_6.RightArmAdditiveHits.AnimStateNode_0.RightArmHit_B.AnimGraphNode_StateMachine_1.RightArmHitChooser_StartTime_DelayFloat.AnimStateNode_2.Hit_Crash1_Choke_R.AnimGraphNode_StateMachine_0.Hit_Crash1_Hard_R_SM.AnimStateTransitionNode_5.Transition
    src=Conduit
    via=AnimStateTransitionNode_5
   }
   {
    dst=Hit_Short
    rule=AnimGraph.AnimGraphNode_StateMachine_6.RightArmAdditiveHits.AnimStateNode_0.RightArmHit_B.AnimGraphNode_StateMachine_1.RightArmHitChooser_StartTime_DelayFloat.AnimStateNode_2.Hit_Crash1_Choke_R.AnimGraphNode_StateMachine_0.Hit_Crash1_Hard_R_SM.AnimStateTransitionNode_4.Transition
    src=Conduit
    via=AnimStateTransitionNode_4
   }
  ]
 }
 Conduit {
  path=AnimGraph.AnimGraphNode_StateMachine_6.RightArmAdditiveHits.AnimStateNode_0.RightArmHit_B.AnimGraphNode_StateMachine_1.RightArmHitChooser_StartTime_DelayFloat.AnimStateNode_2.Hit_Crash1_Choke_R.AnimGraphNode_StateMachine_0.Hit_Crash1_Hard_R_SM.AnimStateConduitNode_0.Conduit
  nodes [Result(bCanEnterTransition=true)]
 }
 Hit_Short {
  path=AnimGraph.AnimGraphNode_StateMachine_6.RightArmAdditiveHits.AnimStateNode_0.RightArmHit_B.AnimGraphNode_StateMachine_1.RightArmHitChooser_StartTime_DelayFloat.AnimStateNode_2.Hit_Crash1_Choke_R.AnimGraphNode_StateMachine_0.Hit_Crash1_Hard_R_SM.AnimStateNode_0.Hit_Short
  nodes [Output Animation Pose(Result="Sequence Player (additive) 'Drum_Hit_Crash1_Tap_Fast'"(StartPosition=Get RightArmMaxDelayShortenedBy(), bLoopAnimation=false) {sequence=&Drum_Hit_Crash1_Tap_Fast, startPosition=0.05})]
 }
 Hit_Medium {
  path=AnimGraph.AnimGraphNode_StateMachine_6.RightArmAdditiveHits.AnimStateNode_0.RightArmHit_B.AnimGraphNode_StateMachine_1.RightArmHitChooser_StartTime_DelayFloat.AnimStateNode_2.Hit_Crash1_Choke_R.AnimGraphNode_StateMachine_0.Hit_Crash1_Hard_R_SM.AnimStateNode_1.Hit_Medium
  nodes [
   n1 = "Sequence Player (additive) 'Drum_Hit_Crash1_Normal_Medium'"(StartPosition=Get RightArmMaxDelayShortenedBy(), bLoopAnimation=false) {sequence=&Drum_Hit_Crash1_Normal_Medium, startPosition=0.05}
   Output Animation Pose(Result=@n1)
  ]
 }
 Hit_Long {
  path=AnimGraph.AnimGraphNode_StateMachine_6.RightArmAdditiveHits.AnimStateNode_0.RightArmHit_B.AnimGraphNode_StateMachine_1.RightArmHitChooser_StartTime_DelayFloat.AnimStateNode_2.Hit_Crash1_Choke_R.AnimGraphNode_StateMachine_0.Hit_Crash1_Hard_R_SM.AnimStateNode_2.Hit_Long
  nodes [Output Animation Pose(Result="Sequence Player (additive) 'Drum_Hit_Crash1_Slow_Big'"(StartPosition=Get RightArmMaxDelayShortenedBy(), bLoopAnimation=false) {sequence=&Drum_Hit_Crash1_Slow_Big, startPosition=0.05})]
 }
 Transition {
  path=AnimGraph.AnimGraphNode_StateMachine_6.RightArmAdditiveHits.AnimStateNode_0.RightArmHit_B.AnimGraphNode_StateMachine_1.RightArmHitChooser_StartTime_DelayFloat.AnimStateNode_2.Hit_Crash1_Choke_R.AnimGraphNode_StateMachine_0.Hit_Crash1_Hard_R_SM.AnimStateTransitionNode_4.Transition
  nodes [Result(bCanEnterTransition=Get RightArmIsShortDelay())]
 }
 Transition {
  path=AnimGraph.AnimGraphNode_StateMachine_6.RightArmAdditiveHits.AnimStateNode_0.RightArmHit_B.AnimGraphNode_StateMachine_1.RightArmHitChooser_StartTime_DelayFloat.AnimStateNode_2.Hit_Crash1_Choke_R.AnimGraphNode_StateMachine_0.Hit_Crash1_Hard_R_SM.AnimStateTransitionNode_5.Transition
  nodes [
   n1 = Get RightArmIsLongDelay()
   n2 = Get RightArmIsShortDelay()
   Result(bCanEnterTransition=AND Boolean(A=NOT Boolean(A=@n2), B=NOT Boolean(A=@n1)))
  ]
 }
 Transition {
  path=AnimGraph.AnimGraphNode_StateMachine_6.RightArmAdditiveHits.AnimStateNode_0.RightArmHit_B.AnimGraphNode_StateMachine_1.RightArmHitChooser_StartTime_DelayFloat.AnimStateNode_2.Hit_Crash1_Choke_R.AnimGraphNode_StateMachine_0.Hit_Crash1_Hard_R_SM.AnimStateTransitionNode_6.Transition
  nodes [Result(bCanEnterTransition=Get RightArmIsLongDelay())]
 }
 Hit_Crash1_Soft_R {
  path=AnimGraph.AnimGraphNode_StateMachine_6.RightArmAdditiveHits.AnimStateNode_0.RightArmHit_B.AnimGraphNode_StateMachine_1.RightArmHitChooser_StartTime_DelayFloat.AnimStateNode_3.Hit_Crash1_Soft_R
  nodes [
   n1 = Get RightArmMaxDelayShortenedBy()
   n2 = "Sequence Player (additive) 'Drum_Hit_Crash1_Normal_Medium'"(StartPosition=Get RightArmMaxDelayShortenedBy(), bLoopAnimation=false) {sequence=&Drum_Hit_Crash1_Normal_Medium, startPosition=0.05}
   Blend Poses by bool(
    BlendPose_0="Sequence Player (additive) 'Drum_Hit_Crash1_Slow_Big'"(StartPosition=Get RightArmMaxDelayShortenedBy(), bLoopAnimation=false) {sequence=&Drum_Hit_Crash1_Slow_Big, startPosition=0.05}
    BlendPose_1=Blend Poses by bool(
     BlendPose_0="Sequence Player (additive) 'Drum_Hit_Crash1_Tap_Fast'"(StartPosition=@n1, bLoopAnimation=false) {sequence=&Drum_Hit_Crash1_Tap_Fast, startPosition=0.05}
     BlendPose_1=@n2
     BlendTime_0=0
     BlendTime_1=0
     bActiveValue=Get RightArmIsShortDelay()
    ) {blendType=Linear}
    BlendTime_0=0
    BlendTime_1=0
    bActiveValue=Get RightArmIsLongDelay()
   ) {blendType=Linear}
   AnimGraphNode_StateMachine_0 = Hit_Crash1_Soft_R_SM() {is=title AnimGraphNode_StateMachine, bAllowConduitEntryStates=true}
   Output Animation Pose(Result=@AnimGraphNode_StateMachine_0)
  ]
 }
 Hit_Crash1_Soft_R_SM {
  entry=Conduit
  path=AnimGraph.AnimGraphNode_StateMachine_6.RightArmAdditiveHits.AnimStateNode_0.RightArmHit_B.AnimGraphNode_StateMachine_1.RightArmHitChooser_StartTime_DelayFloat.AnimStateNode_3.Hit_Crash1_Soft_R.AnimGraphNode_StateMachine_0.Hit_Crash1_Soft_R_SM
  ids=id
  nodes [
   AnimStateConduitNode_0 = Conduit()
   AnimStateNode_0 = Hit_Short()
   AnimStateNode_1 = Hit_Medium()
   AnimStateNode_2 = Hit_Long()
   AnimStateTransitionNode_4 = AnimStateTransitionNode() {crossfadeDuration=0}
   AnimStateTransitionNode_5 = AnimStateTransitionNode() {crossfadeDuration=0}
   AnimStateTransitionNode_6 = AnimStateTransitionNode() {crossfadeDuration=0}
  ]
  links [
   {
    dst=Hit_Long
    rule=AnimGraph.AnimGraphNode_StateMachine_6.RightArmAdditiveHits.AnimStateNode_0.RightArmHit_B.AnimGraphNode_StateMachine_1.RightArmHitChooser_StartTime_DelayFloat.AnimStateNode_3.Hit_Crash1_Soft_R.AnimGraphNode_StateMachine_0.Hit_Crash1_Soft_R_SM.AnimStateTransitionNode_6.Transition
    src=Conduit
    via=AnimStateTransitionNode_6
   }
   {
    dst=Hit_Medium
    rule=AnimGraph.AnimGraphNode_StateMachine_6.RightArmAdditiveHits.AnimStateNode_0.RightArmHit_B.AnimGraphNode_StateMachine_1.RightArmHitChooser_StartTime_DelayFloat.AnimStateNode_3.Hit_Crash1_Soft_R.AnimGraphNode_StateMachine_0.Hit_Crash1_Soft_R_SM.AnimStateTransitionNode_5.Transition
    src=Conduit
    via=AnimStateTransitionNode_5
   }
   {
    dst=Hit_Short
    rule=AnimGraph.AnimGraphNode_StateMachine_6.RightArmAdditiveHits.AnimStateNode_0.RightArmHit_B.AnimGraphNode_StateMachine_1.RightArmHitChooser_StartTime_DelayFloat.AnimStateNode_3.Hit_Crash1_Soft_R.AnimGraphNode_StateMachine_0.Hit_Crash1_Soft_R_SM.AnimStateTransitionNode_4.Transition
    src=Conduit
    via=AnimStateTransitionNode_4
   }
  ]
 }
 Conduit {
  path=AnimGraph.AnimGraphNode_StateMachine_6.RightArmAdditiveHits.AnimStateNode_0.RightArmHit_B.AnimGraphNode_StateMachine_1.RightArmHitChooser_StartTime_DelayFloat.AnimStateNode_3.Hit_Crash1_Soft_R.AnimGraphNode_StateMachine_0.Hit_Crash1_Soft_R_SM.AnimStateConduitNode_0.Conduit
  nodes [Result(bCanEnterTransition=true)]
 }
 Hit_Short {
  path=AnimGraph.AnimGraphNode_StateMachine_6.RightArmAdditiveHits.AnimStateNode_0.RightArmHit_B.AnimGraphNode_StateMachine_1.RightArmHitChooser_StartTime_DelayFloat.AnimStateNode_3.Hit_Crash1_Soft_R.AnimGraphNode_StateMachine_0.Hit_Crash1_Soft_R_SM.AnimStateNode_0.Hit_Short
  nodes [Output Animation Pose(Result="Sequence Player (additive) 'Drum_Hit_Crash1_Tap_Fast'"(StartPosition=Get RightArmMaxDelayShortenedBy(), bLoopAnimation=false) {sequence=&Drum_Hit_Crash1_Tap_Fast, startPosition=0.05})]
 }
 Hit_Medium {
  path=AnimGraph.AnimGraphNode_StateMachine_6.RightArmAdditiveHits.AnimStateNode_0.RightArmHit_B.AnimGraphNode_StateMachine_1.RightArmHitChooser_StartTime_DelayFloat.AnimStateNode_3.Hit_Crash1_Soft_R.AnimGraphNode_StateMachine_0.Hit_Crash1_Soft_R_SM.AnimStateNode_1.Hit_Medium
  nodes [
   n1 = "Sequence Player (additive) 'Drum_Hit_Crash1_Normal_Medium'"(StartPosition=Get RightArmMaxDelayShortenedBy(), bLoopAnimation=false) {sequence=&Drum_Hit_Crash1_Normal_Medium, startPosition=0.05}
   Output Animation Pose(Result=@n1)
  ]
 }
 Hit_Long {
  path=AnimGraph.AnimGraphNode_StateMachine_6.RightArmAdditiveHits.AnimStateNode_0.RightArmHit_B.AnimGraphNode_StateMachine_1.RightArmHitChooser_StartTime_DelayFloat.AnimStateNode_3.Hit_Crash1_Soft_R.AnimGraphNode_StateMachine_0.Hit_Crash1_Soft_R_SM.AnimStateNode_2.Hit_Long
  nodes [Output Animation Pose(Result="Sequence Player (additive) 'Drum_Hit_Crash1_Slow_Big'"(StartPosition=Get RightArmMaxDelayShortenedBy(), bLoopAnimation=false) {sequence=&Drum_Hit_Crash1_Slow_Big, startPosition=0.05})]
 }
 Transition {
  path=AnimGraph.AnimGraphNode_StateMachine_6.RightArmAdditiveHits.AnimStateNode_0.RightArmHit_B.AnimGraphNode_StateMachine_1.RightArmHitChooser_StartTime_DelayFloat.AnimStateNode_3.Hit_Crash1_Soft_R.AnimGraphNode_StateMachine_0.Hit_Crash1_Soft_R_SM.AnimStateTransitionNode_4.Transition
  nodes [Result(bCanEnterTransition=Get RightArmIsShortDelay())]
 }
 Transition {
  path=AnimGraph.AnimGraphNode_StateMachine_6.RightArmAdditiveHits.AnimStateNode_0.RightArmHit_B.AnimGraphNode_StateMachine_1.RightArmHitChooser_StartTime_DelayFloat.AnimStateNode_3.Hit_Crash1_Soft_R.AnimGraphNode_StateMachine_0.Hit_Crash1_Soft_R_SM.AnimStateTransitionNode_5.Transition
  nodes [
   n1 = Get RightArmIsLongDelay()
   n2 = Get RightArmIsShortDelay()
   Result(bCanEnterTransition=AND Boolean(A=NOT Boolean(A=@n2), B=NOT Boolean(A=@n1)))
  ]
 }
 Transition {
  path=AnimGraph.AnimGraphNode_StateMachine_6.RightArmAdditiveHits.AnimStateNode_0.RightArmHit_B.AnimGraphNode_StateMachine_1.RightArmHitChooser_StartTime_DelayFloat.AnimStateNode_3.Hit_Crash1_Soft_R.AnimGraphNode_StateMachine_0.Hit_Crash1_Soft_R_SM.AnimStateTransitionNode_6.Transition
  nodes [Result(bCanEnterTransition=Get RightArmIsLongDelay())]
 }
 Hit_Snare_Soft_R {
  path=AnimGraph.AnimGraphNode_StateMachine_6.RightArmAdditiveHits.AnimStateNode_0.RightArmHit_B.AnimGraphNode_StateMachine_1.RightArmHitChooser_StartTime_DelayFloat.AnimStateNode_4.Hit_Snare_Soft_R
  nodes [Output Animation Pose(Result="Sequence Player (additive) 'Drum_Hit_SnareSoft_Tap_Fast'"(StartPosition=Get LeftArmMaxDelayShortenedBy(), bLoopAnimation=False) {bLoopAnimation=false, sequence=&Drum_Hit_SnareSoft_Tap_Fast})]
 }
 Hit_Crash2_Soft_R {
  path=AnimGraph.AnimGraphNode_StateMachine_6.RightArmAdditiveHits.AnimStateNode_0.RightArmHit_B.AnimGraphNode_StateMachine_1.RightArmHitChooser_StartTime_DelayFloat.AnimStateNode_5.Hit_Crash2_Soft_R
  nodes [
   n1 = Get RightArmMaxDelayShortenedBy()
   n2 = "Sequence Player (additive) 'Drum_Hit_Crash2_Normal_Medium'"(StartPosition=Get RightArmMaxDelayShortenedBy(), bLoopAnimation=false) {sequence=&Drum_Hit_Crash2_Normal_Medium, startPosition=0.05}
   Blend Poses by bool(
    BlendPose_0="Sequence Player (additive) 'Drum_Hit_Crash2_Slow_Big'"(StartPosition=Get RightArmMaxDelayShortenedBy(), bLoopAnimation=false) {sequence=&Drum_Hit_Crash2_Slow_Big, startPosition=0.05}
    BlendPose_1=Blend Poses by bool(
     BlendPose_0="Sequence Player (additive) 'Drum_Hit_Crash2_Tap_Fast'"(StartPosition=@n1, bLoopAnimation=false) {sequence=&Drum_Hit_Crash2_Tap_Fast, startPosition=0.05}
     BlendPose_1=@n2
     BlendTime_0=0
     BlendTime_1=0
     bActiveValue=Get RightArmIsShortDelay()
    ) {blendType=Linear}
    BlendTime_0=0
    BlendTime_1=0
    bActiveValue=Get RightArmIsLongDelay()
   ) {blendType=Linear}
   AnimGraphNode_StateMachine_0 = Hit_Crash2_Soft_R_SM() {is=title AnimGraphNode_StateMachine, bAllowConduitEntryStates=true}
   Output Animation Pose(Result=@AnimGraphNode_StateMachine_0)
  ]
 }
 Hit_Crash2_Soft_R_SM {
  entry=Conduit
  path=AnimGraph.AnimGraphNode_StateMachine_6.RightArmAdditiveHits.AnimStateNode_0.RightArmHit_B.AnimGraphNode_StateMachine_1.RightArmHitChooser_StartTime_DelayFloat.AnimStateNode_5.Hit_Crash2_Soft_R.AnimGraphNode_StateMachine_0.Hit_Crash2_Soft_R_SM
  ids=id
  nodes [
   AnimStateConduitNode_0 = Conduit()
   AnimStateNode_0 = Hit_Short()
   AnimStateNode_1 = Hit_Medium()
   AnimStateNode_2 = Hit_Long()
   AnimStateTransitionNode_4 = AnimStateTransitionNode() {crossfadeDuration=0}
   AnimStateTransitionNode_5 = AnimStateTransitionNode() {crossfadeDuration=0}
   AnimStateTransitionNode_6 = AnimStateTransitionNode() {crossfadeDuration=0}
  ]
  links [
   {
    dst=Hit_Long
    rule=AnimGraph.AnimGraphNode_StateMachine_6.RightArmAdditiveHits.AnimStateNode_0.RightArmHit_B.AnimGraphNode_StateMachine_1.RightArmHitChooser_StartTime_DelayFloat.AnimStateNode_5.Hit_Crash2_Soft_R.AnimGraphNode_StateMachine_0.Hit_Crash2_Soft_R_SM.AnimStateTransitionNode_6.Transition
    src=Conduit
    via=AnimStateTransitionNode_6
   }
   {
    dst=Hit_Medium
    rule=AnimGraph.AnimGraphNode_StateMachine_6.RightArmAdditiveHits.AnimStateNode_0.RightArmHit_B.AnimGraphNode_StateMachine_1.RightArmHitChooser_StartTime_DelayFloat.AnimStateNode_5.Hit_Crash2_Soft_R.AnimGraphNode_StateMachine_0.Hit_Crash2_Soft_R_SM.AnimStateTransitionNode_5.Transition
    src=Conduit
    via=AnimStateTransitionNode_5
   }
   {
    dst=Hit_Short
    rule=AnimGraph.AnimGraphNode_StateMachine_6.RightArmAdditiveHits.AnimStateNode_0.RightArmHit_B.AnimGraphNode_StateMachine_1.RightArmHitChooser_StartTime_DelayFloat.AnimStateNode_5.Hit_Crash2_Soft_R.AnimGraphNode_StateMachine_0.Hit_Crash2_Soft_R_SM.AnimStateTransitionNode_4.Transition
    src=Conduit
    via=AnimStateTransitionNode_4
   }
  ]
 }
 Conduit {
  path=AnimGraph.AnimGraphNode_StateMachine_6.RightArmAdditiveHits.AnimStateNode_0.RightArmHit_B.AnimGraphNode_StateMachine_1.RightArmHitChooser_StartTime_DelayFloat.AnimStateNode_5.Hit_Crash2_Soft_R.AnimGraphNode_StateMachine_0.Hit_Crash2_Soft_R_SM.AnimStateConduitNode_0.Conduit
  nodes [Result(bCanEnterTransition=true)]
 }
 Hit_Short {
  path=AnimGraph.AnimGraphNode_StateMachine_6.RightArmAdditiveHits.AnimStateNode_0.RightArmHit_B.AnimGraphNode_StateMachine_1.RightArmHitChooser_StartTime_DelayFloat.AnimStateNode_5.Hit_Crash2_Soft_R.AnimGraphNode_StateMachine_0.Hit_Crash2_Soft_R_SM.AnimStateNode_0.Hit_Short
  nodes [Output Animation Pose(Result="Sequence Player (additive) 'Drum_Hit_Crash2_Tap_Fast'"(StartPosition=Get RightArmMaxDelayShortenedBy(), bLoopAnimation=false) {sequence=&Drum_Hit_Crash2_Tap_Fast, startPosition=0.05})]
 }
 Hit_Medium {
  path=AnimGraph.AnimGraphNode_StateMachine_6.RightArmAdditiveHits.AnimStateNode_0.RightArmHit_B.AnimGraphNode_StateMachine_1.RightArmHitChooser_StartTime_DelayFloat.AnimStateNode_5.Hit_Crash2_Soft_R.AnimGraphNode_StateMachine_0.Hit_Crash2_Soft_R_SM.AnimStateNode_1.Hit_Medium
  nodes [
   n1 = "Sequence Player (additive) 'Drum_Hit_Crash2_Normal_Medium'"(StartPosition=Get RightArmMaxDelayShortenedBy(), bLoopAnimation=false) {sequence=&Drum_Hit_Crash2_Normal_Medium, startPosition=0.05}
   Output Animation Pose(Result=@n1)
  ]
 }
 Hit_Long {
  path=AnimGraph.AnimGraphNode_StateMachine_6.RightArmAdditiveHits.AnimStateNode_0.RightArmHit_B.AnimGraphNode_StateMachine_1.RightArmHitChooser_StartTime_DelayFloat.AnimStateNode_5.Hit_Crash2_Soft_R.AnimGraphNode_StateMachine_0.Hit_Crash2_Soft_R_SM.AnimStateNode_2.Hit_Long
  nodes [Output Animation Pose(Result="Sequence Player (additive) 'Drum_Hit_Crash2_Slow_Big'"(StartPosition=Get RightArmMaxDelayShortenedBy(), bLoopAnimation=false) {sequence=&Drum_Hit_Crash2_Slow_Big, startPosition=0.05})]
 }
 Transition {
  path=AnimGraph.AnimGraphNode_StateMachine_6.RightArmAdditiveHits.AnimStateNode_0.RightArmHit_B.AnimGraphNode_StateMachine_1.RightArmHitChooser_StartTime_DelayFloat.AnimStateNode_5.Hit_Crash2_Soft_R.AnimGraphNode_StateMachine_0.Hit_Crash2_Soft_R_SM.AnimStateTransitionNode_4.Transition
  nodes [Result(bCanEnterTransition=Get RightArmIsShortDelay())]
 }
 Transition {
  path=AnimGraph.AnimGraphNode_StateMachine_6.RightArmAdditiveHits.AnimStateNode_0.RightArmHit_B.AnimGraphNode_StateMachine_1.RightArmHitChooser_StartTime_DelayFloat.AnimStateNode_5.Hit_Crash2_Soft_R.AnimGraphNode_StateMachine_0.Hit_Crash2_Soft_R_SM.AnimStateTransitionNode_5.Transition
  nodes [
   n1 = Get RightArmIsLongDelay()
   n2 = Get RightArmIsShortDelay()
   Result(bCanEnterTransition=AND Boolean(A=NOT Boolean(A=@n2), B=NOT Boolean(A=@n1)))
  ]
 }
 Transition {
  path=AnimGraph.AnimGraphNode_StateMachine_6.RightArmAdditiveHits.AnimStateNode_0.RightArmHit_B.AnimGraphNode_StateMachine_1.RightArmHitChooser_StartTime_DelayFloat.AnimStateNode_5.Hit_Crash2_Soft_R.AnimGraphNode_StateMachine_0.Hit_Crash2_Soft_R_SM.AnimStateTransitionNode_6.Transition
  nodes [Result(bCanEnterTransition=Get RightArmIsLongDelay())]
 }
 Hit_Ride_R {
  path=AnimGraph.AnimGraphNode_StateMachine_6.RightArmAdditiveHits.AnimStateNode_0.RightArmHit_B.AnimGraphNode_StateMachine_1.RightArmHitChooser_StartTime_DelayFloat.AnimStateNode_6.Hit_Ride_R
  nodes [
   n1 = Get RightArmMaxDelayShortenedBy()
   n2 = "Sequence Player (additive) 'Drum_Hit_Ride_Normal_Medium'"(StartPosition=Get RightArmMaxDelayShortenedBy(), bLoopAnimation=false) {sequence=&Drum_Hit_Ride_Normal_Medium, startPosition=0.05}
   Blend Poses by bool(
    BlendPose_0="Sequence Player (additive) 'Drum_Hit_Ride_Slow_Big'"(StartPosition=Get RightArmMaxDelayShortenedBy(), bLoopAnimation=false) {sequence=&Drum_Hit_Ride_Slow_Big, startPosition=0.05}
    BlendPose_1=Blend Poses by bool(
     BlendPose_0="Sequence Player (additive) 'Drum_Hit_Ride_Tap_Fast'"(StartPosition=@n1, bLoopAnimation=false) {sequence=&Drum_Hit_Ride_Tap_Fast, startPosition=0.05}
     BlendPose_1=@n2
     BlendTime_0=0
     BlendTime_1=0
     bActiveValue=Get RightArmIsShortDelay()
    ) {blendType=Linear}
    BlendTime_0=0
    BlendTime_1=0
    bActiveValue=Get RightArmIsLongDelay()
   ) {blendType=Linear}
   AnimGraphNode_StateMachine_0 = Hit_Ride_R_SM() {is=title AnimGraphNode_StateMachine, bAllowConduitEntryStates=true}
   Output Animation Pose(Result=@AnimGraphNode_StateMachine_0)
  ]
 }
 Hit_Ride_R_SM {
  entry=Conduit
  path=AnimGraph.AnimGraphNode_StateMachine_6.RightArmAdditiveHits.AnimStateNode_0.RightArmHit_B.AnimGraphNode_StateMachine_1.RightArmHitChooser_StartTime_DelayFloat.AnimStateNode_6.Hit_Ride_R.AnimGraphNode_StateMachine_0.Hit_Ride_R_SM
  ids=id
  nodes [
   AnimStateConduitNode_0 = Conduit()
   AnimStateNode_0 = Hit_Short()
   AnimStateNode_1 = Hit_Medium()
   AnimStateNode_2 = Hit_Long()
   AnimStateTransitionNode_4 = AnimStateTransitionNode() {crossfadeDuration=0}
   AnimStateTransitionNode_5 = AnimStateTransitionNode() {crossfadeDuration=0}
   AnimStateTransitionNode_6 = AnimStateTransitionNode() {crossfadeDuration=0}
  ]
  links [
   {
    dst=Hit_Long
    rule=AnimGraph.AnimGraphNode_StateMachine_6.RightArmAdditiveHits.AnimStateNode_0.RightArmHit_B.AnimGraphNode_StateMachine_1.RightArmHitChooser_StartTime_DelayFloat.AnimStateNode_6.Hit_Ride_R.AnimGraphNode_StateMachine_0.Hit_Ride_R_SM.AnimStateTransitionNode_6.Transition
    src=Conduit
    via=AnimStateTransitionNode_6
   }
   {
    dst=Hit_Medium
    rule=AnimGraph.AnimGraphNode_StateMachine_6.RightArmAdditiveHits.AnimStateNode_0.RightArmHit_B.AnimGraphNode_StateMachine_1.RightArmHitChooser_StartTime_DelayFloat.AnimStateNode_6.Hit_Ride_R.AnimGraphNode_StateMachine_0.Hit_Ride_R_SM.AnimStateTransitionNode_5.Transition
    src=Conduit
    via=AnimStateTransitionNode_5
   }
   {
    dst=Hit_Short
    rule=AnimGraph.AnimGraphNode_StateMachine_6.RightArmAdditiveHits.AnimStateNode_0.RightArmHit_B.AnimGraphNode_StateMachine_1.RightArmHitChooser_StartTime_DelayFloat.AnimStateNode_6.Hit_Ride_R.AnimGraphNode_StateMachine_0.Hit_Ride_R_SM.AnimStateTransitionNode_4.Transition
    src=Conduit
    via=AnimStateTransitionNode_4
   }
  ]
 }
 Conduit {
  path=AnimGraph.AnimGraphNode_StateMachine_6.RightArmAdditiveHits.AnimStateNode_0.RightArmHit_B.AnimGraphNode_StateMachine_1.RightArmHitChooser_StartTime_DelayFloat.AnimStateNode_6.Hit_Ride_R.AnimGraphNode_StateMachine_0.Hit_Ride_R_SM.AnimStateConduitNode_0.Conduit
  nodes [Result(bCanEnterTransition=true)]
 }
 Hit_Short {
  path=AnimGraph.AnimGraphNode_StateMachine_6.RightArmAdditiveHits.AnimStateNode_0.RightArmHit_B.AnimGraphNode_StateMachine_1.RightArmHitChooser_StartTime_DelayFloat.AnimStateNode_6.Hit_Ride_R.AnimGraphNode_StateMachine_0.Hit_Ride_R_SM.AnimStateNode_0.Hit_Short
  nodes [Output Animation Pose(Result="Sequence Player (additive) 'Drum_Hit_Ride_Tap_Fast'"(StartPosition=Get RightArmMaxDelayShortenedBy(), bLoopAnimation=false) {sequence=&Drum_Hit_Ride_Tap_Fast, startPosition=0.05})]
 }
 Hit_Medium {
  path=AnimGraph.AnimGraphNode_StateMachine_6.RightArmAdditiveHits.AnimStateNode_0.RightArmHit_B.AnimGraphNode_StateMachine_1.RightArmHitChooser_StartTime_DelayFloat.AnimStateNode_6.Hit_Ride_R.AnimGraphNode_StateMachine_0.Hit_Ride_R_SM.AnimStateNode_1.Hit_Medium
  nodes [
   n1 = "Sequence Player (additive) 'Drum_Hit_Ride_Normal_Medium'"(StartPosition=Get RightArmMaxDelayShortenedBy(), bLoopAnimation=false) {sequence=&Drum_Hit_Ride_Normal_Medium, startPosition=0.05}
   Output Animation Pose(Result=@n1)
  ]
 }
 Hit_Long {
  path=AnimGraph.AnimGraphNode_StateMachine_6.RightArmAdditiveHits.AnimStateNode_0.RightArmHit_B.AnimGraphNode_StateMachine_1.RightArmHitChooser_StartTime_DelayFloat.AnimStateNode_6.Hit_Ride_R.AnimGraphNode_StateMachine_0.Hit_Ride_R_SM.AnimStateNode_2.Hit_Long
  nodes [Output Animation Pose(Result="Sequence Player (additive) 'Drum_Hit_Ride_Slow_Big'"(StartPosition=Get RightArmMaxDelayShortenedBy(), bLoopAnimation=false) {sequence=&Drum_Hit_Ride_Slow_Big, startPosition=0.05})]
 }
 Transition {
  path=AnimGraph.AnimGraphNode_StateMachine_6.RightArmAdditiveHits.AnimStateNode_0.RightArmHit_B.AnimGraphNode_StateMachine_1.RightArmHitChooser_StartTime_DelayFloat.AnimStateNode_6.Hit_Ride_R.AnimGraphNode_StateMachine_0.Hit_Ride_R_SM.AnimStateTransitionNode_4.Transition
  nodes [Result(bCanEnterTransition=Get RightArmIsShortDelay())]
 }
 Transition {
  path=AnimGraph.AnimGraphNode_StateMachine_6.RightArmAdditiveHits.AnimStateNode_0.RightArmHit_B.AnimGraphNode_StateMachine_1.RightArmHitChooser_StartTime_DelayFloat.AnimStateNode_6.Hit_Ride_R.AnimGraphNode_StateMachine_0.Hit_Ride_R_SM.AnimStateTransitionNode_5.Transition
  nodes [
   n1 = Get RightArmIsLongDelay()
   n2 = Get RightArmIsShortDelay()
   Result(bCanEnterTransition=AND Boolean(A=NOT Boolean(A=@n2), B=NOT Boolean(A=@n1)))
  ]
 }
 Transition {
  path=AnimGraph.AnimGraphNode_StateMachine_6.RightArmAdditiveHits.AnimStateNode_0.RightArmHit_B.AnimGraphNode_StateMachine_1.RightArmHitChooser_StartTime_DelayFloat.AnimStateNode_6.Hit_Ride_R.AnimGraphNode_StateMachine_0.Hit_Ride_R_SM.AnimStateTransitionNode_6.Transition
  nodes [Result(bCanEnterTransition=Get RightArmIsLongDelay())]
 }
 Hit_Tom1_R {
  path=AnimGraph.AnimGraphNode_StateMachine_6.RightArmAdditiveHits.AnimStateNode_0.RightArmHit_B.AnimGraphNode_StateMachine_1.RightArmHitChooser_StartTime_DelayFloat.AnimStateNode_7.Hit_Tom1_R
  nodes [
   n1 = Get RightArmMaxDelayShortenedBy()
   n2 = "Sequence Player (additive) 'Drum_Hit_Tom1_Normal_Medium'"(StartPosition=Get RightArmMaxDelayShortenedBy(), bLoopAnimation=false) {sequence=&Drum_Hit_Tom1_Normal_Medium, startPosition=0.05}
   Blend Poses by bool(
    BlendPose_0="Sequence Player (additive) 'Drum_Hit_Tom1_Slow_Big'"(StartPosition=Get RightArmMaxDelayShortenedBy(), bLoopAnimation=false) {sequence=&Drum_Hit_Tom1_Slow_Big, startPosition=0.05}
    BlendPose_1=Blend Poses by bool(
     BlendPose_0="Sequence Player (additive) 'Drum_Hit_Tom1_Tap_Fast'"(StartPosition=@n1, bLoopAnimation=false) {sequence=&Drum_Hit_Tom1_Tap_Fast, startPosition=0.05}
     BlendPose_1=@n2
     BlendTime_0=0
     BlendTime_1=0
     bActiveValue=Get RightArmIsShortDelay()
    ) {blendType=Linear}
    BlendTime_0=0
    BlendTime_1=0
    bActiveValue=Get RightArmIsLongDelay()
   ) {blendType=Linear}
   AnimGraphNode_StateMachine_0 = Hit_Tom1_R_SM() {is=title AnimGraphNode_StateMachine, bAllowConduitEntryStates=true}
   Output Animation Pose(Result=@AnimGraphNode_StateMachine_0)
  ]
 }
 Hit_Tom1_R_SM {
  entry=Conduit
  path=AnimGraph.AnimGraphNode_StateMachine_6.RightArmAdditiveHits.AnimStateNode_0.RightArmHit_B.AnimGraphNode_StateMachine_1.RightArmHitChooser_StartTime_DelayFloat.AnimStateNode_7.Hit_Tom1_R.AnimGraphNode_StateMachine_0.Hit_Tom1_R_SM
  ids=id
  nodes [
   AnimStateConduitNode_0 = Conduit()
   AnimStateNode_0 = Hit_Short()
   AnimStateNode_1 = Hit_Medium()
   AnimStateNode_2 = Hit_Long()
   AnimStateTransitionNode_4 = AnimStateTransitionNode() {crossfadeDuration=0}
   AnimStateTransitionNode_5 = AnimStateTransitionNode() {crossfadeDuration=0}
   AnimStateTransitionNode_6 = AnimStateTransitionNode() {crossfadeDuration=0}
  ]
  links [
   {
    dst=Hit_Long
    rule=AnimGraph.AnimGraphNode_StateMachine_6.RightArmAdditiveHits.AnimStateNode_0.RightArmHit_B.AnimGraphNode_StateMachine_1.RightArmHitChooser_StartTime_DelayFloat.AnimStateNode_7.Hit_Tom1_R.AnimGraphNode_StateMachine_0.Hit_Tom1_R_SM.AnimStateTransitionNode_6.Transition
    src=Conduit
    via=AnimStateTransitionNode_6
   }
   {
    dst=Hit_Medium
    rule=AnimGraph.AnimGraphNode_StateMachine_6.RightArmAdditiveHits.AnimStateNode_0.RightArmHit_B.AnimGraphNode_StateMachine_1.RightArmHitChooser_StartTime_DelayFloat.AnimStateNode_7.Hit_Tom1_R.AnimGraphNode_StateMachine_0.Hit_Tom1_R_SM.AnimStateTransitionNode_5.Transition
    src=Conduit
    via=AnimStateTransitionNode_5
   }
   {
    dst=Hit_Short
    rule=AnimGraph.AnimGraphNode_StateMachine_6.RightArmAdditiveHits.AnimStateNode_0.RightArmHit_B.AnimGraphNode_StateMachine_1.RightArmHitChooser_StartTime_DelayFloat.AnimStateNode_7.Hit_Tom1_R.AnimGraphNode_StateMachine_0.Hit_Tom1_R_SM.AnimStateTransitionNode_4.Transition
    src=Conduit
    via=AnimStateTransitionNode_4
   }
  ]
 }
 Conduit {
  path=AnimGraph.AnimGraphNode_StateMachine_6.RightArmAdditiveHits.AnimStateNode_0.RightArmHit_B.AnimGraphNode_StateMachine_1.RightArmHitChooser_StartTime_DelayFloat.AnimStateNode_7.Hit_Tom1_R.AnimGraphNode_StateMachine_0.Hit_Tom1_R_SM.AnimStateConduitNode_0.Conduit
  nodes [Result(bCanEnterTransition=true)]
 }
 Hit_Short {
  path=AnimGraph.AnimGraphNode_StateMachine_6.RightArmAdditiveHits.AnimStateNode_0.RightArmHit_B.AnimGraphNode_StateMachine_1.RightArmHitChooser_StartTime_DelayFloat.AnimStateNode_7.Hit_Tom1_R.AnimGraphNode_StateMachine_0.Hit_Tom1_R_SM.AnimStateNode_0.Hit_Short
  nodes [Output Animation Pose(Result="Sequence Player (additive) 'Drum_Hit_Tom1_Tap_Fast'"(StartPosition=Get RightArmMaxDelayShortenedBy(), bLoopAnimation=false) {sequence=&Drum_Hit_Tom1_Tap_Fast, startPosition=0.05})]
 }
 Hit_Medium {
  path=AnimGraph.AnimGraphNode_StateMachine_6.RightArmAdditiveHits.AnimStateNode_0.RightArmHit_B.AnimGraphNode_StateMachine_1.RightArmHitChooser_StartTime_DelayFloat.AnimStateNode_7.Hit_Tom1_R.AnimGraphNode_StateMachine_0.Hit_Tom1_R_SM.AnimStateNode_1.Hit_Medium
  nodes [
   n1 = "Sequence Player (additive) 'Drum_Hit_Tom1_Normal_Medium'"(StartPosition=Get RightArmMaxDelayShortenedBy(), bLoopAnimation=false) {sequence=&Drum_Hit_Tom1_Normal_Medium, startPosition=0.05}
   Output Animation Pose(Result=@n1)
  ]
 }
 Hit_Long {
  path=AnimGraph.AnimGraphNode_StateMachine_6.RightArmAdditiveHits.AnimStateNode_0.RightArmHit_B.AnimGraphNode_StateMachine_1.RightArmHitChooser_StartTime_DelayFloat.AnimStateNode_7.Hit_Tom1_R.AnimGraphNode_StateMachine_0.Hit_Tom1_R_SM.AnimStateNode_2.Hit_Long
  nodes [Output Animation Pose(Result="Sequence Player (additive) 'Drum_Hit_Tom1_Slow_Big'"(StartPosition=Get RightArmMaxDelayShortenedBy(), bLoopAnimation=false) {sequence=&Drum_Hit_Tom1_Slow_Big, startPosition=0.05})]
 }
 Transition {
  path=AnimGraph.AnimGraphNode_StateMachine_6.RightArmAdditiveHits.AnimStateNode_0.RightArmHit_B.AnimGraphNode_StateMachine_1.RightArmHitChooser_StartTime_DelayFloat.AnimStateNode_7.Hit_Tom1_R.AnimGraphNode_StateMachine_0.Hit_Tom1_R_SM.AnimStateTransitionNode_4.Transition
  nodes [Result(bCanEnterTransition=Get RightArmIsShortDelay())]
 }
 Transition {
  path=AnimGraph.AnimGraphNode_StateMachine_6.RightArmAdditiveHits.AnimStateNode_0.RightArmHit_B.AnimGraphNode_StateMachine_1.RightArmHitChooser_StartTime_DelayFloat.AnimStateNode_7.Hit_Tom1_R.AnimGraphNode_StateMachine_0.Hit_Tom1_R_SM.AnimStateTransitionNode_5.Transition
  nodes [
   n1 = Get RightArmIsLongDelay()
   n2 = Get RightArmIsShortDelay()
   Result(bCanEnterTransition=AND Boolean(A=NOT Boolean(A=@n2), B=NOT Boolean(A=@n1)))
  ]
 }
 Transition {
  path=AnimGraph.AnimGraphNode_StateMachine_6.RightArmAdditiveHits.AnimStateNode_0.RightArmHit_B.AnimGraphNode_StateMachine_1.RightArmHitChooser_StartTime_DelayFloat.AnimStateNode_7.Hit_Tom1_R.AnimGraphNode_StateMachine_0.Hit_Tom1_R_SM.AnimStateTransitionNode_6.Transition
  nodes [Result(bCanEnterTransition=Get RightArmIsLongDelay())]
 }
 Hit_Tom2_R {
  path=AnimGraph.AnimGraphNode_StateMachine_6.RightArmAdditiveHits.AnimStateNode_0.RightArmHit_B.AnimGraphNode_StateMachine_1.RightArmHitChooser_StartTime_DelayFloat.AnimStateNode_8.Hit_Tom2_R
  nodes [
   n1 = Get RightArmMaxDelayShortenedBy()
   n2 = "Sequence Player (additive) 'Drum_Hit_Tom2_Normal_Medium'"(StartPosition=Get RightArmMaxDelayShortenedBy(), bLoopAnimation=false) {sequence=&Drum_Hit_Tom2_Normal_Medium, startPosition=0.05}
   Blend Poses by bool(
    BlendPose_0="Sequence Player (additive) 'Drum_Hit_Tom2_Slow_Big'"(StartPosition=Get RightArmMaxDelayShortenedBy(), bLoopAnimation=false) {sequence=&Drum_Hit_Tom2_Slow_Big, startPosition=0.05}
    BlendPose_1=Blend Poses by bool(
     BlendPose_0="Sequence Player (additive) 'Drum_Hit_Tom2_Tap_Fast'"(StartPosition=@n1, bLoopAnimation=false) {sequence=&Drum_Hit_Tom2_Tap_Fast, startPosition=0.05}
     BlendPose_1=@n2
     BlendTime_0=0
     BlendTime_1=0
     bActiveValue=Get RightArmIsShortDelay()
    ) {blendType=Linear}
    BlendTime_0=0
    BlendTime_1=0
    bActiveValue=Get RightArmIsLongDelay()
   ) {blendType=Linear}
   AnimGraphNode_StateMachine_0 = Hit_Tom2_R_SM() {is=title AnimGraphNode_StateMachine, bAllowConduitEntryStates=true}
   Output Animation Pose(Result=@AnimGraphNode_StateMachine_0)
  ]
 }
 Hit_Tom2_R_SM {
  entry=Conduit
  path=AnimGraph.AnimGraphNode_StateMachine_6.RightArmAdditiveHits.AnimStateNode_0.RightArmHit_B.AnimGraphNode_StateMachine_1.RightArmHitChooser_StartTime_DelayFloat.AnimStateNode_8.Hit_Tom2_R.AnimGraphNode_StateMachine_0.Hit_Tom2_R_SM
  ids=id
  nodes [
   AnimStateConduitNode_0 = Conduit()
   AnimStateNode_0 = Hit_Short()
   AnimStateNode_1 = Hit_Medium()
   AnimStateNode_2 = Hit_Long()
   AnimStateTransitionNode_4 = AnimStateTransitionNode() {crossfadeDuration=0}
   AnimStateTransitionNode_5 = AnimStateTransitionNode() {crossfadeDuration=0}
   AnimStateTransitionNode_6 = AnimStateTransitionNode() {crossfadeDuration=0}
  ]
  links [
   {
    dst=Hit_Long
    rule=AnimGraph.AnimGraphNode_StateMachine_6.RightArmAdditiveHits.AnimStateNode_0.RightArmHit_B.AnimGraphNode_StateMachine_1.RightArmHitChooser_StartTime_DelayFloat.AnimStateNode_8.Hit_Tom2_R.AnimGraphNode_StateMachine_0.Hit_Tom2_R_SM.AnimStateTransitionNode_6.Transition
    src=Conduit
    via=AnimStateTransitionNode_6
   }
   {
    dst=Hit_Medium
    rule=AnimGraph.AnimGraphNode_StateMachine_6.RightArmAdditiveHits.AnimStateNode_0.RightArmHit_B.AnimGraphNode_StateMachine_1.RightArmHitChooser_StartTime_DelayFloat.AnimStateNode_8.Hit_Tom2_R.AnimGraphNode_StateMachine_0.Hit_Tom2_R_SM.AnimStateTransitionNode_5.Transition
    src=Conduit
    via=AnimStateTransitionNode_5
   }
   {
    dst=Hit_Short
    rule=AnimGraph.AnimGraphNode_StateMachine_6.RightArmAdditiveHits.AnimStateNode_0.RightArmHit_B.AnimGraphNode_StateMachine_1.RightArmHitChooser_StartTime_DelayFloat.AnimStateNode_8.Hit_Tom2_R.AnimGraphNode_StateMachine_0.Hit_Tom2_R_SM.AnimStateTransitionNode_4.Transition
    src=Conduit
    via=AnimStateTransitionNode_4
   }
  ]
 }
 Conduit {
  path=AnimGraph.AnimGraphNode_StateMachine_6.RightArmAdditiveHits.AnimStateNode_0.RightArmHit_B.AnimGraphNode_StateMachine_1.RightArmHitChooser_StartTime_DelayFloat.AnimStateNode_8.Hit_Tom2_R.AnimGraphNode_StateMachine_0.Hit_Tom2_R_SM.AnimStateConduitNode_0.Conduit
  nodes [Result(bCanEnterTransition=true)]
 }
 Hit_Short {
  path=AnimGraph.AnimGraphNode_StateMachine_6.RightArmAdditiveHits.AnimStateNode_0.RightArmHit_B.AnimGraphNode_StateMachine_1.RightArmHitChooser_StartTime_DelayFloat.AnimStateNode_8.Hit_Tom2_R.AnimGraphNode_StateMachine_0.Hit_Tom2_R_SM.AnimStateNode_0.Hit_Short
  nodes [Output Animation Pose(Result="Sequence Player (additive) 'Drum_Hit_Tom2_Tap_Fast'"(StartPosition=Get RightArmMaxDelayShortenedBy(), bLoopAnimation=false) {sequence=&Drum_Hit_Tom2_Tap_Fast, startPosition=0.05})]
 }
 Hit_Medium {
  path=AnimGraph.AnimGraphNode_StateMachine_6.RightArmAdditiveHits.AnimStateNode_0.RightArmHit_B.AnimGraphNode_StateMachine_1.RightArmHitChooser_StartTime_DelayFloat.AnimStateNode_8.Hit_Tom2_R.AnimGraphNode_StateMachine_0.Hit_Tom2_R_SM.AnimStateNode_1.Hit_Medium
  nodes [
   n1 = "Sequence Player (additive) 'Drum_Hit_Tom2_Normal_Medium'"(StartPosition=Get RightArmMaxDelayShortenedBy(), bLoopAnimation=false) {sequence=&Drum_Hit_Tom2_Normal_Medium, startPosition=0.05}
   Output Animation Pose(Result=@n1)
  ]
 }
 Hit_Long {
  path=AnimGraph.AnimGraphNode_StateMachine_6.RightArmAdditiveHits.AnimStateNode_0.RightArmHit_B.AnimGraphNode_StateMachine_1.RightArmHitChooser_StartTime_DelayFloat.AnimStateNode_8.Hit_Tom2_R.AnimGraphNode_StateMachine_0.Hit_Tom2_R_SM.AnimStateNode_2.Hit_Long
  nodes [Output Animation Pose(Result="Sequence Player (additive) 'Drum_Hit_Tom2_Slow_Big'"(StartPosition=Get RightArmMaxDelayShortenedBy(), bLoopAnimation=false) {sequence=&Drum_Hit_Tom2_Slow_Big, startPosition=0.05})]
 }
 Transition {
  path=AnimGraph.AnimGraphNode_StateMachine_6.RightArmAdditiveHits.AnimStateNode_0.RightArmHit_B.AnimGraphNode_StateMachine_1.RightArmHitChooser_StartTime_DelayFloat.AnimStateNode_8.Hit_Tom2_R.AnimGraphNode_StateMachine_0.Hit_Tom2_R_SM.AnimStateTransitionNode_4.Transition
  nodes [Result(bCanEnterTransition=Get RightArmIsShortDelay())]
 }
 Transition {
  path=AnimGraph.AnimGraphNode_StateMachine_6.RightArmAdditiveHits.AnimStateNode_0.RightArmHit_B.AnimGraphNode_StateMachine_1.RightArmHitChooser_StartTime_DelayFloat.AnimStateNode_8.Hit_Tom2_R.AnimGraphNode_StateMachine_0.Hit_Tom2_R_SM.AnimStateTransitionNode_5.Transition
  nodes [
   n1 = Get RightArmIsLongDelay()
   n2 = Get RightArmIsShortDelay()
   Result(bCanEnterTransition=AND Boolean(A=NOT Boolean(A=@n2), B=NOT Boolean(A=@n1)))
  ]
 }
 Transition {
  path=AnimGraph.AnimGraphNode_StateMachine_6.RightArmAdditiveHits.AnimStateNode_0.RightArmHit_B.AnimGraphNode_StateMachine_1.RightArmHitChooser_StartTime_DelayFloat.AnimStateNode_8.Hit_Tom2_R.AnimGraphNode_StateMachine_0.Hit_Tom2_R_SM.AnimStateTransitionNode_6.Transition
  nodes [Result(bCanEnterTransition=Get RightArmIsLongDelay())]
 }
 Hit_Percussion_R {
  path=AnimGraph.AnimGraphNode_StateMachine_6.RightArmAdditiveHits.AnimStateNode_0.RightArmHit_B.AnimGraphNode_StateMachine_1.RightArmHitChooser_StartTime_DelayFloat.AnimStateNode_9.Hit_Percussion_R
  nodes [Output Animation Pose(Result=Additive Identity Pose())]
 }
 Transition {
  path=AnimGraph.AnimGraphNode_StateMachine_6.RightArmAdditiveHits.AnimStateNode_0.RightArmHit_B.AnimGraphNode_StateMachine_1.RightArmHitChooser_StartTime_DelayFloat.AnimStateTransitionNode_2.Transition
  nodes [Result(bCanEnterTransition=Get Hit_Choke1_R() -> (Hit_Choke1_R:bool))]
 }
 Transition {
  path=AnimGraph.AnimGraphNode_StateMachine_6.RightArmAdditiveHits.AnimStateNode_0.RightArmHit_B.AnimGraphNode_StateMachine_1.RightArmHitChooser_StartTime_DelayFloat.AnimStateTransitionNode_3.Transition
  nodes [Result(bCanEnterTransition=Get Hit_Choke2_R() -> (Hit_Choke2_R:bool))]
 }
 Transition {
  path=AnimGraph.AnimGraphNode_StateMachine_6.RightArmAdditiveHits.AnimStateNode_0.RightArmHit_B.AnimGraphNode_StateMachine_1.RightArmHitChooser_StartTime_DelayFloat.AnimStateTransitionNode_30.Transition
  nodes [Result(bCanEnterTransition=Get Hit_Snare_R() -> (Hit_Snare_R:bool))]
 }
 Transition {
  path=AnimGraph.AnimGraphNode_StateMachine_6.RightArmAdditiveHits.AnimStateNode_0.RightArmHit_B.AnimGraphNode_StateMachine_1.RightArmHitChooser_StartTime_DelayFloat.AnimStateTransitionNode_32.Transition
  nodes [Result(bCanEnterTransition=Get Hit_HiHat_R())]
 }
 Transition {
  path=AnimGraph.AnimGraphNode_StateMachine_6.RightArmAdditiveHits.AnimStateNode_0.RightArmHit_B.AnimGraphNode_StateMachine_1.RightArmHitChooser_StartTime_DelayFloat.AnimStateTransitionNode_34.Transition
  nodes [Result(bCanEnterTransition=Get Hit_Crash1_Hard_R() -> (Hit_Crash1_Hard_R:bool))]
 }
 Transition {
  path=AnimGraph.AnimGraphNode_StateMachine_6.RightArmAdditiveHits.AnimStateNode_0.RightArmHit_B.AnimGraphNode_StateMachine_1.RightArmHitChooser_StartTime_DelayFloat.AnimStateTransitionNode_36.Transition
  nodes [Result(bCanEnterTransition=Get Hit_Crash1_Soft_R() -> (Hit_Crash1_Soft_R:bool))]
 }
 Transition {
  path=AnimGraph.AnimGraphNode_StateMachine_6.RightArmAdditiveHits.AnimStateNode_0.RightArmHit_B.AnimGraphNode_StateMachine_1.RightArmHitChooser_StartTime_DelayFloat.AnimStateTransitionNode_38.Transition
  nodes [Result(bCanEnterTransition=Get Hit_Crash2_Hard_R() -> (Hit_Crash2_Hard_R:bool))]
 }
 Transition {
  path=AnimGraph.AnimGraphNode_StateMachine_6.RightArmAdditiveHits.AnimStateNode_0.RightArmHit_B.AnimGraphNode_StateMachine_1.RightArmHitChooser_StartTime_DelayFloat.AnimStateTransitionNode_40.Transition
  nodes [Result(bCanEnterTransition=Get Hit_Crash2_Soft_R() -> (Hit_Crash2_Soft_R:bool))]
 }
 Transition {
  path=AnimGraph.AnimGraphNode_StateMachine_6.RightArmAdditiveHits.AnimStateNode_0.RightArmHit_B.AnimGraphNode_StateMachine_1.RightArmHitChooser_StartTime_DelayFloat.AnimStateTransitionNode_42.Transition
  nodes [Result(bCanEnterTransition=Get Hit_Ride_R() -> (Hit_Ride_R:bool))]
 }
 Transition {
  path=AnimGraph.AnimGraphNode_StateMachine_6.RightArmAdditiveHits.AnimStateNode_0.RightArmHit_B.AnimGraphNode_StateMachine_1.RightArmHitChooser_StartTime_DelayFloat.AnimStateTransitionNode_44.Transition
  nodes [Result(bCanEnterTransition=Get Hit_Tom1_R() -> (Hit_Tom1_R:bool))]
 }
 Transition {
  path=AnimGraph.AnimGraphNode_StateMachine_6.RightArmAdditiveHits.AnimStateNode_0.RightArmHit_B.AnimGraphNode_StateMachine_1.RightArmHitChooser_StartTime_DelayFloat.AnimStateTransitionNode_46.Transition
  nodes [Result(bCanEnterTransition=Get Hit_Tom2_R() -> (Hit_Tom2_R:bool))]
 }
 Transition {
  path=AnimGraph.AnimGraphNode_StateMachine_6.RightArmAdditiveHits.AnimStateNode_0.RightArmHit_B.AnimGraphNode_StateMachine_1.RightArmHitChooser_StartTime_DelayFloat.AnimStateTransitionNode_48.Transition
  nodes [Result(bCanEnterTransition=Get Hit_FloorTom_R() -> (Hit_FloorTom_R:bool))]
 }
 Transition {
  path=AnimGraph.AnimGraphNode_StateMachine_6.RightArmAdditiveHits.AnimStateNode_0.RightArmHit_B.AnimGraphNode_StateMachine_1.RightArmHitChooser_StartTime_DelayFloat.AnimStateTransitionNode_50.Transition
  nodes [Result(bCanEnterTransition=Get Hit_Percussion_R() -> (Hit_Percussion_R:bool))]
 }
 Transition {
  path=AnimGraph.AnimGraphNode_StateMachine_6.RightArmAdditiveHits.AnimStateNode_0.RightArmHit_B.AnimGraphNode_StateMachine_1.RightArmHitChooser_StartTime_DelayFloat.AnimStateTransitionNode_6.Transition
  nodes [Result(bCanEnterTransition=Get Hit_Snare_Soft_R() -> (Hit_Snare_Soft_R:bool))]
 }
 RightArmHit_A {
  path=AnimGraph.AnimGraphNode_StateMachine_6.RightArmAdditiveHits.AnimStateNode_1.RightArmHit_A
  nodes [
   AnimGraphNode_StateMachine_1 = RightArmHitChooser_StartTime_DelayFloat() {bSkipFirstUpdateTransition=false}
   Output Animation Pose(Result=@AnimGraphNode_StateMachine_1)
  ]
 }
 RightArmHitChooser_StartTime_DelayFloat {
  entry=Identity
  path=AnimGraph.AnimGraphNode_StateMachine_6.RightArmAdditiveHits.AnimStateNode_1.RightArmHit_A.AnimGraphNode_StateMachine_1.RightArmHitChooser_StartTime_DelayFloat
  ids=id
  nodes [
   AnimStateNode_0 = Hit_Snare_R()
   AnimStateNode_1 = Hit_HiHat_R()
   AnimStateNode_10 = Identity()
   AnimStateNode_11 = Hit_FloorTom_R()
   AnimStateNode_12 = Hit_Crash2_Choke_R() {
    stateEntered {guid=#49}
    stateFullyBlended {guid=#50}
    stateLeft {guid=#51}
   }
   AnimStateNode_13 = Hit_Crash1_Hard_R()
   AnimStateNode_14 = Hit_Crash2_Hard_R()
   AnimStateNode_2 = Hit_Crash1_Choke_R() {
    stateEntered {guid=#52}
    stateFullyBlended {guid=#53}
    stateLeft {guid=#54}
   }
   AnimStateNode_3 = Hit_Crash1_Soft_R()
   AnimStateNode_4 = Hit_Snare_Soft_R()
   AnimStateNode_5 = Hit_Crash2_Soft_R()
   AnimStateNode_6 = Hit_Ride_R()
   AnimStateNode_7 = Hit_Tom1_R()
   AnimStateNode_8 = Hit_Tom2_R()
   AnimStateNode_9 = Hit_Percussion_R()
   AnimStateTransitionNode_0 = AnimStateTransitionNode()
   AnimStateTransitionNode_1 = AnimStateTransitionNode()
   AnimStateTransitionNode_30 = AnimStateTransitionNode() {crossfadeDuration=0}
   AnimStateTransitionNode_32 = AnimStateTransitionNode() {crossfadeDuration=0}
   AnimStateTransitionNode_34 = AnimStateTransitionNode() {crossfadeDuration=0}
   AnimStateTransitionNode_36 = AnimStateTransitionNode() {crossfadeDuration=0}
   AnimStateTransitionNode_38 = AnimStateTransitionNode() {crossfadeDuration=0}
   AnimStateTransitionNode_40 = AnimStateTransitionNode() {crossfadeDuration=0}
   AnimStateTransitionNode_42 = AnimStateTransitionNode() {crossfadeDuration=0}
   AnimStateTransitionNode_44 = AnimStateTransitionNode() {crossfadeDuration=0}
   AnimStateTransitionNode_46 = AnimStateTransitionNode() {crossfadeDuration=0}
   AnimStateTransitionNode_48 = AnimStateTransitionNode() {crossfadeDuration=0}
   AnimStateTransitionNode_50 = AnimStateTransitionNode() {crossfadeDuration=0}
   AnimStateTransitionNode_7 = AnimStateTransitionNode() {crossfadeDuration=0}
  ]
  links [
   {
    dst=Hit_Crash1_Choke_R
    rule=AnimGraph.AnimGraphNode_StateMachine_6.RightArmAdditiveHits.AnimStateNode_1.RightArmHit_A.AnimGraphNode_StateMachine_1.RightArmHitChooser_StartTime_DelayFloat.AnimStateTransitionNode_0.Transition
    src=Identity
    via=AnimStateTransitionNode_0
   }
   {
    dst=Hit_Crash1_Hard_R
    rule=AnimGraph.AnimGraphNode_StateMachine_6.RightArmAdditiveHits.AnimStateNode_1.RightArmHit_A.AnimGraphNode_StateMachine_1.RightArmHitChooser_StartTime_DelayFloat.AnimStateTransitionNode_34.Transition
    src=Identity
    via=AnimStateTransitionNode_34
   }
   {
    dst=Hit_Crash1_Soft_R
    rule=AnimGraph.AnimGraphNode_StateMachine_6.RightArmAdditiveHits.AnimStateNode_1.RightArmHit_A.AnimGraphNode_StateMachine_1.RightArmHitChooser_StartTime_DelayFloat.AnimStateTransitionNode_36.Transition
    src=Identity
    via=AnimStateTransitionNode_36
   }
   {
    dst=Hit_Crash2_Choke_R
    rule=AnimGraph.AnimGraphNode_StateMachine_6.RightArmAdditiveHits.AnimStateNode_1.RightArmHit_A.AnimGraphNode_StateMachine_1.RightArmHitChooser_StartTime_DelayFloat.AnimStateTransitionNode_1.Transition
    src=Identity
    via=AnimStateTransitionNode_1
   }
   {
    dst=Hit_Crash2_Hard_R
    rule=AnimGraph.AnimGraphNode_StateMachine_6.RightArmAdditiveHits.AnimStateNode_1.RightArmHit_A.AnimGraphNode_StateMachine_1.RightArmHitChooser_StartTime_DelayFloat.AnimStateTransitionNode_38.Transition
    src=Identity
    via=AnimStateTransitionNode_38
   }
   {
    dst=Hit_Crash2_Soft_R
    rule=AnimGraph.AnimGraphNode_StateMachine_6.RightArmAdditiveHits.AnimStateNode_1.RightArmHit_A.AnimGraphNode_StateMachine_1.RightArmHitChooser_StartTime_DelayFloat.AnimStateTransitionNode_40.Transition
    src=Identity
    via=AnimStateTransitionNode_40
   }
   {
    dst=Hit_FloorTom_R
    rule=AnimGraph.AnimGraphNode_StateMachine_6.RightArmAdditiveHits.AnimStateNode_1.RightArmHit_A.AnimGraphNode_StateMachine_1.RightArmHitChooser_StartTime_DelayFloat.AnimStateTransitionNode_48.Transition
    src=Identity
    via=AnimStateTransitionNode_48
   }
   {
    dst=Hit_HiHat_R
    rule=AnimGraph.AnimGraphNode_StateMachine_6.RightArmAdditiveHits.AnimStateNode_1.RightArmHit_A.AnimGraphNode_StateMachine_1.RightArmHitChooser_StartTime_DelayFloat.AnimStateTransitionNode_32.Transition
    src=Identity
    via=AnimStateTransitionNode_32
   }
   {
    dst=Hit_Percussion_R
    rule=AnimGraph.AnimGraphNode_StateMachine_6.RightArmAdditiveHits.AnimStateNode_1.RightArmHit_A.AnimGraphNode_StateMachine_1.RightArmHitChooser_StartTime_DelayFloat.AnimStateTransitionNode_50.Transition
    src=Identity
    via=AnimStateTransitionNode_50
   }
   {
    dst=Hit_Ride_R
    rule=AnimGraph.AnimGraphNode_StateMachine_6.RightArmAdditiveHits.AnimStateNode_1.RightArmHit_A.AnimGraphNode_StateMachine_1.RightArmHitChooser_StartTime_DelayFloat.AnimStateTransitionNode_42.Transition
    src=Identity
    via=AnimStateTransitionNode_42
   }
   {
    dst=Hit_Snare_R
    rule=AnimGraph.AnimGraphNode_StateMachine_6.RightArmAdditiveHits.AnimStateNode_1.RightArmHit_A.AnimGraphNode_StateMachine_1.RightArmHitChooser_StartTime_DelayFloat.AnimStateTransitionNode_30.Transition
    src=Identity
    via=AnimStateTransitionNode_30
   }
   {
    dst=Hit_Snare_Soft_R
    rule=AnimGraph.AnimGraphNode_StateMachine_6.RightArmAdditiveHits.AnimStateNode_1.RightArmHit_A.AnimGraphNode_StateMachine_1.RightArmHitChooser_StartTime_DelayFloat.AnimStateTransitionNode_7.Transition
    src=Identity
    via=AnimStateTransitionNode_7
   }
   {
    dst=Hit_Tom1_R
    rule=AnimGraph.AnimGraphNode_StateMachine_6.RightArmAdditiveHits.AnimStateNode_1.RightArmHit_A.AnimGraphNode_StateMachine_1.RightArmHitChooser_StartTime_DelayFloat.AnimStateTransitionNode_44.Transition
    src=Identity
    via=AnimStateTransitionNode_44
   }
   {
    dst=Hit_Tom2_R
    rule=AnimGraph.AnimGraphNode_StateMachine_6.RightArmAdditiveHits.AnimStateNode_1.RightArmHit_A.AnimGraphNode_StateMachine_1.RightArmHitChooser_StartTime_DelayFloat.AnimStateTransitionNode_46.Transition
    src=Identity
    via=AnimStateTransitionNode_46
   }
  ]
 }
 Hit_Snare_R {
  path=AnimGraph.AnimGraphNode_StateMachine_6.RightArmAdditiveHits.AnimStateNode_1.RightArmHit_A.AnimGraphNode_StateMachine_1.RightArmHitChooser_StartTime_DelayFloat.AnimStateNode_0.Hit_Snare_R
  nodes [
   n1 = Get RightArmMaxDelayShortenedBy()
   n2 = "Sequence Player (additive) 'Drum_Hit_Snare_Normal_Medium'"(StartPosition=Get RightArmMaxDelayShortenedBy(), bLoopAnimation=false) {sequence=&Drum_Hit_Snare_Normal_Medium, startPosition=0.05}
   Blend Poses by bool(
    BlendPose_0="Sequence Player (additive) 'Drum_Hit_Snare_Slow_Big'"(StartPosition=Get RightArmMaxDelayShortenedBy(), bLoopAnimation=false) {sequence=&Drum_Hit_Snare_Slow_Big, startPosition=0.05}
    BlendPose_1=Blend Poses by bool(
     BlendPose_0="Sequence Player (additive) 'Drum_Hit_Snare_Tap_Fast'"(StartPosition=@n1, bLoopAnimation=false) {sequence=&Drum_Hit_Snare_Tap_Fast, startPosition=0.05}
     BlendPose_1=@n2
     BlendTime_0=0
     BlendTime_1=0
     bActiveValue=Get RightArmIsShortDelay()
    ) {blendType=Linear}
    BlendTime_0=0
    BlendTime_1=0
    bActiveValue=Get RightArmIsLongDelay()
   ) {blendType=Linear}
   AnimGraphNode_StateMachine_0 = Hit_Snare_R_SM() {bAllowConduitEntryStates=true}
   Output Animation Pose(Result=@AnimGraphNode_StateMachine_0)
  ]
 }
 Hit_Snare_R_SM {
  entry=Conduit
  path=AnimGraph.AnimGraphNode_StateMachine_6.RightArmAdditiveHits.AnimStateNode_1.RightArmHit_A.AnimGraphNode_StateMachine_1.RightArmHitChooser_StartTime_DelayFloat.AnimStateNode_0.Hit_Snare_R.AnimGraphNode_StateMachine_0.Hit_Snare_R_SM
  ids=id
  nodes [
   AnimStateConduitNode_0 = Conduit()
   AnimStateNode_0 = Hit_Short()
   AnimStateNode_1 = Hit_Medium()
   AnimStateNode_2 = Hit_Long()
   AnimStateTransitionNode_4 = AnimStateTransitionNode() {crossfadeDuration=0, priorityOrder=2}
   AnimStateTransitionNode_5 = AnimStateTransitionNode() {crossfadeDuration=0, priorityOrder=2}
   AnimStateTransitionNode_6 = AnimStateTransitionNode() {crossfadeDuration=0, priorityOrder=2}
  ]
  links [
   {
    dst=Hit_Long
    rule=AnimGraph.AnimGraphNode_StateMachine_6.RightArmAdditiveHits.AnimStateNode_1.RightArmHit_A.AnimGraphNode_StateMachine_1.RightArmHitChooser_StartTime_DelayFloat.AnimStateNode_0.Hit_Snare_R.AnimGraphNode_StateMachine_0.Hit_Snare_R_SM.AnimStateTransitionNode_6.Transition
    src=Conduit
    via=AnimStateTransitionNode_6
   }
   {
    dst=Hit_Medium
    rule=AnimGraph.AnimGraphNode_StateMachine_6.RightArmAdditiveHits.AnimStateNode_1.RightArmHit_A.AnimGraphNode_StateMachine_1.RightArmHitChooser_StartTime_DelayFloat.AnimStateNode_0.Hit_Snare_R.AnimGraphNode_StateMachine_0.Hit_Snare_R_SM.AnimStateTransitionNode_5.Transition
    src=Conduit
    via=AnimStateTransitionNode_5
   }
   {
    dst=Hit_Short
    rule=AnimGraph.AnimGraphNode_StateMachine_6.RightArmAdditiveHits.AnimStateNode_1.RightArmHit_A.AnimGraphNode_StateMachine_1.RightArmHitChooser_StartTime_DelayFloat.AnimStateNode_0.Hit_Snare_R.AnimGraphNode_StateMachine_0.Hit_Snare_R_SM.AnimStateTransitionNode_4.Transition
    src=Conduit
    via=AnimStateTransitionNode_4
   }
  ]
 }
 Conduit {
  path=AnimGraph.AnimGraphNode_StateMachine_6.RightArmAdditiveHits.AnimStateNode_1.RightArmHit_A.AnimGraphNode_StateMachine_1.RightArmHitChooser_StartTime_DelayFloat.AnimStateNode_0.Hit_Snare_R.AnimGraphNode_StateMachine_0.Hit_Snare_R_SM.AnimStateConduitNode_0.Conduit
  nodes [Result(bCanEnterTransition=true)]
 }
 Hit_Short {
  path=AnimGraph.AnimGraphNode_StateMachine_6.RightArmAdditiveHits.AnimStateNode_1.RightArmHit_A.AnimGraphNode_StateMachine_1.RightArmHitChooser_StartTime_DelayFloat.AnimStateNode_0.Hit_Snare_R.AnimGraphNode_StateMachine_0.Hit_Snare_R_SM.AnimStateNode_0.Hit_Short
  nodes [Output Animation Pose(Result="Sequence Player (additive) 'Drum_Hit_Snare_Tap_Fast'"(StartPosition=Get RightArmMaxDelayShortenedBy(), bLoopAnimation=false) {sequence=&Drum_Hit_Snare_Tap_Fast, startPosition=0.05})]
 }
 Hit_Medium {
  path=AnimGraph.AnimGraphNode_StateMachine_6.RightArmAdditiveHits.AnimStateNode_1.RightArmHit_A.AnimGraphNode_StateMachine_1.RightArmHitChooser_StartTime_DelayFloat.AnimStateNode_0.Hit_Snare_R.AnimGraphNode_StateMachine_0.Hit_Snare_R_SM.AnimStateNode_1.Hit_Medium
  nodes [
   n1 = "Sequence Player (additive) 'Drum_Hit_Snare_Normal_Medium'"(StartPosition=Get RightArmMaxDelayShortenedBy(), bLoopAnimation=false) {sequence=&Drum_Hit_Snare_Normal_Medium, startPosition=0.05}
   Output Animation Pose(Result=@n1)
  ]
 }
 Hit_Long {
  path=AnimGraph.AnimGraphNode_StateMachine_6.RightArmAdditiveHits.AnimStateNode_1.RightArmHit_A.AnimGraphNode_StateMachine_1.RightArmHitChooser_StartTime_DelayFloat.AnimStateNode_0.Hit_Snare_R.AnimGraphNode_StateMachine_0.Hit_Snare_R_SM.AnimStateNode_2.Hit_Long
  nodes [Output Animation Pose(Result="Sequence Player (additive) 'Drum_Hit_Snare_Slow_Big'"(StartPosition=Get RightArmMaxDelayShortenedBy(), bLoopAnimation=false) {sequence=&Drum_Hit_Snare_Slow_Big, startPosition=0.05})]
 }
 Transition {
  path=AnimGraph.AnimGraphNode_StateMachine_6.RightArmAdditiveHits.AnimStateNode_1.RightArmHit_A.AnimGraphNode_StateMachine_1.RightArmHitChooser_StartTime_DelayFloat.AnimStateNode_0.Hit_Snare_R.AnimGraphNode_StateMachine_0.Hit_Snare_R_SM.AnimStateTransitionNode_4.Transition
  nodes [Result(bCanEnterTransition=Get RightArmIsShortDelay())]
 }
 Transition {
  path=AnimGraph.AnimGraphNode_StateMachine_6.RightArmAdditiveHits.AnimStateNode_1.RightArmHit_A.AnimGraphNode_StateMachine_1.RightArmHitChooser_StartTime_DelayFloat.AnimStateNode_0.Hit_Snare_R.AnimGraphNode_StateMachine_0.Hit_Snare_R_SM.AnimStateTransitionNode_5.Transition
  nodes [
   n1 = Get RightArmIsLongDelay()
   n2 = Get RightArmIsShortDelay()
   Result(bCanEnterTransition=AND Boolean(A=NOT Boolean(A=@n2), B=NOT Boolean(A=@n1)))
  ]
 }
 Transition {
  path=AnimGraph.AnimGraphNode_StateMachine_6.RightArmAdditiveHits.AnimStateNode_1.RightArmHit_A.AnimGraphNode_StateMachine_1.RightArmHitChooser_StartTime_DelayFloat.AnimStateNode_0.Hit_Snare_R.AnimGraphNode_StateMachine_0.Hit_Snare_R_SM.AnimStateTransitionNode_6.Transition
  nodes [Result(bCanEnterTransition=Get RightArmIsLongDelay())]
 }
 Hit_HiHat_R {
  path=AnimGraph.AnimGraphNode_StateMachine_6.RightArmAdditiveHits.AnimStateNode_1.RightArmHit_A.AnimGraphNode_StateMachine_1.RightArmHitChooser_StartTime_DelayFloat.AnimStateNode_1.Hit_HiHat_R
  nodes [
   AnimGraphNode_StateMachine_0 = Hit_HiHat_R_SM() {bAllowConduitEntryStates=true}
   Output Animation Pose(Result=@AnimGraphNode_StateMachine_0)
  ]
 }
 Hit_HiHat_R_SM {
  entry=Conduit
  path=AnimGraph.AnimGraphNode_StateMachine_6.RightArmAdditiveHits.AnimStateNode_1.RightArmHit_A.AnimGraphNode_StateMachine_1.RightArmHitChooser_StartTime_DelayFloat.AnimStateNode_1.Hit_HiHat_R.AnimGraphNode_StateMachine_0.Hit_HiHat_R_SM
  ids=id
  nodes [
   AnimStateConduitNode_0 = Conduit()
   AnimStateNode_0 = Hit_Short()
   AnimStateNode_2 = Hit_Long()
   AnimStateNode_3 = Hit_Medium()
   AnimStateTransitionNode_4 = AnimStateTransitionNode() {crossfadeDuration=0}
   AnimStateTransitionNode_5 = AnimStateTransitionNode() {crossfadeDuration=0}
   AnimStateTransitionNode_6 = AnimStateTransitionNode() {crossfadeDuration=0}
  ]
  links [
   {
    dst=Hit_Long
    rule=AnimGraph.AnimGraphNode_StateMachine_6.RightArmAdditiveHits.AnimStateNode_1.RightArmHit_A.AnimGraphNode_StateMachine_1.RightArmHitChooser_StartTime_DelayFloat.AnimStateNode_1.Hit_HiHat_R.AnimGraphNode_StateMachine_0.Hit_HiHat_R_SM.AnimStateTransitionNode_6.Transition
    src=Conduit
    via=AnimStateTransitionNode_6
   }
   {
    dst=Hit_Medium
    rule=AnimGraph.AnimGraphNode_StateMachine_6.RightArmAdditiveHits.AnimStateNode_1.RightArmHit_A.AnimGraphNode_StateMachine_1.RightArmHitChooser_StartTime_DelayFloat.AnimStateNode_1.Hit_HiHat_R.AnimGraphNode_StateMachine_0.Hit_HiHat_R_SM.AnimStateTransitionNode_5.Transition
    src=Conduit
    via=AnimStateTransitionNode_5
   }
   {
    dst=Hit_Short
    rule=AnimGraph.AnimGraphNode_StateMachine_6.RightArmAdditiveHits.AnimStateNode_1.RightArmHit_A.AnimGraphNode_StateMachine_1.RightArmHitChooser_StartTime_DelayFloat.AnimStateNode_1.Hit_HiHat_R.AnimGraphNode_StateMachine_0.Hit_HiHat_R_SM.AnimStateTransitionNode_4.Transition
    src=Conduit
    via=AnimStateTransitionNode_4
   }
  ]
 }
 Conduit {
  path=AnimGraph.AnimGraphNode_StateMachine_6.RightArmAdditiveHits.AnimStateNode_1.RightArmHit_A.AnimGraphNode_StateMachine_1.RightArmHitChooser_StartTime_DelayFloat.AnimStateNode_1.Hit_HiHat_R.AnimGraphNode_StateMachine_0.Hit_HiHat_R_SM.AnimStateConduitNode_0.Conduit
  nodes [Result(bCanEnterTransition=true)]
 }
 Hit_Short {
  path=AnimGraph.AnimGraphNode_StateMachine_6.RightArmAdditiveHits.AnimStateNode_1.RightArmHit_A.AnimGraphNode_StateMachine_1.RightArmHitChooser_StartTime_DelayFloat.AnimStateNode_1.Hit_HiHat_R.AnimGraphNode_StateMachine_0.Hit_HiHat_R_SM.AnimStateNode_0.Hit_Short
  nodes [
   n1 = Get RightArmMaxDelayShortenedBy()
   n2 = "Sequence Player (additive) 'Drum_Hit_CrossHiHatSnare_Tap_Fast'"(StartPosition=@n1, bLoopAnimation=False) {bLoopAnimation=false, sequence=&Drum_Hit_CrossHiHatSnare_Tap_Fast}
   n3 = Get Hit_Snare_L()
   n4 = Get Hit_Snare_Soft_L()
   Output Animation Pose(Result=Blend Poses by bool(
     BlendPose_0=@n2
     BlendPose_1="Sequence Player (additive) 'Drum_Hit_HiHat_Tap_Fast'"(StartPosition=@n1, bLoopAnimation=false) {sequence=&Drum_Hit_HiHat_Tap_Fast, startPosition=0.05}
     bActiveValue=OR Boolean(A=@n3, B=@n4)
    ) {blendType=Linear})
  ]
 }
 Hit_Long {
  path=AnimGraph.AnimGraphNode_StateMachine_6.RightArmAdditiveHits.AnimStateNode_1.RightArmHit_A.AnimGraphNode_StateMachine_1.RightArmHitChooser_StartTime_DelayFloat.AnimStateNode_1.Hit_HiHat_R.AnimGraphNode_StateMachine_0.Hit_HiHat_R_SM.AnimStateNode_2.Hit_Long
  nodes [Output Animation Pose(Result="Sequence Player (additive) 'Drum_Hit_HiHat_Slow_Big'"(StartPosition=Get RightArmMaxDelayShortenedBy(), bLoopAnimation=false) {sequence=&Drum_Hit_HiHat_Slow_Big, startPosition=0.05})]
 }
 Hit_Medium {
  path=AnimGraph.AnimGraphNode_StateMachine_6.RightArmAdditiveHits.AnimStateNode_1.RightArmHit_A.AnimGraphNode_StateMachine_1.RightArmHitChooser_StartTime_DelayFloat.AnimStateNode_1.Hit_HiHat_R.AnimGraphNode_StateMachine_0.Hit_HiHat_R_SM.AnimStateNode_3.Hit_Medium
  nodes [
   n1 = Get RightArmMaxDelayShortenedBy()
   n2 = "Sequence Player (additive) 'Drum_Hit_CrossHiHatSnare_Medium'"(StartPosition=@n1, bLoopAnimation=False) {bLoopAnimation=false, sequence=&Drum_Hit_CrossHiHatSnare_Medium}
   n3 = "Sequence Player (additive) 'Drum_Hit_HiHat_Normal_Medium'"(StartPosition=@n1, bLoopAnimation=false) {sequence=&Drum_Hit_HiHat_Normal_Medium, startPosition=0.05}
   n4 = Get Hit_Snare_L()
   n5 = Get Hit_Snare_Soft_L()
   Output Animation Pose(Result=Blend Poses by bool(BlendPose_0=@n2, BlendPose_1=@n3, bActiveValue=OR Boolean(A=@n4, B=@n5)) {blendType=Linear})
  ]
 }
 Transition {
  path=AnimGraph.AnimGraphNode_StateMachine_6.RightArmAdditiveHits.AnimStateNode_1.RightArmHit_A.AnimGraphNode_StateMachine_1.RightArmHitChooser_StartTime_DelayFloat.AnimStateNode_1.Hit_HiHat_R.AnimGraphNode_StateMachine_0.Hit_HiHat_R_SM.AnimStateTransitionNode_4.Transition
  nodes [Result(bCanEnterTransition=Get RightArmIsShortDelay())]
 }
 Transition {
  path=AnimGraph.AnimGraphNode_StateMachine_6.RightArmAdditiveHits.AnimStateNode_1.RightArmHit_A.AnimGraphNode_StateMachine_1.RightArmHitChooser_StartTime_DelayFloat.AnimStateNode_1.Hit_HiHat_R.AnimGraphNode_StateMachine_0.Hit_HiHat_R_SM.AnimStateTransitionNode_5.Transition
  nodes [
   n1 = Get RightArmIsLongDelay()
   n2 = Get RightArmIsShortDelay()
   Result(bCanEnterTransition=AND Boolean(A=NOT Boolean(A=@n2), B=NOT Boolean(A=@n1)))
  ]
 }
 Transition {
  path=AnimGraph.AnimGraphNode_StateMachine_6.RightArmAdditiveHits.AnimStateNode_1.RightArmHit_A.AnimGraphNode_StateMachine_1.RightArmHitChooser_StartTime_DelayFloat.AnimStateNode_1.Hit_HiHat_R.AnimGraphNode_StateMachine_0.Hit_HiHat_R_SM.AnimStateTransitionNode_6.Transition
  nodes [Result(bCanEnterTransition=Get RightArmIsLongDelay())]
 }
 Identity {
  path=AnimGraph.AnimGraphNode_StateMachine_6.RightArmAdditiveHits.AnimStateNode_1.RightArmHit_A.AnimGraphNode_StateMachine_1.RightArmHitChooser_StartTime_DelayFloat.AnimStateNode_10.Identity
  nodes [
   "Sequence Player (additive) 'Drum_SpineTwist_Center_Pose'"() {bLoopAnimation=false, sequence=&Drum_SpineTwist_Center_Pose}
   Output Animation Pose(Result=Additive Identity Pose())
  ]
 }
 Hit_FloorTom_R {
  path=AnimGraph.AnimGraphNode_StateMachine_6.RightArmAdditiveHits.AnimStateNode_1.RightArmHit_A.AnimGraphNode_StateMachine_1.RightArmHitChooser_StartTime_DelayFloat.AnimStateNode_11.Hit_FloorTom_R
  nodes [
   n1 = Get RightArmMaxDelayShortenedBy()
   n2 = "Sequence Player (additive) 'Drum_Hit_FloorTom_Normal_Medium'"(StartPosition=Get RightArmMaxDelayShortenedBy(), bLoopAnimation=false) {sequence=&Drum_Hit_FloorTom_Normal_Medium, startPosition=0.05}
   Blend Poses by bool(
    BlendPose_0="Sequence Player (additive) 'Drum_Hit_FloorTom_Slow_Big'"(StartPosition=Get RightArmMaxDelayShortenedBy(), bLoopAnimation=false) {sequence=&Drum_Hit_FloorTom_Slow_Big, startPosition=0.05}
    BlendPose_1=Blend Poses by bool(
     BlendPose_0="Sequence Player (additive) 'Drum_Hit_FloorTom_Tap_Fast'"(StartPosition=@n1, bLoopAnimation=false) {sequence=&Drum_Hit_FloorTom_Tap_Fast, startPosition=0.05}
     BlendPose_1=@n2
     BlendTime_0=0
     BlendTime_1=0
     bActiveValue=Get RightArmIsShortDelay()
    ) {blendType=Linear}
    BlendTime_0=0
    BlendTime_1=0
    bActiveValue=Get RightArmIsLongDelay()
   ) {blendType=Linear}
   AnimGraphNode_StateMachine_0 = Hit_FloorTom_R_SM() {bAllowConduitEntryStates=true}
   Output Animation Pose(Result=@AnimGraphNode_StateMachine_0)
  ]
 }
 Hit_FloorTom_R_SM {
  entry=Conduit
  path=AnimGraph.AnimGraphNode_StateMachine_6.RightArmAdditiveHits.AnimStateNode_1.RightArmHit_A.AnimGraphNode_StateMachine_1.RightArmHitChooser_StartTime_DelayFloat.AnimStateNode_11.Hit_FloorTom_R.AnimGraphNode_StateMachine_0.Hit_FloorTom_R_SM
  ids=id
  nodes [
   AnimStateConduitNode_0 = Conduit()
   AnimStateNode_0 = Hit_Short()
   AnimStateNode_1 = Hit_Medium()
   AnimStateNode_2 = Hit_Long()
   AnimStateTransitionNode_4 = AnimStateTransitionNode() {crossfadeDuration=0}
   AnimStateTransitionNode_5 = AnimStateTransitionNode() {crossfadeDuration=0}
   AnimStateTransitionNode_6 = AnimStateTransitionNode() {crossfadeDuration=0}
  ]
  links [
   {
    dst=Hit_Long
    rule=AnimGraph.AnimGraphNode_StateMachine_6.RightArmAdditiveHits.AnimStateNode_1.RightArmHit_A.AnimGraphNode_StateMachine_1.RightArmHitChooser_StartTime_DelayFloat.AnimStateNode_11.Hit_FloorTom_R.AnimGraphNode_StateMachine_0.Hit_FloorTom_R_SM.AnimStateTransitionNode_6.Transition
    src=Conduit
    via=AnimStateTransitionNode_6
   }
   {
    dst=Hit_Medium
    rule=AnimGraph.AnimGraphNode_StateMachine_6.RightArmAdditiveHits.AnimStateNode_1.RightArmHit_A.AnimGraphNode_StateMachine_1.RightArmHitChooser_StartTime_DelayFloat.AnimStateNode_11.Hit_FloorTom_R.AnimGraphNode_StateMachine_0.Hit_FloorTom_R_SM.AnimStateTransitionNode_5.Transition
    src=Conduit
    via=AnimStateTransitionNode_5
   }
   {
    dst=Hit_Short
    rule=AnimGraph.AnimGraphNode_StateMachine_6.RightArmAdditiveHits.AnimStateNode_1.RightArmHit_A.AnimGraphNode_StateMachine_1.RightArmHitChooser_StartTime_DelayFloat.AnimStateNode_11.Hit_FloorTom_R.AnimGraphNode_StateMachine_0.Hit_FloorTom_R_SM.AnimStateTransitionNode_4.Transition
    src=Conduit
    via=AnimStateTransitionNode_4
   }
  ]
 }
 Conduit {
  path=AnimGraph.AnimGraphNode_StateMachine_6.RightArmAdditiveHits.AnimStateNode_1.RightArmHit_A.AnimGraphNode_StateMachine_1.RightArmHitChooser_StartTime_DelayFloat.AnimStateNode_11.Hit_FloorTom_R.AnimGraphNode_StateMachine_0.Hit_FloorTom_R_SM.AnimStateConduitNode_0.Conduit
  nodes [Result(bCanEnterTransition=true)]
 }
 Hit_Short {
  path=AnimGraph.AnimGraphNode_StateMachine_6.RightArmAdditiveHits.AnimStateNode_1.RightArmHit_A.AnimGraphNode_StateMachine_1.RightArmHitChooser_StartTime_DelayFloat.AnimStateNode_11.Hit_FloorTom_R.AnimGraphNode_StateMachine_0.Hit_FloorTom_R_SM.AnimStateNode_0.Hit_Short
  nodes [Output Animation Pose(Result="Sequence Player (additive) 'Drum_Hit_FloorTom_Tap_Fast'"(StartPosition=Get RightArmMaxDelayShortenedBy(), bLoopAnimation=false) {sequence=&Drum_Hit_FloorTom_Tap_Fast, startPosition=0.05})]
 }
 Hit_Medium {
  path=AnimGraph.AnimGraphNode_StateMachine_6.RightArmAdditiveHits.AnimStateNode_1.RightArmHit_A.AnimGraphNode_StateMachine_1.RightArmHitChooser_StartTime_DelayFloat.AnimStateNode_11.Hit_FloorTom_R.AnimGraphNode_StateMachine_0.Hit_FloorTom_R_SM.AnimStateNode_1.Hit_Medium
  nodes [
   n1 = "Sequence Player (additive) 'Drum_Hit_FloorTom_Normal_Medium'"(StartPosition=Get RightArmMaxDelayShortenedBy(), bLoopAnimation=false) {sequence=&Drum_Hit_FloorTom_Normal_Medium, startPosition=0.05}
   Output Animation Pose(Result=@n1)
  ]
 }
 Hit_Long {
  path=AnimGraph.AnimGraphNode_StateMachine_6.RightArmAdditiveHits.AnimStateNode_1.RightArmHit_A.AnimGraphNode_StateMachine_1.RightArmHitChooser_StartTime_DelayFloat.AnimStateNode_11.Hit_FloorTom_R.AnimGraphNode_StateMachine_0.Hit_FloorTom_R_SM.AnimStateNode_2.Hit_Long
  nodes [Output Animation Pose(Result="Sequence Player (additive) 'Drum_Hit_FloorTom_Slow_Big'"(StartPosition=Get RightArmMaxDelayShortenedBy(), bLoopAnimation=false) {sequence=&Drum_Hit_FloorTom_Slow_Big, startPosition=0.05})]
 }
 Transition {
  path=AnimGraph.AnimGraphNode_StateMachine_6.RightArmAdditiveHits.AnimStateNode_1.RightArmHit_A.AnimGraphNode_StateMachine_1.RightArmHitChooser_StartTime_DelayFloat.AnimStateNode_11.Hit_FloorTom_R.AnimGraphNode_StateMachine_0.Hit_FloorTom_R_SM.AnimStateTransitionNode_4.Transition
  nodes [Result(bCanEnterTransition=Get RightArmIsShortDelay())]
 }
 Transition {
  path=AnimGraph.AnimGraphNode_StateMachine_6.RightArmAdditiveHits.AnimStateNode_1.RightArmHit_A.AnimGraphNode_StateMachine_1.RightArmHitChooser_StartTime_DelayFloat.AnimStateNode_11.Hit_FloorTom_R.AnimGraphNode_StateMachine_0.Hit_FloorTom_R_SM.AnimStateTransitionNode_5.Transition
  nodes [
   n1 = Get RightArmIsLongDelay()
   n2 = Get RightArmIsShortDelay()
   Result(bCanEnterTransition=AND Boolean(A=NOT Boolean(A=@n2), B=NOT Boolean(A=@n1)))
  ]
 }
 Transition {
  path=AnimGraph.AnimGraphNode_StateMachine_6.RightArmAdditiveHits.AnimStateNode_1.RightArmHit_A.AnimGraphNode_StateMachine_1.RightArmHitChooser_StartTime_DelayFloat.AnimStateNode_11.Hit_FloorTom_R.AnimGraphNode_StateMachine_0.Hit_FloorTom_R_SM.AnimStateTransitionNode_6.Transition
  nodes [Result(bCanEnterTransition=Get RightArmIsLongDelay())]
 }
 Hit_Crash2_Choke_R {
  path=AnimGraph.AnimGraphNode_StateMachine_6.RightArmAdditiveHits.AnimStateNode_1.RightArmHit_A.AnimGraphNode_StateMachine_1.RightArmHitChooser_StartTime_DelayFloat.AnimStateNode_12.Hit_Crash2_Choke_R
  nodes [
   n1 = Get RightArmMaxDelayShortenedBy()
   n2 = "Sequence Player (additive) 'Drum_Hit_Crash2_Normal_Medium'"(StartPosition=Get RightArmMaxDelayShortenedBy(), bLoopAnimation=false) {sequence=&Drum_Hit_Crash2_Normal_Medium, startPosition=0.05}
   Blend Poses by bool(
    BlendPose_0="Sequence Player (additive) 'Drum_Hit_Crash2_Slow_Big'"(StartPosition=Get RightArmMaxDelayShortenedBy(), bLoopAnimation=false) {sequence=&Drum_Hit_Crash2_Slow_Big, startPosition=0.05}
    BlendPose_1=Blend Poses by bool(
     BlendPose_0="Sequence Player (additive) 'Drum_Hit_Crash2_Tap_Fast'"(StartPosition=@n1, bLoopAnimation=false) {sequence=&Drum_Hit_Crash2_Tap_Fast, startPosition=0.05}
     BlendPose_1=@n2
     BlendTime_0=0
     BlendTime_1=0
     bActiveValue=Get RightArmIsShortDelay()
    ) {blendType=Linear}
    BlendTime_0=0
    BlendTime_1=0
    bActiveValue=Get RightArmIsLongDelay()
   ) {blendType=Linear}
   AnimGraphNode_StateMachine_0 = Hit_Crash2_Hard_R_SM() {bAllowConduitEntryStates=true}
   Output Animation Pose(Result=@AnimGraphNode_StateMachine_0)
  ]
 }
 Hit_Crash2_Hard_R_SM {
  entry=Conduit
  path=AnimGraph.AnimGraphNode_StateMachine_6.RightArmAdditiveHits.AnimStateNode_1.RightArmHit_A.AnimGraphNode_StateMachine_1.RightArmHitChooser_StartTime_DelayFloat.AnimStateNode_12.Hit_Crash2_Choke_R.AnimGraphNode_StateMachine_0.Hit_Crash2_Hard_R_SM
  ids=id
  nodes [
   AnimStateConduitNode_0 = Conduit()
   AnimStateNode_0 = Hit_Short()
   AnimStateNode_1 = Hit_Medium()
   AnimStateNode_2 = Hit_Long()
   AnimStateTransitionNode_4 = AnimStateTransitionNode() {crossfadeDuration=0}
   AnimStateTransitionNode_5 = AnimStateTransitionNode() {crossfadeDuration=0}
   AnimStateTransitionNode_6 = AnimStateTransitionNode() {crossfadeDuration=0}
  ]
  links [
   {
    dst=Hit_Long
    rule=AnimGraph.AnimGraphNode_StateMachine_6.RightArmAdditiveHits.AnimStateNode_1.RightArmHit_A.AnimGraphNode_StateMachine_1.RightArmHitChooser_StartTime_DelayFloat.AnimStateNode_12.Hit_Crash2_Choke_R.AnimGraphNode_StateMachine_0.Hit_Crash2_Hard_R_SM.AnimStateTransitionNode_6.Transition
    src=Conduit
    via=AnimStateTransitionNode_6
   }
   {
    dst=Hit_Medium
    rule=AnimGraph.AnimGraphNode_StateMachine_6.RightArmAdditiveHits.AnimStateNode_1.RightArmHit_A.AnimGraphNode_StateMachine_1.RightArmHitChooser_StartTime_DelayFloat.AnimStateNode_12.Hit_Crash2_Choke_R.AnimGraphNode_StateMachine_0.Hit_Crash2_Hard_R_SM.AnimStateTransitionNode_5.Transition
    src=Conduit
    via=AnimStateTransitionNode_5
   }
   {
    dst=Hit_Short
    rule=AnimGraph.AnimGraphNode_StateMachine_6.RightArmAdditiveHits.AnimStateNode_1.RightArmHit_A.AnimGraphNode_StateMachine_1.RightArmHitChooser_StartTime_DelayFloat.AnimStateNode_12.Hit_Crash2_Choke_R.AnimGraphNode_StateMachine_0.Hit_Crash2_Hard_R_SM.AnimStateTransitionNode_4.Transition
    src=Conduit
    via=AnimStateTransitionNode_4
   }
  ]
 }
 Conduit {
  path=AnimGraph.AnimGraphNode_StateMachine_6.RightArmAdditiveHits.AnimStateNode_1.RightArmHit_A.AnimGraphNode_StateMachine_1.RightArmHitChooser_StartTime_DelayFloat.AnimStateNode_12.Hit_Crash2_Choke_R.AnimGraphNode_StateMachine_0.Hit_Crash2_Hard_R_SM.AnimStateConduitNode_0.Conduit
  nodes [Result(bCanEnterTransition=true)]
 }
 Hit_Short {
  path=AnimGraph.AnimGraphNode_StateMachine_6.RightArmAdditiveHits.AnimStateNode_1.RightArmHit_A.AnimGraphNode_StateMachine_1.RightArmHitChooser_StartTime_DelayFloat.AnimStateNode_12.Hit_Crash2_Choke_R.AnimGraphNode_StateMachine_0.Hit_Crash2_Hard_R_SM.AnimStateNode_0.Hit_Short
  nodes [Output Animation Pose(Result="Sequence Player (additive) 'Drum_Hit_Crash2_Tap_Fast'"(StartPosition=Get RightArmMaxDelayShortenedBy(), bLoopAnimation=false) {sequence=&Drum_Hit_Crash2_Tap_Fast, startPosition=0.05})]
 }
 Hit_Medium {
  path=AnimGraph.AnimGraphNode_StateMachine_6.RightArmAdditiveHits.AnimStateNode_1.RightArmHit_A.AnimGraphNode_StateMachine_1.RightArmHitChooser_StartTime_DelayFloat.AnimStateNode_12.Hit_Crash2_Choke_R.AnimGraphNode_StateMachine_0.Hit_Crash2_Hard_R_SM.AnimStateNode_1.Hit_Medium
  nodes [
   n1 = "Sequence Player (additive) 'Drum_Hit_Crash2_Normal_Medium'"(StartPosition=Get RightArmMaxDelayShortenedBy(), bLoopAnimation=false) {sequence=&Drum_Hit_Crash2_Normal_Medium, startPosition=0.05}
   Output Animation Pose(Result=@n1)
  ]
 }
 Hit_Long {
  path=AnimGraph.AnimGraphNode_StateMachine_6.RightArmAdditiveHits.AnimStateNode_1.RightArmHit_A.AnimGraphNode_StateMachine_1.RightArmHitChooser_StartTime_DelayFloat.AnimStateNode_12.Hit_Crash2_Choke_R.AnimGraphNode_StateMachine_0.Hit_Crash2_Hard_R_SM.AnimStateNode_2.Hit_Long
  nodes [Output Animation Pose(Result="Sequence Player (additive) 'Drum_Hit_Crash2_Slow_Big'"(StartPosition=Get RightArmMaxDelayShortenedBy(), bLoopAnimation=false) {sequence=&Drum_Hit_Crash2_Slow_Big, startPosition=0.05})]
 }
 Transition {
  path=AnimGraph.AnimGraphNode_StateMachine_6.RightArmAdditiveHits.AnimStateNode_1.RightArmHit_A.AnimGraphNode_StateMachine_1.RightArmHitChooser_StartTime_DelayFloat.AnimStateNode_12.Hit_Crash2_Choke_R.AnimGraphNode_StateMachine_0.Hit_Crash2_Hard_R_SM.AnimStateTransitionNode_4.Transition
  nodes [Result(bCanEnterTransition=Get RightArmIsShortDelay())]
 }
 Transition {
  path=AnimGraph.AnimGraphNode_StateMachine_6.RightArmAdditiveHits.AnimStateNode_1.RightArmHit_A.AnimGraphNode_StateMachine_1.RightArmHitChooser_StartTime_DelayFloat.AnimStateNode_12.Hit_Crash2_Choke_R.AnimGraphNode_StateMachine_0.Hit_Crash2_Hard_R_SM.AnimStateTransitionNode_5.Transition
  nodes [
   n1 = Get RightArmIsLongDelay()
   n2 = Get RightArmIsShortDelay()
   Result(bCanEnterTransition=AND Boolean(A=NOT Boolean(A=@n2), B=NOT Boolean(A=@n1)))
  ]
 }
 Transition {
  path=AnimGraph.AnimGraphNode_StateMachine_6.RightArmAdditiveHits.AnimStateNode_1.RightArmHit_A.AnimGraphNode_StateMachine_1.RightArmHitChooser_StartTime_DelayFloat.AnimStateNode_12.Hit_Crash2_Choke_R.AnimGraphNode_StateMachine_0.Hit_Crash2_Hard_R_SM.AnimStateTransitionNode_6.Transition
  nodes [Result(bCanEnterTransition=Get RightArmIsLongDelay())]
 }
 Hit_Crash1_Hard_R {
  path=AnimGraph.AnimGraphNode_StateMachine_6.RightArmAdditiveHits.AnimStateNode_1.RightArmHit_A.AnimGraphNode_StateMachine_1.RightArmHitChooser_StartTime_DelayFloat.AnimStateNode_13.Hit_Crash1_Hard_R
  nodes [
   n1 = Get RightArmMaxDelayShortenedBy()
   n2 = "Sequence Player (additive) 'Drum_Hit_Crash1_Normal_Medium'"(StartPosition=Get RightArmMaxDelayShortenedBy(), bLoopAnimation=false) {sequence=&Drum_Hit_Crash1_Normal_Medium, startPosition=0.05}
   Blend Poses by bool(
    BlendPose_0="Sequence Player (additive) 'Drum_Hit_Crash1_Slow_Big'"(StartPosition=Get RightArmMaxDelayShortenedBy(), bLoopAnimation=false) {sequence=&Drum_Hit_Crash1_Slow_Big, startPosition=0.05}
    BlendPose_1=Blend Poses by bool(
     BlendPose_0="Sequence Player (additive) 'Drum_Hit_Crash1_Tap_Fast'"(StartPosition=@n1, bLoopAnimation=false) {sequence=&Drum_Hit_Crash1_Tap_Fast, startPosition=0.05}
     BlendPose_1=@n2
     BlendTime_0=0
     BlendTime_1=0
     bActiveValue=Get RightArmIsShortDelay()
    ) {blendType=Linear}
    BlendTime_0=0
    BlendTime_1=0
    bActiveValue=Get RightArmIsLongDelay()
   ) {blendType=Linear}
   AnimGraphNode_StateMachine_0 = Hit_Crash1_Hard_R_SM() {bAllowConduitEntryStates=true}
   Output Animation Pose(Result=@AnimGraphNode_StateMachine_0)
  ]
 }
 Hit_Crash1_Hard_R_SM {
  entry=Conduit
  path=AnimGraph.AnimGraphNode_StateMachine_6.RightArmAdditiveHits.AnimStateNode_1.RightArmHit_A.AnimGraphNode_StateMachine_1.RightArmHitChooser_StartTime_DelayFloat.AnimStateNode_13.Hit_Crash1_Hard_R.AnimGraphNode_StateMachine_0.Hit_Crash1_Hard_R_SM
  ids=id
  nodes [
   AnimStateConduitNode_0 = Conduit()
   AnimStateNode_0 = Hit_Short()
   AnimStateNode_1 = Hit_Medium()
   AnimStateNode_2 = Hit_Long()
   AnimStateTransitionNode_4 = AnimStateTransitionNode() {crossfadeDuration=0}
   AnimStateTransitionNode_5 = AnimStateTransitionNode() {crossfadeDuration=0}
   AnimStateTransitionNode_6 = AnimStateTransitionNode() {crossfadeDuration=0}
  ]
  links [
   {
    dst=Hit_Long
    rule=AnimGraph.AnimGraphNode_StateMachine_6.RightArmAdditiveHits.AnimStateNode_1.RightArmHit_A.AnimGraphNode_StateMachine_1.RightArmHitChooser_StartTime_DelayFloat.AnimStateNode_13.Hit_Crash1_Hard_R.AnimGraphNode_StateMachine_0.Hit_Crash1_Hard_R_SM.AnimStateTransitionNode_6.Transition
    src=Conduit
    via=AnimStateTransitionNode_6
   }
   {
    dst=Hit_Medium
    rule=AnimGraph.AnimGraphNode_StateMachine_6.RightArmAdditiveHits.AnimStateNode_1.RightArmHit_A.AnimGraphNode_StateMachine_1.RightArmHitChooser_StartTime_DelayFloat.AnimStateNode_13.Hit_Crash1_Hard_R.AnimGraphNode_StateMachine_0.Hit_Crash1_Hard_R_SM.AnimStateTransitionNode_5.Transition
    src=Conduit
    via=AnimStateTransitionNode_5
   }
   {
    dst=Hit_Short
    rule=AnimGraph.AnimGraphNode_StateMachine_6.RightArmAdditiveHits.AnimStateNode_1.RightArmHit_A.AnimGraphNode_StateMachine_1.RightArmHitChooser_StartTime_DelayFloat.AnimStateNode_13.Hit_Crash1_Hard_R.AnimGraphNode_StateMachine_0.Hit_Crash1_Hard_R_SM.AnimStateTransitionNode_4.Transition
    src=Conduit
    via=AnimStateTransitionNode_4
   }
  ]
 }
 Conduit {
  path=AnimGraph.AnimGraphNode_StateMachine_6.RightArmAdditiveHits.AnimStateNode_1.RightArmHit_A.AnimGraphNode_StateMachine_1.RightArmHitChooser_StartTime_DelayFloat.AnimStateNode_13.Hit_Crash1_Hard_R.AnimGraphNode_StateMachine_0.Hit_Crash1_Hard_R_SM.AnimStateConduitNode_0.Conduit
  nodes [Result(bCanEnterTransition=true)]
 }
 Hit_Short {
  path=AnimGraph.AnimGraphNode_StateMachine_6.RightArmAdditiveHits.AnimStateNode_1.RightArmHit_A.AnimGraphNode_StateMachine_1.RightArmHitChooser_StartTime_DelayFloat.AnimStateNode_13.Hit_Crash1_Hard_R.AnimGraphNode_StateMachine_0.Hit_Crash1_Hard_R_SM.AnimStateNode_0.Hit_Short
  nodes [Output Animation Pose(Result="Sequence Player (additive) 'Drum_Hit_Crash1_Tap_Fast'"(StartPosition=Get RightArmMaxDelayShortenedBy(), bLoopAnimation=false) {sequence=&Drum_Hit_Crash1_Tap_Fast, startPosition=0.05})]
 }
 Hit_Medium {
  path=AnimGraph.AnimGraphNode_StateMachine_6.RightArmAdditiveHits.AnimStateNode_1.RightArmHit_A.AnimGraphNode_StateMachine_1.RightArmHitChooser_StartTime_DelayFloat.AnimStateNode_13.Hit_Crash1_Hard_R.AnimGraphNode_StateMachine_0.Hit_Crash1_Hard_R_SM.AnimStateNode_1.Hit_Medium
  nodes [
   n1 = "Sequence Player (additive) 'Drum_Hit_Crash1_Normal_Medium'"(StartPosition=Get RightArmMaxDelayShortenedBy(), bLoopAnimation=false) {sequence=&Drum_Hit_Crash1_Normal_Medium, startPosition=0.05}
   Output Animation Pose(Result=@n1)
  ]
 }
 Hit_Long {
  path=AnimGraph.AnimGraphNode_StateMachine_6.RightArmAdditiveHits.AnimStateNode_1.RightArmHit_A.AnimGraphNode_StateMachine_1.RightArmHitChooser_StartTime_DelayFloat.AnimStateNode_13.Hit_Crash1_Hard_R.AnimGraphNode_StateMachine_0.Hit_Crash1_Hard_R_SM.AnimStateNode_2.Hit_Long
  nodes [Output Animation Pose(Result="Sequence Player (additive) 'Drum_Hit_Crash1_Slow_Big'"(StartPosition=Get RightArmMaxDelayShortenedBy(), bLoopAnimation=false) {sequence=&Drum_Hit_Crash1_Slow_Big, startPosition=0.05})]
 }
 Transition {
  path=AnimGraph.AnimGraphNode_StateMachine_6.RightArmAdditiveHits.AnimStateNode_1.RightArmHit_A.AnimGraphNode_StateMachine_1.RightArmHitChooser_StartTime_DelayFloat.AnimStateNode_13.Hit_Crash1_Hard_R.AnimGraphNode_StateMachine_0.Hit_Crash1_Hard_R_SM.AnimStateTransitionNode_4.Transition
  nodes [Result(bCanEnterTransition=Get RightArmIsShortDelay())]
 }
 Transition {
  path=AnimGraph.AnimGraphNode_StateMachine_6.RightArmAdditiveHits.AnimStateNode_1.RightArmHit_A.AnimGraphNode_StateMachine_1.RightArmHitChooser_StartTime_DelayFloat.AnimStateNode_13.Hit_Crash1_Hard_R.AnimGraphNode_StateMachine_0.Hit_Crash1_Hard_R_SM.AnimStateTransitionNode_5.Transition
  nodes [
   n1 = Get RightArmIsLongDelay()
   n2 = Get RightArmIsShortDelay()
   Result(bCanEnterTransition=AND Boolean(A=NOT Boolean(A=@n2), B=NOT Boolean(A=@n1)))
  ]
 }
 Transition {
  path=AnimGraph.AnimGraphNode_StateMachine_6.RightArmAdditiveHits.AnimStateNode_1.RightArmHit_A.AnimGraphNode_StateMachine_1.RightArmHitChooser_StartTime_DelayFloat.AnimStateNode_13.Hit_Crash1_Hard_R.AnimGraphNode_StateMachine_0.Hit_Crash1_Hard_R_SM.AnimStateTransitionNode_6.Transition
  nodes [Result(bCanEnterTransition=Get RightArmIsLongDelay())]
 }
 Hit_Crash2_Hard_R {
  path=AnimGraph.AnimGraphNode_StateMachine_6.RightArmAdditiveHits.AnimStateNode_1.RightArmHit_A.AnimGraphNode_StateMachine_1.RightArmHitChooser_StartTime_DelayFloat.AnimStateNode_14.Hit_Crash2_Hard_R
  nodes [
   n1 = Get RightArmMaxDelayShortenedBy()
   n2 = "Sequence Player (additive) 'Drum_Hit_Crash2_Normal_Medium'"(StartPosition=Get RightArmMaxDelayShortenedBy(), bLoopAnimation=false) {sequence=&Drum_Hit_Crash2_Normal_Medium, startPosition=0.05}
   Blend Poses by bool(
    BlendPose_0="Sequence Player (additive) 'Drum_Hit_Crash2_Slow_Big'"(StartPosition=Get RightArmMaxDelayShortenedBy(), bLoopAnimation=false) {sequence=&Drum_Hit_Crash2_Slow_Big, startPosition=0.05}
    BlendPose_1=Blend Poses by bool(
     BlendPose_0="Sequence Player (additive) 'Drum_Hit_Crash2_Tap_Fast'"(StartPosition=@n1, bLoopAnimation=false) {sequence=&Drum_Hit_Crash2_Tap_Fast, startPosition=0.05}
     BlendPose_1=@n2
     BlendTime_0=0
     BlendTime_1=0
     bActiveValue=Get RightArmIsShortDelay()
    ) {blendType=Linear}
    BlendTime_0=0
    BlendTime_1=0
    bActiveValue=Get RightArmIsLongDelay()
   ) {blendType=Linear}
   AnimGraphNode_StateMachine_0 = Hit_Crash2_Hard_R_SM() {bAllowConduitEntryStates=true}
   Output Animation Pose(Result=@AnimGraphNode_StateMachine_0)
  ]
 }
 Hit_Crash2_Hard_R_SM {
  entry=Conduit
  path=AnimGraph.AnimGraphNode_StateMachine_6.RightArmAdditiveHits.AnimStateNode_1.RightArmHit_A.AnimGraphNode_StateMachine_1.RightArmHitChooser_StartTime_DelayFloat.AnimStateNode_14.Hit_Crash2_Hard_R.AnimGraphNode_StateMachine_0.Hit_Crash2_Hard_R_SM
  ids=id
  nodes [
   AnimStateConduitNode_0 = Conduit()
   AnimStateNode_0 = Hit_Short()
   AnimStateNode_1 = Hit_Medium()
   AnimStateNode_2 = Hit_Long()
   AnimStateTransitionNode_4 = AnimStateTransitionNode() {crossfadeDuration=0}
   AnimStateTransitionNode_5 = AnimStateTransitionNode() {crossfadeDuration=0}
   AnimStateTransitionNode_6 = AnimStateTransitionNode() {crossfadeDuration=0}
  ]
  links [
   {
    dst=Hit_Long
    rule=AnimGraph.AnimGraphNode_StateMachine_6.RightArmAdditiveHits.AnimStateNode_1.RightArmHit_A.AnimGraphNode_StateMachine_1.RightArmHitChooser_StartTime_DelayFloat.AnimStateNode_14.Hit_Crash2_Hard_R.AnimGraphNode_StateMachine_0.Hit_Crash2_Hard_R_SM.AnimStateTransitionNode_6.Transition
    src=Conduit
    via=AnimStateTransitionNode_6
   }
   {
    dst=Hit_Medium
    rule=AnimGraph.AnimGraphNode_StateMachine_6.RightArmAdditiveHits.AnimStateNode_1.RightArmHit_A.AnimGraphNode_StateMachine_1.RightArmHitChooser_StartTime_DelayFloat.AnimStateNode_14.Hit_Crash2_Hard_R.AnimGraphNode_StateMachine_0.Hit_Crash2_Hard_R_SM.AnimStateTransitionNode_5.Transition
    src=Conduit
    via=AnimStateTransitionNode_5
   }
   {
    dst=Hit_Short
    rule=AnimGraph.AnimGraphNode_StateMachine_6.RightArmAdditiveHits.AnimStateNode_1.RightArmHit_A.AnimGraphNode_StateMachine_1.RightArmHitChooser_StartTime_DelayFloat.AnimStateNode_14.Hit_Crash2_Hard_R.AnimGraphNode_StateMachine_0.Hit_Crash2_Hard_R_SM.AnimStateTransitionNode_4.Transition
    src=Conduit
    via=AnimStateTransitionNode_4
   }
  ]
 }
 Conduit {
  path=AnimGraph.AnimGraphNode_StateMachine_6.RightArmAdditiveHits.AnimStateNode_1.RightArmHit_A.AnimGraphNode_StateMachine_1.RightArmHitChooser_StartTime_DelayFloat.AnimStateNode_14.Hit_Crash2_Hard_R.AnimGraphNode_StateMachine_0.Hit_Crash2_Hard_R_SM.AnimStateConduitNode_0.Conduit
  nodes [Result(bCanEnterTransition=true)]
 }
 Hit_Short {
  path=AnimGraph.AnimGraphNode_StateMachine_6.RightArmAdditiveHits.AnimStateNode_1.RightArmHit_A.AnimGraphNode_StateMachine_1.RightArmHitChooser_StartTime_DelayFloat.AnimStateNode_14.Hit_Crash2_Hard_R.AnimGraphNode_StateMachine_0.Hit_Crash2_Hard_R_SM.AnimStateNode_0.Hit_Short
  nodes [Output Animation Pose(Result="Sequence Player (additive) 'Drum_Hit_Crash2_Tap_Fast'"(StartPosition=Get RightArmMaxDelayShortenedBy(), bLoopAnimation=false) {sequence=&Drum_Hit_Crash2_Tap_Fast, startPosition=0.05})]
 }
 Hit_Medium {
  path=AnimGraph.AnimGraphNode_StateMachine_6.RightArmAdditiveHits.AnimStateNode_1.RightArmHit_A.AnimGraphNode_StateMachine_1.RightArmHitChooser_StartTime_DelayFloat.AnimStateNode_14.Hit_Crash2_Hard_R.AnimGraphNode_StateMachine_0.Hit_Crash2_Hard_R_SM.AnimStateNode_1.Hit_Medium
  nodes [
   n1 = "Sequence Player (additive) 'Drum_Hit_Crash2_Normal_Medium'"(StartPosition=Get RightArmMaxDelayShortenedBy(), bLoopAnimation=false) {sequence=&Drum_Hit_Crash2_Normal_Medium, startPosition=0.05}
   Output Animation Pose(Result=@n1)
  ]
 }
 Hit_Long {
  path=AnimGraph.AnimGraphNode_StateMachine_6.RightArmAdditiveHits.AnimStateNode_1.RightArmHit_A.AnimGraphNode_StateMachine_1.RightArmHitChooser_StartTime_DelayFloat.AnimStateNode_14.Hit_Crash2_Hard_R.AnimGraphNode_StateMachine_0.Hit_Crash2_Hard_R_SM.AnimStateNode_2.Hit_Long
  nodes [Output Animation Pose(Result="Sequence Player (additive) 'Drum_Hit_Crash2_Slow_Big'"(StartPosition=Get RightArmMaxDelayShortenedBy(), bLoopAnimation=false) {sequence=&Drum_Hit_Crash2_Slow_Big, startPosition=0.05})]
 }
 Transition {
  path=AnimGraph.AnimGraphNode_StateMachine_6.RightArmAdditiveHits.AnimStateNode_1.RightArmHit_A.AnimGraphNode_StateMachine_1.RightArmHitChooser_StartTime_DelayFloat.AnimStateNode_14.Hit_Crash2_Hard_R.AnimGraphNode_StateMachine_0.Hit_Crash2_Hard_R_SM.AnimStateTransitionNode_4.Transition
  nodes [Result(bCanEnterTransition=Get RightArmIsShortDelay())]
 }
 Transition {
  path=AnimGraph.AnimGraphNode_StateMachine_6.RightArmAdditiveHits.AnimStateNode_1.RightArmHit_A.AnimGraphNode_StateMachine_1.RightArmHitChooser_StartTime_DelayFloat.AnimStateNode_14.Hit_Crash2_Hard_R.AnimGraphNode_StateMachine_0.Hit_Crash2_Hard_R_SM.AnimStateTransitionNode_5.Transition
  nodes [
   n1 = Get RightArmIsLongDelay()
   n2 = Get RightArmIsShortDelay()
   Result(bCanEnterTransition=AND Boolean(A=NOT Boolean(A=@n2), B=NOT Boolean(A=@n1)))
  ]
 }
 Transition {
  path=AnimGraph.AnimGraphNode_StateMachine_6.RightArmAdditiveHits.AnimStateNode_1.RightArmHit_A.AnimGraphNode_StateMachine_1.RightArmHitChooser_StartTime_DelayFloat.AnimStateNode_14.Hit_Crash2_Hard_R.AnimGraphNode_StateMachine_0.Hit_Crash2_Hard_R_SM.AnimStateTransitionNode_6.Transition
  nodes [Result(bCanEnterTransition=Get RightArmIsLongDelay())]
 }
 Hit_Crash1_Choke_R {
  path=AnimGraph.AnimGraphNode_StateMachine_6.RightArmAdditiveHits.AnimStateNode_1.RightArmHit_A.AnimGraphNode_StateMachine_1.RightArmHitChooser_StartTime_DelayFloat.AnimStateNode_2.Hit_Crash1_Choke_R
  nodes [
   n1 = Get RightArmMaxDelayShortenedBy()
   n2 = "Sequence Player (additive) 'Drum_Hit_Crash1_Normal_Medium'"(StartPosition=Get RightArmMaxDelayShortenedBy(), bLoopAnimation=false) {sequence=&Drum_Hit_Crash1_Normal_Medium, startPosition=0.05}
   Blend Poses by bool(
    BlendPose_0="Sequence Player (additive) 'Drum_Hit_Crash1_Slow_Big'"(StartPosition=Get RightArmMaxDelayShortenedBy(), bLoopAnimation=false) {sequence=&Drum_Hit_Crash1_Slow_Big, startPosition=0.05}
    BlendPose_1=Blend Poses by bool(
     BlendPose_0="Sequence Player (additive) 'Drum_Hit_Crash1_Tap_Fast'"(StartPosition=@n1, bLoopAnimation=false) {sequence=&Drum_Hit_Crash1_Tap_Fast, startPosition=0.05}
     BlendPose_1=@n2
     BlendTime_0=0
     BlendTime_1=0
     bActiveValue=Get RightArmIsShortDelay()
    ) {blendType=Linear}
    BlendTime_0=0
    BlendTime_1=0
    bActiveValue=Get RightArmIsLongDelay()
   ) {blendType=Linear}
   AnimGraphNode_StateMachine_0 = Hit_Crash1_Hard_R_SM() {bAllowConduitEntryStates=true}
   Output Animation Pose(Result=@AnimGraphNode_StateMachine_0)
  ]
 }
 Hit_Crash1_Hard_R_SM {
  entry=Conduit
  path=AnimGraph.AnimGraphNode_StateMachine_6.RightArmAdditiveHits.AnimStateNode_1.RightArmHit_A.AnimGraphNode_StateMachine_1.RightArmHitChooser_StartTime_DelayFloat.AnimStateNode_2.Hit_Crash1_Choke_R.AnimGraphNode_StateMachine_0.Hit_Crash1_Hard_R_SM
  ids=id
  nodes [
   AnimStateConduitNode_0 = Conduit()
   AnimStateNode_0 = Hit_Short()
   AnimStateNode_1 = Hit_Medium()
   AnimStateNode_2 = Hit_Long()
   AnimStateTransitionNode_4 = AnimStateTransitionNode() {crossfadeDuration=0}
   AnimStateTransitionNode_5 = AnimStateTransitionNode() {crossfadeDuration=0}
   AnimStateTransitionNode_6 = AnimStateTransitionNode() {crossfadeDuration=0}
  ]
  links [
   {
    dst=Hit_Long
    rule=AnimGraph.AnimGraphNode_StateMachine_6.RightArmAdditiveHits.AnimStateNode_1.RightArmHit_A.AnimGraphNode_StateMachine_1.RightArmHitChooser_StartTime_DelayFloat.AnimStateNode_2.Hit_Crash1_Choke_R.AnimGraphNode_StateMachine_0.Hit_Crash1_Hard_R_SM.AnimStateTransitionNode_6.Transition
    src=Conduit
    via=AnimStateTransitionNode_6
   }
   {
    dst=Hit_Medium
    rule=AnimGraph.AnimGraphNode_StateMachine_6.RightArmAdditiveHits.AnimStateNode_1.RightArmHit_A.AnimGraphNode_StateMachine_1.RightArmHitChooser_StartTime_DelayFloat.AnimStateNode_2.Hit_Crash1_Choke_R.AnimGraphNode_StateMachine_0.Hit_Crash1_Hard_R_SM.AnimStateTransitionNode_5.Transition
    src=Conduit
    via=AnimStateTransitionNode_5
   }
   {
    dst=Hit_Short
    rule=AnimGraph.AnimGraphNode_StateMachine_6.RightArmAdditiveHits.AnimStateNode_1.RightArmHit_A.AnimGraphNode_StateMachine_1.RightArmHitChooser_StartTime_DelayFloat.AnimStateNode_2.Hit_Crash1_Choke_R.AnimGraphNode_StateMachine_0.Hit_Crash1_Hard_R_SM.AnimStateTransitionNode_4.Transition
    src=Conduit
    via=AnimStateTransitionNode_4
   }
  ]
 }
 Conduit {
  path=AnimGraph.AnimGraphNode_StateMachine_6.RightArmAdditiveHits.AnimStateNode_1.RightArmHit_A.AnimGraphNode_StateMachine_1.RightArmHitChooser_StartTime_DelayFloat.AnimStateNode_2.Hit_Crash1_Choke_R.AnimGraphNode_StateMachine_0.Hit_Crash1_Hard_R_SM.AnimStateConduitNode_0.Conduit
  nodes [Result(bCanEnterTransition=true)]
 }
 Hit_Short {
  path=AnimGraph.AnimGraphNode_StateMachine_6.RightArmAdditiveHits.AnimStateNode_1.RightArmHit_A.AnimGraphNode_StateMachine_1.RightArmHitChooser_StartTime_DelayFloat.AnimStateNode_2.Hit_Crash1_Choke_R.AnimGraphNode_StateMachine_0.Hit_Crash1_Hard_R_SM.AnimStateNode_0.Hit_Short
  nodes [Output Animation Pose(Result="Sequence Player (additive) 'Drum_Hit_Crash1_Tap_Fast'"(StartPosition=Get RightArmMaxDelayShortenedBy(), bLoopAnimation=false) {sequence=&Drum_Hit_Crash1_Tap_Fast, startPosition=0.05})]
 }
 Hit_Medium {
  path=AnimGraph.AnimGraphNode_StateMachine_6.RightArmAdditiveHits.AnimStateNode_1.RightArmHit_A.AnimGraphNode_StateMachine_1.RightArmHitChooser_StartTime_DelayFloat.AnimStateNode_2.Hit_Crash1_Choke_R.AnimGraphNode_StateMachine_0.Hit_Crash1_Hard_R_SM.AnimStateNode_1.Hit_Medium
  nodes [
   n1 = "Sequence Player (additive) 'Drum_Hit_Crash1_Normal_Medium'"(StartPosition=Get RightArmMaxDelayShortenedBy(), bLoopAnimation=false) {sequence=&Drum_Hit_Crash1_Normal_Medium, startPosition=0.05}
   Output Animation Pose(Result=@n1)
  ]
 }
 Hit_Long {
  path=AnimGraph.AnimGraphNode_StateMachine_6.RightArmAdditiveHits.AnimStateNode_1.RightArmHit_A.AnimGraphNode_StateMachine_1.RightArmHitChooser_StartTime_DelayFloat.AnimStateNode_2.Hit_Crash1_Choke_R.AnimGraphNode_StateMachine_0.Hit_Crash1_Hard_R_SM.AnimStateNode_2.Hit_Long
  nodes [Output Animation Pose(Result="Sequence Player (additive) 'Drum_Hit_Crash1_Slow_Big'"(StartPosition=Get RightArmMaxDelayShortenedBy(), bLoopAnimation=false) {sequence=&Drum_Hit_Crash1_Slow_Big, startPosition=0.05})]
 }
 Transition {
  path=AnimGraph.AnimGraphNode_StateMachine_6.RightArmAdditiveHits.AnimStateNode_1.RightArmHit_A.AnimGraphNode_StateMachine_1.RightArmHitChooser_StartTime_DelayFloat.AnimStateNode_2.Hit_Crash1_Choke_R.AnimGraphNode_StateMachine_0.Hit_Crash1_Hard_R_SM.AnimStateTransitionNode_4.Transition
  nodes [Result(bCanEnterTransition=Get RightArmIsShortDelay())]
 }
 Transition {
  path=AnimGraph.AnimGraphNode_StateMachine_6.RightArmAdditiveHits.AnimStateNode_1.RightArmHit_A.AnimGraphNode_StateMachine_1.RightArmHitChooser_StartTime_DelayFloat.AnimStateNode_2.Hit_Crash1_Choke_R.AnimGraphNode_StateMachine_0.Hit_Crash1_Hard_R_SM.AnimStateTransitionNode_5.Transition
  nodes [
   n1 = Get RightArmIsLongDelay()
   n2 = Get RightArmIsShortDelay()
   Result(bCanEnterTransition=AND Boolean(A=NOT Boolean(A=@n2), B=NOT Boolean(A=@n1)))
  ]
 }
 Transition {
  path=AnimGraph.AnimGraphNode_StateMachine_6.RightArmAdditiveHits.AnimStateNode_1.RightArmHit_A.AnimGraphNode_StateMachine_1.RightArmHitChooser_StartTime_DelayFloat.AnimStateNode_2.Hit_Crash1_Choke_R.AnimGraphNode_StateMachine_0.Hit_Crash1_Hard_R_SM.AnimStateTransitionNode_6.Transition
  nodes [Result(bCanEnterTransition=Get RightArmIsLongDelay())]
 }
 Hit_Crash1_Soft_R {
  path=AnimGraph.AnimGraphNode_StateMachine_6.RightArmAdditiveHits.AnimStateNode_1.RightArmHit_A.AnimGraphNode_StateMachine_1.RightArmHitChooser_StartTime_DelayFloat.AnimStateNode_3.Hit_Crash1_Soft_R
  nodes [
   n1 = Get RightArmMaxDelayShortenedBy()
   n2 = "Sequence Player (additive) 'Drum_Hit_Crash1_Normal_Medium'"(StartPosition=Get RightArmMaxDelayShortenedBy(), bLoopAnimation=false) {sequence=&Drum_Hit_Crash1_Normal_Medium, startPosition=0.05}
   Blend Poses by bool(
    BlendPose_0="Sequence Player (additive) 'Drum_Hit_Crash1_Slow_Big'"(StartPosition=Get RightArmMaxDelayShortenedBy(), bLoopAnimation=false) {sequence=&Drum_Hit_Crash1_Slow_Big, startPosition=0.05}
    BlendPose_1=Blend Poses by bool(
     BlendPose_0="Sequence Player (additive) 'Drum_Hit_Crash1_Tap_Fast'"(StartPosition=@n1, bLoopAnimation=false) {sequence=&Drum_Hit_Crash1_Tap_Fast, startPosition=0.05}
     BlendPose_1=@n2
     BlendTime_0=0
     BlendTime_1=0
     bActiveValue=Get RightArmIsShortDelay()
    ) {blendType=Linear}
    BlendTime_0=0
    BlendTime_1=0
    bActiveValue=Get RightArmIsLongDelay()
   ) {blendType=Linear}
   AnimGraphNode_StateMachine_0 = Hit_Crash1_Soft_R_SM() {bAllowConduitEntryStates=true}
   Output Animation Pose(Result=@AnimGraphNode_StateMachine_0)
  ]
 }
 Hit_Crash1_Soft_R_SM {
  entry=Conduit
  path=AnimGraph.AnimGraphNode_StateMachine_6.RightArmAdditiveHits.AnimStateNode_1.RightArmHit_A.AnimGraphNode_StateMachine_1.RightArmHitChooser_StartTime_DelayFloat.AnimStateNode_3.Hit_Crash1_Soft_R.AnimGraphNode_StateMachine_0.Hit_Crash1_Soft_R_SM
  ids=id
  nodes [
   AnimStateConduitNode_0 = Conduit()
   AnimStateNode_0 = Hit_Short()
   AnimStateNode_1 = Hit_Medium()
   AnimStateNode_2 = Hit_Long()
   AnimStateTransitionNode_4 = AnimStateTransitionNode() {crossfadeDuration=0}
   AnimStateTransitionNode_5 = AnimStateTransitionNode() {crossfadeDuration=0}
   AnimStateTransitionNode_6 = AnimStateTransitionNode() {crossfadeDuration=0}
  ]
  links [
   {
    dst=Hit_Long
    rule=AnimGraph.AnimGraphNode_StateMachine_6.RightArmAdditiveHits.AnimStateNode_1.RightArmHit_A.AnimGraphNode_StateMachine_1.RightArmHitChooser_StartTime_DelayFloat.AnimStateNode_3.Hit_Crash1_Soft_R.AnimGraphNode_StateMachine_0.Hit_Crash1_Soft_R_SM.AnimStateTransitionNode_6.Transition
    src=Conduit
    via=AnimStateTransitionNode_6
   }
   {
    dst=Hit_Medium
    rule=AnimGraph.AnimGraphNode_StateMachine_6.RightArmAdditiveHits.AnimStateNode_1.RightArmHit_A.AnimGraphNode_StateMachine_1.RightArmHitChooser_StartTime_DelayFloat.AnimStateNode_3.Hit_Crash1_Soft_R.AnimGraphNode_StateMachine_0.Hit_Crash1_Soft_R_SM.AnimStateTransitionNode_5.Transition
    src=Conduit
    via=AnimStateTransitionNode_5
   }
   {
    dst=Hit_Short
    rule=AnimGraph.AnimGraphNode_StateMachine_6.RightArmAdditiveHits.AnimStateNode_1.RightArmHit_A.AnimGraphNode_StateMachine_1.RightArmHitChooser_StartTime_DelayFloat.AnimStateNode_3.Hit_Crash1_Soft_R.AnimGraphNode_StateMachine_0.Hit_Crash1_Soft_R_SM.AnimStateTransitionNode_4.Transition
    src=Conduit
    via=AnimStateTransitionNode_4
   }
  ]
 }
 Conduit {
  path=AnimGraph.AnimGraphNode_StateMachine_6.RightArmAdditiveHits.AnimStateNode_1.RightArmHit_A.AnimGraphNode_StateMachine_1.RightArmHitChooser_StartTime_DelayFloat.AnimStateNode_3.Hit_Crash1_Soft_R.AnimGraphNode_StateMachine_0.Hit_Crash1_Soft_R_SM.AnimStateConduitNode_0.Conduit
  nodes [Result(bCanEnterTransition=true)]
 }
 Hit_Short {
  path=AnimGraph.AnimGraphNode_StateMachine_6.RightArmAdditiveHits.AnimStateNode_1.RightArmHit_A.AnimGraphNode_StateMachine_1.RightArmHitChooser_StartTime_DelayFloat.AnimStateNode_3.Hit_Crash1_Soft_R.AnimGraphNode_StateMachine_0.Hit_Crash1_Soft_R_SM.AnimStateNode_0.Hit_Short
  nodes [Output Animation Pose(Result="Sequence Player (additive) 'Drum_Hit_Crash1_Tap_Fast'"(StartPosition=Get RightArmMaxDelayShortenedBy(), bLoopAnimation=false) {sequence=&Drum_Hit_Crash1_Tap_Fast, startPosition=0.05})]
 }
 Hit_Medium {
  path=AnimGraph.AnimGraphNode_StateMachine_6.RightArmAdditiveHits.AnimStateNode_1.RightArmHit_A.AnimGraphNode_StateMachine_1.RightArmHitChooser_StartTime_DelayFloat.AnimStateNode_3.Hit_Crash1_Soft_R.AnimGraphNode_StateMachine_0.Hit_Crash1_Soft_R_SM.AnimStateNode_1.Hit_Medium
  nodes [
   n1 = "Sequence Player (additive) 'Drum_Hit_Crash1_Normal_Medium'"(StartPosition=Get RightArmMaxDelayShortenedBy(), bLoopAnimation=false) {sequence=&Drum_Hit_Crash1_Normal_Medium, startPosition=0.05}
   Output Animation Pose(Result=@n1)
  ]
 }
 Hit_Long {
  path=AnimGraph.AnimGraphNode_StateMachine_6.RightArmAdditiveHits.AnimStateNode_1.RightArmHit_A.AnimGraphNode_StateMachine_1.RightArmHitChooser_StartTime_DelayFloat.AnimStateNode_3.Hit_Crash1_Soft_R.AnimGraphNode_StateMachine_0.Hit_Crash1_Soft_R_SM.AnimStateNode_2.Hit_Long
  nodes [Output Animation Pose(Result="Sequence Player (additive) 'Drum_Hit_Crash1_Slow_Big'"(StartPosition=Get RightArmMaxDelayShortenedBy(), bLoopAnimation=false) {sequence=&Drum_Hit_Crash1_Slow_Big, startPosition=0.05})]
 }
 Transition {
  path=AnimGraph.AnimGraphNode_StateMachine_6.RightArmAdditiveHits.AnimStateNode_1.RightArmHit_A.AnimGraphNode_StateMachine_1.RightArmHitChooser_StartTime_DelayFloat.AnimStateNode_3.Hit_Crash1_Soft_R.AnimGraphNode_StateMachine_0.Hit_Crash1_Soft_R_SM.AnimStateTransitionNode_4.Transition
  nodes [Result(bCanEnterTransition=Get RightArmIsShortDelay())]
 }
 Transition {
  path=AnimGraph.AnimGraphNode_StateMachine_6.RightArmAdditiveHits.AnimStateNode_1.RightArmHit_A.AnimGraphNode_StateMachine_1.RightArmHitChooser_StartTime_DelayFloat.AnimStateNode_3.Hit_Crash1_Soft_R.AnimGraphNode_StateMachine_0.Hit_Crash1_Soft_R_SM.AnimStateTransitionNode_5.Transition
  nodes [
   n1 = Get RightArmIsLongDelay()
   n2 = Get RightArmIsShortDelay()
   Result(bCanEnterTransition=AND Boolean(A=NOT Boolean(A=@n2), B=NOT Boolean(A=@n1)))
  ]
 }
 Transition {
  path=AnimGraph.AnimGraphNode_StateMachine_6.RightArmAdditiveHits.AnimStateNode_1.RightArmHit_A.AnimGraphNode_StateMachine_1.RightArmHitChooser_StartTime_DelayFloat.AnimStateNode_3.Hit_Crash1_Soft_R.AnimGraphNode_StateMachine_0.Hit_Crash1_Soft_R_SM.AnimStateTransitionNode_6.Transition
  nodes [Result(bCanEnterTransition=Get RightArmIsLongDelay())]
 }
 Hit_Snare_Soft_R {
  path=AnimGraph.AnimGraphNode_StateMachine_6.RightArmAdditiveHits.AnimStateNode_1.RightArmHit_A.AnimGraphNode_StateMachine_1.RightArmHitChooser_StartTime_DelayFloat.AnimStateNode_4.Hit_Snare_Soft_R
  nodes [Output Animation Pose(Result="Sequence Player (additive) 'Drum_Hit_SnareSoft_Tap_Fast'"(StartPosition=Get LeftArmMaxDelayShortenedBy(), bLoopAnimation=False) {bLoopAnimation=false, sequence=&Drum_Hit_SnareSoft_Tap_Fast})]
 }
 Hit_Crash2_Soft_R {
  path=AnimGraph.AnimGraphNode_StateMachine_6.RightArmAdditiveHits.AnimStateNode_1.RightArmHit_A.AnimGraphNode_StateMachine_1.RightArmHitChooser_StartTime_DelayFloat.AnimStateNode_5.Hit_Crash2_Soft_R
  nodes [
   n1 = Get RightArmMaxDelayShortenedBy()
   n2 = "Sequence Player (additive) 'Drum_Hit_Crash2_Normal_Medium'"(StartPosition=Get RightArmMaxDelayShortenedBy(), bLoopAnimation=false) {sequence=&Drum_Hit_Crash2_Normal_Medium, startPosition=0.05}
   Blend Poses by bool(
    BlendPose_0="Sequence Player (additive) 'Drum_Hit_Crash2_Slow_Big'"(StartPosition=Get RightArmMaxDelayShortenedBy(), bLoopAnimation=false) {sequence=&Drum_Hit_Crash2_Slow_Big, startPosition=0.05}
    BlendPose_1=Blend Poses by bool(
     BlendPose_0="Sequence Player (additive) 'Drum_Hit_Crash2_Tap_Fast'"(StartPosition=@n1, bLoopAnimation=false) {sequence=&Drum_Hit_Crash2_Tap_Fast, startPosition=0.05}
     BlendPose_1=@n2
     BlendTime_0=0
     BlendTime_1=0
     bActiveValue=Get RightArmIsShortDelay()
    ) {blendType=Linear}
    BlendTime_0=0
    BlendTime_1=0
    bActiveValue=Get RightArmIsLongDelay()
   ) {blendType=Linear}
   AnimGraphNode_StateMachine_0 = Hit_Crash2_Soft_R_SM() {bAllowConduitEntryStates=true}
   Output Animation Pose(Result=@AnimGraphNode_StateMachine_0)
  ]
 }
 Hit_Crash2_Soft_R_SM {
  entry=Conduit
  path=AnimGraph.AnimGraphNode_StateMachine_6.RightArmAdditiveHits.AnimStateNode_1.RightArmHit_A.AnimGraphNode_StateMachine_1.RightArmHitChooser_StartTime_DelayFloat.AnimStateNode_5.Hit_Crash2_Soft_R.AnimGraphNode_StateMachine_0.Hit_Crash2_Soft_R_SM
  ids=id
  nodes [
   AnimStateConduitNode_0 = Conduit()
   AnimStateNode_0 = Hit_Short()
   AnimStateNode_1 = Hit_Medium()
   AnimStateNode_2 = Hit_Long()
   AnimStateTransitionNode_4 = AnimStateTransitionNode() {crossfadeDuration=0}
   AnimStateTransitionNode_5 = AnimStateTransitionNode() {crossfadeDuration=0}
   AnimStateTransitionNode_6 = AnimStateTransitionNode() {crossfadeDuration=0}
  ]
  links [
   {
    dst=Hit_Long
    rule=AnimGraph.AnimGraphNode_StateMachine_6.RightArmAdditiveHits.AnimStateNode_1.RightArmHit_A.AnimGraphNode_StateMachine_1.RightArmHitChooser_StartTime_DelayFloat.AnimStateNode_5.Hit_Crash2_Soft_R.AnimGraphNode_StateMachine_0.Hit_Crash2_Soft_R_SM.AnimStateTransitionNode_6.Transition
    src=Conduit
    via=AnimStateTransitionNode_6
   }
   {
    dst=Hit_Medium
    rule=AnimGraph.AnimGraphNode_StateMachine_6.RightArmAdditiveHits.AnimStateNode_1.RightArmHit_A.AnimGraphNode_StateMachine_1.RightArmHitChooser_StartTime_DelayFloat.AnimStateNode_5.Hit_Crash2_Soft_R.AnimGraphNode_StateMachine_0.Hit_Crash2_Soft_R_SM.AnimStateTransitionNode_5.Transition
    src=Conduit
    via=AnimStateTransitionNode_5
   }
   {
    dst=Hit_Short
    rule=AnimGraph.AnimGraphNode_StateMachine_6.RightArmAdditiveHits.AnimStateNode_1.RightArmHit_A.AnimGraphNode_StateMachine_1.RightArmHitChooser_StartTime_DelayFloat.AnimStateNode_5.Hit_Crash2_Soft_R.AnimGraphNode_StateMachine_0.Hit_Crash2_Soft_R_SM.AnimStateTransitionNode_4.Transition
    src=Conduit
    via=AnimStateTransitionNode_4
   }
  ]
 }
 Conduit {
  path=AnimGraph.AnimGraphNode_StateMachine_6.RightArmAdditiveHits.AnimStateNode_1.RightArmHit_A.AnimGraphNode_StateMachine_1.RightArmHitChooser_StartTime_DelayFloat.AnimStateNode_5.Hit_Crash2_Soft_R.AnimGraphNode_StateMachine_0.Hit_Crash2_Soft_R_SM.AnimStateConduitNode_0.Conduit
  nodes [Result(bCanEnterTransition=true)]
 }
 Hit_Short {
  path=AnimGraph.AnimGraphNode_StateMachine_6.RightArmAdditiveHits.AnimStateNode_1.RightArmHit_A.AnimGraphNode_StateMachine_1.RightArmHitChooser_StartTime_DelayFloat.AnimStateNode_5.Hit_Crash2_Soft_R.AnimGraphNode_StateMachine_0.Hit_Crash2_Soft_R_SM.AnimStateNode_0.Hit_Short
  nodes [Output Animation Pose(Result="Sequence Player (additive) 'Drum_Hit_Crash2_Tap_Fast'"(StartPosition=Get RightArmMaxDelayShortenedBy(), bLoopAnimation=false) {sequence=&Drum_Hit_Crash2_Tap_Fast, startPosition=0.05})]
 }
 Hit_Medium {
  path=AnimGraph.AnimGraphNode_StateMachine_6.RightArmAdditiveHits.AnimStateNode_1.RightArmHit_A.AnimGraphNode_StateMachine_1.RightArmHitChooser_StartTime_DelayFloat.AnimStateNode_5.Hit_Crash2_Soft_R.AnimGraphNode_StateMachine_0.Hit_Crash2_Soft_R_SM.AnimStateNode_1.Hit_Medium
  nodes [
   n1 = "Sequence Player (additive) 'Drum_Hit_Crash2_Normal_Medium'"(StartPosition=Get RightArmMaxDelayShortenedBy(), bLoopAnimation=false) {sequence=&Drum_Hit_Crash2_Normal_Medium, startPosition=0.05}
   Output Animation Pose(Result=@n1)
  ]
 }
 Hit_Long {
  path=AnimGraph.AnimGraphNode_StateMachine_6.RightArmAdditiveHits.AnimStateNode_1.RightArmHit_A.AnimGraphNode_StateMachine_1.RightArmHitChooser_StartTime_DelayFloat.AnimStateNode_5.Hit_Crash2_Soft_R.AnimGraphNode_StateMachine_0.Hit_Crash2_Soft_R_SM.AnimStateNode_2.Hit_Long
  nodes [Output Animation Pose(Result="Sequence Player (additive) 'Drum_Hit_Crash2_Slow_Big'"(StartPosition=Get RightArmMaxDelayShortenedBy(), bLoopAnimation=false) {sequence=&Drum_Hit_Crash2_Slow_Big, startPosition=0.05})]
 }
 Transition {
  path=AnimGraph.AnimGraphNode_StateMachine_6.RightArmAdditiveHits.AnimStateNode_1.RightArmHit_A.AnimGraphNode_StateMachine_1.RightArmHitChooser_StartTime_DelayFloat.AnimStateNode_5.Hit_Crash2_Soft_R.AnimGraphNode_StateMachine_0.Hit_Crash2_Soft_R_SM.AnimStateTransitionNode_4.Transition
  nodes [Result(bCanEnterTransition=Get RightArmIsShortDelay())]
 }
 Transition {
  path=AnimGraph.AnimGraphNode_StateMachine_6.RightArmAdditiveHits.AnimStateNode_1.RightArmHit_A.AnimGraphNode_StateMachine_1.RightArmHitChooser_StartTime_DelayFloat.AnimStateNode_5.Hit_Crash2_Soft_R.AnimGraphNode_StateMachine_0.Hit_Crash2_Soft_R_SM.AnimStateTransitionNode_5.Transition
  nodes [
   n1 = Get RightArmIsLongDelay()
   n2 = Get RightArmIsShortDelay()
   Result(bCanEnterTransition=AND Boolean(A=NOT Boolean(A=@n2), B=NOT Boolean(A=@n1)))
  ]
 }
 Transition {
  path=AnimGraph.AnimGraphNode_StateMachine_6.RightArmAdditiveHits.AnimStateNode_1.RightArmHit_A.AnimGraphNode_StateMachine_1.RightArmHitChooser_StartTime_DelayFloat.AnimStateNode_5.Hit_Crash2_Soft_R.AnimGraphNode_StateMachine_0.Hit_Crash2_Soft_R_SM.AnimStateTransitionNode_6.Transition
  nodes [Result(bCanEnterTransition=Get RightArmIsLongDelay())]
 }
 Hit_Ride_R {
  path=AnimGraph.AnimGraphNode_StateMachine_6.RightArmAdditiveHits.AnimStateNode_1.RightArmHit_A.AnimGraphNode_StateMachine_1.RightArmHitChooser_StartTime_DelayFloat.AnimStateNode_6.Hit_Ride_R
  nodes [
   n1 = Get RightArmMaxDelayShortenedBy()
   n2 = "Sequence Player (additive) 'Drum_Hit_Ride_Normal_Medium'"(StartPosition=Get RightArmMaxDelayShortenedBy(), bLoopAnimation=false) {sequence=&Drum_Hit_Ride_Normal_Medium, startPosition=0.05}
   Blend Poses by bool(
    BlendPose_0="Sequence Player (additive) 'Drum_Hit_Ride_Slow_Big'"(StartPosition=Get RightArmMaxDelayShortenedBy(), bLoopAnimation=false) {sequence=&Drum_Hit_Ride_Slow_Big, startPosition=0.05}
    BlendPose_1=Blend Poses by bool(
     BlendPose_0="Sequence Player (additive) 'Drum_Hit_Ride_Tap_Fast'"(StartPosition=@n1, bLoopAnimation=false) {sequence=&Drum_Hit_Ride_Tap_Fast, startPosition=0.05}
     BlendPose_1=@n2
     BlendTime_0=0
     BlendTime_1=0
     bActiveValue=Get RightArmIsShortDelay()
    ) {blendType=Linear}
    BlendTime_0=0
    BlendTime_1=0
    bActiveValue=Get RightArmIsLongDelay()
   ) {blendType=Linear}
   AnimGraphNode_StateMachine_0 = Hit_Ride_R_SM() {bAllowConduitEntryStates=true}
   Output Animation Pose(Result=@AnimGraphNode_StateMachine_0)
  ]
 }
 Hit_Ride_R_SM {
  entry=Conduit
  path=AnimGraph.AnimGraphNode_StateMachine_6.RightArmAdditiveHits.AnimStateNode_1.RightArmHit_A.AnimGraphNode_StateMachine_1.RightArmHitChooser_StartTime_DelayFloat.AnimStateNode_6.Hit_Ride_R.AnimGraphNode_StateMachine_0.Hit_Ride_R_SM
  ids=id
  nodes [
   AnimStateConduitNode_0 = Conduit()
   AnimStateNode_0 = Hit_Short()
   AnimStateNode_1 = Hit_Medium()
   AnimStateNode_2 = Hit_Long()
   AnimStateTransitionNode_4 = AnimStateTransitionNode() {crossfadeDuration=0}
   AnimStateTransitionNode_5 = AnimStateTransitionNode() {crossfadeDuration=0}
   AnimStateTransitionNode_6 = AnimStateTransitionNode() {crossfadeDuration=0}
  ]
  links [
   {
    dst=Hit_Long
    rule=AnimGraph.AnimGraphNode_StateMachine_6.RightArmAdditiveHits.AnimStateNode_1.RightArmHit_A.AnimGraphNode_StateMachine_1.RightArmHitChooser_StartTime_DelayFloat.AnimStateNode_6.Hit_Ride_R.AnimGraphNode_StateMachine_0.Hit_Ride_R_SM.AnimStateTransitionNode_6.Transition
    src=Conduit
    via=AnimStateTransitionNode_6
   }
   {
    dst=Hit_Medium
    rule=AnimGraph.AnimGraphNode_StateMachine_6.RightArmAdditiveHits.AnimStateNode_1.RightArmHit_A.AnimGraphNode_StateMachine_1.RightArmHitChooser_StartTime_DelayFloat.AnimStateNode_6.Hit_Ride_R.AnimGraphNode_StateMachine_0.Hit_Ride_R_SM.AnimStateTransitionNode_5.Transition
    src=Conduit
    via=AnimStateTransitionNode_5
   }
   {
    dst=Hit_Short
    rule=AnimGraph.AnimGraphNode_StateMachine_6.RightArmAdditiveHits.AnimStateNode_1.RightArmHit_A.AnimGraphNode_StateMachine_1.RightArmHitChooser_StartTime_DelayFloat.AnimStateNode_6.Hit_Ride_R.AnimGraphNode_StateMachine_0.Hit_Ride_R_SM.AnimStateTransitionNode_4.Transition
    src=Conduit
    via=AnimStateTransitionNode_4
   }
  ]
 }
 Conduit {
  path=AnimGraph.AnimGraphNode_StateMachine_6.RightArmAdditiveHits.AnimStateNode_1.RightArmHit_A.AnimGraphNode_StateMachine_1.RightArmHitChooser_StartTime_DelayFloat.AnimStateNode_6.Hit_Ride_R.AnimGraphNode_StateMachine_0.Hit_Ride_R_SM.AnimStateConduitNode_0.Conduit
  nodes [Result(bCanEnterTransition=true)]
 }
 Hit_Short {
  path=AnimGraph.AnimGraphNode_StateMachine_6.RightArmAdditiveHits.AnimStateNode_1.RightArmHit_A.AnimGraphNode_StateMachine_1.RightArmHitChooser_StartTime_DelayFloat.AnimStateNode_6.Hit_Ride_R.AnimGraphNode_StateMachine_0.Hit_Ride_R_SM.AnimStateNode_0.Hit_Short
  nodes [Output Animation Pose(Result="Sequence Player (additive) 'Drum_Hit_Ride_Tap_Fast'"(StartPosition=Get RightArmMaxDelayShortenedBy(), bLoopAnimation=false) {sequence=&Drum_Hit_Ride_Tap_Fast, startPosition=0.05})]
 }
 Hit_Medium {
  path=AnimGraph.AnimGraphNode_StateMachine_6.RightArmAdditiveHits.AnimStateNode_1.RightArmHit_A.AnimGraphNode_StateMachine_1.RightArmHitChooser_StartTime_DelayFloat.AnimStateNode_6.Hit_Ride_R.AnimGraphNode_StateMachine_0.Hit_Ride_R_SM.AnimStateNode_1.Hit_Medium
  nodes [
   n1 = "Sequence Player (additive) 'Drum_Hit_Ride_Normal_Medium'"(StartPosition=Get RightArmMaxDelayShortenedBy(), bLoopAnimation=false) {sequence=&Drum_Hit_Ride_Normal_Medium, startPosition=0.05}
   Output Animation Pose(Result=@n1)
  ]
 }
 Hit_Long {
  path=AnimGraph.AnimGraphNode_StateMachine_6.RightArmAdditiveHits.AnimStateNode_1.RightArmHit_A.AnimGraphNode_StateMachine_1.RightArmHitChooser_StartTime_DelayFloat.AnimStateNode_6.Hit_Ride_R.AnimGraphNode_StateMachine_0.Hit_Ride_R_SM.AnimStateNode_2.Hit_Long
  nodes [Output Animation Pose(Result="Sequence Player (additive) 'Drum_Hit_Ride_Slow_Big'"(StartPosition=Get RightArmMaxDelayShortenedBy(), bLoopAnimation=false) {sequence=&Drum_Hit_Ride_Slow_Big, startPosition=0.05})]
 }
 Transition {
  path=AnimGraph.AnimGraphNode_StateMachine_6.RightArmAdditiveHits.AnimStateNode_1.RightArmHit_A.AnimGraphNode_StateMachine_1.RightArmHitChooser_StartTime_DelayFloat.AnimStateNode_6.Hit_Ride_R.AnimGraphNode_StateMachine_0.Hit_Ride_R_SM.AnimStateTransitionNode_4.Transition
  nodes [Result(bCanEnterTransition=Get RightArmIsShortDelay())]
 }
 Transition {
  path=AnimGraph.AnimGraphNode_StateMachine_6.RightArmAdditiveHits.AnimStateNode_1.RightArmHit_A.AnimGraphNode_StateMachine_1.RightArmHitChooser_StartTime_DelayFloat.AnimStateNode_6.Hit_Ride_R.AnimGraphNode_StateMachine_0.Hit_Ride_R_SM.AnimStateTransitionNode_5.Transition
  nodes [
   n1 = Get RightArmIsLongDelay()
   n2 = Get RightArmIsShortDelay()
   Result(bCanEnterTransition=AND Boolean(A=NOT Boolean(A=@n2), B=NOT Boolean(A=@n1)))
  ]
 }
 Transition {
  path=AnimGraph.AnimGraphNode_StateMachine_6.RightArmAdditiveHits.AnimStateNode_1.RightArmHit_A.AnimGraphNode_StateMachine_1.RightArmHitChooser_StartTime_DelayFloat.AnimStateNode_6.Hit_Ride_R.AnimGraphNode_StateMachine_0.Hit_Ride_R_SM.AnimStateTransitionNode_6.Transition
  nodes [Result(bCanEnterTransition=Get RightArmIsLongDelay())]
 }
 Hit_Tom1_R {
  path=AnimGraph.AnimGraphNode_StateMachine_6.RightArmAdditiveHits.AnimStateNode_1.RightArmHit_A.AnimGraphNode_StateMachine_1.RightArmHitChooser_StartTime_DelayFloat.AnimStateNode_7.Hit_Tom1_R
  nodes [
   n1 = Get RightArmMaxDelayShortenedBy()
   n2 = "Sequence Player (additive) 'Drum_Hit_Tom1_Normal_Medium'"(StartPosition=Get RightArmMaxDelayShortenedBy(), bLoopAnimation=false) {sequence=&Drum_Hit_Tom1_Normal_Medium, startPosition=0.05}
   Blend Poses by bool(
    BlendPose_0="Sequence Player (additive) 'Drum_Hit_Tom1_Slow_Big'"(StartPosition=Get RightArmMaxDelayShortenedBy(), bLoopAnimation=false) {sequence=&Drum_Hit_Tom1_Slow_Big, startPosition=0.05}
    BlendPose_1=Blend Poses by bool(
     BlendPose_0="Sequence Player (additive) 'Drum_Hit_Tom1_Tap_Fast'"(StartPosition=@n1, bLoopAnimation=false) {sequence=&Drum_Hit_Tom1_Tap_Fast, startPosition=0.05}
     BlendPose_1=@n2
     BlendTime_0=0
     BlendTime_1=0
     bActiveValue=Get RightArmIsShortDelay()
    ) {blendType=Linear}
    BlendTime_0=0
    BlendTime_1=0
    bActiveValue=Get RightArmIsLongDelay()
   ) {blendType=Linear}
   AnimGraphNode_StateMachine_0 = Hit_Tom1_R_SM() {bAllowConduitEntryStates=true}
   Output Animation Pose(Result=@AnimGraphNode_StateMachine_0)
  ]
 }
 Hit_Tom1_R_SM {
  entry=Conduit
  path=AnimGraph.AnimGraphNode_StateMachine_6.RightArmAdditiveHits.AnimStateNode_1.RightArmHit_A.AnimGraphNode_StateMachine_1.RightArmHitChooser_StartTime_DelayFloat.AnimStateNode_7.Hit_Tom1_R.AnimGraphNode_StateMachine_0.Hit_Tom1_R_SM
  ids=id
  nodes [
   AnimStateConduitNode_0 = Conduit()
   AnimStateNode_0 = Hit_Short()
   AnimStateNode_1 = Hit_Medium()
   AnimStateNode_2 = Hit_Long()
   AnimStateTransitionNode_4 = AnimStateTransitionNode() {crossfadeDuration=0}
   AnimStateTransitionNode_5 = AnimStateTransitionNode() {crossfadeDuration=0}
   AnimStateTransitionNode_6 = AnimStateTransitionNode() {crossfadeDuration=0}
  ]
  links [
   {
    dst=Hit_Long
    rule=AnimGraph.AnimGraphNode_StateMachine_6.RightArmAdditiveHits.AnimStateNode_1.RightArmHit_A.AnimGraphNode_StateMachine_1.RightArmHitChooser_StartTime_DelayFloat.AnimStateNode_7.Hit_Tom1_R.AnimGraphNode_StateMachine_0.Hit_Tom1_R_SM.AnimStateTransitionNode_6.Transition
    src=Conduit
    via=AnimStateTransitionNode_6
   }
   {
    dst=Hit_Medium
    rule=AnimGraph.AnimGraphNode_StateMachine_6.RightArmAdditiveHits.AnimStateNode_1.RightArmHit_A.AnimGraphNode_StateMachine_1.RightArmHitChooser_StartTime_DelayFloat.AnimStateNode_7.Hit_Tom1_R.AnimGraphNode_StateMachine_0.Hit_Tom1_R_SM.AnimStateTransitionNode_5.Transition
    src=Conduit
    via=AnimStateTransitionNode_5
   }
   {
    dst=Hit_Short
    rule=AnimGraph.AnimGraphNode_StateMachine_6.RightArmAdditiveHits.AnimStateNode_1.RightArmHit_A.AnimGraphNode_StateMachine_1.RightArmHitChooser_StartTime_DelayFloat.AnimStateNode_7.Hit_Tom1_R.AnimGraphNode_StateMachine_0.Hit_Tom1_R_SM.AnimStateTransitionNode_4.Transition
    src=Conduit
    via=AnimStateTransitionNode_4
   }
  ]
 }
 Conduit {
  path=AnimGraph.AnimGraphNode_StateMachine_6.RightArmAdditiveHits.AnimStateNode_1.RightArmHit_A.AnimGraphNode_StateMachine_1.RightArmHitChooser_StartTime_DelayFloat.AnimStateNode_7.Hit_Tom1_R.AnimGraphNode_StateMachine_0.Hit_Tom1_R_SM.AnimStateConduitNode_0.Conduit
  nodes [Result(bCanEnterTransition=true)]
 }
 Hit_Short {
  path=AnimGraph.AnimGraphNode_StateMachine_6.RightArmAdditiveHits.AnimStateNode_1.RightArmHit_A.AnimGraphNode_StateMachine_1.RightArmHitChooser_StartTime_DelayFloat.AnimStateNode_7.Hit_Tom1_R.AnimGraphNode_StateMachine_0.Hit_Tom1_R_SM.AnimStateNode_0.Hit_Short
  nodes [Output Animation Pose(Result="Sequence Player (additive) 'Drum_Hit_Tom1_Tap_Fast'"(StartPosition=Get RightArmMaxDelayShortenedBy(), bLoopAnimation=false) {sequence=&Drum_Hit_Tom1_Tap_Fast, startPosition=0.05})]
 }
 Hit_Medium {
  path=AnimGraph.AnimGraphNode_StateMachine_6.RightArmAdditiveHits.AnimStateNode_1.RightArmHit_A.AnimGraphNode_StateMachine_1.RightArmHitChooser_StartTime_DelayFloat.AnimStateNode_7.Hit_Tom1_R.AnimGraphNode_StateMachine_0.Hit_Tom1_R_SM.AnimStateNode_1.Hit_Medium
  nodes [
   n1 = "Sequence Player (additive) 'Drum_Hit_Tom1_Normal_Medium'"(StartPosition=Get RightArmMaxDelayShortenedBy(), bLoopAnimation=false) {sequence=&Drum_Hit_Tom1_Normal_Medium, startPosition=0.05}
   Output Animation Pose(Result=@n1)
  ]
 }
 Hit_Long {
  path=AnimGraph.AnimGraphNode_StateMachine_6.RightArmAdditiveHits.AnimStateNode_1.RightArmHit_A.AnimGraphNode_StateMachine_1.RightArmHitChooser_StartTime_DelayFloat.AnimStateNode_7.Hit_Tom1_R.AnimGraphNode_StateMachine_0.Hit_Tom1_R_SM.AnimStateNode_2.Hit_Long
  nodes [Output Animation Pose(Result="Sequence Player (additive) 'Drum_Hit_Tom1_Slow_Big'"(StartPosition=Get RightArmMaxDelayShortenedBy(), bLoopAnimation=false) {sequence=&Drum_Hit_Tom1_Slow_Big, startPosition=0.05})]
 }
 Transition {
  path=AnimGraph.AnimGraphNode_StateMachine_6.RightArmAdditiveHits.AnimStateNode_1.RightArmHit_A.AnimGraphNode_StateMachine_1.RightArmHitChooser_StartTime_DelayFloat.AnimStateNode_7.Hit_Tom1_R.AnimGraphNode_StateMachine_0.Hit_Tom1_R_SM.AnimStateTransitionNode_4.Transition
  nodes [Result(bCanEnterTransition=Get RightArmIsShortDelay())]
 }
 Transition {
  path=AnimGraph.AnimGraphNode_StateMachine_6.RightArmAdditiveHits.AnimStateNode_1.RightArmHit_A.AnimGraphNode_StateMachine_1.RightArmHitChooser_StartTime_DelayFloat.AnimStateNode_7.Hit_Tom1_R.AnimGraphNode_StateMachine_0.Hit_Tom1_R_SM.AnimStateTransitionNode_5.Transition
  nodes [
   n1 = Get RightArmIsLongDelay()
   n2 = Get RightArmIsShortDelay()
   Result(bCanEnterTransition=AND Boolean(A=NOT Boolean(A=@n2), B=NOT Boolean(A=@n1)))
  ]
 }
 Transition {
  path=AnimGraph.AnimGraphNode_StateMachine_6.RightArmAdditiveHits.AnimStateNode_1.RightArmHit_A.AnimGraphNode_StateMachine_1.RightArmHitChooser_StartTime_DelayFloat.AnimStateNode_7.Hit_Tom1_R.AnimGraphNode_StateMachine_0.Hit_Tom1_R_SM.AnimStateTransitionNode_6.Transition
  nodes [Result(bCanEnterTransition=Get RightArmIsLongDelay())]
 }
 Hit_Tom2_R {
  path=AnimGraph.AnimGraphNode_StateMachine_6.RightArmAdditiveHits.AnimStateNode_1.RightArmHit_A.AnimGraphNode_StateMachine_1.RightArmHitChooser_StartTime_DelayFloat.AnimStateNode_8.Hit_Tom2_R
  nodes [
   n1 = Get RightArmMaxDelayShortenedBy()
   n2 = "Sequence Player (additive) 'Drum_Hit_Tom2_Normal_Medium'"(StartPosition=Get RightArmMaxDelayShortenedBy(), bLoopAnimation=false) {sequence=&Drum_Hit_Tom2_Normal_Medium, startPosition=0.05}
   Blend Poses by bool(
    BlendPose_0="Sequence Player (additive) 'Drum_Hit_Tom2_Slow_Big'"(StartPosition=Get RightArmMaxDelayShortenedBy(), bLoopAnimation=false) {sequence=&Drum_Hit_Tom2_Slow_Big, startPosition=0.05}
    BlendPose_1=Blend Poses by bool(
     BlendPose_0="Sequence Player (additive) 'Drum_Hit_Tom2_Tap_Fast'"(StartPosition=@n1, bLoopAnimation=false) {sequence=&Drum_Hit_Tom2_Tap_Fast, startPosition=0.05}
     BlendPose_1=@n2
     BlendTime_0=0
     BlendTime_1=0
     bActiveValue=Get RightArmIsShortDelay()
    ) {blendType=Linear}
    BlendTime_0=0
    BlendTime_1=0
    bActiveValue=Get RightArmIsLongDelay()
   ) {blendType=Linear}
   AnimGraphNode_StateMachine_0 = Hit_Tom2_R_SM() {bAllowConduitEntryStates=true}
   Output Animation Pose(Result=@AnimGraphNode_StateMachine_0)
  ]
 }
 Hit_Tom2_R_SM {
  entry=Conduit
  path=AnimGraph.AnimGraphNode_StateMachine_6.RightArmAdditiveHits.AnimStateNode_1.RightArmHit_A.AnimGraphNode_StateMachine_1.RightArmHitChooser_StartTime_DelayFloat.AnimStateNode_8.Hit_Tom2_R.AnimGraphNode_StateMachine_0.Hit_Tom2_R_SM
  ids=id
  nodes [
   AnimStateConduitNode_0 = Conduit()
   AnimStateNode_0 = Hit_Short()
   AnimStateNode_1 = Hit_Medium()
   AnimStateNode_2 = Hit_Long()
   AnimStateTransitionNode_4 = AnimStateTransitionNode() {crossfadeDuration=0}
   AnimStateTransitionNode_5 = AnimStateTransitionNode() {crossfadeDuration=0}
   AnimStateTransitionNode_6 = AnimStateTransitionNode() {crossfadeDuration=0}
  ]
  links [
   {
    dst=Hit_Long
    rule=AnimGraph.AnimGraphNode_StateMachine_6.RightArmAdditiveHits.AnimStateNode_1.RightArmHit_A.AnimGraphNode_StateMachine_1.RightArmHitChooser_StartTime_DelayFloat.AnimStateNode_8.Hit_Tom2_R.AnimGraphNode_StateMachine_0.Hit_Tom2_R_SM.AnimStateTransitionNode_6.Transition
    src=Conduit
    via=AnimStateTransitionNode_6
   }
   {
    dst=Hit_Medium
    rule=AnimGraph.AnimGraphNode_StateMachine_6.RightArmAdditiveHits.AnimStateNode_1.RightArmHit_A.AnimGraphNode_StateMachine_1.RightArmHitChooser_StartTime_DelayFloat.AnimStateNode_8.Hit_Tom2_R.AnimGraphNode_StateMachine_0.Hit_Tom2_R_SM.AnimStateTransitionNode_5.Transition
    src=Conduit
    via=AnimStateTransitionNode_5
   }
   {
    dst=Hit_Short
    rule=AnimGraph.AnimGraphNode_StateMachine_6.RightArmAdditiveHits.AnimStateNode_1.RightArmHit_A.AnimGraphNode_StateMachine_1.RightArmHitChooser_StartTime_DelayFloat.AnimStateNode_8.Hit_Tom2_R.AnimGraphNode_StateMachine_0.Hit_Tom2_R_SM.AnimStateTransitionNode_4.Transition
    src=Conduit
    via=AnimStateTransitionNode_4
   }
  ]
 }
 Conduit {
  path=AnimGraph.AnimGraphNode_StateMachine_6.RightArmAdditiveHits.AnimStateNode_1.RightArmHit_A.AnimGraphNode_StateMachine_1.RightArmHitChooser_StartTime_DelayFloat.AnimStateNode_8.Hit_Tom2_R.AnimGraphNode_StateMachine_0.Hit_Tom2_R_SM.AnimStateConduitNode_0.Conduit
  nodes [Result(bCanEnterTransition=true)]
 }
 Hit_Short {
  path=AnimGraph.AnimGraphNode_StateMachine_6.RightArmAdditiveHits.AnimStateNode_1.RightArmHit_A.AnimGraphNode_StateMachine_1.RightArmHitChooser_StartTime_DelayFloat.AnimStateNode_8.Hit_Tom2_R.AnimGraphNode_StateMachine_0.Hit_Tom2_R_SM.AnimStateNode_0.Hit_Short
  nodes [Output Animation Pose(Result="Sequence Player (additive) 'Drum_Hit_Tom2_Tap_Fast'"(StartPosition=Get RightArmMaxDelayShortenedBy(), bLoopAnimation=false) {sequence=&Drum_Hit_Tom2_Tap_Fast, startPosition=0.05})]
 }
 Hit_Medium {
  path=AnimGraph.AnimGraphNode_StateMachine_6.RightArmAdditiveHits.AnimStateNode_1.RightArmHit_A.AnimGraphNode_StateMachine_1.RightArmHitChooser_StartTime_DelayFloat.AnimStateNode_8.Hit_Tom2_R.AnimGraphNode_StateMachine_0.Hit_Tom2_R_SM.AnimStateNode_1.Hit_Medium
  nodes [
   n1 = "Sequence Player (additive) 'Drum_Hit_Tom2_Normal_Medium'"(StartPosition=Get RightArmMaxDelayShortenedBy(), bLoopAnimation=false) {sequence=&Drum_Hit_Tom2_Normal_Medium, startPosition=0.05}
   Output Animation Pose(Result=@n1)
  ]
 }
 Hit_Long {
  path=AnimGraph.AnimGraphNode_StateMachine_6.RightArmAdditiveHits.AnimStateNode_1.RightArmHit_A.AnimGraphNode_StateMachine_1.RightArmHitChooser_StartTime_DelayFloat.AnimStateNode_8.Hit_Tom2_R.AnimGraphNode_StateMachine_0.Hit_Tom2_R_SM.AnimStateNode_2.Hit_Long
  nodes [Output Animation Pose(Result="Sequence Player (additive) 'Drum_Hit_Tom2_Slow_Big'"(StartPosition=Get RightArmMaxDelayShortenedBy(), bLoopAnimation=false) {sequence=&Drum_Hit_Tom2_Slow_Big, startPosition=0.05})]
 }
 Transition {
  path=AnimGraph.AnimGraphNode_StateMachine_6.RightArmAdditiveHits.AnimStateNode_1.RightArmHit_A.AnimGraphNode_StateMachine_1.RightArmHitChooser_StartTime_DelayFloat.AnimStateNode_8.Hit_Tom2_R.AnimGraphNode_StateMachine_0.Hit_Tom2_R_SM.AnimStateTransitionNode_4.Transition
  nodes [Result(bCanEnterTransition=Get RightArmIsShortDelay())]
 }
 Transition {
  path=AnimGraph.AnimGraphNode_StateMachine_6.RightArmAdditiveHits.AnimStateNode_1.RightArmHit_A.AnimGraphNode_StateMachine_1.RightArmHitChooser_StartTime_DelayFloat.AnimStateNode_8.Hit_Tom2_R.AnimGraphNode_StateMachine_0.Hit_Tom2_R_SM.AnimStateTransitionNode_5.Transition
  nodes [
   n1 = Get RightArmIsLongDelay()
   n2 = Get RightArmIsShortDelay()
   Result(bCanEnterTransition=AND Boolean(A=NOT Boolean(A=@n2), B=NOT Boolean(A=@n1)))
  ]
 }
 Transition {
  path=AnimGraph.AnimGraphNode_StateMachine_6.RightArmAdditiveHits.AnimStateNode_1.RightArmHit_A.AnimGraphNode_StateMachine_1.RightArmHitChooser_StartTime_DelayFloat.AnimStateNode_8.Hit_Tom2_R.AnimGraphNode_StateMachine_0.Hit_Tom2_R_SM.AnimStateTransitionNode_6.Transition
  nodes [Result(bCanEnterTransition=Get RightArmIsLongDelay())]
 }
 Hit_Percussion_R {
  path=AnimGraph.AnimGraphNode_StateMachine_6.RightArmAdditiveHits.AnimStateNode_1.RightArmHit_A.AnimGraphNode_StateMachine_1.RightArmHitChooser_StartTime_DelayFloat.AnimStateNode_9.Hit_Percussion_R
  nodes [Output Animation Pose(Result=Additive Identity Pose())]
 }
 Transition {
  path=AnimGraph.AnimGraphNode_StateMachine_6.RightArmAdditiveHits.AnimStateNode_1.RightArmHit_A.AnimGraphNode_StateMachine_1.RightArmHitChooser_StartTime_DelayFloat.AnimStateTransitionNode_0.Transition
  nodes [Result(bCanEnterTransition=Get Hit_Choke1_R())]
 }
 Transition {
  path=AnimGraph.AnimGraphNode_StateMachine_6.RightArmAdditiveHits.AnimStateNode_1.RightArmHit_A.AnimGraphNode_StateMachine_1.RightArmHitChooser_StartTime_DelayFloat.AnimStateTransitionNode_1.Transition
  nodes [Result(bCanEnterTransition=Get Hit_Choke2_R())]
 }
 Transition {
  path=AnimGraph.AnimGraphNode_StateMachine_6.RightArmAdditiveHits.AnimStateNode_1.RightArmHit_A.AnimGraphNode_StateMachine_1.RightArmHitChooser_StartTime_DelayFloat.AnimStateTransitionNode_30.Transition
  nodes [Result(bCanEnterTransition=Get Hit_Snare_R())]
 }
 Transition {
  path=AnimGraph.AnimGraphNode_StateMachine_6.RightArmAdditiveHits.AnimStateNode_1.RightArmHit_A.AnimGraphNode_StateMachine_1.RightArmHitChooser_StartTime_DelayFloat.AnimStateTransitionNode_32.Transition
  nodes [Result(bCanEnterTransition=Get Hit_HiHat_R())]
 }
 Transition {
  path=AnimGraph.AnimGraphNode_StateMachine_6.RightArmAdditiveHits.AnimStateNode_1.RightArmHit_A.AnimGraphNode_StateMachine_1.RightArmHitChooser_StartTime_DelayFloat.AnimStateTransitionNode_34.Transition
  nodes [Result(bCanEnterTransition=Get Hit_Crash1_Hard_R())]
 }
 Transition {
  path=AnimGraph.AnimGraphNode_StateMachine_6.RightArmAdditiveHits.AnimStateNode_1.RightArmHit_A.AnimGraphNode_StateMachine_1.RightArmHitChooser_StartTime_DelayFloat.AnimStateTransitionNode_36.Transition
  nodes [Result(bCanEnterTransition=Get Hit_Crash1_Soft_R())]
 }
 Transition {
  path=AnimGraph.AnimGraphNode_StateMachine_6.RightArmAdditiveHits.AnimStateNode_1.RightArmHit_A.AnimGraphNode_StateMachine_1.RightArmHitChooser_StartTime_DelayFloat.AnimStateTransitionNode_38.Transition
  nodes [Result(bCanEnterTransition=Get Hit_Crash2_Hard_R())]
 }
 Transition {
  path=AnimGraph.AnimGraphNode_StateMachine_6.RightArmAdditiveHits.AnimStateNode_1.RightArmHit_A.AnimGraphNode_StateMachine_1.RightArmHitChooser_StartTime_DelayFloat.AnimStateTransitionNode_40.Transition
  nodes [Result(bCanEnterTransition=Get Hit_Crash2_Soft_R())]
 }
 Transition {
  path=AnimGraph.AnimGraphNode_StateMachine_6.RightArmAdditiveHits.AnimStateNode_1.RightArmHit_A.AnimGraphNode_StateMachine_1.RightArmHitChooser_StartTime_DelayFloat.AnimStateTransitionNode_42.Transition
  nodes [Result(bCanEnterTransition=Get Hit_Ride_R())]
 }
 Transition {
  path=AnimGraph.AnimGraphNode_StateMachine_6.RightArmAdditiveHits.AnimStateNode_1.RightArmHit_A.AnimGraphNode_StateMachine_1.RightArmHitChooser_StartTime_DelayFloat.AnimStateTransitionNode_44.Transition
  nodes [Result(bCanEnterTransition=Get Hit_Tom1_R())]
 }
 Transition {
  path=AnimGraph.AnimGraphNode_StateMachine_6.RightArmAdditiveHits.AnimStateNode_1.RightArmHit_A.AnimGraphNode_StateMachine_1.RightArmHitChooser_StartTime_DelayFloat.AnimStateTransitionNode_46.Transition
  nodes [Result(bCanEnterTransition=Get Hit_Tom2_R())]
 }
 Transition {
  path=AnimGraph.AnimGraphNode_StateMachine_6.RightArmAdditiveHits.AnimStateNode_1.RightArmHit_A.AnimGraphNode_StateMachine_1.RightArmHitChooser_StartTime_DelayFloat.AnimStateTransitionNode_48.Transition
  nodes [Result(bCanEnterTransition=Get Hit_FloorTom_R())]
 }
 Transition {
  path=AnimGraph.AnimGraphNode_StateMachine_6.RightArmAdditiveHits.AnimStateNode_1.RightArmHit_A.AnimGraphNode_StateMachine_1.RightArmHitChooser_StartTime_DelayFloat.AnimStateTransitionNode_50.Transition
  nodes [Result(bCanEnterTransition=Get Hit_Percussion_R())]
 }
 Transition {
  path=AnimGraph.AnimGraphNode_StateMachine_6.RightArmAdditiveHits.AnimStateNode_1.RightArmHit_A.AnimGraphNode_StateMachine_1.RightArmHitChooser_StartTime_DelayFloat.AnimStateTransitionNode_7.Transition
  nodes [Result(bCanEnterTransition=Get Hit_Snare_Soft_R())]
 }
 Neutral {
  path=AnimGraph.AnimGraphNode_StateMachine_6.RightArmAdditiveHits.AnimStateNode_2.Neutral
  nodes [
   "Sequence Player (additive) 'Drum_SpineTwist_Center_Pose'"() {bLoopAnimation=false, sequence=&Drum_SpineTwist_Center_Pose}
   Output Animation Pose(Result=Additive Identity Pose())
  ]
 }
 Transition {
  path=AnimGraph.AnimGraphNode_StateMachine_6.RightArmAdditiveHits.AnimStateTransitionNode_0.Transition
  nodes [
   n1 = Get RightArmHitLongBlend() -> (RightArmHitLongBlend:bool)
   n2 = Get RightArmHitMediumBlend() -> (RightArmHitMediumBlend:bool)
   Result(bCanEnterTransition=AND Boolean!(A=Get TriggerRightArmHit_B() -> (TriggerRightArmHit_B:bool), B=NOT Boolean(A=@n1), C:bool=NOT Boolean(A=@n2)) -> (ReturnValue))
  ]
 }
 Transition {
  path=AnimGraph.AnimGraphNode_StateMachine_6.RightArmAdditiveHits.AnimStateTransitionNode_1.Transition
  nodes [
   n1 = Get RightArmHitLongBlend()
   n2 = Get RightArmHitMediumBlend()
   Result(bCanEnterTransition=AND Boolean(A=Get TriggerRightArmHit_A() -> (TriggerRightArmHit_A:bool), B=NOT Boolean(A=@n1), C=NOT Boolean(A=@n2)))
   Get RightArmIsShortDelay()
  ]
 }
 Transition {
  path=AnimGraph.AnimGraphNode_StateMachine_6.RightArmAdditiveHits.AnimStateTransitionNode_11.Transition
  nodes [
   n1 = Get TriggerRightArmHit_A()
   Result(bCanEnterTransition=@n1)
   OR Boolean(A=@n1, B=Get TriggerRightArmHit_B())
  ]
 }
 Transition {path=AnimGraph.AnimGraphNode_StateMachine_6.RightArmAdditiveHits.AnimStateTransitionNode_2.Transition, nodes [Result()]}
 Transition {path=AnimGraph.AnimGraphNode_StateMachine_6.RightArmAdditiveHits.AnimStateTransitionNode_3.Transition, nodes [Result()]}
 Transition {
  path=AnimGraph.AnimGraphNode_StateMachine_6.RightArmAdditiveHits.AnimStateTransitionNode_4.Transition
  nodes [
   n1 = Get TriggerRightArmHit_B()
   Result(bCanEnterTransition=@n1)
   OR Boolean(A=Get TriggerRightArmHit_A(), B=@n1)
  ]
 }
 Transition {
  path=AnimGraph.AnimGraphNode_StateMachine_6.RightArmAdditiveHits.AnimStateTransitionNode_5.Transition
  nodes [
   n1 = Get RightArmHitLongBlend()
   Result(bCanEnterTransition=AND Boolean(A=Get TriggerRightArmHit_A(), B=NOT Boolean(A=@n1), C=Get RightArmHitMediumBlend()))
   Get RightArmIsShortDelay()
  ]
 }
 Transition {
  path=AnimGraph.AnimGraphNode_StateMachine_6.RightArmAdditiveHits.AnimStateTransitionNode_6.Transition
  nodes [
   n1 = Get RightArmHitLongBlend()
   Result(bCanEnterTransition=AND Boolean(A=Get TriggerRightArmHit_B(), B=NOT Boolean(A=@n1), C=Get RightArmHitMediumBlend()))
  ]
 }
 Transition {
  path=AnimGraph.AnimGraphNode_StateMachine_6.RightArmAdditiveHits.AnimStateTransitionNode_7.Transition
  nodes [
   Result(bCanEnterTransition=AND Boolean(A=Get TriggerRightArmHit_A(), B=Get RightArmHitLongBlend(), C=Get RightArmHitMediumBlend()))
   NOT Boolean(A=false)
  ]
 }
 Transition {
  path=AnimGraph.AnimGraphNode_StateMachine_6.RightArmAdditiveHits.AnimStateTransitionNode_8.Transition
  nodes [Result(bCanEnterTransition=AND Boolean(A=Get TriggerRightArmHit_B(), B=Get RightArmHitLongBlend(), C=Get RightArmHitMediumBlend()))]
 }
 ClearMidiListener {nodes [
   ClearMidiListener() {is=title K2Node_FunctionEntry}
   n1 = Get IntensityMidiListener() -> (IntensityMidiListener:Async Action Listen For Midi Text Event Object Reference)
   > Is Valid(InputObject=@n1)
   > Cancel(self:Cancellable Async Action Object Reference=@n1)
   > Set IntensityMidiListener(IntensityMidiListener:Async Action Listen For Midi Text Event Object Reference) -> (Output_Get:Async Action Listen For Midi Text Event Object Reference)
  ]}
 DrumKit Cowbell Hit {nodes [DrumKit Cowbell Hit() {is=title K2Node_FunctionEntry}]}
 DrumKit Crash1 Choke {nodes [DrumKit Crash1 Choke() {is=title K2Node_FunctionEntry}]}
 DrumKit Crash1 Hard Hit {nodes [DrumKit Crash1 Hard Hit() {is=title K2Node_FunctionEntry}]}
 DrumKit Crash1 Soft Hit {nodes [DrumKit Crash1 Soft Hit() {is=title K2Node_FunctionEntry}]}
 DrumKit Crash2 Choke {nodes [DrumKit Crash2 Choke() {is=title K2Node_FunctionEntry}]}
 DrumKit Crash2 Hard Hit {nodes [DrumKit Crash2 Hard Hit() {is=title K2Node_FunctionEntry}]}
 DrumKit Crash2 Soft Hit {nodes [DrumKit Crash2 Soft Hit() {is=title K2Node_FunctionEntry}]}
 DrumKit FloorTom Hit {nodes [DrumKit FloorTom Hit() {is=title K2Node_FunctionEntry}]}
 DrumKit HiHat Hit {nodes [DrumKit HiHat Hit() {is=title K2Node_FunctionEntry}]}
 DrumKit HiHat Pedal Up {nodes [DrumKit HiHat Pedal Up() {is=title K2Node_FunctionEntry}]}
 DrumKit Kick Hit {nodes [DrumKit Kick Hit() {is=title K2Node_FunctionEntry}]}
 DrumKit Ride Hard Hit {nodes [DrumKit Ride Hard Hit() {is=title K2Node_FunctionEntry}]}
 DrumKit Snare Hit {nodes [DrumKit Snare Hit() {is=title K2Node_FunctionEntry}]}
 DrumKit Tom1 Hit {nodes [DrumKit Tom1 Hit() {is=title K2Node_FunctionEntry}]}
 DrumKit Tom2 Hit {nodes [DrumKit Tom2 Hit() {is=title K2Node_FunctionEntry}]}
 EventGraph {nodes [
   Delay(Duration=8)
   > Set IsIdleBeatSynced(IsIdleBeatSynced:bool=true) -> (Output_Get:bool)
   > Delay(Duration=8)
   > Set IsIdleBeatSynced(IsIdleBeatSynced=false)
   SetLeftArmHitAngle() {is=title K2Node_CustomEvent}
   n1 = Get LeftArmHitLocation() -> (LeftArmHitLocation:Vector)
   n2 = BreakVector(InVec=Normalize!(A:Vector=@n1, Tolerance:float=0.0001) -> (ReturnValue:Vector) {is=title K2Node_CallFunction})
   > n3 = Set LeftArmHitAngle(LeftArmHitAngle:double=float * float(A="Atan2 (Degrees)"!(Y:double=@n2, X:double=@n2.Y) -> (ReturnValue:double), B=-1) {is=title K2Node_PromotableOperator}) -> (Output_Get:double)
   n4 = Get RightArmHitLocation() -> (RightArmHitLocation:Vector)
   n5 = BreakVector(InVec=Normalize(A=@n4, Tolerance=0.0001))
   n6 = Set RightArmHitAngle(execute=@n7.then, RightArmHitAngle:double=float * float(A="Atan2 (Degrees)"(Y=@n5, X=@n5.Y), B=-1)) -> (Output_Get:double)
   Set SpineTwist(execute=[@n3.then, @n6.then], SpineTwist:double=SafeDivide(A=float + float(A=@n3, B=@n6), B=2)) -> (Output_Get:double)
   n7 = SetRightArmHitAngle() {is=title K2Node_CustomEvent}
   n8 = Stage Prop Actor Changed!() -> (OutputDelegate, then, StageProp:Actor Object Reference) {is=title K2Node_CustomEvent}
   K2Node_Composite_1 = Twist Rotator Setup!(execute=[@K2Node_Composite_4.Out, @n8.then]) -> (then:Exec) {is=title K2Node_Composite}
   n9 = Event Bind_Crash2_SoftHit!() -> (OutputDelegate, then, Callback:"Delegate (by ref)") {is=title K2Node_Event}
   > Bind Event to Drum Kit Crash 2 Soft Hit(self:Sparks Drum Anim BP Object Reference, Delegate=@n9.Callback) {is=title K2Node_AddDelegate}
   AnimNotify_RightArm_Neutral_Entered() {is=title K2Node_Event}
   > Set RightArmIsPlayingHit(RightArmIsPlayingHit:bool=false) -> (Output_Get:bool)
   n10 = Event Bind_SnareHit!() -> (OutputDelegate, then, Callback:"Delegate (by ref)") {is=title K2Node_Event}
   > Bind Event to Drum Kit Snare Hit(self:Sparks Drum Anim BP Object Reference, Delegate=@n10.Callback) {is=title K2Node_AddDelegate}
   n11 = Event Bind_RideHardHit!() -> (OutputDelegate, then, Callback:"Delegate (by ref)") {is=title K2Node_Event}
   > Bind Event to Drum Kit Ride Hard Hit(self:Sparks Drum Anim BP Object Reference, Delegate=@n11.Callback) {is=title K2Node_AddDelegate}
   n12 = Event Bind_KickHit!() -> (OutputDelegate, then, Callback:"Delegate (by ref)") {is=title K2Node_Event}
   > Bind Event to Drum Kit Kick Hit(self:Sparks Drum Anim BP Object Reference, Delegate=@n12.Callback) {is=title K2Node_AddDelegate}
   n13 = Event Bind_HiHatPedalUp!() -> (OutputDelegate, then, Callback:"Delegate (by ref)") {is=title K2Node_Event}
   > Bind Event to Drum Kit Hi Hat Pedal Up(self:Sparks Drum Anim BP Object Reference, Delegate=@n13.Callback) {is=title K2Node_AddDelegate}
   n14 = Event Bind_HiHatHit!() -> (OutputDelegate, then, Callback:"Delegate (by ref)") {is=title K2Node_Event}
   > Bind Event to Drum Kit Hi Hat Hit(self:Sparks Drum Anim BP Object Reference, Delegate=@n14.Callback) {is=title K2Node_AddDelegate}
   n15 = Event Bind_FloorTomHit!() -> (OutputDelegate, then, Callback:"Delegate (by ref)") {is=title K2Node_Event}
   > Bind Event to Drum Kit Floor Tom Hit(self:Sparks Drum Anim BP Object Reference, Delegate=@n15.Callback) {is=title K2Node_AddDelegate}
   Event On New Song() {is=title K2Node_Event}
   > SetSongDefaults!(
    execute
    AnimInstance:Sparks Instrument Anim Instance Object Reference=GetSelfAsSparksInstrumentAnimInstance!() -> (ReturnValue:Sparks Instrument Anim Instance Object Reference)
   ) -> (then)
   Event SparksCharAttackEventReceived() {is=title K2Node_Event}
   > PrintString!(execute, InString="Play Attack Landed (Reaction) Anim here.", bPrintToScreen, bPrintToLog, TextColor, Duration, Key) -> (then)
   n16 = Event On MIDI Event Driver Changed!() -> (OutputDelegate, then, NewMidiEventDriver:"Midi Event Driver Interface Interface (by ref)") {is=title K2Node_Event}
   > n17 = Branch(Condition=IsValidInterface!(Interface:"Interface Interface (by ref)"=@n16.NewMidiEventDriver) -> (ReturnValue:bool))
   ClearMidiListener(execute=@n17.else) {is=title K2Node_CallFunction}
   AnimNotify_RightArmHit_A_Entered() {is=title K2Node_Event}
   > Set RightArmIsPlayingHit(RightArmIsPlayingHit=true)
   > Set TriggerRightArmHit_A(TriggerRightArmHit_A:bool=false) -> (Output_Get:bool)
   n18 = Event Bind_Crash1_SoftHit!() -> (OutputDelegate, then, Callback:"Delegate (by ref)") {is=title K2Node_Event}
   > Bind Event to Drum Kit Crash 1 Soft Hit(self:Sparks Drum Anim BP Object Reference, Delegate=@n18.Callback) {is=title K2Node_AddDelegate}
   n19 = Event Bind_Crash1_HardHit!() -> (OutputDelegate, then, Callback:"Delegate (by ref)") {is=title K2Node_Event}
   > Bind Event to Drum Kit Crash 1 Hard Hit(self:Sparks Drum Anim BP Object Reference, Delegate=@n19.Callback) {is=title K2Node_AddDelegate}
   n20 = Event Bind_Crash1_Choke!() -> (OutputDelegate, then, Callback:"Delegate (by ref)") {is=title K2Node_Event}
   > Bind Event to Drum Kit Crash 1 Choke(self:Sparks Drum Anim BP Object Reference, Delegate=@n20.Callback) {is=title K2Node_AddDelegate}
   n21 = Event Bind_CowbellHit!() -> (OutputDelegate, then, Callback:"Delegate (by ref)") {is=title K2Node_Event}
   > Bind Event to Drum Kit Cowbell Hit(self:Sparks Drum Anim BP Object Reference, Delegate=@n21.Callback) {is=title K2Node_AddDelegate}
   n22 = Event On LipSync Asset Changed!() -> (OutputDelegate, then, NewLipSyncAsset:Anim Sequence Object Reference) {is=title K2Node_Event}
   > n23 = Is Valid(InputObject=@n22.NewLipSyncAsset)
   Set LipSyncAnimSequence(execute=@n23.Is Not Valid, LipSyncAnimSequence:Anim Sequence Object Reference) -> (Output_Get:Anim Sequence Object Reference)
   Set LipSyncAnimSequence(execute=@n23.Is Valid, LipSyncAnimSequence=@n22.NewLipSyncAsset)
   n24 = Event On Set InstrumentPSC!() -> (
    OutputDelegate
    then
    OldInstrumentPSC:Sparks Instrument Anim Pawn Component Object Reference
    NewInstrumentPSC:Sparks Instrument Anim Pawn Component Object Reference
   ) {is=title K2Node_Event}
   > n25 = Sequence()
   > Is Valid(InputObject=@n24.OldInstrumentPSC)
   > Unbind Event from On Stage Prop Actor Changed(self:Sparks Instrument Anim Pawn Component Object Reference=@n24.OldInstrumentPSC, Delegate=@n8) {is=title K2Node_RemoveDelegate}
   Is Valid(exec=@n25.then_1, InputObject=@n24.NewInstrumentPSC)
   > Bind Event to On Stage Prop Actor Changed(self:Sparks Instrument Anim Pawn Component Object Reference=@n24.NewInstrumentPSC, Delegate=@n8) {is=title K2Node_AddDelegate}
   n26 = Event BlueprintUpdateAnimation()
   > DrumUpdateArmHit!(
    execute
    Input:"Float (double-precision) (by ref)"=Get LeftArmIsPlayingHitBlend()
    Target:bool=Get LeftArmIsPlayingHit() -> (LeftArmIsPlayingHit:bool)
    DeltaTime:double=@n26.DeltaTimeX
   ) -> (then)
   > DrumUpdateArmHit(Input=Get RightArmIsPlayingHitBlend(), Target=Get RightArmIsPlayingHit() -> (RightArmIsPlayingHit:bool), DeltaTime=@n26.DeltaTimeX)
   n27 = Event Bind_Crash2_HardHit!() -> (OutputDelegate, then, Callback:"Delegate (by ref)") {is=title K2Node_Event}
   > Bind Event to Drum Kit Crash 2 Hard Hit(self:Sparks Drum Anim BP Object Reference, Delegate=@n27.Callback) {is=title K2Node_AddDelegate}
   n28 = Event Bind_Crash2_Choke!() -> (OutputDelegate, then, Callback:"Delegate (by ref)") {is=title K2Node_Event}
   > Bind Event to Drum Kit Crash 2 Choke(self:Sparks Drum Anim BP Object Reference, Delegate=@n28.Callback) {is=title K2Node_AddDelegate}
   n29 = Event On MIDI Event Data Parsed!() -> (OutputDelegate, then, EventDriver:"Midi Event Driver Interface Interface (by ref)") {is=title K2Node_Event}
   > Set MidiTrackName(MidiTrackName:Name=SelectName!(
     A:Name=GetMidiTrackNameFromSparksInstrument!(self:Midi Event Driver Interface Interface=@n29.EventDriver, InstrumentType:ESparksInstrumentType Enum=Drum) -> (ReturnValue:Name=None)
     B:Name=None
     bPickA:bool=IsValidInterface(Interface=@n29.EventDriver)
    ) -> (ReturnValue:Name=None)) -> (Output_Get:Name=None)
   > K2Node_Composite_3 = Song Intensity!(In:Exec) -> (Out:Exec) {is=title K2Node_Composite}
   > K2Node_Composite_4 = Drum Hits!(In:Exec) -> (Out:Exec) {is=title K2Node_Composite}
   AnimNotify_RightArmHit_B_Entered() {is=title K2Node_Event}
   > Set RightArmIsPlayingHit(RightArmIsPlayingHit=true)
   > Set TriggerRightArmHit_B(TriggerRightArmHit_B:bool=false) -> (Output_Get:bool)
   Event SparksCharAttackEventLaunched() {is=title K2Node_Event}
   > PrintString(InString=Play Attack Launched Anim Here.)
   n30 = TryGetPawnOwner()
   > n31 = Get Fort Pawn From Object!(execute, InObject:Object Reference=@n30) -> (Valid:Exec, NotValid:Exec, FortPawn:Fort Pawn Object Reference) {is=title K2Node_CallFunction}
   > n32 = FindGender!(execute, Object:Object Reference=@n31) -> (True:Exec, False:Exec, OutGender:EFortCustomGender Enum=Invalid)
   > n33 = Switch on EFortCustomGender!(execute, Selection:EFortCustomGender Enum=@n32) -> (Invalid:Exec, Male:Exec, Female:Exec, Both:Exec) {is=title K2Node_SwitchEnum}
   PlayAnimMontage!(
    execute=[@n33.Both, @n33.Male]
    Actor:Actor Object Reference=@n30
    NewAnimMontage:Anim Montage Object Reference=/SparksCharacterFull/Animation/MainPlayer/BandPlay/Drum/Attacks/Drum_Attack_DoubleCrash_CMM_A_M
    InPlayRate:float=1
    StartSectionName:Name=Default
   ) -> (then, ReturnValue:float)
   PlayAnimMontage(
    execute=@n33.Female
    Actor=@n30
    NewAnimMontage=/SparksCharacterFull/Animation/MainPlayer/BandPlay/Drum/Attacks/Drum_Attack_DoubleCrash_CMF_A_M
    InPlayRate=1
    StartSectionName=Default
   )
   Event SparksCharAttackEventCharging() {is=title K2Node_Event}
   > PrintString(InString=Play Start Charging Attack Anim Here.)
   n34 = TryGetPawnOwner()
   > n35 = Get Fort Pawn From Object(InObject=@n34)
   > n36 = FindGender(Object=@n35) -> (OutGender=Invalid)
   > n37 = Switch on EFortCustomGender(Selection=@n36)
   PlayAnimMontage(
    execute=[@n37.Both, @n37.Male]
    Actor=@n34
    NewAnimMontage=/SparksCharacterFull/Animation/MainPlayer/BandPlay/Drum/Attacks/Drum_Charge_DoubleCrash_CMM_A_M
    InPlayRate=1
    StartSectionName=Intro
   )
   PlayAnimMontage(
    execute=@n37.Female
    Actor=@n34
    NewAnimMontage=/SparksCharacterFull/Animation/MainPlayer/BandPlay/Drum/Attacks/Drum_Charge_DoubleCrash_CMF_A_M
    InPlayRate=1
    StartSectionName=Intro
   )
   AnimNotify_LeftArmHit_A_Entered() {is=title K2Node_Event}
   > Set LeftArmIsPlayingHit(LeftArmIsPlayingHit:bool=true) -> (Output_Get:bool)
   > Set TriggerLeftArmHit_A(TriggerLeftArmHit_A:bool=false) -> (Output_Get:bool)
   AnimNotify_LeftArmHit_B_Entered() {is=title K2Node_Event}
   > Set LeftArmIsPlayingHit(LeftArmIsPlayingHit=true)
   > Set TriggerLeftArmHit_B(TriggerLeftArmHit_B:bool=false) -> (Output_Get:bool)
   Event BlueprintInitializeAnimation()
   > n38 = IntensityMidiTextEvents!(execute) -> (then, Events:Array of Strings)
   > Set MidiTextEvents(MidiTextEvents:Array of Strings=@n38) -> (Output_Get:Array of Strings)
   n39 = TryGetPawnOwner()
   > Is Valid(InputObject=@n39)
   > Set PawnOwner(PawnOwner:Pawn Object Reference=@n39) -> (Output_Get:Pawn Object Reference)
   AnimNotify_LeftArm_Neutral_Entered() {is=title K2Node_Event}
   > Set LeftArmIsPlayingHit(LeftArmIsPlayingHit=false)
   n40 = Event Bind_Tom2_Hit!() -> (OutputDelegate, then, Callback:"Delegate (by ref)") {is=title K2Node_Event}
   > Bind Event to Drum Kit Tom 2 Hit(self:Sparks Drum Anim BP Object Reference, Delegate=@n40.Callback) {is=title K2Node_AddDelegate}
   n41 = Event Bind_Tom1_Hit!() -> (OutputDelegate, then, Callback:"Delegate (by ref)") {is=title K2Node_Event}
   > Bind Event to Drum Kit Tom 1 Hit(self:Sparks Drum Anim BP Object Reference, Delegate=@n41.Callback) {is=title K2Node_AddDelegate}
   PrintString(InString=Drums: Set up drumkit after spawn)
   InverseTransformLocation!(T:"Transform (by ref)"=GetSocketTransform(self=GetOwningComponent(), InSocketName=Root, TransformSpace=RTS_World), Location:Vector=@n1) -> (ReturnValue:Vector)
   InverseTransformLocation(T=GetSocketTransform(self=GetOwningComponent(), InSocketName=Root, TransformSpace=RTS_World), Location=@n4)
   K2Node_Composite_2 = Location Setup() {is=title K2Node_Composite}
   LeftArmTest() {is=title K2Node_CustomEvent}
   RightArmTest() {is=title K2Node_CustomEvent}
   Event Sparks Debug Update() {is=title K2Node_Event}
  ]}
 Twist Rotator Setup {
  path=EventGraph.K2Node_Composite_1.Twist Rotator Setup
  nodes [
   n1 = GetOwningComponent()
   n2 = GetSocketTransform(self=@n1, InSocketName=Root, TransformSpace=RTS_World)
   n3 = vector + vector(A=TransformLocation!(T:"Transform (by ref)"=@n2, Location:Vector=MakeVector(Y=50)) -> (ReturnValue:Vector), B="0,0,20")
   DrawDebugArrow!(
    execute
    LineStart:Vector=vector + vector(A=GetSocketLocation!(self:Scene Component Object Reference=@n1, InSocketName:Name=Root) -> (ReturnValue:Vector), B="0,0,20")
    LineEnd:Vector=@n3
    ArrowSize:float=5
    LineColor:Linear Color Structure=(R=1,G=1,B=1,A=1)
    Duration:float=60
    Thickness:float=1
    DepthPriority:EDrawDebugSceneDepthPriorityGroup Enum=World
   ) -> (then)
   n4 = InverseTransformLocation(T=@n2, Location=Get CymbalRideLocation() -> (CymbalRideLocation:Vector))
   n5 = BreakVector(InVec=Normalize(A=@n4, Tolerance=0.0001))
   n6 = "Atan2 (Degrees)"(Y=@n5, X=@n5.Y)
   > PrintString(InString="To String (Float)"(InDouble=float * float(A=@n6, B=-1)), TextColor=(R=0.6811,G=1,B=0,A=1), Duration=60)
   n7 = Get PawnOwner() -> (PawnOwner:Pawn Object Reference)
   PrintString(
    InString=Append(A="PlayerPawn = ", B=GetDisplayName(Object=@n7))
    TextColor=(R=0,G=0.5,B=1,A=1)
    Duration=GetWorldDeltaSeconds!() -> (ReturnValue:double)
   )
   > PrintString(InString=Append(A="Drumkit Actor = ", B=GetDisplayName()), TextColor=(R=0,G=0.502,B=1,A=1), Duration=GetWorldDeltaSeconds())
   n8 = GetSkeletalMeshAsset!(self:Skeletal Mesh Component Object Reference=Get DrumKitSkelComp() -> (DrumKitSkelComp:Skeletal Mesh Component Object Reference)) -> (ReturnValue:Skeletal Mesh Object Reference)
   > PrintString(
    InString=Append(A="DrumKit Skeletal Mesh Asset = ", B=GetDisplayName(Object=@n8))
    TextColor=(R=0,G=0.5,B=1,A=1)
    Duration=GetWorldDeltaSeconds()
   )
   n9 = Get MidpointLocation() -> (MidpointLocation:Vector)
   > PrintString(
    InString=Append(A="Midpoint Location = ", B="To String (Vector)"!(InVec:Vector=@n9) -> (ReturnValue:String))
    TextColor=(R=1,G=0.749,B=0,A=1)
    Duration=GetWorldDeltaSeconds()
   )
   n10 = Get SnareLocation() -> (SnareLocation:Vector)
   > PrintString(
    InString=Append(A="Snare Socket Location = ", B="To String (Vector)"(InVec=@n10))
    TextColor=(R=1,G=0.749,B=0,A=1)
    Duration=GetWorldDeltaSeconds()
   )
   n11 = Get FloorTomLocation() -> (FloorTomLocation:Vector)
   > PrintString(
    InString=Append(A="FloorTom Socket Location = ", B="To String (Vector)"(InVec=@n11))
    TextColor=(R=1,G=0.749,B=0,A=1)
    Duration=GetWorldDeltaSeconds()
   )
   Inputs()
   > n12 = Sequence()
   > n13 = Cast To FortPawn!(execute, Object=Get PawnOwner()) -> (then, CastFailed, AsFort Pawn:Fort Pawn Object Reference) {is=title K2Node_DynamicCast}
   n14 = Get EmoteMeshCompDatas!(self:Fort Pawn Object Reference=@n13) -> (EmoteMeshCompDatas:Map of Integers to Emote Mesh Comp Data Structures)
   > n15 = Cast To SkeletalMeshComponent!(
    execute
    Object=Break Emote Mesh Comp Data!(EmoteMeshCompData:"Emote Mesh Comp Data Structure (by ref)"=Find(TargetMap:Map of Integers to Emote Mesh Comp Data Structures=@n14, Key:"Integer (by ref)"=3103) -> (Value:Emote Mesh Comp Data Structure)) -> (MeshComp:Mesh Component Object Reference, PropExclusionTag:Gameplay Tag Structure, bTickEvenWhenNotRendered:bool=~) {is=title K2Node_BreakStruct}
   ) -> (then, CastFailed, AsSkeletal Mesh Component:Skeletal Mesh Component Object Reference) {is=title K2Node_DynamicCast}
   > n16 = Set DrumKitSkelComp(DrumKitSkelComp:Skeletal Mesh Component Object Reference=@n15) -> (Output_Get:Skeletal Mesh Component Object Reference)
   > n17 = Cast To Sparks_DrumKit_Interface!(execute, Object=GetAnimInstance(self=@n16)) -> (then, CastFailed, AsSparks Drum Kit Interface:Sparks Drum Kit Interface Interface) {is=title K2Node_DynamicCast}
   > Set DrumKit_AnimBP(DrumKit_AnimBP:Sparks Drum Kit Interface Interface=@n17) -> (Output_Get:Sparks Drum Kit Interface Interface)
   > Set RootBoneTransform(RootBoneTransform:Transform=GetSocketTransform(self=GetOwningComponent(), InSocketName=Root, TransformSpace=RTS_World)) -> (Output_Get:Transform)
   n18 = Get DrumKitSkelComp()
   > Set SnareLocation(SnareLocation:Vector=InverseTransformLocation(T=Get RootBoneTransform() -> (RootBoneTransform:Transform), Location=GetSocketLocation(self=@n18, InSocketName=bone_target_snare))) -> (Output_Get:Vector)
   n19 = Get DrumKitSkelComp()
   > Set FloorTomLocation(FloorTomLocation:Vector=InverseTransformLocation(T=Get RootBoneTransform(), Location=GetSocketLocation(self=@n19, InSocketName=bone_target_floor_tom))) -> (Output_Get:Vector)
   n20 = Get DrumKitSkelComp()
   > Set CymbalLeftLocation(CymbalLeftLocation:Vector=InverseTransformLocation(T=Get RootBoneTransform(), Location=GetSocketLocation(self=@n20, InSocketName=bone_target_L-cymbal))) -> (Output_Get:Vector)
   n21 = Get DrumKitSkelComp()
   > Set CymbalRightLocation(CymbalRightLocation:Vector=InverseTransformLocation(T=Get RootBoneTransform(), Location=GetSocketLocation(self=@n21, InSocketName=bone_target_R-cymbal))) -> (Output_Get:Vector)
   n22 = Get DrumKitSkelComp()
   > Set HiHatLocation(HiHatLocation:Vector=InverseTransformLocation(T=Get RootBoneTransform(), Location=GetSocketLocation(self=@n22, InSocketName=bone_target_snare))) -> (Output_Get:Vector)
   n23 = Get DrumKitSkelComp()
   > Set CymbalRideLocation(CymbalRideLocation:Vector=InverseTransformLocation(T=Get RootBoneTransform(), Location=GetSocketLocation(self=@n23, InSocketName=bone_target_ride))) -> (Output_Get:Vector)
   n24 = Get DrumKitSkelComp()
   > Set LeftTomLocation(LeftTomLocation:Vector=InverseTransformLocation(T=Get RootBoneTransform(), Location=GetSocketLocation(self=@n24, InSocketName=bone_target_L-tom))) -> (Output_Get:Vector)
   n25 = Get DrumKitSkelComp()
   > Set RightTomLocation(RightTomLocation:Vector=InverseTransformLocation(T=Get RootBoneTransform(), Location=GetSocketLocation(self=@n25, InSocketName=bone_target_R-tom))) -> (Output_Get:Vector)
   n26 = Get DrumKitSkelComp()
   > Set CowbellLocation(CowbellLocation:Vector=InverseTransformLocation(T=Get RootBoneTransform(), Location=GetSocketLocation(self=@n26, InSocketName=bone_target_R-cowbell))) -> (Output_Get:Vector)
   PrintString(execute=@n15.CastFailed, InString="Cast To skel mesh DrumKit Failed!", TextColor=(R=0.5859,G=1,B=0,A=1), Duration=20)
   Reinitialize(execute=@n12.then_1, self:Sparks Drum Kit Interface Interface=Get DrumKit_AnimBP() -> (DrumKit_AnimBP:Sparks Drum Kit Interface Interface))
   > Outputs!(then:Exec)
   Normalize(A=@n3, Tolerance=0.0001)
   "To String (Float)"(InDouble=0)
   "To String (Vector)"(InVec=@n4)
   GetSocketLocation(InSocketName=bone_target_hihat)
   Get HiHatLocation() -> (HiHatLocation:Vector)
   Get RightTomLocation() -> (RightTomLocation:Vector)
   Get SnareLocation()
  ]
 }
 Location Setup {
  path=EventGraph.K2Node_Composite_2.Location Setup
  nodes [
   Inputs()
   > Set PlayerPawnLocation(PlayerPawnLocation:Vector=Get Actor Location(self=Get PawnOwner())) -> (Output_Get:Vector)
   n1 = GetOwningComponent()
   > Set PawnSpineLocation(PawnSpineLocation:Vector=GetSocketLocation(self=@n1, InSocketName=spine_05)) -> (Output_Get:Vector)
   > Set PawnLeftArmLocation(PawnLeftArmLocation:Vector=GetSocketLocation(self=@n1, InSocketName=upperarm_l)) -> (Output_Get:Vector)
   > Set PawnRightArmLocation(PawnRightArmLocation:Vector=GetSocketLocation(self=@n1, InSocketName=upperarm_r)) -> (Output_Get:Vector)
   Outputs!()
  ]
 }
 Song Intensity {
  path=EventGraph.K2Node_Composite_3.Song Intensity
  nodes [
   Inputs!() -> (In:Exec)
   > n1 = Sequence!(execute) -> (then_0, then_1, then_2:Exec)
   > ClearMidiListener()
   Branch(execute=@n1.then_1, Condition=IsValidInterface(Interface=Get MidiEventDriver() -> (MidiEventDriver:Midi Event Driver Interface Interface)))
   > n2 = ListenForMidiTextEvents!(
    execute:Exec
    MidiEventDriver:Midi Event Driver Interface Interface=Get MidiEventDriver()
    MidiTrackName:Name=Get MidiTrackName() -> (MidiTrackName:Name=None)
    EventStrings:Array of Strings=Get MidiTextEvents() -> (MidiTextEvents:Array of Strings)
    SecondsBeforeEventStart:float=Get SecBeforeIntensity() -> (SecBeforeIntensity:float)
    InitBroadcast:EMidiEventListenerInitBroadcast Enum=MostRecentEvent
   ) -> (
    then:Exec
    AsyncTaskProxy:Async Action Listen For Midi Text Event Object Reference
    OnMidiEventStart:Exec
    OnFinished:Exec
    MidiEventText:String
    MidiEventData:"Parsed Midi Text Event Structure (by ref)"
   ) {is=title K2Node_AsyncAction}
   > Set IntensityMidiListener(IntensityMidiListener=@n2)
   MidiEvent_Intensity_OnEvent!(execute=@n2.OnMidiEventStart, self, MidiEventText:String=@n2.MidiEventText) -> (then)
   Outputs!(Out:Exec=@n1.then_2)
  ]
 }
 Drum Hits {
  path=EventGraph.K2Node_Composite_4.Drum Hits
  nodes [
   n1 = RightArm_ResetHitBool()
   n2 = Branch(execute=[@n1.then, @n28.else], Condition=Get RightArmLastHitWasB() -> (RightArmLastHitWasB:bool))
   > Set TriggerRightArmHit_B(TriggerRightArmHit_B=false)
   > Set TriggerRightArmHit_A(TriggerRightArmHit_A=true)
   > n3 = Set RightArmLastHitWasB(RightArmLastHitWasB:bool=false) -> (Output_Get:bool)
   Set RightArmIsPlayingHit(execute=[@n4.then, @n3.then], RightArmIsPlayingHit=true)
   > SetRightArmHitAngle() {is=title K2Node_CallFunction}
   Set TriggerRightArmHit_A(execute=@n2.else, TriggerRightArmHit_A=false)
   > Set TriggerRightArmHit_B(TriggerRightArmHit_B=true)
   > n4 = Set RightArmLastHitWasB(RightArmLastHitWasB=true)
   n5 = LeftArm_ResetHitBool()
   n6 = Branch(execute=[@n5.then, @K2Node_Composite_1.Exec Out], Condition=Get LeftArmLastHitWasB() -> (LeftArmLastHitWasB:bool))
   > Set TriggerLeftArmHit_B(TriggerLeftArmHit_B=false)
   > Set TriggerLeftArmHit_A(TriggerLeftArmHit_A=true)
   > n7 = Set LeftArmLastHitWasB(LeftArmLastHitWasB:bool=false) -> (Output_Get:bool)
   Set LeftArmIsPlayingHit(execute=[@n8.then, @n7.then], LeftArmIsPlayingHit=true)
   > SetLeftArmHitAngle() {is=title K2Node_CallFunction}
   Set TriggerLeftArmHit_A(execute=@n6.else, TriggerLeftArmHit_A=false)
   > Set TriggerLeftArmHit_B(TriggerLeftArmHit_B=true)
   > n8 = Set LeftArmLastHitWasB(LeftArmLastHitWasB=true)
   Reroute Node()
   > PrintString(InString=Hello, TextColor=(R=1,G=1,B=0,A=1), Duration=1)
   Reroute Node()
   > Set RightArmHitLocation(RightArmHitLocation:Vector=Get HiHatLocation()) -> (Output_Get:Vector)
   Reroute Node()
   > PrintString(InString=Hello, TextColor=(R=1,G=1,B=0,A=1), Duration=1)
   Inputs()
   > n9 = Sequence()
   > K2Node_Composite_0 = ClearDrumHits!(In:Exec) -> (Out:Exec) {is=title K2Node_Composite}
   Branch(execute=@n9.then_1, Condition=IsValidInterface(Interface=Get MidiEventDriver()))
   > n10 = Sequence!(execute) -> (then_0, then_1, then_2, then_3:Exec, then_4:Exec, then_5:Exec)
   n11 = Get LeftArmMidiNoteIntArray() -> (LeftArmMidiNoteIntArray:Array of Bytes)
   n12 = Make Array!(
    "[0]":Midi Note Event Delay Pool Structure=Make Midi Note Event Delay Pool!(Id:Name=Short, MaxSecondsBefore:float=0.25, MinSecondsBefore:float=0) -> (MidiNoteEvent_DelayPool:Midi Note Event Delay Pool Structure) {is=title K2Node_MakeStruct}
    "[1]":Midi Note Event Delay Pool Structure=Make Midi Note Event Delay Pool(Id=Medium, MaxSecondsBefore=0.5, MinSecondsBefore=0.25)
   ) -> (Array:Array of Midi Note Event Delay Pool Structures)
   > n13 = ListenForMidiNoteEventsPools!(
    execute:Exec
    MidiEventDriver:Midi Event Driver Interface Interface=Get MidiEventDriver()
    MidiTrackName:Name=Get MidiTrackName() -> (MidiTrackName=None)
    EventNotes:Array of Bytes=@n11
    WatchNotes:Array of Bytes=@n11
    DelayPools:Array of Midi Note Event Delay Pool Structures=@n12
    InitBroadcast:EMidiEventListenerInitBroadcast Enum=MostRecentEvent
   ) -> (
    then:Exec
    AsyncTaskProxy:Async Action Listen For Midi Note Event Pools Object Reference
    OnMidiNoteStart:Exec
    OnFinished:Exec
    MidiNote:Byte
    MidiNoteData:"Parsed Midi Note Event Structure (by ref)"
    WaitPoolID:Name
    ActualDelay:float=~
    MaxDelayShortenedBy:float=~
   ) {is=title K2Node_AsyncAction}
   > Set LeftArmMidiListener(LeftArmMidiListener:Async Action Listen For Midi Note Event Pools Object Reference=@n13) -> (Output_Get:Async Action Listen For Midi Note Event Pools Object Reference)
   Do Once(execute=@n13.OnMidiNoteStart, Reset=@n20.then)
   > n14 = Sequence!(execute) -> (then_0, then_1)
   > Set LeftArmHitLongBlend(LeftArmHitLongBlend:bool="float >= float"(A=@n13.ActualDelay, B=0.25)) -> (Output_Get:bool)
   > Set LeftArmHitMediumBlend(LeftArmHitMediumBlend:bool="float >= float"(A=@n13.ActualDelay, B=0.15)) -> (Output_Get:bool)
   n15 = GetDrumMidiNoteEnumFromInteger!(MidiNoteEventInt:Byte=@n13.MidiNote) -> (ReturnValue:EDrumMidiNoteEvent Enum=None)
   > Set LeftArmMidiNoteEventEnum(LeftArmMidiNoteEventEnum:EDrumMidiNoteEvent Enum=@n15) -> (Output_Get:EDrumMidiNoteEvent Enum=None)
   > n16 = Set LeftArmWaitPoolID(LeftArmWaitPoolID:Name=@n13.WaitPoolID) -> (Output_Get:Name=None)
   > n17 = Switch on Name!(execute:Exec, Selection:Name=@n16) -> (Short:Exec, Medium:Exec, Long:Exec) {is=title K2Node_SwitchName, bHasDefaultPin=false, pinNames [Short,Medium,Long]}
   Set LeftArmIsShortDelay(execute=@n17.Medium, LeftArmIsShortDelay:bool=false) -> (Output_Get:bool)
   > Set LeftArmIsLongDelay(LeftArmIsLongDelay:bool=false) -> (Output_Get:bool)
   > n18 = Set LeftArmMaxDelayShortenedBy(LeftArmMaxDelayShortenedBy:float=@n13.MaxDelayShortenedBy) -> (Output_Get:float)
   K2Node_Composite_1 = CollapseGraph!(Exec In:Exec=[@n40.then, @n19.then, @n18.then]) -> (Exec Out:Exec) {is=title K2Node_Composite}
   Set LeftArmIsShortDelay(execute=@n17.Short, LeftArmIsShortDelay=true)
   > Set LeftArmIsLongDelay(LeftArmIsLongDelay=false)
   > n19 = Set LeftArmMaxDelayShortenedBy(LeftArmMaxDelayShortenedBy=float + float(A=@n13.MaxDelayShortenedBy, B=0.25))
   n20 = DelayUntilNextTick!(execute=@n14.then_1) -> (then)
   n21 = Get RightArmMidiNoteIntArray() -> (RightArmMidiNoteIntArray:Array of Bytes)
   n22 = ListenForMidiNoteEventsPools(
    execute=@n10.then_1
    MidiEventDriver=Get MidiEventDriver()
    MidiTrackName=Get MidiTrackName() -> (MidiTrackName=None)
    EventNotes=@n21
    WatchNotes=@n21
    DelayPools=@n12
    InitBroadcast=MostRecentEvent
   ) -> (ActualDelay=~, MaxDelayShortenedBy=~)
   > Set RightArmMidiListener(RightArmMidiListener:Async Action Listen For Midi Note Event Pools Object Reference=@n22) -> (Output_Get:Async Action Listen For Midi Note Event Pools Object Reference)
   Do Once(execute=@n22.OnMidiNoteStart, Reset=@n30.then)
   > n23 = Sequence()
   > Set RightArmHitLongBlend(RightArmHitLongBlend:bool="float >= float"(A=@n22.ActualDelay, B=0.25)) -> (Output_Get:bool)
   > Set RightArmHitMediumBlend(RightArmHitMediumBlend:bool="float >= float"(A=@n22.ActualDelay, B=0.15)) -> (Output_Get:bool)
   n24 = GetDrumMidiNoteEnumFromInteger(MidiNoteEventInt=@n22.MidiNote) -> (ReturnValue=None)
   > Set RightArmMidiNoteEventEnum(RightArmMidiNoteEventEnum:EDrumMidiNoteEvent Enum=@n24) -> (Output_Get:EDrumMidiNoteEvent Enum=None)
   > n25 = Set RightArmWaitPoolID(RightArmWaitPoolID:Name=@n22.WaitPoolID) -> (Output_Get:Name=None)
   > n26 = Switch on Name(Selection=@n25) {bHasDefaultPin=false, pinNames [Short,Medium,Long]}
   Set RightArmIsShortDelay(execute=@n26.Medium, RightArmIsShortDelay:bool=false) -> (Output_Get:bool)
   > Set RightArmIsLongDelay(RightArmIsLongDelay:bool=false) -> (Output_Get:bool)
   > n27 = Set RightArmMaxDelayShortenedBy(RightArmMaxDelayShortenedBy:float=@n22.MaxDelayShortenedBy) -> (Output_Get:float)
   K2Node_Composite_3 = CollapseGraph_0!(Exec In:Exec=[@n39.then, @n27.then, @n29.then]) -> (Exec Out:Exec) {is=title K2Node_Composite}
   > n28 = Branch(Condition=Get Hit_Percussion_R())
   Set RightArmIsShortDelay(execute=@n26.Short, RightArmIsShortDelay=true)
   > Set RightArmIsLongDelay(RightArmIsLongDelay=false)
   > n29 = Set RightArmMaxDelayShortenedBy(RightArmMaxDelayShortenedBy=float + float(A=@n22.MaxDelayShortenedBy, B=0.25))
   n30 = DelayUntilNextTick(execute=@n23.then_1)
   n31 = ListenForMidiNoteEvents!(
    execute:Exec=@n10.then_2
    MidiEventDriver:Midi Event Driver Interface Interface=Get MidiEventDriver()
    MidiTrackName:Name=Get MidiTrackName() -> (MidiTrackName=None)
    EventNotes:Array of Bytes=Get AllDrumKitMidiNoteIntArray() -> (AllDrumKitMidiNoteIntArray:Array of Bytes)
    SecondsBeforeEventStart:float=0
    SecondsBeforeEventEnd:float=0
    InitBroadcast:EMidiEventListenerInitBroadcast Enum=MostRecentEvent
   ) -> (
    then:Exec
    AsyncTaskProxy:Async Action Listen For Midi Note Event Object Reference
    OnMidiEventStart:Exec
    OnMidiEventEnd:Exec
    OnFinished:Exec
    MidiEventNote:Byte
    MidiEventData:"Parsed Midi Note Event Structure (by ref)"
   ) {is=title K2Node_AsyncAction}
   > Set DrumKitMidiListener(DrumKitMidiListener:Async Action Listen For Midi Note Event Object Reference=@n31) -> (Output_Get:Async Action Listen For Midi Note Event Object Reference)
   K2Node_Composite_6 = DrumKitHitEvents!(
    execute=@n31.OnMidiEventStart
    Selection:EDrumMidiNoteEvent Enum=GetDrumMidiNoteEnumFromInteger(MidiNoteEventInt=@n31.MidiEventNote) -> (ReturnValue=None)
   ) -> (OutputPin:Exec) {is=title K2Node_Composite}
   n32 = ListenForMidiNoteEvent!(
    execute:Exec=@n10.then_3
    MidiEventDriver:Midi Event Driver Interface Interface=Get MidiEventDriver()
    MidiTrackName:Name=Get MidiTrackName() -> (MidiTrackName=None)
    EventNote:Byte=25
    SecondsBeforeEventStart:float=0.1
    SecondsBeforeEventEnd:float=0.15
    InitBroadcast:EMidiEventListenerInitBroadcast Enum=MostRecentEvent
   ) -> (
    then:Exec
    AsyncTaskProxy:Async Action Listen For Midi Note Event Object Reference
    OnMidiEventStart:Exec
    OnMidiEventEnd:Exec
    OnFinished:Exec
    MidiEventNote:Byte
    MidiEventData:"Parsed Midi Note Event Structure (by ref)"
   ) {is=title K2Node_AsyncAction}
   > Set PedalUpMidiListener(PedalUpMidiListener:Async Action Listen For Midi Note Event Object Reference=@n32) -> (Output_Get:Async Action Listen For Midi Note Event Object Reference)
   Set Trigger_HiHat_Open(execute=@n32.OnMidiEventEnd, Trigger_HiHat_Open:bool=false) -> (Output_Get:bool)
   > Set Trigger_HiHat_Close(Trigger_HiHat_Close:bool=true) -> (Output_Get:bool)
   Set Trigger_HiHat_Open(execute=@n32.OnMidiEventStart, Trigger_HiHat_Open=true)
   > Set Trigger_HiHat_Close(Trigger_HiHat_Close=false)
   n33 = Get RightLegMidiNoteIntArray() -> (RightLegMidiNoteIntArray:Array of Bytes)
   n34 = ListenForMidiNoteEventsPools(
    execute=@n10.then_4
    MidiEventDriver=Get MidiEventDriver()
    MidiTrackName=Get MidiTrackName() -> (MidiTrackName=None)
    EventNotes=@n33
    WatchNotes=@n33
    DelayPools=Make Array!("[0]"=Make Midi Note Event Delay Pool(Id=Medium, MaxSecondsBefore=0.28, MinSecondsBefore=0)) -> (Array)
    InitBroadcast=MostRecentEvent
   ) -> (ActualDelay=~, MaxDelayShortenedBy=~)
   > Set RightLegMidiListener(RightLegMidiListener:Async Action Listen For Midi Note Event Pools Object Reference=@n34) -> (Output_Get:Async Action Listen For Midi Note Event Pools Object Reference)
   Do Once(execute=@n34.OnMidiNoteStart, Reset=@n36.then)
   > n35 = Sequence()
   > Set RightLegMaxDelayShortenedBy(RightLegMaxDelayShortenedBy:float=@n34.MaxDelayShortenedBy) -> (Output_Get:float)
   > Increment Int(Value=Get HitBassIndex() -> (HitBassIndex:int=0))
   n36 = DelayUntilNextTick(execute=@n35.then_1)
   n37 = ListenForMidiNoteEvents(
    execute=@n10.then_5
    MidiEventDriver=Get MidiEventDriver()
    MidiTrackName=Get MidiTrackName() -> (MidiTrackName=None)
    EventNotes=Get AllDrumKitMidiNoteIntArray()
    SecondsBeforeEventStart=0.7
    SecondsBeforeEventEnd=0
    InitBroadcast=MostRecentEvent
   )
   > Set SpineRotationMidiListener(SpineRotationMidiListener:Async Action Listen For Midi Note Event Object Reference=@n37) -> (Output_Get:Async Action Listen For Midi Note Event Object Reference)
   n38 = Switch on EDrumMidiNoteEvent!(
    execute=@n37.OnMidiEventStart
    Selection:EDrumMidiNoteEvent Enum=GetDrumMidiNoteEnumFromInteger(MidiNoteEventInt=@n37.MidiEventNote) -> (ReturnValue=None)
   ) -> (
    None:Exec
    KickHit_RightFoot:Exec
    HiHatPedalUp_LeftFoot:Exec
    SnareHit_LeftHand:Exec
    SnareHit_RightHand:Exec
    HiHatHit_LeftHand:Exec
    HiHatHit_RightHand:Exec
    Percussion_RightHand:Exec
    Crash1HardHit_LeftHand:Exec
    Crash1SoftHit_LeftHand:Exec
    Crash1Hardhit_RightHand:Exec
    Crash1SoftHit_RightHand:Exec
    Crash2HardHit_RightHand:Exec
    Crash2SoftHit_RightHand:Exec
    Crash1Choke:Exec
    Crash2Choke:Exec
    RideCymbalHit_RightHand:Exec
    RideCymbalHit_LeftHand:Exec
    Crash2Hit_LeftHand:Exec
    Crash2SoftHit_LeftHand:Exec
    Tom1Hit_LeftHand:Exec
    Tom1Hit_RightHand:Exec
    Tom2Hit_LeftHand:Exec
    Tom2Hit_RightHand:Exec
    FloorTomHit_LeftHand:Exec
    FloorTomHit_RightHand:Exec
    SticksHeld:Exec
    SnareHit_LeftHand_Soft:Exec
    SnareHit_RightHand_Soft:Exec
   ) {is=title K2Node_SwitchEnum}
   Set RightArmHitLocation(execute=@n38.Crash1Choke, RightArmHitLocation=Get CymbalLeftLocation() -> (CymbalLeftLocation:Vector))
   Set LeftArmHitLocation(execute=@n38.Crash1HardHit_LeftHand, LeftArmHitLocation:Vector=Get CymbalLeftLocation()) -> (Output_Get:Vector)
   Set RightArmHitLocation(execute=@n38.Crash1Hardhit_RightHand, RightArmHitLocation=Get CymbalLeftLocation())
   Set LeftArmHitLocation(execute=@n38.Crash1SoftHit_LeftHand, LeftArmHitLocation=Get CymbalLeftLocation())
   Set RightArmHitLocation(execute=@n38.Crash1SoftHit_RightHand, RightArmHitLocation=Get CymbalLeftLocation())
   Set RightArmHitLocation(execute=@n38.Crash2Choke, RightArmHitLocation=Get CymbalRightLocation() -> (CymbalRightLocation:Vector))
   Set RightArmHitLocation(execute=@n38.Crash2HardHit_RightHand, RightArmHitLocation=Get CymbalRightLocation())
   Set LeftArmHitLocation(execute=@n38.Crash2Hit_LeftHand, LeftArmHitLocation=Get CymbalRightLocation())
   Set LeftArmHitLocation(execute=@n38.Crash2SoftHit_LeftHand, LeftArmHitLocation=Get CymbalRightLocation())
   Set RightArmHitLocation(execute=@n38.Crash2SoftHit_RightHand, RightArmHitLocation=Get CymbalRightLocation())
   Set LeftArmHitLocation(execute=@n38.FloorTomHit_LeftHand, LeftArmHitLocation=Get FloorTomLocation())
   Set RightArmHitLocation(execute=@n38.FloorTomHit_RightHand, RightArmHitLocation=Get FloorTomLocation())
   Set LeftArmHitLocation(execute=@n38.HiHatHit_LeftHand, LeftArmHitLocation=Get HiHatLocation())
   Set RightArmHitLocation(execute=[@n38.HiHatHit_RightHand, @n38.SnareHit_RightHand, @n38.SnareHit_RightHand_Soft], RightArmHitLocation=Get SnareLocation())
   Set RightArmHitLocation(execute=@n38.Percussion_RightHand, RightArmHitLocation=Get CowbellLocation() -> (CowbellLocation:Vector))
   Set LeftArmHitLocation(execute=@n38.RideCymbalHit_LeftHand, LeftArmHitLocation=Get CymbalRideLocation())
   Set RightArmHitLocation(execute=@n38.RideCymbalHit_RightHand, RightArmHitLocation=Get CymbalRideLocation())
   Set LeftArmHitLocation(execute=[@n38.SnareHit_LeftHand, @n38.SnareHit_LeftHand_Soft], LeftArmHitLocation=Get SnareLocation())
   Set LeftArmHitLocation(execute=@n38.Tom1Hit_LeftHand, LeftArmHitLocation=Get LeftTomLocation() -> (LeftTomLocation:Vector))
   Set RightArmHitLocation(execute=@n38.Tom1Hit_RightHand, RightArmHitLocation=Get LeftTomLocation())
   Set LeftArmHitLocation(execute=@n38.Tom2Hit_LeftHand, LeftArmHitLocation=Get RightTomLocation())
   Set RightArmHitLocation(execute=@n38.Tom2Hit_RightHand, RightArmHitLocation=Get RightTomLocation())
   Outputs(Out=@n9.then_2)
   Set RightArmIsShortDelay(RightArmIsShortDelay=false)
   > n39 = Set RightArmIsLongDelay(RightArmIsLongDelay=true)
   Set LeftArmIsShortDelay(LeftArmIsShortDelay=false)
   > n40 = Set LeftArmIsLongDelay(LeftArmIsLongDelay=true)
   ListenForMidiNoteEvent(MidiTrackName=None, EventNote=0, SecondsBeforeEventStart=0, SecondsBeforeEventEnd=0, InitBroadcast=MostRecentEvent)
   PrintString(InString=GetPlayerNameSafe!(Actor:Actor Object Reference=TryGetPawnOwner()) -> (ReturnValue:String))
   PrintString(InString=Hi Hat Pedal Up Start)
   PrintString(InString=Hi Hat Pedal Up STOP)
   GetAllDrumMidiNoteIntegers!() -> (ReturnValue:Array of Bytes)
   Append!(
    A="Row Name: "
    B="To String (Byte)"!(InByte:Byte=@n22.MidiNote) -> (ReturnValue:String)
    C:String="Midi Note Event: "
    D:String=Enum to String!(Enumerator:EDrumMidiNoteEvent Enum=@n24) -> (ReturnValue:String)
   ) -> (ReturnValue)
   Append(A="Row Name: ", B="To String (Byte)"(InByte=@n13.MidiNote), C="Midi Note Event: ", D=Enum to String(Enumerator=@n15))
   Flip Flop!(None:Exec) -> (A:Exec, B:Exec, IsA:bool=~) {is=title K2Node_MacroInstance}
   Flip Flop() -> (IsA=~)
   Make Midi Note Event Delay Pool(Id=Long, MaxSecondsBefore=7, MinSecondsBefore=0.4)
   "float <= float"(A=@n13.MaxDelayShortenedBy, B=0.25) -> (ReturnValue=~)
   "float <= float"(A=@n13.MaxDelayShortenedBy, B=0.15) -> (ReturnValue=~)
   "float <= float"(A=@n22.MaxDelayShortenedBy, B=0.15) -> (ReturnValue=~)
   "float <= float"(A=@n22.MaxDelayShortenedBy, B=0.25) -> (ReturnValue=~)
   Get TriggerRightArmHit_B()
   Get TriggerLeftArmHit_B()
  ]
 }
 ClearDrumHits {
  path=EventGraph.K2Node_Composite_4.Drum Hits.K2Node_Composite_0.ClearDrumHits
  nodes [
   Inputs()
   > n1 = Sequence()
   > n2 = Sequence!(execute) -> (then_0, then_1, then_2:Exec, then_3:Exec, then_4:Exec, then_5:Exec)
   n3 = Get LeftArmMidiListener() -> (LeftArmMidiListener:Async Action Listen For Midi Note Event Pools Object Reference)
   > Is Valid(InputObject=@n3)
   > Cancel(self=@n3)
   > Set LeftArmMidiListener()
   n4 = Get RightArmMidiListener() -> (RightArmMidiListener:Async Action Listen For Midi Note Event Pools Object Reference)
   Is Valid(exec=@n2.then_1, InputObject=@n4)
   > Cancel(self=@n4)
   > Set RightArmMidiListener()
   n5 = Get DrumKitMidiListener() -> (DrumKitMidiListener:Async Action Listen For Midi Note Event Object Reference)
   Is Valid(exec=@n2.then_2, InputObject=@n5)
   > Cancel(self=@n5)
   > Set DrumKitMidiListener()
   n6 = Get PedalUpMidiListener() -> (PedalUpMidiListener:Async Action Listen For Midi Note Event Object Reference)
   Is Valid(exec=@n2.then_3, InputObject=@n6)
   > Cancel(self=@n6)
   > Set PedalUpMidiListener()
   n7 = Get RightLegMidiListener() -> (RightLegMidiListener:Async Action Listen For Midi Note Event Pools Object Reference)
   Is Valid(exec=@n2.then_4, InputObject=@n7)
   > Cancel(self=@n7)
   > Set RightLegMidiListener()
   n8 = Get SpineRotationMidiListener() -> (SpineRotationMidiListener:Async Action Listen For Midi Note Event Object Reference)
   Is Valid(exec=@n2.then_5, InputObject=@n8)
   > Cancel(self=@n8)
   > Set SpineRotationMidiListener()
   Outputs(Out=@n1.then_1)
  ]
 }
 CollapseGraph {
  path=EventGraph.K2Node_Composite_4.Drum Hits.K2Node_Composite_1.CollapseGraph
  nodes [
   Inputs!() -> (Exec In:Exec)
   > n1 = Sequence!(execute) -> (then_0, then_1, then_2, then_3, then_4, then_5, then_6:Exec, then_7:Exec, then_8:Exec, then_9:Exec, then_10:Exec, then_11:Exec)
   n2 = Get LeftArmMidiNoteEventEnum() -> (LeftArmMidiNoteEventEnum:EDrumMidiNoteEvent Enum=None)
   > Set Hit_Snare_Soft_L(Hit_Snare_Soft_L:bool="Equal (Enum)"(A:EDrumMidiNoteEvent Enum=@n2, B:EDrumMidiNoteEvent Enum=SnareHit_LeftHand_Soft)) -> (Output_Get:bool)
   n3 = Set Hit_Snare_L(execute=@n1.then_1, Hit_Snare_L:bool="Equal (Enum)"(A=@n2, B=SnareHit_LeftHand)) -> (Output_Get:bool)
   > Branch(Condition=@n3)
   n4 = Set Hit_FloorTom_L(execute=@n1.then_10, Hit_FloorTom_L:bool="Equal (Enum)"(A=@n2, B=FloorTomHit_LeftHand)) -> (Output_Get:bool)
   > Branch(Condition=@n4)
   Outputs!(Exec Out:Exec=@n1.then_11)
   n5 = Set Hit_HiHat_L(execute=@n1.then_2, Hit_HiHat_L:bool="Equal (Enum)"(A=@n2, B=HiHatHit_LeftHand)) -> (Output_Get:bool)
   > Branch(Condition=@n5)
   n6 = Set Hit_Crash1_Hard_L(execute=@n1.then_3, Hit_Crash1_Hard_L:bool="Equal (Enum)"(A=@n2, B=Crash1HardHit_LeftHand)) -> (Output_Get:bool)
   > Branch(Condition=@n6)
   n7 = Set Hit_Crash1_Soft_L(execute=@n1.then_4, Hit_Crash1_Soft_L:bool="Equal (Enum)"(A=@n2, B=Crash1SoftHit_LeftHand)) -> (Output_Get:bool)
   > Branch(Condition=@n7)
   n8 = Set Hit_Crash2_Hard_L(execute=@n1.then_5, Hit_Crash2_Hard_L:bool="Equal (Enum)"(A=@n2, B=Crash2Hit_LeftHand)) -> (Output_Get:bool)
   > Branch(Condition=@n8)
   n9 = Set Hit_Crash2_Soft_L(execute=@n1.then_6, Hit_Crash2_Soft_L:bool="Equal (Enum)"(A=@n2, B=Crash2SoftHit_LeftHand)) -> (Output_Get:bool)
   > Branch(Condition=@n9)
   n10 = Set Hit_Ride_L(execute=@n1.then_7, Hit_Ride_L:bool="Equal (Enum)"(A=@n2, B=RideCymbalHit_LeftHand)) -> (Output_Get:bool)
   > Branch(Condition=@n10)
   n11 = Set Hit_Tom1_L(execute=@n1.then_8, Hit_Tom1_L:bool="Equal (Enum)"(A=@n2, B=Tom1Hit_LeftHand)) -> (Output_Get:bool)
   > Branch(Condition=@n11)
   n12 = Set Hit_Tom2_L(execute=@n1.then_9, Hit_Tom2_L:bool="Equal (Enum)"(A=@n2, B=Tom2Hit_LeftHand)) -> (Output_Get:bool)
   > Branch(Condition=@n12)
   Set LeftArmHitLocation(LeftArmHitLocation=Get FloorTomLocation())
   Set LeftArmHitLocation(LeftArmHitLocation=Get CymbalLeftLocation())
   Set LeftArmHitLocation(LeftArmHitLocation=Get CymbalRightLocation())
   Set LeftArmHitLocation(LeftArmHitLocation=Get CymbalRightLocation())
   Set LeftArmHitLocation(LeftArmHitLocation=Get CymbalRideLocation())
   Set LeftArmHitLocation(LeftArmHitLocation=Get LeftTomLocation())
   Set LeftArmHitLocation(LeftArmHitLocation=Get RightTomLocation())
   Set LeftArmHitLocation(LeftArmHitLocation=Get SnareLocation())
   Set LeftArmHitLocation(LeftArmHitLocation=Get HiHatLocation())
   Set LeftArmHitLocation(LeftArmHitLocation=Get CymbalLeftLocation())
  ]
 }
 CollapseGraph_0 {
  path=EventGraph.K2Node_Composite_4.Drum Hits.K2Node_Composite_3.CollapseGraph_0
  nodes [
   Inputs()
   > n1 = Sequence!(execute) -> (then_0, then_1, then_2, then_3, then_4, then_5, then_6, then_7, then_8, then_9, then_10, then_11, then_12:Exec, then_13:Exec, then_14:Exec)
   n2 = Get RightArmMidiNoteEventEnum() -> (RightArmMidiNoteEventEnum:EDrumMidiNoteEvent Enum=None)
   > Set Hit_Snare_Soft_R(Hit_Snare_Soft_R:bool="Equal (Enum)"(A=@n2, B=SnareHit_RightHand_Soft)) -> (Output_Get:bool)
   n3 = Set Hit_Snare_R(execute=@n1.then_1, Hit_Snare_R:bool="Equal (Enum)"(A=@n2, B=SnareHit_RightHand)) -> (Output_Get:bool)
   > Branch(Condition=@n3)
   n4 = Set Hit_Tom2_R(execute=@n1.then_10, Hit_Tom2_R:bool="Equal (Enum)"(A=@n2, B=Tom2Hit_RightHand)) -> (Output_Get:bool)
   > Branch(Condition=@n4)
   n5 = Set Hit_FloorTom_R(execute=@n1.then_11, Hit_FloorTom_R:bool="Equal (Enum)"(A=@n2, B=FloorTomHit_RightHand)) -> (Output_Get:bool)
   > Branch(Condition=@n5)
   n6 = Set Hit_Choke1_R(execute=@n1.then_12, Hit_Choke1_R:bool="Equal (Enum)"(A=@n2, B=Crash1Choke)) -> (Output_Get:bool)
   > Branch(Condition=@n6)
   n7 = Set Hit_Choke2_R(execute=@n1.then_13, Hit_Choke2_R:bool="Equal (Enum)"(A=@n2, B=Crash2Choke)) -> (Output_Get:bool)
   > Branch(Condition=@n7)
   Outputs(Exec Out=@n1.then_14)
   n8 = Set Hit_HiHat_R(execute=@n1.then_2, Hit_HiHat_R:bool="Equal (Enum)"(A=@n2, B=HiHatHit_RightHand)) -> (Output_Get:bool)
   > Branch(Condition=@n8)
   n9 = Set Hit_Percussion_R(execute=@n1.then_3, Hit_Percussion_R:bool="Equal (Enum)"(A=@n2, B=Percussion_RightHand)) -> (Output_Get:bool)
   > Branch(Condition=@n9)
   n10 = Set Hit_Crash1_Hard_R(execute=@n1.then_4, Hit_Crash1_Hard_R:bool="Equal (Enum)"(A=@n2, B=Crash1Hardhit_RightHand)) -> (Output_Get:bool)
   > Branch(Condition=@n10)
   n11 = Set Hit_Crash1_Soft_R(execute=@n1.then_5, Hit_Crash1_Soft_R:bool="Equal (Enum)"(A=@n2, B=Crash1SoftHit_RightHand)) -> (Output_Get:bool)
   > Branch(Condition=@n11)
   n12 = Set Hit_Crash2_Hard_R(execute=@n1.then_6, Hit_Crash2_Hard_R:bool="Equal (Enum)"(A=@n2, B=Crash2HardHit_RightHand)) -> (Output_Get:bool)
   > Branch(Condition=@n12)
   n13 = Set Hit_Crash2_Soft_R(execute=@n1.then_7, Hit_Crash2_Soft_R:bool="Equal (Enum)"(A=@n2, B=Crash2SoftHit_RightHand)) -> (Output_Get:bool)
   > Branch(Condition=@n13)
   n14 = Set Hit_Ride_R(execute=@n1.then_8, Hit_Ride_R:bool="Equal (Enum)"(A=@n2, B=RideCymbalHit_RightHand)) -> (Output_Get:bool)
   > Branch(Condition=@n14)
   n15 = Set Hit_Tom1_R(execute=@n1.then_9, Hit_Tom1_R:bool="Equal (Enum)"(A=@n2, B=Tom1Hit_RightHand)) -> (Output_Get:bool)
   > Branch(Condition=@n15)
   Set RightArmHitLocation(RightArmHitLocation=Get FloorTomLocation())
   Set RightArmHitLocation(RightArmHitLocation=Get CymbalRightLocation())
   Set RightArmHitLocation(RightArmHitLocation=Get CymbalRideLocation())
   Set RightArmHitLocation(RightArmHitLocation=Get CymbalLeftLocation())
   Set RightArmHitLocation(RightArmHitLocation=Get CymbalRightLocation())
   Set RightArmHitLocation(RightArmHitLocation=Get LeftTomLocation())
   Set RightArmHitLocation(RightArmHitLocation=Get RightTomLocation())
   Set RightArmHitLocation(RightArmHitLocation=Get SnareLocation())
   Set RightArmHitLocation(RightArmHitLocation=Get HiHatLocation())
   Set RightArmHitLocation(RightArmHitLocation=Get CowbellLocation())
   Set RightArmHitLocation(RightArmHitLocation=Get CymbalLeftLocation())
   Set RightArmHitLocation(RightArmHitLocation=Get CymbalLeftLocation())
   Set RightArmHitLocation(RightArmHitLocation=Get CymbalRightLocation())
  ]
 }
 DrumKitHitEvents {
  path=EventGraph.K2Node_Composite_4.Drum Hits.K2Node_Composite_6.DrumKitHitEvents
  nodes [
   n1 = Inputs!() -> (execute:Exec, Selection:EDrumMidiNoteEvent Enum=None)
   > n2 = Sequence!(execute) -> (then_0, then_1)
   > n3 = Switch on EDrumMidiNoteEvent(Selection=@n1)
   Call Drum Kit Crash 1 Choke!(execute:Exec=@n3.Crash1Choke, self:Sparks Drum Anim BP Object Reference) -> (then:Exec) {is=title K2Node_CallDelegate}
   Call Drum Kit Crash 1 Hard Hit!(execute:Exec=[@n3.Crash1HardHit_LeftHand, @n3.Crash1Hardhit_RightHand], self:Sparks Drum Anim BP Object Reference) -> (then:Exec) {is=title K2Node_CallDelegate}
   Call Drum Kit Crash 1 Soft Hit!(execute:Exec=[@n3.Crash1SoftHit_LeftHand, @n3.Crash1SoftHit_RightHand], self:Sparks Drum Anim BP Object Reference) -> (then:Exec) {is=title K2Node_CallDelegate}
   Call Drum Kit Crash 2 Choke!(execute:Exec=@n3.Crash2Choke, self:Sparks Drum Anim BP Object Reference) -> (then:Exec) {is=title K2Node_CallDelegate}
   Call Drum Kit Crash 2 Hard Hit!(execute:Exec=[@n3.Crash2HardHit_RightHand, @n3.Crash2Hit_LeftHand], self:Sparks Drum Anim BP Object Reference) -> (then:Exec) {is=title K2Node_CallDelegate}
   Call Drum Kit Crash 2 Soft Hit!(execute:Exec=[@n3.Crash2SoftHit_LeftHand, @n3.Crash2SoftHit_RightHand], self:Sparks Drum Anim BP Object Reference) -> (then:Exec) {is=title K2Node_CallDelegate}
   Call Drum Kit Floor Tom Hit!(execute:Exec=[@n3.FloorTomHit_LeftHand, @n3.FloorTomHit_RightHand], self:Sparks Drum Anim BP Object Reference) -> (then:Exec) {is=title K2Node_CallDelegate}
   Call Drum Kit Hi Hat Hit!(execute:Exec=[@n3.HiHatHit_LeftHand, @n3.HiHatHit_RightHand], self:Sparks Drum Anim BP Object Reference) -> (then:Exec) {is=title K2Node_CallDelegate}
   Call Drum Kit Cowbell Hit!(execute:Exec=@n3.Percussion_RightHand, self:Sparks Drum Anim BP Object Reference) -> (then:Exec) {is=title K2Node_CallDelegate}
   Call Drum Kit Ride Hard Hit!(execute:Exec=[@n3.RideCymbalHit_LeftHand, @n3.RideCymbalHit_RightHand], self:Sparks Drum Anim BP Object Reference) -> (then:Exec) {is=title K2Node_CallDelegate}
   Call Drum Kit Snare Hit!(execute:Exec=[@n3.SnareHit_LeftHand, @n3.SnareHit_RightHand], self:Sparks Drum Anim BP Object Reference) -> (then:Exec) {is=title K2Node_CallDelegate}
   Call Drum Kit Tom 1 Hit!(execute:Exec=[@n3.Tom1Hit_LeftHand, @n3.Tom1Hit_RightHand], self:Sparks Drum Anim BP Object Reference) -> (then:Exec) {is=title K2Node_CallDelegate}
   Call Drum Kit Tom 2 Hit!(execute:Exec=[@n3.Tom2Hit_LeftHand, @n3.Tom2Hit_RightHand], self:Sparks Drum Anim BP Object Reference) -> (then:Exec) {is=title K2Node_CallDelegate}
   Outputs!(OutputPin:Exec=@n2.then_1)
  ]
 }
 Get_HitBassIndex {nodes [
   Get_HitBassIndex() {is=title K2Node_FunctionEntry}
   > Return Node!(execute, Index:int=Get HitBassIndex() -> (HitBassIndex=0))
  ]}
 Get_RightLegMaxDelayShortenedBy {nodes [
   Get_RightLegMaxDelayShortenedBy() {is=title K2Node_FunctionEntry}
   > Return Node!(execute, RightLegMaxDelayShortenedBy:double=Get RightLegMaxDelayShortenedBy())
  ]}
 Get_TriggerHiHatClose {nodes [
   Get_TriggerHiHatClose() {is=title K2Node_FunctionEntry}
   > Return Node!(execute, TriggerHiHatClose:bool=Get Trigger_HiHat_Close())
  ]}
 Get_TriggerHiHatOpen {nodes [
   Get_TriggerHiHatOpen() {is=title K2Node_FunctionEntry}
   > Return Node!(execute, TriggerHiHatOpen:bool=Get Trigger_HiHat_Open())
  ]}
 LeftArm_ResetHitBool {nodes [
   LeftArm_ResetHitBool() {is=title K2Node_FunctionEntry}
   > n1 = Sequence!(execute) -> (then_0, then_1, then_2:Exec, then_3:Exec, then_4:Exec, then_5:Exec, then_6:Exec, then_7:Exec, then_8:Exec, then_9:Exec)
   n2 = Get LeftArmMidiNoteEventEnum() -> (LeftArmMidiNoteEventEnum=None)
   > n3 = Set Hit_Snare_L(Hit_Snare_L="Equal (Enum)"(A=@n2, B=SnareHit_LeftHand))
   > Branch(Condition=@n3)
   > Set LeftArmHitLocation(LeftArmHitLocation=Get SnareLocation())
   n4 = Set Hit_HiHat_L(execute=@n1.then_1, Hit_HiHat_L="Equal (Enum)"(A=@n2, B=HiHatHit_LeftHand))
   > Branch(Condition=@n4)
   > Set LeftArmHitLocation(LeftArmHitLocation=Get HiHatLocation())
   n5 = Set Hit_Crash1_Hard_L(execute=@n1.then_2, Hit_Crash1_Hard_L="Equal (Enum)"(A=@n2, B=Crash1HardHit_LeftHand))
   > Branch(Condition=@n5)
   > Set LeftArmHitLocation(LeftArmHitLocation=Get CymbalLeftLocation())
   n6 = Set Hit_Crash1_Soft_L(execute=@n1.then_3, Hit_Crash1_Soft_L="Equal (Enum)"(A=@n2, B=Crash1SoftHit_LeftHand))
   > Branch(Condition=@n6)
   > Set LeftArmHitLocation(LeftArmHitLocation=Get CymbalLeftLocation())
   n7 = Set Hit_Crash2_Hard_L(execute=@n1.then_4, Hit_Crash2_Hard_L="Equal (Enum)"(A=@n2, B=Crash2Hit_LeftHand))
   > Branch(Condition=@n7)
   > Set LeftArmHitLocation(LeftArmHitLocation=Get CymbalRightLocation())
   n8 = Set Hit_Crash2_Soft_L(execute=@n1.then_5, Hit_Crash2_Soft_L="Equal (Enum)"(A=@n2, B=Crash2SoftHit_LeftHand))
   > Branch(Condition=@n8)
   > Set LeftArmHitLocation(LeftArmHitLocation=Get CymbalRightLocation())
   n9 = Set Hit_Ride_L(execute=@n1.then_6, Hit_Ride_L="Equal (Enum)"(A=@n2, B=RideCymbalHit_LeftHand))
   > Branch(Condition=@n9)
   > Set LeftArmHitLocation(LeftArmHitLocation=Get CymbalRideLocation())
   n10 = Set Hit_Tom1_L(execute=@n1.then_7, Hit_Tom1_L="Equal (Enum)"(A=@n2, B=Tom1Hit_LeftHand))
   > Branch(Condition=@n10)
   > Set LeftArmHitLocation(LeftArmHitLocation=Get LeftTomLocation())
   n11 = Set Hit_Tom2_L(execute=@n1.then_8, Hit_Tom2_L="Equal (Enum)"(A=@n2, B=Tom2Hit_LeftHand))
   > Branch(Condition=@n11)
   > Set LeftArmHitLocation(LeftArmHitLocation=Get RightTomLocation())
   n12 = Set Hit_FloorTom_L(execute=@n1.then_9, Hit_FloorTom_L="Equal (Enum)"(A=@n2, B=FloorTomHit_LeftHand))
   > Branch(Condition=@n12)
   > Set LeftArmHitLocation(LeftArmHitLocation=Get FloorTomLocation())
  ]}
 MidiEvent_Intensity_OnEvent {nodes [
   n1 = MidiEvent_Intensity_OnEvent!() -> (then, MidiEventText:String) {is=title K2Node_FunctionEntry}
   > n2 = TranslateMidiEventTextToSongIntensity!(execute, MidiEventText:Name=String To Name(InString=@n1)) -> (then, IdleRealtime:bool, MidiSongIntensityIsIdle:bool, SongIntensity:Gameplay Tag Structure)
   > Set IdleRealtime(IdleRealtime:bool=@n2) -> (Output_Get:bool)
   > Set DesiredSongIntensity(DesiredSongIntensity:Gameplay Tag Structure=@n2.SongIntensity) -> (Output_Get:Gameplay Tag Structure)
   > Set IsIdleBeatSynced(IsIdleBeatSynced=@n2.MidiSongIntensityIsIdle)
  ]}
 OnHitBassEntry {nodes [
   OnHitBassEntry!() -> (then, Context:"Anim Update Context Structure (by ref)", Node:"Anim Node Reference Structure (by ref)") {is=title K2Node_FunctionEntry}
   > Set HitBassPlayed(HitBassPlayed:int=Get HitBassIndex() -> (HitBassIndex=0)) -> (Output_Get:int=0)
   > Return Node!(execute)
  ]}
 OnPlay0Entry {nodes [
   OnPlay0Entry!() -> (then, Context:"Anim Update Context Structure (by ref)", Node:"Anim Node Reference Structure (by ref)") {is=title K2Node_FunctionEntry}
   > SetPlayAnimAndTempoMultiplier!(execute, self, Index:int=0) -> (then)
   > Return Node()
  ]}
 OnPlay1Entry {nodes [
   OnPlay1Entry!() -> (then, Context:"Anim Update Context Structure (by ref)", Node:"Anim Node Reference Structure (by ref)") {is=title K2Node_FunctionEntry}
   > SetPlayAnimAndTempoMultiplier(Index=1)
   > Return Node()
  ]}
 RightArm_ResetHitBool {nodes [
   RightArm_ResetHitBool() {is=title K2Node_FunctionEntry}
   > n1 = Sequence!(execute) -> (then_0, then_1, then_2, then_3, then_4, then_5, then_6, then_7, then_8, then_9, then_10:Exec, then_11:Exec, then_12:Exec)
   n2 = Get RightArmMidiNoteEventEnum() -> (RightArmMidiNoteEventEnum=None)
   > n3 = Set Hit_Snare_R(Hit_Snare_R="Equal (Enum)"(A=@n2, B=SnareHit_RightHand))
   > Branch(Condition=@n3)
   > Set RightArmHitLocation(RightArmHitLocation=Get SnareLocation())
   n4 = Set Hit_HiHat_R(execute=@n1.then_1, Hit_HiHat_R="Equal (Enum)"(A=@n2, B=HiHatHit_RightHand))
   > Branch(Condition=@n4)
   > Set RightArmHitLocation(RightArmHitLocation=Get HiHatLocation())
   n5 = Set Hit_FloorTom_R(execute=@n1.then_10, Hit_FloorTom_R="Equal (Enum)"(A=@n2, B=FloorTomHit_RightHand))
   > Branch(Condition=@n5)
   > Set RightArmHitLocation(RightArmHitLocation=Get FloorTomLocation())
   n6 = Set Hit_Choke1_R(execute=@n1.then_11, Hit_Choke1_R="Equal (Enum)"(A=@n2, B=Crash1Choke))
   > Branch(Condition=@n6)
   > Set RightArmHitLocation(RightArmHitLocation=Get CymbalLeftLocation())
   n7 = Set Hit_Choke2_R(execute=@n1.then_12, Hit_Choke2_R="Equal (Enum)"(A=@n2, B=Crash2Choke))
   > Branch(Condition=@n7)
   > Set RightArmHitLocation(RightArmHitLocation=Get CymbalRightLocation())
   n8 = Set Hit_Percussion_R(execute=@n1.then_2, Hit_Percussion_R="Equal (Enum)"(A=@n2, B=Percussion_RightHand))
   > Branch(Condition=@n8)
   > Set RightArmHitLocation(RightArmHitLocation=Get CowbellLocation())
   n9 = Set Hit_Crash1_Hard_R(execute=@n1.then_3, Hit_Crash1_Hard_R="Equal (Enum)"(A=@n2, B=Crash1Hardhit_RightHand))
   > Branch(Condition=@n9)
   > Set RightArmHitLocation(RightArmHitLocation=Get CymbalLeftLocation())
   n10 = Set Hit_Crash1_Soft_R(execute=@n1.then_4, Hit_Crash1_Soft_R="Equal (Enum)"(A=@n2, B=Crash1SoftHit_RightHand))
   > Branch(Condition=@n10)
   > Set RightArmHitLocation(RightArmHitLocation=Get CymbalLeftLocation())
   n11 = Set Hit_Crash2_Hard_R(execute=@n1.then_5, Hit_Crash2_Hard_R="Equal (Enum)"(A=@n2, B=Crash2HardHit_RightHand))
   > Branch(Condition=@n11)
   > Set RightArmHitLocation(RightArmHitLocation=Get CymbalRightLocation())
   n12 = Set Hit_Crash2_Soft_R(execute=@n1.then_6, Hit_Crash2_Soft_R="Equal (Enum)"(A=@n2, B=Crash2SoftHit_RightHand))
   > Branch(Condition=@n12)
   > Set RightArmHitLocation(RightArmHitLocation=Get CymbalRightLocation())
   n13 = Set Hit_Ride_R(execute=@n1.then_7, Hit_Ride_R="Equal (Enum)"(A=@n2, B=RideCymbalHit_RightHand))
   > Branch(Condition=@n13)
   > Set RightArmHitLocation(RightArmHitLocation=Get CymbalRideLocation())
   n14 = Set Hit_Tom1_R(execute=@n1.then_8, Hit_Tom1_R="Equal (Enum)"(A=@n2, B=Tom1Hit_RightHand))
   > Branch(Condition=@n14)
   > Set RightArmHitLocation(RightArmHitLocation=Get LeftTomLocation())
   n15 = Set Hit_Tom2_R(execute=@n1.then_9, Hit_Tom2_R="Equal (Enum)"(A=@n2, B=Tom2Hit_RightHand))
   > Branch(Condition=@n15)
   > Set RightArmHitLocation(RightArmHitLocation=Get RightTomLocation())
  ]}
 SetPlayAnimAndTempoMultiplier {nodes [
   n1 = SetPlayAnimAndTempoMultiplier!() -> (then, Index:int) {is=title K2Node_FunctionEntry}
   > Set SongIntensity(SongIntensity:Gameplay Tag Structure=Get DesiredSongIntensity() -> (DesiredSongIntensity:Gameplay Tag Structure)) -> (Output_Get:Gameplay Tag Structure)
   > n2 = PickIntensityAnimLoopCycle!(
    execute
    self:Sparks Instrument Anim Instance Object Reference
    BPMSet:"Sparks Anim BPMSet Structure (by ref)"=Get HomebaseBPMs() -> (HomebaseBPMs:Sparks Anim BPMSet Structure)
    TransitionTag:Gameplay Tag Structure
    ForceVariant:bool=false
    VariantIndex:int=0
   ) -> (then, ReturnValue:"Sparks Anim Loop Cycle Structure (by ref)")
   > Set Array Elem!(
    execute
    TargetArray:Array of Anim Sequence Object References=Get PlaySequences()
    Index:int=@n1
    Item:"Anim Sequence Object Reference (by ref)"=Break Sparks Anim Loop Cycle!(SparksAnimLoopCycle:"Sparks Anim Loop Cycle Structure (by ref)"=@n2) -> (Loop:Anim Sequence Object Reference, Entry:Anim Sequence Object Reference, Exit:Anim Sequence Object Reference) {is=title K2Node_BreakStruct}
    bSizeToFit:bool=true
   ) -> (then)
   > Set Array Elem(
    TargetArray:"Array of Float (double-precision)s"=Get TempoMultipliers()
    Index=@n1
    Item:"Float (double-precision) (by ref)"=Break Sparks Anim BPMSet!(SparksAnimBPMSet:"Sparks Anim BPMSet Structure (by ref)"=Get HomebaseBPMs()) -> (BPMBuckets:Array of Sparks Anim Intensity Set Object References, CurrentTempoMultiplier:float=~) {is=title K2Node_BreakStruct}.CurrentTempoMultiplier
    bSizeToFit=true
   )
  ]}
 ShouldHitBassTransition {nodes [
   ShouldHitBassTransition() {is=title K2Node_FunctionEntry}
   > Return Node!(execute, Transition:bool="Not Equal (Integer)"(A=Get HitBassIndex() -> (HitBassIndex=0), B=Get HitBassPlayed() -> (HitBassPlayed:int=0)))
  ]}
 ShouldTransition {nodes [
   ShouldTransition() {is=title K2Node_FunctionEntry}
   n1 = Get HomebaseBPMs()
   n2 = Get DesiredSongIntensity()
   n3 = Get SongIntensity() -> (SongIntensity:Gameplay Tag Structure)
   > Return Node(Transition=OR Boolean(
     A="Not Equal (GameplayTag)"(A:Gameplay Tag Structure=@n2, B:Gameplay Tag Structure=@n3) -> (ReturnValue:bool=~) {is=title K2Node_PromotableOperator}
     B=HasInvalidBucket!(self:Sparks Instrument Anim Instance Object Reference, BPMSet:"Sparks Anim BPMSet Structure (by ref)"=@n1) -> (ReturnValue:bool)
    ))
  ]}
]
variables [
 AllDrumKitMidiNoteIntArray {type="array<integer>", default [26,30,34,35,43,44,45,46,48,50,27,31,32,36,37,38,39,42,47,49,51,24,25,28,29]}
 BPM Preview {type=number, default=125}
 CowbellLocation {type=Vector}
 CymbalLeftLocation {type=Vector}
 CymbalRideLocation {type=Vector}
 CymbalRightLocation {type=Vector}
 DEBUG_AllowFeature_3200 {type=boolean}
 DEBUG_AnimNotify_TextColor {type=LinearColor, default=<rgba 1 0 1 1>}
 DelayBefore_EnterPlayVariant {type=number, default=12}
 DeltaTimeX {type=number}
 DrumKit Cowbell Hit {default=()}
 DrumKit Crash1 Choke {default=()}
 DrumKit Crash1 Hard Hit {default=()}
 DrumKit Crash1 Soft Hit {default=()}
 DrumKit Crash2 Choke {default=()}
 DrumKit Crash2 Hard Hit {default=()}
 DrumKit Crash2 Soft Hit {default=()}
 DrumKit FloorTom Hit {default=()}
 DrumKit HiHat Hit {default=()}
 DrumKit HiHat Pedal Up {default=()}
 DrumKit Kick Hit {default=()}
 DrumKit Ride Hard Hit {default=()}
 DrumKit Snare Hit {default=()}
 DrumKit Tom1 Hit {default=()}
 DrumKit Tom2 Hit {default=()}
 DrumKit_AnimBP {default=None}
 DrumKitMidiListener {type=AsyncAction_ListenForMidiNoteEvent, default=null}
 DrumKitSkelComp {type=SkeletalMeshComponent}
 FloorTomLocation {type=Vector}
 HiHatLocation {type=Vector}
 Hit_Choke1_R {type=boolean}
 Hit_Choke2_R {type=boolean}
 Hit_Crash1_Hard_L {type=boolean}
 Hit_Crash1_Hard_R {type=boolean}
 Hit_Crash1_Soft_L {type=boolean}
 Hit_Crash1_Soft_R {type=boolean}
 Hit_Crash2_Hard_L {type=boolean}
 Hit_Crash2_Hard_R {type=boolean}
 Hit_Crash2_Soft_L {type=boolean}
 Hit_Crash2_Soft_R {type=boolean}
 Hit_FloorTom_L {type=boolean}
 Hit_FloorTom_R {type=boolean}
 Hit_HiHat_L {type=boolean}
 Hit_HiHat_R {type=boolean}
 Hit_Percussion_R {type=boolean}
 Hit_Ride_L {type=boolean}
 Hit_Ride_R {type=boolean}
 Hit_Snare_L {type=boolean}
 Hit_Snare_R {type=boolean}
 Hit_Snare_Soft_L {type=boolean}
 Hit_Snare_Soft_R {type=boolean}
 Hit_Tom1_L {type=boolean}
 Hit_Tom1_R {type=boolean}
 Hit_Tom2_L {type=boolean}
 Hit_Tom2_R {type=boolean}
 HitBassIndex {type=integer}
 HitBassPlayed {type=integer}
 IdleRealtime {type=boolean, default=true}
 IntensityMidiListener {type=AsyncAction_ListenForMidiTextEvent, default=null}
 IsIdleBeatSynced {type=boolean}
 LeftArmHitAngle {type=number}
 LeftArmHitLocation {type=Vector}
 LeftArmHitLongBlend {type=boolean}
 LeftArmHitMediumBlend {type=boolean}
 LeftArmHitShortBlend {type=boolean}
 LeftArmIsLongDelay {type=boolean}
 LeftArmIsPlayingHit {type=boolean}
 LeftArmIsPlayingHitBlend {type=number}
 LeftArmIsShortDelay {type=boolean}
 LeftArmLastHitWasB {type=boolean}
 LeftArmMaxDelayShortenedBy {type=number}
 LeftArmMidiListener {type=AsyncAction_ListenForMidiNoteEventPools, default=null}
 LeftArmMidiNoteEventEnum {type=EDrumMidiNoteEvent, default=None}
 LeftArmMidiNoteIntArray {type="array<integer>", default [26,30,34,35,43,44,45,46,48,50,28]}
 LeftArmWaitPoolID {type=string, default=None}
 LeftExtreme {type=Rotator}
 LeftLegMidiNoteIntArray {type="array<integer>", default [25]}
 LeftTomLocation {type=Vector}
 LipSyncAnimSequence {type=AnimSequence, default=null}
 MidiTextEvents {type="array<string>", default []}
 MidiTrackName {type=string, default=None}
 MidpointLocation {type=Vector}
 PawnLeftArmLocation {type=Vector}
 PawnOwner {type=Pawn, default=null}
 PawnRightArmLocation {type=Vector}
 PawnSpineLocation {type=Vector}
 PedalUpMidiListener {type=AsyncAction_ListenForMidiNoteEvent, default=null}
 PlayerCenter {type=Rotator}
 PlayerDirectionRotator {type=Rotator}
 PlayerPawnLocation {type=Vector}
 PlaySequences {type="array<AnimSequence>", default []}
 RightArmHitAngle {type=number}
 RightArmHitLocation {type=Vector}
 RightArmHitLongBlend {type=boolean}
 RightArmHitMediumBlend {type=boolean}
 RightArmHitShortBlend {type=boolean}
 RightArmIsLongDelay {type=boolean}
 RightArmIsPlayingHit {type=boolean}
 RightArmIsPlayingHitBlend {type=number}
 RightArmIsShortDelay {type=boolean}
 RightArmLastHitWasB {type=boolean}
 RightArmMaxDelayShortenedBy {type=number}
 RightArmMidiListener {type=AsyncAction_ListenForMidiNoteEventPools, default=null}
 RightArmMidiNoteEventEnum {type=EDrumMidiNoteEvent, default=None}
 RightArmMidiNoteIntArray {type="array<integer>", default [27,31,32,36,37,38,39,42,47,49,51,29,40,41]}
 RightArmWaitPoolID {type=string, default=None}
 RightExtreme {type=Rotator}
 RightLegMaxDelayShortenedBy {type=number}
 RightLegMidiListener {type=AsyncAction_ListenForMidiNoteEventPools, default=null}
 RightLegMidiNoteIntArray {type="array<integer>", default [24]}
 RightTomLocation {type=Vector}
 RootBoneTransform {type=Transform, default {location=<0 0 0>, rotation=<rot 0 0 0>, scale=<1 1 1>}}
 SecBeforeIntensity {type=number, default=10}
 SnareLocation {type=Vector}
 SpineRotationMidiListener {type=AsyncAction_ListenForMidiNoteEvent, default=null}
 SpineTwist {type=number}
 TempoMultipliers {type="array<number>", default []}
 TimeSinceLastTransition {type=number}
 Trigger_HiHat_Close {type=boolean}
 Trigger_HiHat_Open {type=boolean}
 TriggerLeftArmHit_A {type=boolean}
 TriggerLeftArmHit_B {type=boolean}
 TriggerRightArmHit_A {type=boolean}
 TriggerRightArmHit_B {type=boolean}
]
